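# Editor utility Blueprint that batch reparents selected Blueprint assets to a chosen class; it is editor tooling.
# Its one function, Reparent_BP, takes an Actor class reference as the new parent.
# It loops over the selected assets, resolves each to a Blueprint with GetBlueprintAsset and passes it to ReparentBlueprint.
# The graph has no explicit validity check, compile or save step.
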EditorUtilityBlueprint /FigmentEnvironment/Tools/LevelInstanceMaterialInstanceSetup/AAU_Reparent
graphs [
 EventGraph {links [], nodes []}
 Reparent_BP {nodes [
   n1 = Reparent_BP!() -> (then, Class to Reparent To:Actor Class Reference) {is=title K2Node_FunctionEntry}
   > n2 = GetSelectedAssets()
   > n3 = For Each Loop(Array:Array of Object References=@n2) -> (Array Element:Object Reference)
   > n4 = GetBlueprintAsset(Object=@n3)
   > ReparentBlueprint(Blueprint=@n4, NewParentClass=@n1)
  ]}
]
variables []
