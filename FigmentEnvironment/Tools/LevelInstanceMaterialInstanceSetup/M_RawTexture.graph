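# M_RawTexture is an unlit, alpha-composite material in the LevelInstanceMaterialInstanceSetup tools folder that samples a texture parameter, RawTex.
# PointSampledUVs snaps the sample coordinates to texel centers for a fixed 1024 texture size, avoiding filtered blending.
# The texture's alpha channel feeds a OneMinus inversion, with no material output connection shown.
# RawTex defaults to a gradient effects texture, and automatic usage flagging in the editor is off.
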Material /FigmentEnvironment/Tools/LevelInstanceMaterialInstanceSetup/M_RawTexture
graph {
 properties {bAutomaticallySetUsageInEditor=false, blendMode=BLEND_AlphaComposite, shadingModel=MSM_Unlit, usedShadingModels=Unlit}
 nodes [
  n1 = Constant() {r=1024}
  OneMinus(None=TextureSampleParameter2D(UVs=MaterialFunctionCall!(UVs, Texture Size=@n1) {materialFunction=/Engine/Functions/Engine_MaterialFunctions01/Texturing/PointSampledUVs}.Result) {parameterName=RawTex, texture=/Game/Effects/Fort_Effects/Textures/Grad/gradient_stasis}.A)
 ]
}
