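# Editor utility Blueprint that renders copies of landscape grass textures into render targets, apparently to compare duplicates for consolidation.
# `Go` loads hardcoded texture paths, switches each to HDR compression without restoring it, and draws it through M_RawTexture into a size-matched target.
# Its pixel-comparison nodes are unfinished and mostly disconnected.
# `fill` draws textureA and textureB into preset targets, temporarily switching compression and restoring it afterward.
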EditorUtilityBlueprint /FigmentEnvironment/Tools/LevelInstanceMaterialInstanceSetup/EUA_Consolidate
graphs [
 EventGraph {nodes [
   Event BeginPlay()
   Event ActorBeginOverlap()
   Event Tick!() -> (OutputDelegate, then, DeltaSeconds:float)
  ]}
 Go {nodes [
   Go() {is=title K2Node_FunctionEntry}
   > n1 = GetEditorWorld(self=Get UnrealEditorSubsystem())
   > n2 = CreateDynamicMaterialInstance(Parent=&/M_RawTexture, OptionalName=None, CreationFlags=Transient)
   > Set MI(MI:Material Instance Dynamic Object Reference=@n2) -> (Output_Get:Material Instance Dynamic Object Reference)
   > Clear(TargetArray:Array of Texture Render Target 2D Object References=Get RT2D() -> (RT2D:Array of Texture Render Target 2D Object References))
   > Clear(TargetArray:Array of Int Vector 2 Structures=Get sizes() -> (sizes:Array of Int Vector 2 Structures))
   > Clear(TargetArray:Array of Texture 2D Object References=Get TexArray() -> (TexArray:Array of Texture 2D Object References))
   > n3 = Sequence()
   n4 = ParseIntoArray(
    SourceString=MakeLiteralString(Value="/Figment_S05/Migration51/Athena/Environments/Landscape/Textures/T_Athena_FarmGrass_01\r\n/Figment_S06/FN_6_20/Athena/Environments/Landscape/Textures/T_Athena_FarmGrass_01\r\n/Figment_S07/LM_7_00/Athena/Environments/Landscape/Textures/T_Athena_FarmGrass_01\r\n/Figment_S08/LM_8_1_1/Athena/Environments/Landscape/Textures/T_Athena_FarmGrass_01")
    Delimiter="\r\n"
   )
   > n5 = For Each Loop(Array=@n4)
   n6 = Get TexArray()
   PrintString(execute=@n5.Completed, InString="To String (Integer)"(InInt=Length(TargetArray:Array of Texture 2D Object References=@n6)))
   n7 = LoadAsset(execute=@n5.LoopBody, AssetPath=@n5)
   > FullyLoadAssets!(execute, Assets:Array of Object References=Make Array("[0]":Object Reference=@n7) -> (Array:Array of Object References)) -> (then)
   > n8 = Cast To Texture2D(Object=@n7)
   > Add(TargetArray:Array of Texture 2D Object References=Get TexArray(), NewItem:"Texture 2D Object Reference (by ref)"=@n8) {is=title K2Node_CallArrayFunction}
   n9 = GetSizeX!(self:Texture 2D Object Reference=@n8) -> (ReturnValue:int=0)
   n10 = GetSizeY!(self:Texture 2D Object Reference=@n8) -> (ReturnValue:int=0)
   > Add(
    TargetArray:Array of Int Vector 2 Structures=Get sizes()
    NewItem:"Int Vector 2 Structure (by ref)"=Make Int Vector 2!(X:int=@n9, Y:int=@n10) -> (IntVector2:Int Vector 2 Structure) {is=title K2Node_MakeStruct}
   )
   > n11 = CreateRenderTarget2D!(
    execute
    WorldContextObject:Object Reference=@n1
    Width:int=@n9
    Height:int=@n10
    Format:ETextureRenderTargetFormat Enum=RTF_RGBA16f
    ClearColor:Linear Color Structure=(R=0,G=0,B=0,A=1)
    bAutoGenerateMipMaps:bool=false
    bSupportUAVs:bool=false
   ) -> (then, ReturnValue:Texture Render Target 2D Object Reference)
   > Add(
    TargetArray:Array of Texture Render Target 2D Object References=Get RT2D()
    NewItem:"Texture Render Target 2D Object Reference (by ref)"=@n11
   )
   n12 = Get MI() -> (MI:Material Instance Dynamic Object Reference)
   > SetTextureParameterValue(self=@n12, ParameterName=RawTex, Value=@n8)
   > Set CompressionSettings(CompressionSettings=TC_HDR, self=@n8)
   > DrawMaterialToRenderTarget!(
    execute
    WorldContextObject:Object Reference=@n1
    TextureRenderTarget:Texture Render Target 2D Object Reference=@n11
    Material:Material Interface Object Reference=@n12
   ) -> (then)
   PrintString(execute=@n3.then_1, InString=Start)
   n13 = Get RT2D()
   n14 = "Get (a copy)"(Array:Array of Texture Render Target 2D Object References=@n13) -> (Output:Texture Render Target 2D Object Reference)
   > n15 = ReadRenderTarget!(execute, WorldContextObject:Object Reference=@n1, TextureRenderTarget:Texture Render Target 2D Object Reference=@n14, bNormalize:bool=false) -> (then, OutSamples:Array of Color Structures, ReturnValue:bool)
   n16 = "Get (a copy)"(Array=@n13, Dimension 1=1)
   n17 = Branch(Condition="Equal (Object)"(A=@n14, B=@n16))
   > PrintString(InString=True)
   PrintString(execute=@n17.else, InString=False)
   n18 = ReadRenderTarget(WorldContextObject=@n1, TextureRenderTarget=@n16, bNormalize=false)
   Identical!(ArrayA:Array of Color Structures=@n15, ArrayB:Array of Color Structures=@n18) -> (ReturnValue:bool) {is=title K2Node_CallArrayFunction}
   Length(TargetArray:Array of Strings=@n4)
   GetCPUCopy!(self:Texture 2D Object Reference=@n8) -> (ReturnValue:FShared Image Const Structure)
   CreateAssetData(InAsset=@n7, bAllowBlueprintClass=false)
   GetAssetTools()
   ReloadPackages!(
    execute
    PackagesToReload:Array of Package Object References=Make Array("[0]":Package Object Reference=GetPackageForObject!(Object:Object Reference=@n8) -> (ReturnValue:Package Object Reference)) -> (Array:Array of Package Object References)
    InteractionMode:EReloadPackagesInteractionMode Enum=AssumePositive
   ) -> (then, bOutAnyPackagesReloaded:bool, OutErrorMessage:Text)
   Execute Python Script!(
    execute
    PythonScript="import unreal\r\n\r\nprint(array_a == array_b)"
    arrayA:Array of Color Structures=@n15
    arrayB:Array of Color Structures=@n18
   ) -> (then, ReturnValue) {inputs [arrayA,arrayB]}
   Execute Python Script!(execute, PythonScript="import unreal\r\nimport numpy as np\r\n\r\n\r\nprint(np.array(aa)[0])", aa:Array of Color Structures=@n15) -> (then, ReturnValue) {inputs [aa]}
   For Each Loop(Array:Array of Texture 2D Object References=@n6) -> (Array Element:Texture 2D Object Reference)
   n19 = Get CompressionSettings(self=@n8)
   Print Enum!(Execute:Exec, Enum:Byte=@n19, OptionalString:String, Duration:float=2, TextColor:Linear Color Structure=(R=0,G=0.6588,B=1,A=1)) -> (Then:Exec) {is=title K2Node_MacroInstance}
   WaitForPackage!(execute:Exec, self:Object Reference=@n8, PackageName:String) -> (then:Exec) {is=title K2Node_Message}
   Get RT2D()
   Set CompressionSettings(CompressionSettings=@n19, self=@n8)
  ]}
 UserConstructionScript {nodes [Construction Script()]}
 fill {nodes [
   fill() {is=title K2Node_FunctionEntry}
   > n1 = GetEditorWorld(self=Get UnrealEditorSubsystem())
   n2 = Get textureA() -> (textureA:Texture 2D Object Reference)
   n3 = GetSizeX(self=@n2) -> (ReturnValue=0)
   n4 = GetSizeY(self=@n2) -> (ReturnValue=0)
   n5 = Get RTA() -> (RTA:Texture Render Target 2D Object Reference)
   > ResizeRenderTarget2D!(execute, TextureRenderTarget:Texture Render Target 2D Object Reference=@n5, Width:int=@n3, Height:int=@n4) -> (then)
   n6 = Get RTB() -> (RTB:Texture Render Target 2D Object Reference)
   > ResizeRenderTarget2D(TextureRenderTarget=@n6, Width=@n3, Height=@n4)
   > ResizeRenderTarget2D(
    TextureRenderTarget=Get RTB_C() -> (RTB_C:Texture Render Target 2D Object Reference)
    Width=Truncate(A=float * float(A:int=@n3, B=1.5) {is=title K2Node_PromotableOperator})
    Height=@n4
   )
   > ClearRenderTarget2D!(
    execute
    WorldContextObject:Object Reference=@n1
    TextureRenderTarget:Texture Render Target 2D Object Reference=@n5
    ClearColor:Linear Color Structure=(R=0,G=0,B=0,A=0)
   ) -> (then)
   > ClearRenderTarget2D(WorldContextObject=@n1, TextureRenderTarget=@n6, ClearColor=(R=0,G=0,B=0,A=0))
   > ClearRenderTarget2D(WorldContextObject=@n1, TextureRenderTarget=Get RTB_C(), ClearColor=(R=0,G=0,B=0,A=0))
   > n7 = CreateDynamicMaterialInstance(
    WorldContextObject=@n1
    Parent=&M_RawTexture
    OptionalName=None
    CreationFlags=Transient
   )
   > Set MI(MI=@n7)
   n8 = Get MI()
   > SetTextureParameterValue(self=@n8, ParameterName=RawTex, Value=Get textureA())
   n9 = Get textureA()
   > Set prevcomp(prevcomp:TextureCompressionSettings Enum=Get CompressionSettings(self=@n9)) -> (Output_Get:TextureCompressionSettings Enum)
   > Set CompressionSettings(CompressionSettings=TC_HDR, self=@n9)
   > DrawMaterialToRenderTarget(WorldContextObject=@n1, TextureRenderTarget=Get RTA(), Material=@n8)
   > Set CompressionSettings(CompressionSettings=Get prevcomp() -> (prevcomp:TextureCompressionSettings Enum), self=@n9)
   n10 = Get MI()
   n11 = Get textureB() -> (textureB:Texture 2D Object Reference)
   > SetTextureParameterValue(self=@n10, ParameterName=RawTex, Value=@n11)
   > Set prevcomp(prevcomp=Get CompressionSettings(self=@n11))
   > Set CompressionSettings(CompressionSettings=TC_HDR, self=@n11)
   > DrawMaterialToRenderTarget(WorldContextObject=@n1, TextureRenderTarget=Get RTB(), Material=@n10)
   > Set CompressionSettings(CompressionSettings=Get prevcomp(), self=@n11)
   DrawMaterialToRenderTarget()
   Get RTA()
  ]}
]
variables [
 MI {type=MaterialInstanceDynamic}
 RT2D {type="array<TextureRenderTarget2D>", default []}
 RTA {type=TextureRenderTarget2D, default=&/NewTextureRenderTarget2D}
 RTB {type=TextureRenderTarget2D, default=&/NewTextureRenderTarget2D1}
 RTB_C {type=TextureRenderTarget2D, default=&/NewTextureRenderTarget2D2}
 sizes {type="array<IntVector2>", default []}
 TexArray {type="array<Texture2D>", default []}
 textureA {type=Texture2D, default=/FigmentEnvironment/FN_6_20/Athena/Environments/Landscape/Textures/T_Athena_Grass_Farm_ColorMatched_D_2}
 textureB {type=Texture2D, default=/Game/Athena/Environments/Landscape/Textures/T_Athena_Grass_Farm_ColorMatched_D_2}
]
