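# Editor utility widget, apparently an unfinished texture consolidation tool, with a single Go button.
# The button clears its arrays, then loads a hardcoded list of same-named Athena FarmGrass texture paths from different Figment season folders.
# Each texture and its size are stored, with a matching RGBA16f render target created in the editor world, then the count is printed.
# A second texture loop is empty, the resize node is unconnected, and nothing is drawn or consolidated.
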EditorUtilityWidgetBlueprint /FigmentEnvironment/Tools/LevelInstanceMaterialInstanceSetup/EUW_Consolidation
graphs [EventGraph {nodes [
   "On Clicked (EditorUtilityButton_0)"() {is=title K2Node_ComponentBoundEvent}
   > n1 = GetEditorWorld(self=Get UnrealEditorSubsystem())
   > Clear(TargetArray:Array of Texture Render Target 2D Object References=Get RT2D() -> (RT2D:Array of Texture Render Target 2D Object References))
   > Clear(TargetArray:Array of Int Vector 2 Structures=Get sizes() -> (sizes:Array of Int Vector 2 Structures))
   > Clear(TargetArray:Array of Texture 2D Object References=Get TexArray() -> (TexArray:Array of Texture 2D Object References))
   > n2 = Sequence()
   > n3 = For Each Loop(Array=ParseIntoArray(
     SourceString=MakeLiteralString(Value="/Figment_S05/Migration51/Athena/Environments/Landscape/Textures/T_Athena_FarmGrass_01\r\n/Figment_S06/FN_6_20/Athena/Environments/Landscape/Textures/T_Athena_FarmGrass_01\r\n/Figment_S07/LM_7_00/Athena/Environments/Landscape/Textures/T_Athena_FarmGrass_01\r\n/Figment_S08/LM_8_1_1/Athena/Environments/Landscape/Textures/T_Athena_FarmGrass_01")
     Delimiter="\r\n"
    ))
   n4 = Get TexArray()
   PrintString(execute=@n3.Completed, InString="To String (Integer)"(InInt=Length(TargetArray:Array of Texture 2D Object References=@n4)))
   n5 = LoadAsset(execute=@n3.LoopBody, AssetPath=@n3)
   > n6 = Cast To Texture2D(Object=@n5)
   > Add(TargetArray:Array of Texture 2D Object References=Get TexArray(), NewItem:"Texture 2D Object Reference (by ref)"=@n6) {is=title K2Node_CallArrayFunction}
   n7 = GetSizeX!(self:Texture 2D Object Reference=@n6) -> (ReturnValue:int=0)
   n8 = GetSizeY!(self:Texture 2D Object Reference=@n6) -> (ReturnValue:int=0)
   > Add(
    TargetArray:Array of Int Vector 2 Structures=Get sizes()
    NewItem:"Int Vector 2 Structure (by ref)"=Make Int Vector 2!(X:int=@n7, Y:int=@n8) -> (IntVector2:Int Vector 2 Structure) {is=title K2Node_MakeStruct}
   )
   > n9 = CreateRenderTarget2D!(
    execute
    WorldContextObject:Object Reference=@n1
    Width:int=@n7
    Height:int=@n8
    Format:ETextureRenderTargetFormat Enum=RTF_RGBA16f
    ClearColor:Linear Color Structure=(R=0,G=0,B=0,A=1)
    bAutoGenerateMipMaps:bool=false
    bSupportUAVs:bool=false
   ) -> (then, ReturnValue:Texture Render Target 2D Object Reference)
   > Add(TargetArray:Array of Texture Render Target 2D Object References=Get RT2D(), NewItem:"Texture Render Target 2D Object Reference (by ref)"=@n9)
   For Each Loop(Exec=@n2.then_1, Array:Array of Texture 2D Object References=@n4) -> (Array Element:Texture 2D Object Reference)
   ResizeRenderTarget2D!(execute, TextureRenderTarget:Texture Render Target 2D Object Reference, Width:int=256, Height:int=256) -> (then)
   Event PreConstruct()
   Event Construct()
   Event Tick()
  ]}]
variables [
 RT2D {type="array<TextureRenderTarget2D>", default []}
 sizes {type="array<IntVector2>", default []}
 TexArray {type="array<Texture2D>", default []}
]
widgets {
 info {rootWidgetClass=&CanvasPanel, widgetCount=3}
 widgets [
  CanvasPanel_0 {widgetClassPath=&CanvasPanel, widget=CanvasPanel_0}
  EditorUtilityButton_0 {
   widgetClassPath=&EditorUtilityButton
   parent=CanvasPanel_0
   widget=EditorUtilityButton_0
   slot {
    class=CanvasPanelSlot
    properties {bAutoSize=true, layoutData {alignment=<0 0>, anchors {maximum=<0 0>, minimum=<0 0>}, offsets=<ltrb 132 152 100 40>}}
   }
  }
  TextBlock_0 {widgetClassPath=&TextBlock, parent=EditorUtilityButton_0, widget=TextBlock_0, properties {text=Go}, slot {class=ButtonSlot}}
 ]
}
