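# Editor utility widget that batch-converts Figment environment levels, mainly cliff static meshes, materials and BuildingTextureData.
# Buttons act on selected actors, the open level, or each World in a path, some through temporary level instances.
# Functions map meshes to cliff Blueprints for replacement, duplicate texture data and material instances with reparenting, delete foliage actors and clear tags.
# Inline Python reads and writes TextureData arrays, flagging changed actors for saving.
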EditorUtilityWidgetBlueprint /FigmentEnvironment/Tools/LevelInstanceMaterialInstanceSetup/EUW_ReplaceMaterialsWithInstances
graphs [
 AssignTextureDataSlot4 {nodes [
   AssignTextureDataSlot4() {is=title K2Node_FunctionEntry}
   > n1 = GetEditorWorld(self=Get UnrealEditorSubsystem())
   > n2 = GetAllActorsOfClass!(
    execute
    WorldContextObject:Object Reference=@n1
    ActorClass:Actor Class Reference=/FigmentEnvironment/Blueprints/Cliffs/BP__Cliff_Figment.BP__Cliff_Figment_C
   ) -> (then, OutActors:"Array of BP  Cliff Figment Object References")
   > n3 = For Each Loop(Array:"Array of BP  Cliff Figment Object References"=@n2) -> (Array Element:"BP  Cliff Figment Object Reference")
   > n4 = Execute Python Script!(
    execute
    PythonScript="import unreal\r\nno_need_to_save = False\r\nTextureDataArray = geo.get_editor_property('TextureData')\r\nprint(TextureDataArray)\r\nif(TextureDataArray[3] == texture_data):\r\n    no_need_to_save = True\r\nelse:\r\n    TextureDataArray[3] = texture_data\r\n    geo.set_editor_property('TextureData', TextureDataArray)"
    geo:"BP  Cliff Figment Object Reference"=@n3
    texture_data:Building Texture Data Object Reference=Get data() -> (data:Building Texture Data Object Reference)
   ) -> (then, ReturnValue, no_need_to_save:bool=~) {inputs [geo,texture_data], outputs [no_need_to_save]}
   > n5 = Branch(Condition=@n4.no_need_to_save)
   Add(
    execute=@n5.else
    TargetArray:"Array of BP  Cliff Figment Object References"=Get Cliff Actors to Save() -> (Cliff Actors to Save:"Array of BP  Cliff Figment Object References")
    NewItem:"BP  Cliff Figment Object Reference (by ref)"=@n3
   ) {is=title K2Node_CallArrayFunction}
  ]}
 AssignTextureDataToCliffActorCreateIfMissing {nodes [
   n1 = AssignTextureDataToCliffActorCreateIfMissing!() -> (then, CliffActor:"BP  Cliff Figment Object Reference") {is=title K2Node_FunctionEntry}
   > n2 = Execute Python Script!(execute, PythonScript="import unreal\r\ntexture_data = (geo.get_editor_property('TextureData')[0])", geo=@n1) -> (then, ReturnValue, texture_data:Object Reference) {inputs [geo], outputs [texture_data]}
   n3 = GetDisplayName(Object=GetPackageForObject!(Object:Object Reference=@n2.texture_data) -> (ReturnValue:Package Object Reference))
   > n4 = Branch(Condition=IsEmpty(InString=@n3))
   > PrintString(InString=Nothing Assigned)
   n5 = GetEditorWorld(execute=@n4.else, self=Get UnrealEditorSubsystem())
   > n6 = StripRightMostFolderFromPath_Util!(execute, Path:String=@n3, WorldContextObject:Object Reference=@n5, Strip Count:int=1, __WorldContext:Object Reference=@n5) -> (then, FinalPath1:String, AssetName:String)
   n7 = MakeLiteralString(Value=/TextureData/S03)
   n8 = MakeLiteralString(Value=/FigmentEnvironment/Blueprints/Cliffs/Materials)
   > n9 = Branch(Condition="Equal, Case Insensitive (String)"(A=Append(A=@n8, B=@n7), B=@n6))
   > PrintString(InString=already assigned)
   n10 = Append(A=@n6.AssetName, B=MakeLiteralString(Value=_S03))
   n11 = Append(A=@n6, B=MakeLiteralString(Value=/S03/))
   n12 = Append(A=@n11, B=@n10)
   n13 = DoesAssetExist(execute=@n9.else, AssetPath=@n12)
   > n14 = Branch(Condition=@n13)
   > PrintString(InString=TD asset existsi)
   > n15 = GetAssetByObjectPath!(
    execute
    self:Asset Registry Interface=GetAssetRegistry()
    ObjectPath:Name=String To Name(InString=Append!(A=@n12, B=., C:String=@n10) -> (ReturnValue))
    bIncludeOnlyOnDiskAssets:bool=false
   ) -> (then, ReturnValue:Asset Data Structure)
   > n16 = Cast To BuildingTextureData!(execute, Object=GetAsset(InAssetData=@n15)) -> (then, CastFailed, AsBuilding Texture Data:Building Texture Data Object Reference) {is=title K2Node_DynamicCast}
   > n17 = Execute Python Script!(
    execute
    PythonScript="import unreal\r\nmaterial_interface = tex_data.get_editor_property('OverrideMaterial')"
    tex_data:Building Texture Data Object Reference=@n16
   ) -> (then, ReturnValue, material_interface:Material Instance Constant Object Reference) {inputs [tex_data], outputs [material_interface]}
   > n18 = Set material_on_TD(material_on_TD:Material Instance Constant Object Reference=@n17.material_interface) -> (Output_Get:Material Instance Constant Object Reference)
   > PrintString(InString=GetDisplayName(Object=Get Parent!(self:Material Instance Object Reference=@n18) -> (Parent:Material Interface Object Reference)))
   > n19 = Execute Python Script!(
    execute
    PythonScript="import unreal\r\nno_need_to_save = False\r\nTextureDataArray = geo.get_editor_property('TextureData')\r\nprint(TextureDataArray)\r\nif(TextureDataArray[0] == texture_data):\r\n    no_need_to_save = True\r\nelse:\r\n    TextureDataArray[0] = texture_data\r\n    geo.set_editor_property('TextureData', TextureDataArray)"
    geo:"BP  Cliff Figment Object Reference"=@n1
    texture_data:Building Texture Data Object Reference=@n16
   ) -> (then, ReturnValue, no_need_to_save:bool=~) {inputs [geo,texture_data], outputs [no_need_to_save]}
   > n20 = Branch(Condition=@n19.no_need_to_save)
   Return Node!(execute=[@n39.then, @n20.then], ActorToSave:"BP  Cliff Figment Object Reference"=@n1)
   PrintString(execute=@n14.else, InString=Append!(A="TD asset doesnt exist, duplicating", B=@n12) -> (ReturnValue))
   > n21 = GetAssetTools()
   > n22 = DuplicateAsset!(execute, self:Asset Tools Interface=@n21, AssetName:String=@n10, PackagePath:String=@n11, OriginalObject:Object Reference=@n2.texture_data) -> (then, ReturnValue:Object Reference)
   > n23 = Execute Python Script!(
    execute
    PythonScript="import unreal\r\nmaterial_interface = tex_data.get_editor_property('OverrideMaterial')"
    tex_data:Object Reference=@n22
   ) -> (then, ReturnValue, material_interface:Material Instance Constant Object Reference) {inputs [tex_data], outputs [material_interface]}
   > n24 = Set material_on_TD(material_on_TD=@n23.material_interface)
   n25 = GetDisplayName(Object=GetPackageForObject(Object=@n24))
   > PrintString(InString=@n25)
   > n26 = GetEditorWorld(self=Get UnrealEditorSubsystem())
   > n27 = StripRightMostFolderFromPath_Util(Path=@n25, WorldContextObject=@n26, Strip Count=1, __WorldContext=@n26)
   n28 = Append(A=@n8, B=MakeLiteralString(Value=/S03/))
   n29 = Append(A=@n27.AssetName, B=MakeLiteralString(Value=_S03))
   > n30 = DoesAssetExist(AssetPath=Append(A=@n28, B=@n29))
   > n31 = Branch(Condition=@n30)
   n32 = GetAssetTools(execute=@n31.else)
   > n33 = DuplicateAsset(self=@n32, AssetName=@n29, PackagePath=@n28, OriginalObject=@n24)
   > n34 = Cast To MaterialInstanceConstant(Object=@n33)
   > n35 = Branch(Condition="Equal (Object)"(A=Get Parent(self=@n34), B=/Game/Athena/Environments/Landscape/Material/Athena_Terrain_Mat_Biome))
   > n36 = SetMaterialInstanceParent!(
    execute
    Instance:Material Instance Constant Object Reference=@n34
    NewParent:Material Interface Object Reference=Get material_orig() -> (material_orig:Material Interface Object Reference)
   ) -> (then)
   SaveLoadedAsset(execute=[@n36.then, @n35.else], AssetToSave=@n33, bOnlyIfIsDirty=true)
   > n37 = Cast To BuildingTextureData(Object=@n22)
   > Set OverrideMaterial!(execute, OverrideMaterial:Material Interface Object Reference=@n34, self:Building Texture Data Object Reference=@n37) -> (then, Output_Get:Material Interface Object Reference)
   > SaveLoadedAsset(AssetToSave=@n37, bOnlyIfIsDirty=false)
   > n38 = Execute Python Script!(
    execute
    PythonScript="import unreal\r\nno_need_to_save = False\r\nTextureDataArray = geo.get_editor_property('TextureData')\r\nprint(TextureDataArray)\r\nif(TextureDataArray[0] == texture_data):\r\n    no_need_to_save = True\r\nelse:\r\n    TextureDataArray[0] = texture_data\r\n    geo.set_editor_property('TextureData', TextureDataArray)"
    geo:"BP  Cliff Figment Object Reference"=@n1
    texture_data:Building Texture Data Object Reference=@n37
   ) -> (then, ReturnValue, no_need_to_save:bool=~) {inputs [geo,texture_data], outputs [no_need_to_save]}
   > n39 = Branch(Condition=@n38.no_need_to_save)
   Cast To MaterialInstanceConstant(Object:Wildcard)
  ]}
 BuildMeshMap {nodes [
   n1 = Keys!(
    execute
    TargetMap:Map of Static Mesh Object References to Strings=Get mapOfMeshesToBPs() -> (mapOfMeshesToBPs:Map of Static Mesh Object References to Strings)
   ) -> (then, Keys:Array of Static Mesh Object References)
   > n2 = For Each Loop(Array:Array of Static Mesh Object References=@n1) -> (Array Element:Static Mesh Object Reference)
   n3 = Values!(execute, TargetMap:Map of Static Mesh Object References to Strings=Get mapOfMeshesToBPs()) -> (then, Values:Array of Strings)
   PrintString(
    execute=@n2.LoopBody
    InString=Append!(
     A=GetDisplayName(Object=@n2)
     B=": "
     C:String="Get (a ref)"(Array:Array of Strings=@n3, Dimension 1=@n2.Array Index) -> (Output:"String (by ref)")
    ) -> (ReturnValue)
   )
   BuildMeshMap() {is=title K2Node_FunctionEntry}
   > n4 = Sequence!(execute) -> (then_0, then_1, then_2:Exec)
   > Clear!(execute, TargetMap:Map of Static Mesh Object References to Strings=Get mapOfMeshesToBPs()) -> (then) {is=title K2Node_CallFunction}
   > Clear(TargetMap:Map of Strings to Strings=Get mapOfMeshNamesToBp() -> (mapOfMeshNamesToBp:Map of Strings to Strings))
   > n5 = GetAssetsByPath!(
    execute
    self:Asset Registry Interface=GetAssetRegistry()
    PackagePath:Name=String To Name(InString=MakeLiteralString(Value=/FigmentEnvironment/Blueprints/Cliffs/))
    bRecursive:bool=false
    bIncludeOnlyOnDiskAssets:bool=true
   ) -> (then, OutAssetData:Array of Asset Data Structures, ReturnValue:bool)
   > n6 = For Each Loop(Array:Array of Asset Data Structures=@n5) -> (Array Element:Asset Data Structure)
   > n7 = GetBlueprintAsset(Object=GetAsset(InAssetData=@n6))
   > n8 = GeneratedClass!(execute, BlueprintObj:Blueprint Object Reference=@n7) -> (then, ReturnValue:Object Class Reference)
   > n9 = GetClassDefaultObject!(execute, Class:Object Class Reference=@n8) -> (then, ReturnValue:Object Reference)
   > n10 = Cast To BP__Cliff_Figment!(execute, Object=@n9) -> (then, CastFailed, "AsBP  Cliff Figment":"BP  Cliff Figment Object Reference") {is=title K2Node_DynamicCast}
   n11 = Get StaticMesh(self=Get StaticMeshComponent(self=@n10))
   n12 = GetDisplayName(Object=@n11)
   > n13 = Branch(Condition=IsEmpty(InString=@n12))
   n14 = "To String (Name)"(InName=Break Asset Data(AssetData=@n6))
   > PrintString(InString=Append!(A="Map Build Empty: ", B=@n12, C="->", D:String=@n14) -> (ReturnValue))
   Add!(execute=@n13.else, TargetMap:Map of Strings to Strings=Get mapOfMeshNamesToBp(), Key:"String (by ref)"=@n12, Value:"String (by ref)"=@n14) -> (then) {is=title K2Node_CallFunction}
   > Add(
    TargetMap:Map of Static Mesh Object References to Strings=Get mapOfMeshesToBPs()
    Key:"Static Mesh Object Reference (by ref)"=@n11
    Value=@n14
   )
   n15 = Keys(execute=@n4.then_1, TargetMap=Get mapOfMeshesToBPs())
   > Clear!(
    execute
    TargetSet:Set of Static Mesh Object References=Get staticMeshSetCheckDuplicattes() -> (staticMeshSetCheckDuplicattes:Set of Static Mesh Object References)
   ) -> (then)
   n16 = Get staticMeshSetCheckDuplicattes()
   > Add Items!(execute, TargetSet:Set of Static Mesh Object References=@n16, NewItems:Array of Static Mesh Object References=@n15) -> (then)
   > n17 = Branch(Condition="Equal (Integer)"(
     A=Length!(TargetSet:Set of Static Mesh Object References=@n16) -> (ReturnValue:int=0) {is=title K2Node_CallFunction}
     B=Length(TargetArray:Array of Static Mesh Object References=@n15) {is=title K2Node_CallArrayFunction}
    ))
   > PrintString(InString=No static Mesh Duplicate blueprints found)
   PrintString(execute=@n17.else, InString="Ahhhhhh duplicated!\r\n######################################################")
   Return Node!(execute=@n4.then_2, mapOfMeshNamesToBp:Map of Strings to Strings=Get mapOfMeshNamesToBp())
   PrintString(InString=Hello)
   Remove!(execute, TargetMap:Map of Strings to Strings=Get mapOfMeshNamesToBp(), Key:"String (by ref)") -> (then, ReturnValue:bool)
  ]}
 ClearTags {nodes [
   n1 = GetEditorWorld(execute=@n2.then, self=Get UnrealEditorSubsystem())
   n2 = ClearTags!() -> (then, WorldContextObject:Object Reference) {is=title K2Node_FunctionEntry}
   n3 = GetAllActorsOfClass(
    execute=@n1.then
    WorldContextObject=SelectObject!(A:Object Reference=@n2, B:Object Reference=@n1, bSelectA:bool=IsValid(Object=@n2)) -> (ReturnValue:Object Reference)
    ActorClass=&BP__Cliff_Figment_C
   )
   n4 = For Each Loop(Exec=@n8.then, Array:Array of Actor Object References=@n3) -> (Array Element:Actor Object Reference)
   > n5 = Cast To BP__Cliff_Figment(Object=@n4)
   n6 = PrintString(InString=GetDisplayName(Object=@n5))
   ResetEditorProperty!(
    execute=[@n6.then, @n5.then]
    Object:Object Reference=@n5
    PropertyName:Name=Tags
    ChangeNotifyMode:EPropertyAccessChangeNotifyMode Enum=Never
   ) -> (then, ReturnValue:bool)
   > Add(TargetArray=Get Cliff Actors to Save(), NewItem=@n5) {is=title K2Node_CallArrayFunction}
   n7 = Get Cliff Actors to Save()
   SaveLoadedAssets(execute=@n9.then, AssetsToSave=@n7, bOnlyIfIsDirty=false)
   n8 = Clear(execute=@n3.then, TargetArray:"Array of BP  Cliff Figment Object References"=Get Cliff Actors to Save()) {is=title K2Node_CallArrayFunction}
   n9 = CheckoutLoadedAssets(execute=@n4.Completed, AssetsToCheckout=@n7)
  ]}
 Create BP Asset From Fort Static Mesh {nodes [
   n1 = Create BP Asset From Fort Static Mesh!() -> (then, Fort Static Mesh Actor:Static Mesh Actor Object Reference) {is=title K2Node_FunctionEntry}
   > GetAssetTools()
   n2 = Get StaticMesh(self=Get StaticMeshComponent(self=@n1))
   n3 = Append!(A=MakeLiteralString(Value=&Cliffs/), B=Append(A=BP_, B=GetDisplayName(Object=@n2))) -> (ReturnValue)
   > n4 = CreateBlueprintAssetWithParent!(
    execute
    AssetPath:String=@n3
    ParentClass:Object Class Reference=&BP__Cliff_Figment_C
   ) -> (then, ReturnValue:Blueprint Object Reference)
   > n5 = LoadBlueprintClass!(execute, AssetPath:String=@n3) -> (then, ReturnValue:Object Class Reference)
   > n6 = GetClassDefaultObject(Class=@n5)
   > n7 = Cast To BP__Cliff_Figment(Object=@n6)
   > SetStaticMesh(self=Get StaticMeshComponent(self=@n7), NewMesh=@n2)
   > SaveAsset!(execute, AssetToSave:String=@n3, bOnlyIfIsDirty:bool=true) -> (then, ReturnValue:bool)
   PrintString(execute=@n7.CastFailed, InString=Nah)
   GetDisplayName()
   GetBlueprintAsset(Object=@n4)
   PrintString(InString=IS CLIFF)
   GetSelectionSet!(execute) -> (then, ReturnValue:Array of Actor Object References)
   Add!(
    execute
    ElementList:Typed Element List Structure=Create Element List!(
     execute
     Registry:Typed Element Registry Object Reference=Get Default Typed Element Registry!() -> (ReturnValue:Typed Element Registry Object Reference) {is=title K2Node_CallFunction}
    ) -> (then, ReturnValue:Typed Element List Structure)
    ElementHandle:"Script Typed Element Handle Structure (by ref)"
   ) -> (then, ReturnValue:bool) {is=title K2Node_CallFunction}
   GetClass!(InAssetData:"Asset Data Structure (by ref)") -> (ReturnValue:Object Class Reference)
   Get Element LIst() -> (Element LIst:Typed Element List Structure)
  ]}
 DeleteFoliageActors {nodes [
   n1 = Get actorsToSAve2() -> (actorsToSAve2:Array of Object References)
   SaveLoadedAssets!(
    execute
    self:Editor Asset Subsystem Object Reference=Get EditorAssetSubsystem() -> (ReturnValue:Editor Asset Subsystem Object Reference) {is=title K2Node_GetEditorSubsystem}
    AssetsToSave=@n1
    bOnlyIfIsDirty=false
   ) -> (then, ReturnValue)
   > EndTransaction()
   n2 = Get ExternalFoliageActorExternalPathsToKill() -> (ExternalFoliageActorExternalPathsToKill:Array of Strings)
   n3 = For Each Loop(Array:Array of Strings=@n2) -> (Array Element:String)
   n4 = GetAssetsByPackageName!(
    execute
    self:Asset Registry Interface=GetAssetRegistry()
    PackageName:Name=String To Name(InString=@n3)
    bIncludeOnlyOnDiskAssets:bool=false
    bSkipARFilteredAssets:bool=true
   ) -> (then, OutAssetData:Array of Asset Data Structures, ReturnValue:bool)
   n5 = "Get (a copy)"(Array:Array of Asset Data Structures=@n4) -> (Output:Asset Data Structure)
   > PrintString(InString="To String (Name)"(InName=Break Asset Data(AssetData=@n5)))
   n6 = Get ExternalFoliageActorsToKill() -> (ExternalFoliageActorsToKill:Array of Strings)
   n7 = CheckOutFiles!(execute, InFiles:Array of Strings=@n6, bSilent:bool=true) -> (then, ReturnValue:bool)
   n8 = Get FoliageActorsToKill() -> (FoliageActorsToKill:Array of Actor Object References)
   DestroyActors!(
    execute=[@n7.then, @n10.then_2]
    self:Editor Actor Subsystem Object Reference=Get EditorActorSubsystem()
    ActorsToDestroy:Array of Actor Object References=@n8
   ) -> (then, ReturnValue:bool)
   DeleteFoliageActors() {is=title K2Node_FunctionEntry}
   > n9 = GetEditorWorld(self=Get UnrealEditorSubsystem())
   > Clear(TargetArray:Array of Object References=Get actorsToSAve2())
   > BeginTransaction(Context=Mod, Description="LOCTEXT(\"Mod\")")
   > n10 = Sequence!(execute) -> (then_0, then_1, then_2, then_3:Exec)
   > n11 = GetAllActorsOfClass(WorldContextObject=@n9, ActorClass=/Script/Foliage.InstancedFoliageActor) -> (OutActors:Array of Instanced Foliage Actor Object References)
   > n12 = Execute Python Script!(
    execute
    PythonScript="import unreal\r\narr = []\r\nfor i in actors:\r\n    arr.append(i.get_package())"
    Actors:Array of Instanced Foliage Actor Object References=@n11
   ) -> (then, ReturnValue, arr:Array of Package Object References) {inputs [Actors], outputs [arr]}
   > Clear(TargetArray:Array of Actor Object References=Get FoliageActorsToKill())
   > Clear(TargetArray:Array of Strings=Get ExternalFoliageActorsToKill())
   > n13 = For Each Loop(Array:Array of Instanced Foliage Actor Object References=@n11) -> (Array Element:Instanced Foliage Actor Object Reference)
   n14 = Get Object Path String(Object=@n13)
   n15 = ParseIntoArray(SourceString=@n14, Delimiter=/)
   > n16 = Branch(Condition="Equal Exactly (String)"(A="Get (a copy)"(Array:Array of Strings=@n15) -> (Output:String), B=Temp))
   n17 = GetDisplayName(Object=GetPackageForObject(Object=@n13))
   Add(execute=@n16.else, TargetArray:Array of Strings=Get ExternalFoliageActorsToKill(), NewItem:"String (by ref)"=@n17) {is=title K2Node_CallArrayFunction}
   > Add(TargetArray=Get ExternalFoliageActorExternalPathsToKill(), NewItem=GetSystemPath!(Object:Object Reference=@n13) -> (ReturnValue:String))
   > Add(TargetArray:Array of Actor Object References=Get FoliageActorsToKill(), NewItem:"Actor Object Reference (by ref)"=@n13)
   n18 = GetPackageForObject(Object=@n13)
   > Add(TargetArray:Array of Object References=Get actorsToSAve2(), NewItem:"Object Reference (by ref)"=@n18)
   > PrintString(InString=@n17)
   SavePackages!(execute=@n10.then_3, PackagesToSave:Array of Package Object References=@n12.arr, bOnlyDirty:bool=false) -> (then, ReturnValue:bool)
   n19 = For Each Loop(Array:Array of Strings=@n6) -> (Array Element:String)
   > n20 = GetAssetsByPath(self=GetAssetRegistry(), PackagePath=String To Name(InString=@n19), bRecursive=false, bIncludeOnlyOnDiskAssets=false)
   n21 = "Get (a copy)"(Array:Array of Asset Data Structures=@n20) -> (Output:Asset Data Structure)
   Break Asset Data(AssetData=@n21)
   Add(
    TargetArray:Array of Package Object References=Make Array("[0]":Package Object Reference) -> (Array:Array of Package Object References)
    NewItem:"Package Object Reference (by ref)"=GetPackageForObject(Object="Get (a copy)"(Array:Array of Instanced Foliage Actor Object References=@n11) -> (Output:Instanced Foliage Actor Object Reference))
   )
   Set Array Elem!(execute, TargetArray=@n2, Index:int=@n3.Array Index, Item:"String (by ref)", bSizeToFit:bool=false) -> (then)
   SaveLoadedAsset(bOnlyIfIsDirty=true)
   SavePackages(bOnlyDirty=false)
   GetAssetByObjectPath(ObjectPath=None, bIncludeOnlyOnDiskAssets=false)
   GetSystemPath(Object=GetAsset(InAssetData=@n21))
   SavePackages(bOnlyDirty=false)
   GetDisplayName(Object=@n13)
   CheckoutLoadedAssets(AssetsToCheckout=@n1)
   SavePackages(PackagesToSave=@n12.arr, bOnlyDirty=false)
   GetDirtyContentPackages!(execute) -> (then, OutDirtyPackages:Array of Package Object References)
   "To String (Name)"(InName=GetFolderPath!(self:Actor Object Reference=@n13) -> (ReturnValue:Name=None))
   SavePackages(bOnlyDirty=false)
   CheckOutFile!(execute=@n3.LoopBody, InFile:String=@n3, bSilent:bool=false) -> (then, ReturnValue:bool)
   CheckOutFiles(InFiles=@n2, bSilent=false)
   SaveLoadedAssets!(execute, AssetsToSave=@n8, bOnlyIfIsDirty=false) -> (then, ReturnValue)
   Return Node!(execute, Package Array To Save:Array of Package Object References=@n12.arr)
   Reroute Node(InputPin:String=@n14) -> (OutputPin:String)
   Make Array("[0]"=@n18)
   Get actorsToSAve2()
   Get actorsToSAve2()
   Get FoliageActorsToKill()
  ]}
 DeleteFoliageActorsClean {nodes [
   n1 = DeleteFoliageActorsClean!() -> (then, WorldContextObject:Object Reference) {is=title K2Node_FunctionEntry}
   > n2 = GetEditorWorld(self=Get UnrealEditorSubsystem())
   > Clear(TargetArray:Array of Object References=Get actorsToSAve2())
   > n3 = Sequence()
   > n4 = GetAllActorsOfClass(WorldContextObject=SelectObject(A=@n1, B=@n2, bSelectA=IsValid(Object=@n1)), ActorClass=&InstancedFoliageActor)
   > n5 = Execute Python Script(PythonScript="import unreal\r\narr = []\r\nfor i in actors:\r\n    arr.append(i.get_package())", Actors=@n4) {inputs [Actors], outputs [arr]}
   > Clear(TargetArray:Array of Actor Object References=Get FoliageActorsToKill())
   > Clear(TargetArray:Array of Strings=Get ExternalFoliageActorsToKill())
   > n6 = For Each Loop(Array:Array of Instanced Foliage Actor Object References=@n4) -> (Array Element:Instanced Foliage Actor Object Reference)
   n7 = ParseIntoArray(SourceString=Get Object Path String(Object=@n6), Delimiter=/)
   > n8 = Branch(Condition="Equal Exactly (String)"(A="Get (a copy)"(Array:Array of Strings=@n7) -> (Output:String), B=Temp))
   n9 = GetDisplayName(Object=GetPackageForObject(Object=@n6))
   Add(execute=@n8.else, TargetArray:Array of Strings=Get ExternalFoliageActorsToKill(), NewItem:"String (by ref)"=@n9)
   > Add(TargetArray=Get ExternalFoliageActorExternalPathsToKill(), NewItem=GetSystemPath(Object=@n6))
   > Add(TargetArray:Array of Actor Object References=Get FoliageActorsToKill(), NewItem:"Actor Object Reference (by ref)"=@n6)
   > Add(TargetArray:Array of Object References=Get actorsToSAve2(), NewItem:"Object Reference (by ref)"=GetPackageForObject(Object=@n6))
   > PrintString(InString=@n9)
   DestroyActors(execute=@n3.then_2, self=Get EditorActorSubsystem(), ActorsToDestroy=Get FoliageActorsToKill())
   SavePackages(execute=@n3.then_3, PackagesToSave=@n5.arr, bOnlyDirty=false)
   SavePackages(bOnlyDirty=false)
   Get actorsToSAve2()
  ]}
 Dup Texture Data and Material Instances of All Cliff Actors {nodes [
   n1 = GetEditorWorld(execute=@n4.then, self=Get UnrealEditorSubsystem())
   n2 = GetAllActorsOfClass(WorldContextObject=@n1, ActorClass=/Script/FortniteGame.BuildingLevelInstance) -> (OutActors:Array of Building Level Instance Object References)
   > n3 = For Each Loop(Array:Array of Building Level Instance Object References=@n2) -> (Array Element:Building Level Instance Object Reference)
   > PrintString(InString=GetDisplayName(Object=@n3))
   > PrintString(InString="To String (Boolean)"(InBool=IsLoaded!(self:Level Instance Interface Interface=@n3) -> (ReturnValue:bool)))
   n4 = Dup Texture Data and Material Instances of All Cliff Actors() {is=title K2Node_FunctionEntry}
   n5 = GetAllActorsOfClass(execute=@n1.then, WorldContextObject=@n1, ActorClass=&BP__Cliff_Figment_C) -> (OutActors:"Array of BP  Cliff Figment Object References")
   n6 = For Each Loop(Exec=@n10.then, Array:Array of Actor Object References=@n5) -> (Array Element:Actor Object Reference)
   n7 = Cast To BP__Cliff_Figment(execute=@n11.then, Object=@n6)
   > n8 = AssignTextureDataToCliffActorCreateIfMissing!(execute, self, CliffActor:"BP  Cliff Figment Object Reference"=@n7) -> (then, ActorToSave:"BP  Cliff Figment Object Reference") {is=title K2Node_CallFunction}
   > Add(
    TargetArray:"Array of BP  Cliff Figment Object References"=Get Cliff Actors to Save()
    NewItem:"BP  Cliff Figment Object Reference (by ref)"=@n8
   )
   n9 = Get Cliff Actors to Save()
   SaveLoadedAssets(execute=@n12.then, AssetsToSave=@n9, bOnlyIfIsDirty=true)
   n10 = Clear(execute=@n5.then, TargetArray:"Array of BP  Cliff Figment Object References"=Get Cliff Actors to Save())
   n11 = PrintString(execute=@n6.LoopBody, InString=GetDisplayName(Object=@n6))
   n12 = CheckoutLoadedAssets(execute=@n6.Completed, AssetsToCheckout=@n9)
  ]}
 EnsureMaterialAndTextureDataExist {nodes [
   n1 = EnsureMaterialAndTextureDataExist!() -> (then, Mesh Material:Object Reference) {is=title K2Node_FunctionEntry}
   n2 = GetDisplayName(Object=@n1)
   > PrintString(InString=@n2)
   > n3 = TextureAssetPathPrep!(execute, self, Material Asset Name:String=@n2) -> (then, "Material Full  Path":String, TD Asset Full Path:String, TD Asset Name:String, Material Folder:String, TD Asset Folder:String)
   > PrintString(InString=Append!(A="does ", B=@n3, C:String=exist) -> (ReturnValue))
   > n4 = DoesAssetExist(AssetPath=@n3)
   > n5 = Branch(Condition=@n4)
   > n6 = LoadAsset!(execute, self:Editor Asset Subsystem Object Reference=Get EditorAssetSubsystem(), AssetPath=@n3) -> (then, ReturnValue)
   > n7 = Cast To MaterialInterface!(execute, Object=@n6) -> (then, CastFailed, AsMaterial Interface:Material Interface Object Reference) {is=title K2Node_DynamicCast}
   > n8 = Set CurrentMaterialForStaticMesh(CurrentMaterialForStaticMesh:Material Interface Object Reference=@n7) -> (Output_Get:Material Interface Object Reference)
   PrintString(execute=[@n8.then, @n23.then], InString=Append(A="does ", B=@n3.TD Asset Full Path, C=exist))
   > n9 = DoesAssetExist(AssetPath=@n3.TD Asset Full Path)
   > n10 = Branch(Condition=@n9)
   > n11 = LoadAsset(self=Get EditorAssetSubsystem(), AssetPath=@n3.TD Asset Full Path)
   > n12 = Cast To BuildingTextureData(Object=@n11)
   > n13 = Set CurrentTDForStaticMesh(CurrentTDForStaticMesh:Building Texture Data Object Reference=@n12) -> (Output_Get:Building Texture Data Object Reference)
   n14 = Get CurrentTDForStaticMesh() -> (CurrentTDForStaticMesh:Building Texture Data Object Reference)
   Set OverrideMaterial(
    execute=[@n13.then, @n18.then]
    OverrideMaterial=Get CurrentMaterialForStaticMesh() -> (CurrentMaterialForStaticMesh:Material Interface Object Reference)
    self=@n14
   )
   > SaveLoadedAsset(AssetToSave=@n14, bOnlyIfIsDirty=true)
   > Return Node!(execute, " Texture Data To Assign":Building Texture Data Object Reference=@n14)
   n15 = GetAssetTools(execute=@n5.else)
   n16 = CreateAsset(
    execute=@n10.else
    self=@n15
    AssetName=@n3.TD Asset Name
    PackagePath=@n3.TD Asset Folder
    AssetClass=/Script/FortniteGame.BuildingTextureData
   )
   > n17 = Cast To BuildingTextureData(Object=@n16)
   > n18 = Set CurrentTDForStaticMesh(CurrentTDForStaticMesh=@n17)
   n19 = DuplicateAsset(execute=@n15.then, self=@n15, AssetName=@n2, PackagePath=@n3.Material Folder, OriginalObject=@n1)
   > n20 = Cast To MaterialInstanceConstant(Object:Object Reference=@n19)
   n21 = SetMaterialInstanceParent(execute=@n26.then, Instance=@n20, NewParent=/FigmentEnvironment/Materials/Cliff/Athena_Terrain_Mat_Biome_Figment_Inst)
   SaveLoadedAsset(execute=[@n21.then, @n26.else], bOnlyIfIsDirty=true)
   > n22 = Cast To MaterialInterface(Object=@n19)
   > n23 = Set CurrentMaterialForStaticMesh(CurrentMaterialForStaticMesh=@n22)
   n24 = Get Parent(self=@n20)
   n25 = OR Boolean!(
    A="Equal (Object)"(A=@n24, B=&Athena_Terrain_Mat_Biome)
    B="Equal (Object)"(A=@n24, B=/Game/Athena/Environments/Landscape/Material/Athena_Terrain_Mat_Biome_Inst)
    C:bool="Equal (Object)"(A=@n24, B=/Game/Packages/Fortnite_UniqueMaterials_1/DS_Fortnight/Terrain_Mat)
   ) -> (ReturnValue)
   n26 = Branch(execute=@n20.then, Condition=@n25)
  ]}
 EventGraph {nodes [
   CheckoutLoadedAssets()
   > SaveLoadedAssets(bOnlyIfIsDirty=true)
   "On Clicked (FixUpInLevel)"() {is=title K2Node_ComponentBoundEvent}
   > PaintConcreteVertex1000!(execute, self, WorldContextObject:Object Reference) -> (then)
   "On Clicked (ConvertSMinPathtoBP)"() {is=title K2Node_ComponentBoundEvent}
   > GetEditorWorld(self=Get UnrealEditorSubsystem())
   n1 = "GetText (Text Box)"!(self:Text Box Object Reference=Get SourceTextBox() -> (SourceTextBox:Editor Utility Editable Text Box Object Reference)) -> (ReturnValue:Text)
   > n2 = GetAssetsByPath(self=GetAssetRegistry(), PackagePath=String To Name(InString="To String (Text)"(InText=@n1)), bRecursive=true, bIncludeOnlyOnDiskAssets=true)
   > Sequence!(execute) -> (then_0, then_1)
   > n3 = For Each Loop(Array:Array of Asset Data Structures=@n2) -> (Array Element:Asset Data Structure)
   n4 = GetAsset(InAssetData=@n3)
   > Set Current Level To Process(Current Level To Process:Object Reference=@n4) -> (Output_Get:Object Reference)
   n5 = Break Asset Data(AssetData=@n3)
   > n6 = Branch(Condition=Contains(SearchIn="To String (Name)"(InName=@n5), Substring=ExternalObjects, bUseCase=true))
   n7 = Get Class Display Name(Class=GetClass(InAssetData=@n3))
   PrintString(execute=@n6.else, InString=@n7)
   > n8 = Branch(Condition="Equal Exactly (String)"(A=@n7, B=World))
   > PrintString(InString="To String (Name)"(InName=@n5))
   > n9 = GetEditorWorld(self=Get UnrealEditorSubsystem())
   > n10 = SpawnActorFromClass!(
    execute
    self:Editor Actor Subsystem Object Reference=Get EditorActorSubsystem()
    ActorClass:Actor Class Reference=/Script/Engine.LevelInstance
    Location:Vector="0, 0, 0"
    Rotation:Rotator
    bTransient:bool=true
   ) -> (then, ReturnValue:Level Instance Object Reference)
   > Set LevelInstanceActor(LevelInstanceActor:Level Instance Object Reference=@n10) -> (Output_Get:Level Instance Object Reference)
   > SetEditorProperty(
    Object=@n10
    PropertyName=WorldAsset
    PropertyValue:Soft Object Reference=Make Soft Reference!(Input:Object Reference=@n4) -> (Output:Soft Object Reference)
    ChangeNotifyMode=Never
   )
   > Clear(TargetArray=Get Cliff Actors to Save())
   > n11 = Sequence!(execute) -> (then_0, then_1, then_2:Exec)
   > LevelInstanceEnterEditMode!(execute, LevelInstance:Level Instance Object Reference=Get LevelInstanceActor() -> (LevelInstanceActor:Level Instance Object Reference)) -> (then, ReturnValue:bool)
   > Sequence!(execute) -> (then_0, then_1, then_2, then_3:Exec, then_4:Exec)
   > n12 = Set Current Level To Process(Current Level To Process=@n4)
   > ReplaceAllStaticMeshesWithExistingBP!(execute, self, WorldContextObject:Object Reference) -> (then)
   LevelInstanceCommit!(execute=@n11.then_1, LevelInstance:Level Instance Object Reference=Get LevelInstanceActor(), bDiscardChanges:bool=true) -> (then, ReturnValue:bool)
   n13 = Get EditorActorSubsystem()
   SelectNothing(execute=@n11.then_2, self:Editor Actor Subsystem Object Reference=@n13)
   > DestroyActor!(execute, self:Editor Actor Subsystem Object Reference=@n13, ActorToDestroy:Actor Object Reference=Get LevelInstanceActor()) -> (then, ReturnValue:bool)
   > CollectGarbage!(execute) -> (then)
   PrintString(execute=@n8.else, InString="Found Object other than World SKIPPING\r\n", TextColor=(R=1,G=0.4929,B=0,A=1))
   "On Clicked (EditorUtilityButton_34)"() {is=title K2Node_ComponentBoundEvent}
   > SetActiveWidgetIndex(self=Get WidgetSwitcher_184() -> (WidgetSwitcher_184:Widget Switcher Object Reference), Index=0)
   "On Clicked (EditorUtilityButton_60)"() {is=title K2Node_ComponentBoundEvent}
   > SetActiveWidgetIndex(self=Get WidgetSwitcher_184(), Index=1)
   "On Clicked (ConvertSMinLeveltoBP)"() {is=title K2Node_ComponentBoundEvent}
   > n14 = GetEditorWorld(self=Get UnrealEditorSubsystem())
   > ReplaceAllStaticMeshesWithExistingBP(WorldContextObject=@n14)
   "On Clicked (DeleteFoliage_1)"() {is=title K2Node_ComponentBoundEvent}
   n15 = "To String (Text)"(InText="GetText (Text Box)"(self=Get SourceTextBox_pg2() -> (SourceTextBox_pg2:Editor Utility Editable Text Box Object Reference)))
   > n16 = For Each LI!(execute:Exec, InputPin:String=@n15) -> (then:Exec, Output_Get:Object Reference) {is=title K2Node_MacroInstance}
   DeleteFoliageActorsClean!(execute=[@n16.then, @n40.then], self, WorldContextObject:Object Reference=@n16) -> (then) {is=title K2Node_CallFunction}
   "On Clicked (ClearTagsInPath)"() {is=title K2Node_ComponentBoundEvent}
   n17 = "To String (Text)"(InText="GetText (Text Box)"(self=Get SourceTextBox_pg2()))
   > n18 = For Each Context!(execute:Exec, InputPin:String=@n17) -> (then:Exec, Output_Get:Object Reference) {is=title K2Node_MacroInstance}
   ClearTags!(execute=[@n18.then, @n41.then], self, WorldContextObject:Object Reference=@n18) -> (then) {is=title K2Node_CallFunction}
   "On Clicked (ClearTagsInLevel)"() {is=title K2Node_ComponentBoundEvent}
   > ClearTags()
   "On Clicked (PaintConcreteVertexColorsRedInLevel)"() {is=title K2Node_ComponentBoundEvent}
   n19 = Get SourceTextBox_pg2()
   > n20 = For Each LI(InputPin="To String (Text)"(InText="GetText (Text Box)"(self=@n19)))
   > PaintConcreteVertex1000(WorldContextObject=@n20)
   "On Clicked (ConverttoBlueprint)"() {is=title K2Node_ComponentBoundEvent}
   > n21 = GetSelectedLevelActors(self=Get EditorActorSubsystem())
   > n22 = Cast To FortStaticMeshActor!(execute, Object="Get (a copy)"(Array:Array of Actor Object References=@n21) -> (Output:Actor Object Reference)) -> (then, CastFailed, AsFort Static Mesh Actor:Fort Static Mesh Actor Object Reference) {is=title K2Node_DynamicCast}
   > Create BP Asset From Fort Static Mesh!(execute, self, Fort Static Mesh Actor:Static Mesh Actor Object Reference=@n22) -> (then) {is=title K2Node_CallFunction}
   > n23 = Branch(Condition=Get SM_to_BP_all() -> (SM_to_BP_all:bool))
   > ReplaceAllStaticMeshesWithExistingBP()
   n24 = BuildMeshMap!(execute=@n23.else, self) -> (then, mapOfMeshNamesToBp:Map of Strings to Strings) {is=title K2Node_CallFunction}
   > Replace Mesh Actor With Existing BP Single!(execute, self, Static Mesh Actor:Static Mesh Actor Object Reference=@n22, mapOfMeshNamesToBp:Map of Strings to Strings=@n24) -> (then)
   "On Clicked (DeleteFoliage)"() {is=title K2Node_ComponentBoundEvent}
   > DeleteFoliageActorsClean()
   "On Property Changed (SinglePropertyView_0)"!() -> (OutputDelegate, then, PropertyName:Name=None) {is=title K2Node_ComponentBoundEvent}
   > PrintString(InString=GetDisplayName(Object=Get MICFrom() -> (MICFrom:Material Instance Constant Object Reference)))
   "On Property Changed (SinglePropertyView)"!() -> (OutputDelegate, then, PropertyName:Name=None) {is=title K2Node_ComponentBoundEvent}
   > PrintString(InString=GetDisplayName(Object=Get MICTo() -> (MICTo:Material Instance Constant Object Reference)))
   "On Clicked (EditorUtilityButton_4)"() {is=title K2Node_ComponentBoundEvent}
   > n25 = GetEditorWorld(self=Get UnrealEditorSubsystem())
   > n26 = GetAllActorsOfClass(WorldContextObject=@n25, ActorClass=/Script/FortniteGame.FortStaticMeshActor) -> (OutActors:Array of Fort Static Mesh Actor Object References)
   > n27 = For Each Loop(Array:Array of Fort Static Mesh Actor Object References=@n26) -> (Array Element:Fort Static Mesh Actor Object Reference)
   n28 = Get StaticMeshComponent(self=@n27)
   > Branch(Condition="Equal (Object)"(
     A=GetMaterial!(self:Primitive Component Object Reference=@n28, ElementIndex:int=0) -> (ReturnValue:Material Interface Object Reference)
     B=Get MICFrom()
    ))
   > n29 = Branch(Condition=IsValid(Object=Get MICTo()))
   > SetMaterial(self=@n28, ElementIndex=0, Material=Get MICTo())
   PrintString(execute=@n29.else, InString=target material not set in UI)
   "On Clicked (EditorUtilityButton_5)"() {is=title K2Node_ComponentBoundEvent}
   > n30 = GetSelectedActors()
   > n31 = Cast To FortStaticMeshActor(Object="Get (a copy)"(Array=@n30))
   > n32 = Cast To MaterialInstanceConstant(Object=GetMaterial(self=Get StaticMeshComponent(self=@n31), ElementIndex=0))
   > Set MICFrom(MICFrom:Material Instance Constant Object Reference=@n32) -> (Output_Get:Material Instance Constant Object Reference)
   "On Clicked (EditorUtilityButton_7)"() {is=title K2Node_ComponentBoundEvent}
   > n33 = GetEditorWorld(self=Get UnrealEditorSubsystem())
   > n34 = GetAllActorsOfClass(WorldContextObject=@n33, ActorClass=&BP__Cliff_Figment_C) -> (OutActors:"Array of BP  Cliff Figment Object References")
   > n35 = For Each Loop(Array:"Array of BP  Cliff Figment Object References"=@n34) -> (Array Element:"BP  Cliff Figment Object Reference")
   > n36 = GetEditorProperty!(execute:Exec, Object:Object Reference=@n35, PropertyName:Name=TextureData) -> (then:Exec, PropertyValue:Array of Building Texture Data Object References, ReturnValue:bool)
   n37 = GetDisplayName(Object=GetMaterial(self=Get StaticMeshComponent(self=@n35), ElementIndex=0))
   n38 = IsValid(Object="Get (a copy)"(Array:Array of Building Texture Data Object References=@n36) -> (Output:Building Texture Data Object Reference))
   > Branch(Condition=AND Boolean(A=Contains(SearchIn=@n37, Substring=Terrain_Mat), B=NOT Boolean(A=@n38)))
   > SetEditorProperty(
    Object=@n35
    PropertyName=TextureData
    PropertyValue:Array of Building Texture Data Object References=Make Array!(
     "[0]":Building Texture Data Object Reference=Get TD Asset() -> (TD Asset:Building Texture Data Object Reference)
     "[1]":Building Texture Data Object Reference="Get (a copy)"(Array=@n36, Dimension 1=1)
     "[2]":Building Texture Data Object Reference="Get (a copy)"(Array=@n36, Dimension 1=2)
     "[3]":Building Texture Data Object Reference="Get (a copy)"(Array=@n36, Dimension 1=3)
    ) -> (Array:Array of Building Texture Data Object References)
   )
   > PrintString(InString=GetDisplayName(Object=@n35))
   n39 = "On Check State Changed (SMConvertApplyToAll)"!() -> (OutputDelegate, then, bIsChecked:bool) {is=title K2Node_ComponentBoundEvent}
   > Set SM_to_BP_all(SM_to_BP_all:bool=@n39.bIsChecked) -> (Output_Get:bool)
   Event PreConstruct()
   > SetObject(self=Get SinglePropertyView_0() -> (SinglePropertyView_0:Single Property View Object Reference), NewObject=Self-Reference())
   > SetObject(self=Get SinglePropertyView() -> (SinglePropertyView:Single Property View Object Reference), NewObject=Self-Reference())
   > SetObject(self=Get SinglePropertyView_1() -> (SinglePropertyView_1:Single Property View Object Reference), NewObject=Self-Reference())
   > SetObject(self=Get SinglePropertyView_2() -> (SinglePropertyView_2:Single Property View Object Reference), NewObject=Self-Reference())
   n40 = For Each Context(InputPin=@n15)
   n41 = For Each LI(InputPin=@n17)
   FixupTextureData_S03!(execute, self, Level Object Reference:Object Reference) -> (then)
   Remove_appended_TDs!(execute, self, Level Object:Object Reference) -> (then)
   Remove_appended_TDs(Level Object=@n12)
   "SetText (Text Box)"!(
    execute
    self:Text Box Object Reference=Get SourceTextBox()
    InText:Text="To Text (String)"(InString=Get TargetPAth() -> (TargetPAth:String))
   ) -> (then)
   PrintString(InString=GetDisplayName())
   Dup Texture Data and Material Instances of All Cliff Actors() {is=title K2Node_CallFunction}
   Dup Texture Data and Material Instances of All Cliff Actors()
   AssignTextureDataSlot4() {is=title K2Node_CallFunction}
   Reassign TextureData with existing!(execute, self, Level Object Reference:Object Reference) -> (then)
   ReplaceAllStaticMeshesWithExistingBP()
   Reassign TextureData with existing(Level Object Reference=@n4)
   ForEachLevelInstanceContext!(execute, self, InString:String) -> (then, Output_Get:Object Reference)
   DeleteFoliageActors!(execute, self) -> (then, Package Array To Save:Array of Package Object References) {is=title K2Node_CallFunction}
   DeleteFoliageActors()
   PrintString(WorldContextObject=@n18, InString=GetDisplayName())
   "Get Class Path Name (Top Level Asset Path)"!(Class:Object Class Reference) -> (ReturnValue:Top Level Asset Path Structure) {is=title K2Node_CallFunction}
   GetAllActorsOfClass() -> (OutActors:Array of Actor Object References)
   ReplaceAllStaticMeshesWithExistingBP()
   BuildMeshMap()
   GetClass(InAssetData=@n3)
   RemoveChild(self=Get CanvasPanel_X() -> (CanvasPanel_X:Canvas Panel Object Reference), Content=Get viewport() -> (viewport:Viewport Object Reference))
   Make Soft Reference(Input=@n4)
   Resolve Soft Reference!(Input:Soft Object Reference=Make Soft Reference(Input=@n4)) -> (Output:Object Reference)
   Event Construct()
   Event Tick()
   Sequence!(execute) -> (then_0, then_1)
   "Get (a copy)"(Array:Array of Actor Object References=GetSelectedActors()) -> (Output:Actor Object Reference)
   Get EditorActorSubsystem()
   Get ContentBrowserDataSubsystem() -> (ReturnValue:Content Browser Data Subsystem Object Reference) {is=title K2Node_GetEditorSubsystem}
   Get LevelEditorSubsystem()
   Reroute Node(InputPin:World Object Reference=@n9) -> (OutputPin:World Object Reference)
   n42 = Get SourceTextBox_pg2()
   For Each Context(InputPin="To String (Text)"(InText="GetText (Text Box)"(self=@n42)))
   Get Cliff Actors to Save()
   Get TD Asset()
  ]}
 FixupTextureData_S03 {nodes [
   n1 = GetSelectedLevelActors(self=Get EditorActorSubsystem())
   > n2 = For Each Loop(Array:Array of Actor Object References=@n1) -> (Array Element:Actor Object Reference)
   > Cast To BP__Cliff_Figment(Object=@n2)
   n3 = GetEditorWorld(execute=@n4.then, self=Get UnrealEditorSubsystem())
   n4 = FixupTextureData_S03!() -> (then, Level Object Reference:Object Reference) {is=title K2Node_FunctionEntry}
   n5 = GetAllActorsOfClass(
    execute=@n3.then
    WorldContextObject=SelectObject(A=@n4, B=@n3, bSelectA=IsValid(Object=@n4))
    ActorClass=&BP__Cliff_Figment_C
   ) -> (OutActors:"Array of BP  Cliff Figment Object References")
   n6 = For Each Loop(Exec=@n58.then, Array=@n5)
   PrintString(InString=GetDisplayName(Object=@n6))
   > n7 = Cast To BP__Cliff_Figment(Object=@n6)
   > n8 = Execute Python Script!(
    execute
    PythonScript="import unreal\r\ntexture_data = (geo.get_editor_property('TextureData')[0])"
    geo:"BP  Cliff Figment Object Reference"=@n7
   ) -> (then, ReturnValue, texture_data:Object Reference) {inputs [geo], outputs [texture_data]}
   n9 = GetDisplayName(Object=GetPackageForObject(Object=@n8.texture_data))
   > n10 = Branch(Condition=IsEmpty(InString=@n9))
   > PrintString(InString=Nothing Assigned)
   n11 = GetEditorWorld(execute=@n10.else, self=Get UnrealEditorSubsystem())
   > n12 = StripRightMostFolderFromPath_Util(Path=@n9, WorldContextObject=@n11, Strip Count=1, __WorldContext=@n11)
   n13 = MakeLiteralString(Value=/Figment_S03/Athena/Environments/TextureData/CliffGround/)
   > n14 = Branch(Condition="Equal, Case Insensitive (String)"(A=@n13, B=@n12))
   > PrintString(InString=already assigned)
   n15 = MakeLiteralString(Value=&CliffGround/)
   n16 = Append!(A=@n12.AssetName, B=_S03) -> (ReturnValue)
   n17 = Append(A=@n15, B=@n16)
   n18 = DoesAssetExist(execute=@n14.else, AssetPath=@n17)
   > n19 = Branch(Condition=@n18)
   > n20 = PrintString(InString=TD asset existsi)
   n21 = GetAssetByObjectPath(
    execute=[@n20.then, @n50.then]
    self=GetAssetRegistry()
    ObjectPath=String To Name(InString=Append!(A=@n17, B=., C:String=@n16) -> (ReturnValue))
    bIncludeOnlyOnDiskAssets=false
   )
   n22 = Cast To BuildingTextureData(execute=[@n21.then, @n49.then], Object=GetAsset(InAssetData=@n21))
   > n23 = Execute Python Script!(
    execute
    PythonScript="import unreal\r\nmaterial_interface = tex_data.get_editor_property('OverrideMaterial')"
    tex_data:Building Texture Data Object Reference=@n22
   ) -> (then, ReturnValue, material_interface:Material Instance Constant Object Reference) {inputs [tex_data], outputs [material_interface]}
   > n24 = Set material_on_TD(material_on_TD=@n23.material_interface)
   n25 = Get Parent(self=@n24)
   > n26 = PrintString(InString=GetDisplayName(Object=@n25))
   n27 = Execute Python Script!(
    execute=[@n26.then, @n51.then]
    PythonScript="import unreal\r\nno_need_to_save = False\r\nTextureDataArray = geo.get_editor_property('TextureData')\r\nprint(TextureDataArray)\r\nif(TextureDataArray[0] == texture_data):\r\n    no_need_to_save = True\r\nelse:\r\n    TextureDataArray[0] = texture_data\r\n    geo.set_editor_property('TextureData', TextureDataArray)"
    geo:"BP  Cliff Figment Object Reference"=@n7
    texture_data:Building Texture Data Object Reference=@n22
   ) -> (then, ReturnValue, no_need_to_save:bool=~) {inputs [geo,texture_data], outputs [no_need_to_save]}
   > n28 = Branch(Condition=@n27.no_need_to_save)
   Add(execute=@n28.else, TargetArray=Get Cliff Actors to Save(), NewItem=@n7)
   PrintString(execute=@n19.else, InString=Append!(A="TD asset doesnt exist, duplicating", B) -> (ReturnValue))
   > n29 = GetAssetTools()
   > n30 = DuplicateAsset(self=@n29, AssetName=@n16, PackagePath=@n15, OriginalObject=@n8.texture_data)
   > n31 = Execute Python Script!(
    execute
    PythonScript="import unreal\r\nmaterial_interface = tex_data.get_editor_property('OverrideMaterial')"
    tex_data:Object Reference=@n30
   ) -> (then, ReturnValue, material_interface:Material Instance Constant Object Reference) {inputs [tex_data], outputs [material_interface]}
   > n32 = Set material_on_TD(material_on_TD=@n31.material_interface)
   n33 = GetDisplayName(Object=GetPackageForObject(Object=@n32))
   > PrintString(InString=@n33)
   > n34 = GetEditorWorld(self=Get UnrealEditorSubsystem())
   > n35 = StripRightMostFolderFromPath_Util(Path=@n33, WorldContextObject=@n34, Strip Count=1, __WorldContext=@n34)
   n36 = MakeLiteralString(Value=/Figment_S03/Athena/Environments/Masterials/CliffGround/)
   n37 = Append(A=@n35.AssetName, B=MakeLiteralString(Value=_S03))
   > n38 = DoesAssetExist(AssetPath=Append(A=@n36, B=@n37))
   > n39 = Branch(Condition=@n38)
   n40 = GetAssetTools(execute=@n39.else)
   > n41 = DuplicateAsset(self=@n40, AssetName=@n37, PackagePath=@n36, OriginalObject=@n32)
   > n42 = Cast To MaterialInstanceConstant(Object=@n41)
   > n43 = Branch(Condition="Equal (Object)"(A=Get Parent(self=@n42), B=&Athena_Terrain_Mat_Biome))
   n44 = Get FigmentParentMAt() -> (FigmentParentMAt:Material Instance Constant Object Reference)
   > Cast To MaterialInterface(Object=SelectObject(A=Get FigmentParentMAt(), B=@n42, bSelectA=IsValid(Object=@n44)))
   > SetMaterialInstanceParent(Instance=@n42, NewParent=@n42)
   > SaveLoadedAsset(AssetToSave=@n41, bOnlyIfIsDirty=true)
   > n45 = Cast To BuildingTextureData(Object=@n30)
   > Set OverrideMaterial(OverrideMaterial=@n42, self=@n45)
   > SaveLoadedAsset(AssetToSave=@n45, bOnlyIfIsDirty=false)
   > n46 = Execute Python Script!(
    execute
    PythonScript="import unreal\r\nno_need_to_save = False\r\nTextureDataArray = geo.get_editor_property('TextureData')\r\nprint(TextureDataArray)\r\nif(TextureDataArray[0] == texture_data):\r\n    no_need_to_save = True\r\nelse:\r\n    TextureDataArray[0] = texture_data\r\n    geo.set_editor_property('TextureData', TextureDataArray)"
    geo:"BP  Cliff Figment Object Reference"=@n7
    texture_data:Building Texture Data Object Reference=@n45
   ) -> (then, ReturnValue, no_need_to_save:bool=~) {inputs [geo,texture_data], outputs [no_need_to_save]}
   > n47 = Branch(Condition=@n46.no_need_to_save)
   Add(execute=@n47.else, TargetArray=Get Cliff Actors to Save(), NewItem=@n7)
   n48 = GetPackageForObject(Object=@n42)
   PrintString(
    execute=@n43.else
    InString=Append(A=GetDisplayName(Object=@n48), B="\r\n########################\r\n is not a child of Athena Material\r\n########################################")
   )
   n49 = PrintString(InString=Hello)
   n50 = PrintString(InString=Hello)
   n51 = Branch(Condition="Equal (Object)"(A=@n25, B=Get material_orig()))
   PrintString(execute=@n51.else, InString=Wrong parent)
   n52 = Is Valid(InputObject=Get Color Map TD override() -> (Color Map TD override:Building Texture Data Object Reference))
   PrintString(execute=@n52.Is Not Valid, InString=Please Assign TD Asset)
   n53 = GetSelectedLevelActors(execute=@n52.Is Valid, self=Get EditorActorSubsystem())
   n54 = Cast To BP__Cliff_Figment(execute=[@n53.then, @n6.LoopBody], Object=@n6)
   > n55 = GetEditorProperty(Object=@n54, PropertyName=TextureData)
   > n56 = Branch(Condition="Equal (Object)"(
     A="Get (a copy)"(Array:Array of Building Texture Data Object References=@n55, Dimension 1=3) -> (Output:Building Texture Data Object Reference)
     B=Get Color Map TD override()
    ))
   SetEditorProperty(
    execute=@n56.else
    Object=@n54
    PropertyName=TextureData
    PropertyValue=Make Array(
     "[0]"="Get (a copy)"(Array=@n55)
     "[1]"="Get (a copy)"(Array=@n55, Dimension 1=1)
     "[2]"="Get (a copy)"(Array=@n55, Dimension 1=2)
     "[3]"=Get Color Map TD override()
    )
   )
   PrintString(InString="TD asset doesnt exist, duplicating")
   GetAssetByObjectPath(ObjectPath=None, bIncludeOnlyOnDiskAssets=false)
   Get Dependencies!(execute, self:Asset Registry Interface, PackageName:Name=None, DependencyOptions:"Asset Registry Dependency Options Structure (by ref)") -> (then, OutDependencies:Array of Names, ReturnValue:bool) {is=title K2Node_CallFunction}
   GetAsset()
   CollectGarbage()
   n57 = Get Cliff Actors to Save()
   SaveLoadedAssets(execute=@n60.then, AssetsToSave=@n57, bOnlyIfIsDirty=true)
   GetSystemPath(Object=@n8.texture_data)
   PrintString(InString=@n9)
   Append!(A, B=., C:String) -> (ReturnValue)
   Append!(A=@n13, B=MakeLiteralString(Value=&S03)) -> (ReturnValue)
   Cast To MaterialInstanceConstant(Object:Wildcard)
   Execute Python Script!(
    execute
    PythonScript="import unreal\r\nTextureData = geo.get_editor_property('TextureData')[0]\r\nOverrideMaterial = TextureData.get_editor_property('OverrideMaterial')\r\nprint(TextureData)"
    Geo:"BP  Cliff Figment Object Reference"=@n7
   ) -> (then, ReturnValue) {inputs [Geo]}
   "Get (a copy)"(Array:Array of Actor Object References=@n53) -> (Output:Actor Object Reference)
   Reroute Node(InputPin:Object Reference=@n41) -> (OutputPin:Object Reference)
   Get data()
   Get material_orig()
   n58 = Clear(execute=@n59.then, TargetArray=Get Actors to Revert() -> (Actors to Revert:"Array of BP  Cliff Figment Object References"))
   n59 = Clear(execute=@n5.then, TargetArray=Get Cliff Actors to Save())
   n60 = CheckoutLoadedAssets(execute=@n6.Completed, AssetsToCheckout=@n57)
  ]}
 ForEachContext {nodes [
   n1 = Inputs!() -> (execute, InputPin:String)
   n2 = Get SourceTextBox()
   n3 = Select(Option 0:String=@n1, Option 1:String="To String (Text)"(InText="GetText (Text Box)"(self=@n2)), Index=IsEmpty(InString=@n1)) -> (ReturnValue:String)
   > PrintString(InString=@n3)
   > n4 = GetAssetsByPath(self=GetAssetRegistry(), PackagePath=String To Name(InString=@n3), bRecursive=true, bIncludeOnlyOnDiskAssets=true)
   > Sequence()
   > n5 = For Each Loop(Array:Array of Asset Data Structures=@n4) -> (Array Element:Asset Data Structure)
   n6 = GetAsset(InAssetData=@n5)
   > Set Current Level To Process(Current Level To Process=@n6)
   n7 = Break Asset Data(AssetData=@n5)
   > n8 = Branch(Condition=Contains(SearchIn="To String (Name)"(InName=@n7), Substring=ExternalObjects, bUseCase=true))
   n9 = Get Class Display Name(Class=GetClass(InAssetData=@n5))
   PrintString(execute=@n8.else, InString=@n9)
   > n10 = Branch(Condition="Equal Exactly (String)"(A=@n9, B=World))
   > PrintString(InString="To String (Name)"(InName=@n7))
   > Clear(TargetArray=Get Cliff Actors to Save())
   > Sequence!(execute) -> (then_0, then_1, then_2:Exec)
   > Sequence!(execute) -> (then_0, then_1, then_2, then_3:Exec, then_4:Exec)
   > n11 = Set Current Level To Process(Current Level To Process=@n6)
   > Outputs!(then:Exec, Output_Get:Object Reference=@n11)
   PrintString(execute=@n10.else, InString="Found Object other than World SKIPPING\r\n", TextColor=(R=1,G=0.4929,B=0,A=1))
  ]}
 ForEachLI {nodes [
   n1 = Inputs()
   n2 = "To String (Text)"(InText="GetText (Text Box)"(self=Get SourceTextBox()))
   n3 = IsEmpty(InString=@n1)
   > n4 = GetAssetsByPath(
    self=GetAssetRegistry()
    PackagePath=String To Name(InString=Select(Option 0=@n1, Option 1=@n2, Index=@n3))
    bRecursive=true
    bIncludeOnlyOnDiskAssets=true
   )
   > Sequence!(execute) -> (then_0, then_1)
   > n5 = For Each Loop(Array=@n4)
   n6 = GetAsset(InAssetData=@n5)
   > Set Current Level To Process(Current Level To Process=@n6)
   n7 = Break Asset Data(AssetData=@n5)
   > n8 = Branch(Condition=Contains(SearchIn="To String (Name)"(InName=@n7), Substring=ExternalObjects, bUseCase=true))
   n9 = Get Class Display Name(Class=GetClass(InAssetData=@n5))
   PrintString(execute=@n8.else, InString=@n9)
   > n10 = Branch(Condition="Equal Exactly (String)"(A=@n9, B=World))
   > PrintString(InString="To String (Name)"(InName=@n7))
   > n11 = GetEditorWorld(self=Get UnrealEditorSubsystem())
   > n12 = SpawnActorFromClass(self=Get EditorActorSubsystem(), ActorClass=&LevelInstance, Location="0, 0, 0", bTransient=true)
   > Set LevelInstanceActor(LevelInstanceActor=@n12)
   > SetEditorProperty(Object=@n12, PropertyName=WorldAsset, PropertyValue:Soft Object Reference=Make Soft Reference(Input=@n6), ChangeNotifyMode=Never)
   > Clear(TargetArray=Get Cliff Actors to Save())
   > n13 = Sequence!(execute) -> (then_0, then_1, then_2:Exec)
   > LevelInstanceEnterEditMode(LevelInstance=Get LevelInstanceActor())
   > Sequence!(execute) -> (then_0, then_1, then_2, then_3:Exec, then_4:Exec)
   > n14 = Set Current Level To Process(Current Level To Process=@n6)
   > Outputs(Output_Get=@n14)
   LevelInstanceCommit(execute=@n13.then_1, LevelInstance=Get LevelInstanceActor(), bDiscardChanges=true)
   n15 = Get EditorActorSubsystem()
   SelectNothing(execute=@n13.then_2, self=@n15)
   > DestroyActor(self=@n15, ActorToDestroy=Get LevelInstanceActor())
   > CollectGarbage()
   PrintString(execute=@n10.else, InString="Found Object other than World SKIPPING\r\n", TextColor=(R=1,G=0.4929,B=0,A=1))
   "To String (Text)"(InText="GetText (Text Box)"(self=Get SourceTextBox()))
   GetClass(InAssetData=@n5)
   Make Soft Reference(Input=@n6)
   Resolve Soft Reference(Input=Make Soft Reference(Input=@n6))
   Reroute Node(InputPin:World Object Reference=@n11) -> (OutputPin:World Object Reference)
  ]}
 ForEachLevelInstanceActor {nodes [
   n1 = ForEachLevelInstanceActor!() -> (then, InputPin:String) {is=title K2Node_FunctionEntry}
   > For Each LI(InputPin=@n1)
   > Return Node!(execute)
  ]}
 ForEachLevelInstanceContext {nodes [
   n1 = ForEachLevelInstanceContext!() -> (then, InString:String) {is=title K2Node_FunctionEntry}
   > n2 = For Each Context(InputPin=@n1)
   > Return Node!(execute, Output_Get:Object Reference=@n2)
  ]}
 GetStaticMeshActorsExclusive {nodes [
   n1 = GetStaticMeshActorsExclusive!() -> (then, WorldContextObject:Object Reference) {is=title K2Node_FunctionEntry}
   > n2 = GetAllActorsOfClass(WorldContextObject=@n1, ActorClass=&FortStaticMeshActor) -> (OutActors:Array of Fort Static Mesh Actor Object References)
   > n3 = GetAllActorsOfClass(WorldContextObject=@n1, ActorClass=&BP__Cliff_Figment_C) -> (OutActors:"Array of BP  Cliff Figment Object References")
   n4 = Get RemoveCliffSet() -> (RemoveCliffSet:Set of Actor Object References)
   > Add Items(TargetSet:Set of Actor Object References=@n4, NewItems:Array of Actor Object References=@n2)
   > Remove Items!(execute, TargetSet:Set of Actor Object References=@n4, Items:Array of Actor Object References=@n3) -> (then)
   > n5 = To Array!(execute, A:Set of Actor Object References=Get RemoveCliffSet()) -> (then, Result:Array of Actor Object References)
   > Return Node!(execute, Result:Array of Actor Object References=@n5)
  ]}
 PaintConcreteVertex1000 {nodes [
   n1 = PaintConcreteVertex1000!() -> (then, WorldContextObject:Object Reference) {is=title K2Node_FunctionEntry}
   > n2 = GetEditorWorld(self=Get UnrealEditorSubsystem())
   > n3 = GetAllActorsOfClass(
    WorldContextObject=SelectObject(A=@n1, B=@n2, bSelectA=IsValid(Object=@n1))
    ActorClass=&BP__Cliff_Figment_C
   )
   > Clear(TargetArray=Get Cliff Actors to Save())
   > n4 = For Each Loop(Array:"Array of BP  Cliff Figment Object References"=@n3) -> (Array Element:"BP  Cliff Figment Object Reference")
   n5 = Get Cliff Actors to Save()
   CheckoutLoadedAssets(execute=@n4.Completed, AssetsToCheckout=@n5)
   > SaveLoadedAssets(AssetsToSave=@n5, bOnlyIfIsDirty=true)
   n6 = Cast To BP__Cliff_Figment(execute=@n4.LoopBody, Object=@n4)
   n7 = Get StaticMeshComponent(self=@n6)
   > n8 = Cast To MaterialInstanceDynamic!(execute, Object=GetMaterial(self=@n7, ElementIndex=0)) -> (then, CastFailed, AsMaterial Instance Dynamic:Material Instance Dynamic Object Reference) {is=title K2Node_DynamicCast}
   > Branch(Condition="Equal (Object)"(A=Get Parent(self=@n8), B=/Figment_S04/Athena/Environments/Materials/CliffGround/Athena_Terrain_Mat_Concerte_Inst))
   > PaintVerticesSingleColor!(
    execute
    StaticMeshComponent:Static Mesh Component Object Reference=@n7
    FillColor:"Linear Color Structure (by ref)"=Make Linear Color!(R:float=1, G:float=0, B:float=0, A:float=0) -> (LinearColor:Linear Color Structure) {is=title K2Node_MakeStruct}
    bConvertToSRGB:bool=false
   ) -> (then)
   > Add(TargetArray=Get Cliff Actors to Save(), NewItem=@n6)
   PrintString(InString=mid)
  ]}
 Reassign TextureData with existing {nodes [
   n1 = GetSelectedLevelActors(self=Get EditorActorSubsystem())
   > n2 = For Each Loop(Array:Array of Actor Object References=@n1) -> (Array Element:Actor Object Reference)
   > Cast To BP__Cliff_Figment(Object=@n2)
   n3 = GetEditorWorld(execute=@n4.then, self=Get UnrealEditorSubsystem())
   n4 = Reassign TextureData with existing!() -> (then, Level Object Reference:Object Reference) {is=title K2Node_FunctionEntry}
   n5 = GetAllActorsOfClass(
    execute=@n3.then
    WorldContextObject=SelectObject(A=@n4, B=@n3, bSelectA=IsValid(Object=@n4))
    ActorClass=&BP__Cliff_Figment_C
   )
   n6 = For Each Loop(Exec=@n53.then, Array=@n5)
   n7 = PrintString(InString=GetDisplayName(Object=@n6))
   n8 = Cast To BP__Cliff_Figment(execute=[@n7.then, @n6.LoopBody], Object=@n6)
   > n9 = Execute Python Script!(
    execute
    PythonScript="import unreal\r\ntexture_data = (geo.get_editor_property('TextureData')[0])"
    geo:"BP  Cliff Figment Object Reference"=@n8
   ) -> (then, ReturnValue, texture_data:Object Reference) {inputs [geo], outputs [texture_data]}
   n10 = GetDisplayName(Object=GetPackageForObject(Object=@n9.texture_data))
   > n11 = Branch(Condition=IsEmpty(InString=@n10))
   > PrintString(InString=Nothing Assigned)
   n12 = GetEditorWorld(execute=@n11.else, self=Get UnrealEditorSubsystem())
   > n13 = StripRightMostFolderFromPath_Util(Path=@n10, WorldContextObject=@n12, Strip Count=1, __WorldContext=@n12)
   n14 = MakeLiteralString(Value=/Figment_S03/Athena/Environments/TextureData/CliffGround/)
   > n15 = Branch(Condition="Equal, Case Insensitive (String)"(A=@n14, B=@n13))
   > PrintString(InString=already assigned)
   n16 = MakeLiteralString(Value=&CliffGround/)
   n17 = Append(A=@n13.AssetName, B=_S03)
   n18 = Append(A=@n16, B=@n17)
   n19 = DoesAssetExist(execute=@n15.else, AssetPath=@n18)
   > n20 = Branch(Condition=@n19)
   > n21 = PrintString(InString=TD asset existsi)
   n22 = GetAssetByObjectPath(
    execute=[@n21.then, @n51.then]
    self=GetAssetRegistry()
    ObjectPath=String To Name(InString=Append!(A=@n18, B=., C:String=@n17) -> (ReturnValue))
    bIncludeOnlyOnDiskAssets=false
   )
   n23 = Cast To BuildingTextureData(execute=[@n22.then, @n50.then], Object=GetAsset(InAssetData=@n22))
   > n24 = Execute Python Script!(
    execute
    PythonScript="import unreal\r\nmaterial_interface = tex_data.get_editor_property('OverrideMaterial')"
    tex_data:Building Texture Data Object Reference=@n23
   ) -> (then, ReturnValue, material_interface:Material Instance Constant Object Reference) {inputs [tex_data], outputs [material_interface]}
   > n25 = Set material_on_TD(material_on_TD=@n24.material_interface)
   n26 = Get Parent(self=@n25)
   > n27 = PrintString(InString=GetDisplayName(Object=@n26))
   n28 = Execute Python Script!(
    execute=[@n27.then, @n52.then]
    PythonScript="import unreal\r\nno_need_to_save = False\r\nTextureDataArray = geo.get_editor_property('TextureData')\r\nprint(TextureDataArray)\r\nif(TextureDataArray[0] == texture_data):\r\n    no_need_to_save = True\r\nelse:\r\n    TextureDataArray[0] = texture_data\r\n    geo.set_editor_property('TextureData', TextureDataArray)"
    geo:"BP  Cliff Figment Object Reference"=@n8
    texture_data:Building Texture Data Object Reference=@n23
   ) -> (then, ReturnValue, no_need_to_save:bool=~) {inputs [geo,texture_data], outputs [no_need_to_save]}
   > n29 = Branch(Condition=@n28.no_need_to_save)
   Add(execute=@n29.else, TargetArray=Get Cliff Actors to Save(), NewItem=@n8)
   PrintString(execute=@n20.else, InString=Append!(A="TD asset doesnt exist, duplicating", B) -> (ReturnValue))
   > n30 = GetAssetTools()
   > n31 = DuplicateAsset(self=@n30, AssetName=@n17, PackagePath=@n16, OriginalObject=@n9.texture_data)
   > n32 = Execute Python Script!(
    execute
    PythonScript="import unreal\r\nmaterial_interface = tex_data.get_editor_property('OverrideMaterial')"
    tex_data:Object Reference=@n31
   ) -> (then, ReturnValue, material_interface:Material Instance Constant Object Reference) {inputs [tex_data], outputs [material_interface]}
   > n33 = Set material_on_TD(material_on_TD=@n32.material_interface)
   n34 = GetDisplayName(Object=GetPackageForObject(Object=@n33))
   > PrintString(InString=@n34)
   > n35 = GetEditorWorld(self=Get UnrealEditorSubsystem())
   > n36 = StripRightMostFolderFromPath_Util(Path=@n34, WorldContextObject=@n35, Strip Count=1, __WorldContext=@n35)
   n37 = MakeLiteralString(Value=/Figment_S03/Athena/Environments/Masterials/CliffGround/)
   n38 = Append(A=@n36.AssetName, B=MakeLiteralString(Value=_S03))
   > n39 = DoesAssetExist(AssetPath=Append(A=@n37, B=@n38))
   > n40 = Branch(Condition=@n39)
   n41 = GetAssetTools(execute=@n40.else)
   > n42 = DuplicateAsset(self=@n41, AssetName=@n38, PackagePath=@n37, OriginalObject=@n33)
   > n43 = Cast To MaterialInstanceConstant(Object:Object Reference=@n42)
   > n44 = Branch(Condition="Equal (Object)"(A=Get Parent(self=@n43), B=&Athena_Terrain_Mat_Biome))
   n45 = Get FigmentParentMAt()
   > Cast To MaterialInterface(Object=SelectObject(A=Get FigmentParentMAt(), B=@n43, bSelectA=IsValid(Object=@n45)))
   > SetMaterialInstanceParent(Instance=@n43, NewParent=@n43)
   > SaveLoadedAsset(AssetToSave=@n42, bOnlyIfIsDirty=true)
   > n46 = Cast To BuildingTextureData(Object=@n31)
   > Set OverrideMaterial(OverrideMaterial=@n43, self=@n46)
   > SaveLoadedAsset(AssetToSave=@n46, bOnlyIfIsDirty=false)
   > n47 = Execute Python Script!(
    execute
    PythonScript="import unreal\r\nno_need_to_save = False\r\nTextureDataArray = geo.get_editor_property('TextureData')\r\nprint(TextureDataArray)\r\nif(TextureDataArray[0] == texture_data):\r\n    no_need_to_save = True\r\nelse:\r\n    TextureDataArray[0] = texture_data\r\n    geo.set_editor_property('TextureData', TextureDataArray)"
    geo:"BP  Cliff Figment Object Reference"=@n8
    texture_data:Building Texture Data Object Reference=@n46
   ) -> (then, ReturnValue, no_need_to_save:bool=~) {inputs [geo,texture_data], outputs [no_need_to_save]}
   > n48 = Branch(Condition=@n47.no_need_to_save)
   Add(execute=@n48.else, TargetArray=Get Cliff Actors to Save(), NewItem=@n8)
   n49 = GetPackageForObject(Object=@n43)
   PrintString(
    execute=@n44.else
    InString=Append(A=GetDisplayName(Object=@n49), B="\r\n########################\r\n is not a child of Athena Material\r\n########################################")
   )
   n50 = PrintString(InString=Hello)
   n51 = PrintString(InString=Hello)
   n52 = Branch(Condition="Equal (Object)"(A=@n26, B=Get material_orig()))
   PrintString(execute=@n52.else, InString=Wrong parent)
   PrintString(InString="TD asset doesnt exist, duplicating")
   GetAssetByObjectPath(ObjectPath=None, bIncludeOnlyOnDiskAssets=false)
   Get Dependencies(PackageName=None)
   GetAsset()
   CollectGarbage(execute=@n56.then)
   GetSystemPath(Object=@n9.texture_data)
   PrintString(InString=@n10)
   Append!(A, B=., C:String) -> (ReturnValue)
   Append!(A=@n14, B=MakeLiteralString(Value=&S03)) -> (ReturnValue)
   Cast To MaterialInstanceConstant(Object:Wildcard)
   Execute Python Script!(
    execute
    PythonScript="import unreal\r\nTextureData = geo.get_editor_property('TextureData')[0]\r\nOverrideMaterial = TextureData.get_editor_property('OverrideMaterial')\r\nprint(TextureData)"
    Geo:"BP  Cliff Figment Object Reference"=@n8
   ) -> (then, ReturnValue) {inputs [Geo]}
   Reroute Node(InputPin:Object Reference=@n42) -> (OutputPin:Object Reference)
   Get material_orig()
   n53 = Clear(execute=@n54.then, TargetArray=Get Actors to Revert())
   n54 = Clear(execute=@n5.then, TargetArray=Get Cliff Actors to Save())
   n55 = Get Cliff Actors to Save()
   n56 = SaveLoadedAssets(execute=@n57.then, AssetsToSave=@n55, bOnlyIfIsDirty=true)
   n57 = CheckoutLoadedAssets(execute=@n6.Completed, AssetsToCheckout=@n55)
  ]}
 Remove_appended_TDs {nodes [
   n1 = GetSelectedLevelActors(self=Get EditorActorSubsystem())
   n2 = GetEditorWorld(execute=@n3.then, self=Get UnrealEditorSubsystem())
   n3 = Remove_appended_TDs!() -> (then, Level Object:Object Reference) {is=title K2Node_FunctionEntry}
   n4 = GetAllActorsOfClass(
    execute=@n2.then
    WorldContextObject=SelectObject(A=@n3, B=@n2, bSelectA=IsValid(Object=@n3))
    ActorClass=&BP__Cliff_Figment_C
   )
   n5 = For Each Loop(Exec=@n28.then, Array:"Array of BP  Cliff Figment Object References"=@n4) -> (Array Element:"BP  Cliff Figment Object Reference")
   n6 = Cast To BP__Cliff_Figment(execute=[@n1.then, @n5.LoopBody], Object=@n5)
   > n7 = GetEditorProperty(Object=@n6, PropertyName=TextureData)
   n8 = "Get (a copy)"(Array:Array of Building Texture Data Object References=@n7) -> (Output:Building Texture Data Object Reference)
   n9 = GetDisplayName(Object=GetPackageForObject(Object=@n8))
   > n10 = StripRightMostFolderFromPath_Util(Path=@n9, Strip Count=1)
   > n11 = Set Folder(Folder:String=@n10) -> (Output_Get:String)
   PrintString(execute=[@n26.then, @n11.then], InString=@n9)
   n12 = GetDisplayName(Object=@n8)
   > PrintString(InString=@n12)
   n13 = ParseIntoArray(SourceString=@n12, Delimiter=_)
   > Set Path(Path:Array of Strings=@n13) -> (Output_Get:Array of Strings)
   > Set AppendCount(AppendCount:int=0) -> (Output_Get:int=0)
   n14 = Length(TargetArray:Array of Strings=@n13)
   n15 = int - int(A=@n14)
   > n16 = For Loop with Break!(Execute:Exec, FirstIndex:int, LastIndex:int=@n15, Break:Exec=@n25.else) -> (LoopBody:Exec, Index:int, Completed:Exec) {is=title K2Node_MacroInstance}
   n17 = Get AppendCount() -> (AppendCount:int=0)
   Branch(execute=@n16.Completed, Condition="integer > integer"(A=@n17, B=1))
   > Resize!(execute, TargetArray=Get Path() -> (Path:Array of Strings), Size:int=int - int(A=@n14, B=int - int(A=@n17, B=~))) -> (then) {is=title K2Node_CallArrayFunction}
   n18 = Append(A=JoinStringArray(SourceArray=Get Path(), Separator=_), B=_S03)
   n19 = Append!(A=Get Folder() -> (Folder:String), B=/, C:String=@n18) -> (ReturnValue)
   > PrintString(InString=@n19)
   > n20 = DoesAssetExist(AssetPath=@n19)
   > Branch(Condition=@n20)
   > n21 = GetAssetByObjectPath(self=GetAssetRegistry(), ObjectPath=String To Name(InString=Append(A=@n19, B=., C=@n18)), bIncludeOnlyOnDiskAssets=false)
   > n22 = Cast To BuildingTextureData(Object=GetAsset(InAssetData=@n21))
   > PrintString(InString="exists!")
   > SetEditorProperty(
    Object=@n6
    PropertyName=TextureData
    PropertyValue:Array of Building Texture Data Object References=Make Array(
     "[0]"=@n22
     "[1]"="Get (a copy)"(Array=@n7, Dimension 1=1)
     "[2]"="Get (a copy)"(Array=@n7, Dimension 1=2)
     "[3]"="Get (a copy)"(Array=@n7, Dimension 1=3)
    )
   )
   > Add(TargetArray=Get Cliff Actors to Save(), NewItem=@n6)
   n23 = int - int(A=@n15, B=@n16)
   n24 = "Get (a copy)"(Array:Array of Strings=@n13, Dimension 1=@n23) -> (Output:String)
   n25 = Branch(
    execute=@n16.LoopBody
    Condition=OR Boolean!(A="Equal Exactly (String)"(A=@n24, B=S03), B="Equal Exactly (String)"(A=@n24, B=S02)) -> (ReturnValue)
   )
   > Increment Int(Value=Get AppendCount() -> (AppendCount=0))
   n26 = GetPathNameForLoadedAsset(LoadedAsset=@n8)
   "To String (Integer)"(InInt=@n23)
   String To Name()
   JoinStringArray(Separator=" ")
   n27 = Get Cliff Actors to Save()
   SaveLoadedAssets(execute=@n30.then, AssetsToSave=@n27, bOnlyIfIsDirty=true)
   "To String (Integer)"(InInt=@n17)
   Append(B=.)
   Append(B=.)
   "Get (a copy)"(Array:Array of Actor Object References=@n1) -> (Output:Actor Object Reference)
   Get AssetEditorSubsystem()
   n28 = Clear(execute=@n29.then, TargetArray=Get Actors to Revert())
   n29 = Clear(execute=@n4.then, TargetArray=Get Cliff Actors to Save())
   n30 = CheckoutLoadedAssets(execute=@n5.Completed, AssetsToCheckout=@n27)
  ]}
 Replace Mesh Actor With Existing BP Single {nodes [
   n1 = Replace Mesh Actor With Existing BP Single!() -> (then, Static Mesh Actor:Static Mesh Actor Object Reference, mapOfMeshNamesToBp:Map of Strings to Strings) {is=title K2Node_FunctionEntry}
   n2 = GetMaterial(self=Get StaticMeshComponent(self=@n1), ElementIndex=0)
   n3 = PrintString(InString=Append!(A="testing: ", B=GetObjectName(Object=@n2)) -> (ReturnValue))
   Replace Static Mesh Actor Impl!(
    execute=[@n3.then, @n1.then]
    self
    Static Mesh Actor:Static Mesh Actor Object Reference=@n1
    mapOfMeshNamesToBp:Map of Strings to Strings=@n1.mapOfMeshNamesToBp
   ) -> (then)
  ]}
 Replace Static Mesh Actor Impl {nodes [
   n1 = Replace Static Mesh Actor Impl!() -> (then, Static Mesh Actor:Static Mesh Actor Object Reference, mapOfMeshNamesToBp:Map of Strings to Strings) {is=title K2Node_FunctionEntry}
   n2 = Get StaticMeshComponent(self=@n1)
   n3 = GetDisplayName(Object=Get StaticMesh(self=@n2))
   n4 = Find(TargetMap=@n1.mapOfMeshNamesToBp, Key=@n3)
   n5 = Branch(execute=[@n1.then, @n10.else], Condition=@n4.ReturnValue)
   > PrintString(InString=Append!(A="Found BP for:  ", B=@n3, C:String=@n4) -> (ReturnValue))
   > n6 = EnsureMaterialAndTextureDataExist!(execute, self, Mesh Material:Object Reference=GetMaterial(self=@n2, ElementIndex=0)) -> (then, " Texture Data To Assign":Building Texture Data Object Reference) {is=title K2Node_CallFunction}
   > SetSelectedLevelActors(self=Get EditorActorSubsystem(), ActorsToSelect=Make Array!("[0]":Actor Object Reference=@n1) -> (Array:Array of Actor Object References))
   > n7 = DoesAssetExist(AssetPath=@n4)
   > Branch(Condition=@n7)
   > ReplaceSelectedActors!(execute, InAssetPath:String=@n4) -> (then)
   > n8 = GetSelectedActors()
   n9 = "Get (a copy)"(Array=@n8)
   > Add(TargetArray:Array of Object References=Get actorsToSAve2(), NewItem:"Object Reference (by ref)"=@n9)
   > SelectNothing(self=Get EditorActorSubsystem())
   > ResetTextureData!(execute, self, Actor:Object Reference=@n9, TD asset:Building Texture Data Object Reference=@n6) -> (then)
   PrintString(execute=@n5.else, InString=Append!(A="NO FOUND !!!!!!!!!!!!!!!", B=@n3) -> (ReturnValue))
   n10 = Branch(Condition=IsEmpty(InString=@n3))
   "Equal Exactly (String)"(A=@n4, B=/FigmentEnvironment/Blueprints/Cliffs/BP__Cliff_Figment)
   Get LevelEditorSubsystem()
   Get LevelInstanceSubsystem!(WorldContext:Object Reference) -> (ReturnValue:Level Instance Subsystem Object Reference) {is=title K2Node_GetSubsystem}
   Get mapOfMeshNamesToBp()
  ]}
 ReplaceAllStaticMeshesWithExistingBP {nodes [
   n1 = ReplaceAllStaticMeshesWithExistingBP!() -> (then, WorldContextObject:Object Reference) {is=title K2Node_FunctionEntry}
   > n2 = GetEditorWorld(self=Get UnrealEditorSubsystem())
   > Clear(TargetArray:Array of Object References=Get actorsToSAve2())
   > n3 = Sequence!(execute) -> (then_0, then_1, then_2)
   > n4 = BuildMeshMap()
   n5 = GetStaticMeshActorsExclusive!(execute=@n3.then_1, self, WorldContextObject:Object Reference=SelectObject(A=@n1, B=@n2, bSelectA=IsValid(Object=@n1))) -> (then, Result:Array of Actor Object References) {is=title K2Node_CallFunction}
   > n6 = For Each Loop(Array:Array of Actor Object References=@n5) -> (Array Element:Actor Object Reference)
   > n7 = Cast To FortStaticMeshActor(Object=@n6)
   > Replace Static Mesh Actor Impl(Static Mesh Actor=@n7, mapOfMeshNamesToBp=@n4) {is=title K2Node_CallFunction}
   n8 = Get actorsToSAve2()
   CheckoutLoadedAssets(execute=@n3.then_2, AssetsToCheckout=@n8)
   > SaveLoadedAssets(AssetsToSave=@n8, bOnlyIfIsDirty=true)
  ]}
 ResetTextureData {nodes [
   n1 = ResetTextureData!() -> (then, Actor:Object Reference, TD asset:Building Texture Data Object Reference) {is=title K2Node_FunctionEntry}
   > n2 = GetEditorProperty(Object=@n1, PropertyName=TextureData)
   > SetEditorProperty(
    Object=@n1
    PropertyName=TextureData
    PropertyValue=Make Array!(
     "[0]":Building Texture Data Object Reference=@n1.TD asset
     "[1]":Building Texture Data Object Reference="Get (a copy)"(Array:Array of Building Texture Data Object References=@n2, Dimension 1=1) -> (Output:Building Texture Data Object Reference)
     "[2]":Building Texture Data Object Reference="Get (a copy)"(Array=@n2, Dimension 1=2)
     "[3]":Building Texture Data Object Reference=Get data()
    ) -> (Array:Array of Building Texture Data Object References)
   )
   "Get (a copy)"(Array=@n2, Dimension 1=3)
   Get CurrentTDForStaticMesh()
  ]}
 TextureAssetPathPrep {nodes [
   n1 = TextureAssetPathPrep!() -> (then, Material Asset Name:String) {is=title K2Node_FunctionEntry}
   > n2 = GetEditorWorld(self=Get UnrealEditorSubsystem())
   n3 = "To String (Text)"(InText="GetText (Text Box)"(self=Get TargetTextBox_() -> (TargetTextBox_:Editor Utility Editable Text Box Object Reference)))
   > n4 = StripRightMostFolderFromPath_Util(Path=@n3, WorldContextObject=@n2, Strip Count=2, __WorldContext=@n2)
   n5 = Append(A=TD_, B=@n1)
   n6 = Append(A=@n4, B=/TextureData/CliffGround/)
   n7 = EndsWith(SourceString=@n3, InSuffix=/)
   n8 = Append(A=@n3, B=/)
   > Return Node!(
    execute
    "Material Full  Path":String=Append(A=Select(Option 0=@n8, Option 1=@n3, Index=@n7), B=@n1)
    TD Asset Full Path:String=Append(A=@n6, B=@n5)
    TD Asset Name:String=@n5
    Material Folder:String=@n3
    TD Asset Folder:String=@n6
   )
   MakeLiteralString(Value=/FigmentEnvironment/Blueprints/Cliffs/TextureData)
   MakeLiteralString(Value=/FigmentEnvironment/Blueprints/Cliffs/TextureData/CliffGround/)
   MakeLiteralString(Value=/FigmentEnvironment/Blueprints/Cliffs/Materials/CliffGround/)
  ]}
]
variables [
 Actors to Revert {type="array<BP__Cliff_Figment_C>", default []}
 actorsToSAve2 {type="array<Object>", default []}
 Cliff Actors to Save {type="array<BP__Cliff_Figment_C>", default []}
 Color Map TD override {type=BuildingTextureData, default=/FigmentEnvironment/Blueprints/Cliffs/Biome/TD_BiomeMap_S03}
 Current Level To Process {type=Object, default=null}
 CurrentMaterialForStaticMesh {type=MaterialInterface, default=null}
 CurrentTDForStaticMesh {type=BuildingTextureData, default=null}
 data {type=BuildingTextureData, default=/FigmentEnvironment/Blueprints/Cliffs/Biome/TD_BiomeMap_S04}
 Element LIst {type=ScriptTypedElementListProxy, default {}}
 ExternalFoliageActorExternalPathsToKill {type="array<string>", default []}
 ExternalFoliageActorsToKill {type="array<string>", default []}
 FigmentParentMAt {type=MaterialInstanceConstant, default=null}
 FoliageActorsToKill {type="array<Actor>"}
 Level Instance Actor {type=LevelInstance, default=null}
 LevelInstanceActor {type=LevelInstance, default=null}
 MapOfMaterials {default {None=null}}
 mapOfMeshesToBPs {default {}}
 mapOfMeshNamesToBp {type="map<string>", default {}}
 material_on_TD {type=MaterialInstanceConstant, default=null}
 material_orig {type=MaterialInterface, default=&Athena_Terrain_Mat_Biome}
 MICFrom {type=MaterialInstanceConstant, default=null}
 MICTo {type=MaterialInstanceConstant, default=null}
 OriginalMIs {type="array<MaterialInstanceConstant>", default []}
 PackagesToRevert {type="array<string>", default []}
 RemoveCliffSet {type="array<Actor>"}
 SM_to_BP_all {type=boolean}
 staticMeshSetCheckDuplicattes {type="array<StaticMesh>", default []}
 TargetPAth {type=string, default=/Figment_S04_LevelInstances/}
 TD Asset {type=BuildingTextureData, default=null}
 viewport {type=Viewport, default=null}
]
widgets {
 info {rootWidgetClass=&CanvasPanel, widgetCount=70}
 widgets [
  CanvasPanel_X {
   widgetClassPath=&CanvasPanel
   bIsVariable=true
   widget=CanvasPanel_X
   properties {accessibleBehavior=Auto, bOverrideAccessibleDefaults=true, visibility=Visible}
  }
  VerticalBox_146 {
   widgetClassPath=&VerticalBox
   parent=CanvasPanel_X
   widget=VerticalBox_146
   slot {class=CanvasPanelSlot, properties {bAutoSize=true, layoutData=~}}
  }
  HorizontalBox_174 {widgetClassPath=&HorizontalBox, parent=VerticalBox_146, widget=HorizontalBox_174, slot {class=VerticalBoxSlot}}
  EditorUtilityButton_34 {
   widgetClassPath=&EditorUtilityButton
   parent=HorizontalBox_174
   widget=EditorUtilityButton_34
   slot {class=HorizontalBoxSlot}
  }
  TextBlock_178 {
   widgetClassPath=&TextBlock
   parent=EditorUtilityButton_34
   widget=TextBlock_178
   properties {text=Figment Processing Tools}
   slot {class=ButtonSlot}
  }
  EditorUtilityButton_60 {
   widgetClassPath=&EditorUtilityButton
   parent=HorizontalBox_174
   widget=EditorUtilityButton_60
   slot {class=HorizontalBoxSlot}
  }
  TextBlock_228 {
   widgetClassPath=&TextBlock
   parent=EditorUtilityButton_60
   widget=TextBlock_228
   properties {text=Experimental}
   slot {class=ButtonSlot}
  }
  WidgetSwitcher_184 {
   widgetClassPath=&WidgetSwitcher
   bIsVariable=true
   parent=VerticalBox_146
   widget=WidgetSwitcher_184
   slot {class=VerticalBoxSlot}
  }
  CanvasPanel_36 {widgetClassPath=&CanvasPanel, parent=WidgetSwitcher_184, widget=CanvasPanel_36, slot {class=WidgetSwitcherSlot}}
  VerticalBox_0 {
   widgetClassPath=&VerticalBox
   parent=CanvasPanel_36
   widget=VerticalBox_0
   slot {
    class=CanvasPanelSlot
    properties {bAutoSize=true, layoutData {alignment=<0 0>, anchors {maximum=<0 0>, minimum=<0 0>}, offsets=<ltrb 80 76 100 40>}}
   }
  }
  ConvertSMinLeveltoBP {
   widgetClassPath=&EditorUtilityButton
   parent=VerticalBox_0
   widget=ConvertSMinLeveltoBP
   slot {class=VerticalBoxSlot, properties {horizontalAlignment=HAlign_Left}}
  }
  TextBlock {
   widgetClassPath=&TextBlock
   parent=ConvertSMinLeveltoBP
   widget=TextBlock
   properties {text=Convert SM in Level to BP}
   slot {class=ButtonSlot, properties {horizontalAlignment=HAlign_Left}}
  }
  ConvertSMinPathtoBP {
   widgetClassPath=&EditorUtilityButton
   parent=VerticalBox_0
   widget=ConvertSMinPathtoBP
   slot {class=VerticalBoxSlot, properties {horizontalAlignment=HAlign_Left}}
  }
  TextBlock_6 {
   widgetClassPath=&TextBlock
   parent=ConvertSMinPathtoBP
   widget=TextBlock_6
   properties {text=Convert SM in Path to BP}
   slot {class=ButtonSlot, properties {horizontalAlignment=HAlign_Left}}
  }
  HorizontalBox_376 {widgetClassPath=&HorizontalBox, parent=VerticalBox_0, widget=HorizontalBox_376, slot {class=VerticalBoxSlot}}
  TextBlock_106 {
   widgetClassPath=&TextBlock
   parent=HorizontalBox_376
   widget=TextBlock_106
   properties {text=Source Level Instance Top Level Path:}
   slot {class=HorizontalBoxSlot, properties {padding=<ltrb 8 8 8 8>}}
  }
  VerticalBox_163 {
   widgetClassPath=&VerticalBox
   parent=HorizontalBox_376
   widget=VerticalBox_163
   slot {class=HorizontalBoxSlot, properties {verticalAlignment=VAlign_Center}}
  }
  SourceTextBox {
   widgetClassPath=&EditorUtilityEditableTextBox
   bIsVariable=true
   parent=VerticalBox_163
   widget=SourceTextBox
   properties {
    text=/Figment_S04_LevelInstances/
    widgetStyle {
     backgroundColor {colorUseRule=UseColor_Specified, specifiedColor=<rgba 1 1 1 1>}
     backgroundImageFocused {
      bIsDynamicallyLoaded=false
      drawAs=RoundedBox
      imageType=NoImage
      mirroring=NoMirror
      resourceName=None
      resourceObject=null
      tiling=NoTile
      imageSize=<0 0>
      margin=<ltrb 0 0 0 0>
      outlineSettings {
       bUseBrushTransparency=false
       roundingType=FixedRadius
       width=1
       color {colorUseRule=UseColor_ColorTable, specifiedColor=<rgba 0 0 0 0>}
       cornerRadii=<4 4 4 4>
      }
      tintColor {colorUseRule=UseColor_Specified, specifiedColor=<rgba 0.0048 0.0048 0.0048 1>}
      uVRegion {bIsValid=false, max=<0 0>, min=<0 0>}
     }
     backgroundImageHovered {
      bIsDynamicallyLoaded=false
      drawAs=RoundedBox
      imageType=NoImage
      mirroring=NoMirror
      resourceName=None
      resourceObject=null
      tiling=NoTile
      imageSize=<0 0>
      margin=<ltrb 0 0 0 0>
      outlineSettings {
       bUseBrushTransparency=false
       roundingType=FixedRadius
       width=1
       color {colorUseRule=UseColor_ColorTable, specifiedColor=<rgba 0 0 0 0>}
       cornerRadii=<4 4 4 4>
      }
      tintColor {colorUseRule=UseColor_Specified, specifiedColor=<rgba 0.0048 0.0048 0.0048 1>}
      uVRegion {bIsValid=false, max=<0 0>, min=<0 0>}
     }
     backgroundImageNormal {
      bIsDynamicallyLoaded=false
      drawAs=RoundedBox
      imageType=NoImage
      mirroring=NoMirror
      resourceName=None
      resourceObject=null
      tiling=NoTile
      imageSize=<0 0>
      margin=<ltrb 0 0 0 0>
      outlineSettings {
       bUseBrushTransparency=false
       roundingType=FixedRadius
       width=1
       color {colorUseRule=UseColor_ColorTable, specifiedColor=<rgba 0 0 0 0>}
       cornerRadii=<4 4 4 4>
      }
      tintColor {colorUseRule=UseColor_Specified, specifiedColor=<rgba 0.0048 0.0048 0.0048 1>}
      uVRegion {bIsValid=false, max=<0 0>, min=<0 0>}
     }
     backgroundImageReadOnly {
      bIsDynamicallyLoaded=false
      drawAs=RoundedBox
      imageType=NoImage
      mirroring=NoMirror
      resourceName=None
      resourceObject=null
      tiling=NoTile
      imageSize=<0 0>
      margin=<ltrb 0 0 0 0>
      outlineSettings {
       bUseBrushTransparency=false
       roundingType=FixedRadius
       width=1
       color {colorUseRule=UseColor_ColorTable, specifiedColor=<rgba 0 0 0 0>}
       cornerRadii=<4 4 4 4>
      }
      tintColor {colorUseRule=UseColor_Specified, specifiedColor=<rgba 0.0284 0.0284 0.0284 1>}
      uVRegion {bIsValid=false, max=<0 0>, min=<0 0>}
     }
     focusedForegroundColor {colorUseRule=UseColor_Specified, specifiedColor=<rgba 1 1 1 1>}
     foregroundColor {colorUseRule=UseColor_Specified, specifiedColor=<rgba 0.5271 0.5271 0.5271 1>}
     hScrollBarPadding=<ltrb 0 0 0 0>
     padding=<ltrb 8 3 8 4>
     readOnlyForegroundColor {colorUseRule=UseColor_Foreground, specifiedColor=<rgba 1 0 1 1>}
     scrollBarStyle {
      thickness=8
      draggedThumbImage {
       bIsDynamicallyLoaded=false
       drawAs=RoundedBox
       imageType=NoImage
       mirroring=NoMirror
       resourceName=None
       resourceObject=null
       tiling=NoTile
       imageSize=<0 0>
       margin=<ltrb 0 0 0 0>
       outlineSettings {
        bUseBrushTransparency=false
        roundingType=FixedRadius
        width=0
        color {colorUseRule=UseColor_Specified, specifiedColor=<rgba 0 0 0 0>}
        cornerRadii=<4 4 4 4>
       }
       tintColor {colorUseRule=UseColor_Specified, specifiedColor=<rgba 0.2159 0.2159 0.2159 1>}
       uVRegion {bIsValid=false, max=<0 0>, min=<0 0>}
      }
      horizontalBackgroundImage {
       bIsDynamicallyLoaded=false
       drawAs=NoDrawType
       imageType=NoImage
       mirroring=NoMirror
       resourceName=None
       resourceObject=null
       tiling=NoTile
       imageSize=<0 0>
       margin=<ltrb 0 0 0 0>
       outlineSettings {
        bUseBrushTransparency=false
        roundingType=HalfHeightRadius
        width=0
        color {colorUseRule=UseColor_Specified, specifiedColor=<rgba 0 0 0 0>}
        cornerRadii=<0 0 0 0>
       }
       tintColor {colorUseRule=UseColor_Specified, specifiedColor=<rgba 1 1 1 1>}
       uVRegion {bIsValid=false, max=<0 0>, min=<0 0>}
      }
      horizontalBottomSlotImage {
       bIsDynamicallyLoaded=false
       drawAs=NoDrawType
       imageType=NoImage
       mirroring=NoMirror
       resourceName=None
       resourceObject=null
       tiling=NoTile
       imageSize=<0 0>
       margin=<ltrb 0 0 0 0>
       outlineSettings {
        bUseBrushTransparency=false
        roundingType=HalfHeightRadius
        width=0
        color {colorUseRule=UseColor_Specified, specifiedColor=<rgba 0 0 0 0>}
        cornerRadii=<0 0 0 0>
       }
       tintColor {colorUseRule=UseColor_Specified, specifiedColor=<rgba 1 1 1 1>}
       uVRegion {bIsValid=false, max=<0 0>, min=<0 0>}
      }
      horizontalTopSlotImage {
       bIsDynamicallyLoaded=false
       drawAs=NoDrawType
       imageType=NoImage
       mirroring=NoMirror
       resourceName=None
       resourceObject=null
       tiling=NoTile
       imageSize=<0 0>
       margin=<ltrb 0 0 0 0>
       outlineSettings {
        bUseBrushTransparency=false
        roundingType=HalfHeightRadius
        width=0
        color {colorUseRule=UseColor_Specified, specifiedColor=<rgba 0 0 0 0>}
        cornerRadii=<0 0 0 0>
       }
       tintColor {colorUseRule=UseColor_Specified, specifiedColor=<rgba 1 1 1 1>}
       uVRegion {bIsValid=false, max=<0 0>, min=<0 0>}
      }
      hoveredThumbImage {
       bIsDynamicallyLoaded=false
       drawAs=RoundedBox
       imageType=NoImage
       mirroring=NoMirror
       resourceName=None
       resourceObject=null
       tiling=NoTile
       imageSize=<0 0>
       margin=<ltrb 0 0 0 0>
       outlineSettings {
        bUseBrushTransparency=false
        roundingType=FixedRadius
        width=0
        color {colorUseRule=UseColor_Specified, specifiedColor=<rgba 0 0 0 0>}
        cornerRadii=<4 4 4 4>
       }
       tintColor {colorUseRule=UseColor_Specified, specifiedColor=<rgba 0.2159 0.2159 0.2159 1>}
       uVRegion {bIsValid=false, max=<0 0>, min=<0 0>}
      }
      normalThumbImage {
       bIsDynamicallyLoaded=false
       drawAs=RoundedBox
       imageType=NoImage
       mirroring=NoMirror
       resourceName=None
       resourceObject=null
       tiling=NoTile
       imageSize=<0 0>
       margin=<ltrb 0 0 0 0>
       outlineSettings {
        bUseBrushTransparency=false
        roundingType=FixedRadius
        width=0
        color {colorUseRule=UseColor_Specified, specifiedColor=<rgba 0 0 0 0>}
        cornerRadii=<4 4 4 4>
       }
       tintColor {colorUseRule=UseColor_Specified, specifiedColor=<rgba 0.0953 0.0953 0.0953 1>}
       uVRegion {bIsValid=false, max=<0 0>, min=<0 0>}
      }
      verticalBackgroundImage {
       bIsDynamicallyLoaded=false
       drawAs=NoDrawType
       imageType=NoImage
       mirroring=NoMirror
       resourceName=None
       resourceObject=null
       tiling=NoTile
       imageSize=<0 0>
       margin=<ltrb 0 0 0 0>
       outlineSettings {
        bUseBrushTransparency=false
        roundingType=HalfHeightRadius
        width=0
        color {colorUseRule=UseColor_Specified, specifiedColor=<rgba 0 0 0 0>}
        cornerRadii=<0 0 0 0>
       }
       tintColor {colorUseRule=UseColor_Specified, specifiedColor=<rgba 1 1 1 1>}
       uVRegion {bIsValid=false, max=<0 0>, min=<0 0>}
      }
      verticalBottomSlotImage {
       bIsDynamicallyLoaded=false
       drawAs=NoDrawType
       imageType=NoImage
       mirroring=NoMirror
       resourceName=None
       resourceObject=null
       tiling=NoTile
       imageSize=<0 0>
       margin=<ltrb 0 0 0 0>
       outlineSettings {
        bUseBrushTransparency=false
        roundingType=HalfHeightRadius
        width=0
        color {colorUseRule=UseColor_Specified, specifiedColor=<rgba 0 0 0 0>}
        cornerRadii=<0 0 0 0>
       }
       tintColor {colorUseRule=UseColor_Specified, specifiedColor=<rgba 1 1 1 1>}
       uVRegion {bIsValid=false, max=<0 0>, min=<0 0>}
      }
      verticalTopSlotImage {
       bIsDynamicallyLoaded=false
       drawAs=NoDrawType
       imageType=NoImage
       mirroring=NoMirror
       resourceName=None
       resourceObject=null
       tiling=NoTile
       imageSize=<0 0>
       margin=<ltrb 0 0 0 0>
       outlineSettings {
        bUseBrushTransparency=false
        roundingType=HalfHeightRadius
        width=0
        color {colorUseRule=UseColor_Specified, specifiedColor=<rgba 0 0 0 0>}
        cornerRadii=<0 0 0 0>
       }
       tintColor {colorUseRule=UseColor_Specified, specifiedColor=<rgba 1 1 1 1>}
       uVRegion {bIsValid=false, max=<0 0>, min=<0 0>}
      }
     }
     textStyle {
      overflowPolicy=Clip
      transformPolicy=None
      colorAndOpacity {colorUseRule=UseColor_Foreground, specifiedColor=<rgba 1 0 1 1>}
      font {
       bForceMonospaced=false
       bMaterialIsStencil=false
       fontMaterial=null
       fontObject=null
       letterSpacing=0
       monospacedWidth=1
       size=18
       skewAmount=0
       typefaceFontName=Regular
       outlineSettings {
        bApplyOutlineToDropShadows=false
        bMiteredCorners=false
        bSeparateFillAlpha=false
        outlineMaterial=null
        outlineSize=0
        outlineColor=<rgba 0 0 0 1>
       }
      }
      highlightColor {colorUseRule=UseColor_ColorTable, specifiedColor=<rgba 0 0 0 0>}
      highlightShape {
       bIsDynamicallyLoaded=false
       drawAs=Image
       imageType=NoImage
       mirroring=NoMirror
       resourceName=None
       resourceObject=null
       tiling=NoTile
       imageSize=<0 0>
       margin=<ltrb 0 0 0 0>
       outlineSettings {
        bUseBrushTransparency=false
        roundingType=HalfHeightRadius
        width=0
        color {colorUseRule=UseColor_Specified, specifiedColor=<rgba 0 0 0 0>}
        cornerRadii=<0 0 0 0>
       }
       tintColor {colorUseRule=UseColor_ColorTable, specifiedColor=<rgba 0 0 0 0>}
       uVRegion {bIsValid=false, max=<0 0>, min=<0 0>}
      }
      selectedBackgroundColor {colorUseRule=UseColor_ColorTable, specifiedColor=<rgba 0 0 0 0>}
      shadowColorAndOpacity=<rgba 0 0 0 1>
      shadowOffset=<0 0>
      strikeBrush {
       bIsDynamicallyLoaded=false
       drawAs=Image
       imageType=NoImage
       mirroring=NoMirror
       resourceName=None
       resourceObject=null
       tiling=NoTile
       imageSize=<32 32>
       margin=<ltrb 0 0 0 0>
       outlineSettings {
        bUseBrushTransparency=false
        roundingType=HalfHeightRadius
        width=0
        color {colorUseRule=UseColor_Specified, specifiedColor=<rgba 0 0 0 0>}
        cornerRadii=<0 0 0 0>
       }
       tintColor {colorUseRule=UseColor_Specified, specifiedColor=<rgba 1 1 1 1>}
       uVRegion {bIsValid=false, max=<0 0>, min=<0 0>}
      }
      underlineBrush {
       bIsDynamicallyLoaded=false
       drawAs=Image
       imageType=NoImage
       mirroring=NoMirror
       resourceName=None
       resourceObject=null
       tiling=NoTile
       imageSize=<32 32>
       margin=<ltrb 0 0 0 0>
       outlineSettings {
        bUseBrushTransparency=false
        roundingType=HalfHeightRadius
        width=0
        color {colorUseRule=UseColor_Specified, specifiedColor=<rgba 0 0 0 0>}
        cornerRadii=<0 0 0 0>
       }
       tintColor {colorUseRule=UseColor_Specified, specifiedColor=<rgba 1 1 1 1>}
       uVRegion {bIsValid=false, max=<0 0>, min=<0 0>}
      }
     }
     vScrollBarPadding=<ltrb 0 0 0 0>
    }
   }
   slot {class=VerticalBoxSlot}
  }
  HorizontalBox_1 {widgetClassPath=&HorizontalBox, parent=VerticalBox_0, widget=HorizontalBox_1, slot {class=VerticalBoxSlot}}
  TextBlock_192 {
   widgetClassPath=&TextBlock
   parent=HorizontalBox_1
   widget=TextBlock_192
   properties {text=Target Material Path:}
   slot {class=HorizontalBoxSlot, properties {padding=<ltrb 8 8 8 8>}}
  }
  VerticalBox_339 {
   widgetClassPath=&VerticalBox
   parent=HorizontalBox_1
   widget=VerticalBox_339
   slot {class=HorizontalBoxSlot, properties {verticalAlignment=VAlign_Center}}
  }
  TargetTextBox_ {
   widgetClassPath=&EditorUtilityEditableTextBox
   bIsVariable=true
   parent=VerticalBox_339
   widget=TargetTextBox_
   properties {
    text=/Figment_S04/Athena/Environments/Materials/CliffGround
    widgetStyle {
     backgroundColor {colorUseRule=UseColor_Specified, specifiedColor=<rgba 1 1 1 1>}
     backgroundImageFocused {
      bIsDynamicallyLoaded=false
      drawAs=RoundedBox
      imageType=NoImage
      mirroring=NoMirror
      resourceName=None
      resourceObject=null
      tiling=NoTile
      imageSize=<0 0>
      margin=<ltrb 0 0 0 0>
      outlineSettings {
       bUseBrushTransparency=false
       roundingType=FixedRadius
       width=1
       color {colorUseRule=UseColor_ColorTable, specifiedColor=<rgba 0 0 0 0>}
       cornerRadii=<4 4 4 4>
      }
      tintColor {colorUseRule=UseColor_Specified, specifiedColor=<rgba 0.0048 0.0048 0.0048 1>}
      uVRegion {bIsValid=false, max=<0 0>, min=<0 0>}
     }
     backgroundImageHovered {
      bIsDynamicallyLoaded=false
      drawAs=RoundedBox
      imageType=NoImage
      mirroring=NoMirror
      resourceName=None
      resourceObject=null
      tiling=NoTile
      imageSize=<0 0>
      margin=<ltrb 0 0 0 0>
      outlineSettings {
       bUseBrushTransparency=false
       roundingType=FixedRadius
       width=1
       color {colorUseRule=UseColor_ColorTable, specifiedColor=<rgba 0 0 0 0>}
       cornerRadii=<4 4 4 4>
      }
      tintColor {colorUseRule=UseColor_Specified, specifiedColor=<rgba 0.0048 0.0048 0.0048 1>}
      uVRegion {bIsValid=false, max=<0 0>, min=<0 0>}
     }
     backgroundImageNormal {
      bIsDynamicallyLoaded=false
      drawAs=RoundedBox
      imageType=NoImage
      mirroring=NoMirror
      resourceName=None
      resourceObject=null
      tiling=NoTile
      imageSize=<0 0>
      margin=<ltrb 0 0 0 0>
      outlineSettings {
       bUseBrushTransparency=false
       roundingType=FixedRadius
       width=1
       color {colorUseRule=UseColor_ColorTable, specifiedColor=<rgba 0 0 0 0>}
       cornerRadii=<4 4 4 4>
      }
      tintColor {colorUseRule=UseColor_Specified, specifiedColor=<rgba 0.0048 0.0048 0.0048 1>}
      uVRegion {bIsValid=false, max=<0 0>, min=<0 0>}
     }
     backgroundImageReadOnly {
      bIsDynamicallyLoaded=false
      drawAs=RoundedBox
      imageType=NoImage
      mirroring=NoMirror
      resourceName=None
      resourceObject=null
      tiling=NoTile
      imageSize=<0 0>
      margin=<ltrb 0 0 0 0>
      outlineSettings {
       bUseBrushTransparency=false
       roundingType=FixedRadius
       width=1
       color {colorUseRule=UseColor_ColorTable, specifiedColor=<rgba 0 0 0 0>}
       cornerRadii=<4 4 4 4>
      }
      tintColor {colorUseRule=UseColor_Specified, specifiedColor=<rgba 0.0284 0.0284 0.0284 1>}
      uVRegion {bIsValid=false, max=<0 0>, min=<0 0>}
     }
     focusedForegroundColor {colorUseRule=UseColor_Specified, specifiedColor=<rgba 1 1 1 1>}
     foregroundColor {colorUseRule=UseColor_Specified, specifiedColor=<rgba 0.5271 0.5271 0.5271 1>}
     hScrollBarPadding=<ltrb 0 0 0 0>
     padding=<ltrb 8 3 8 4>
     readOnlyForegroundColor {colorUseRule=UseColor_Foreground, specifiedColor=<rgba 1 0 1 1>}
     scrollBarStyle {
      thickness=8
      draggedThumbImage {
       bIsDynamicallyLoaded=false
       drawAs=RoundedBox
       imageType=NoImage
       mirroring=NoMirror
       resourceName=None
       resourceObject=null
       tiling=NoTile
       imageSize=<0 0>
       margin=<ltrb 0 0 0 0>
       outlineSettings {
        bUseBrushTransparency=false
        roundingType=FixedRadius
        width=0
        color {colorUseRule=UseColor_Specified, specifiedColor=<rgba 0 0 0 0>}
        cornerRadii=<4 4 4 4>
       }
       tintColor {colorUseRule=UseColor_Specified, specifiedColor=<rgba 0.2159 0.2159 0.2159 1>}
       uVRegion {bIsValid=false, max=<0 0>, min=<0 0>}
      }
      horizontalBackgroundImage {
       bIsDynamicallyLoaded=false
       drawAs=NoDrawType
       imageType=NoImage
       mirroring=NoMirror
       resourceName=None
       resourceObject=null
       tiling=NoTile
       imageSize=<0 0>
       margin=<ltrb 0 0 0 0>
       outlineSettings {
        bUseBrushTransparency=false
        roundingType=HalfHeightRadius
        width=0
        color {colorUseRule=UseColor_Specified, specifiedColor=<rgba 0 0 0 0>}
        cornerRadii=<0 0 0 0>
       }
       tintColor {colorUseRule=UseColor_Specified, specifiedColor=<rgba 1 1 1 1>}
       uVRegion {bIsValid=false, max=<0 0>, min=<0 0>}
      }
      horizontalBottomSlotImage {
       bIsDynamicallyLoaded=false
       drawAs=NoDrawType
       imageType=NoImage
       mirroring=NoMirror
       resourceName=None
       resourceObject=null
       tiling=NoTile
       imageSize=<0 0>
       margin=<ltrb 0 0 0 0>
       outlineSettings {
        bUseBrushTransparency=false
        roundingType=HalfHeightRadius
        width=0
        color {colorUseRule=UseColor_Specified, specifiedColor=<rgba 0 0 0 0>}
        cornerRadii=<0 0 0 0>
       }
       tintColor {colorUseRule=UseColor_Specified, specifiedColor=<rgba 1 1 1 1>}
       uVRegion {bIsValid=false, max=<0 0>, min=<0 0>}
      }
      horizontalTopSlotImage {
       bIsDynamicallyLoaded=false
       drawAs=NoDrawType
       imageType=NoImage
       mirroring=NoMirror
       resourceName=None
       resourceObject=null
       tiling=NoTile
       imageSize=<0 0>
       margin=<ltrb 0 0 0 0>
       outlineSettings {
        bUseBrushTransparency=false
        roundingType=HalfHeightRadius
        width=0
        color {colorUseRule=UseColor_Specified, specifiedColor=<rgba 0 0 0 0>}
        cornerRadii=<0 0 0 0>
       }
       tintColor {colorUseRule=UseColor_Specified, specifiedColor=<rgba 1 1 1 1>}
       uVRegion {bIsValid=false, max=<0 0>, min=<0 0>}
      }
      hoveredThumbImage {
       bIsDynamicallyLoaded=false
       drawAs=RoundedBox
       imageType=NoImage
       mirroring=NoMirror
       resourceName=None
       resourceObject=null
       tiling=NoTile
       imageSize=<0 0>
       margin=<ltrb 0 0 0 0>
       outlineSettings {
        bUseBrushTransparency=false
        roundingType=FixedRadius
        width=0
        color {colorUseRule=UseColor_Specified, specifiedColor=<rgba 0 0 0 0>}
        cornerRadii=<4 4 4 4>
       }
       tintColor {colorUseRule=UseColor_Specified, specifiedColor=<rgba 0.2159 0.2159 0.2159 1>}
       uVRegion {bIsValid=false, max=<0 0>, min=<0 0>}
      }
      normalThumbImage {
       bIsDynamicallyLoaded=false
       drawAs=RoundedBox
       imageType=NoImage
       mirroring=NoMirror
       resourceName=None
       resourceObject=null
       tiling=NoTile
       imageSize=<0 0>
       margin=<ltrb 0 0 0 0>
       outlineSettings {
        bUseBrushTransparency=false
        roundingType=FixedRadius
        width=0
        color {colorUseRule=UseColor_Specified, specifiedColor=<rgba 0 0 0 0>}
        cornerRadii=<4 4 4 4>
       }
       tintColor {colorUseRule=UseColor_Specified, specifiedColor=<rgba 0.0953 0.0953 0.0953 1>}
       uVRegion {bIsValid=false, max=<0 0>, min=<0 0>}
      }
      verticalBackgroundImage {
       bIsDynamicallyLoaded=false
       drawAs=NoDrawType
       imageType=NoImage
       mirroring=NoMirror
       resourceName=None
       resourceObject=null
       tiling=NoTile
       imageSize=<0 0>
       margin=<ltrb 0 0 0 0>
       outlineSettings {
        bUseBrushTransparency=false
        roundingType=HalfHeightRadius
        width=0
        color {colorUseRule=UseColor_Specified, specifiedColor=<rgba 0 0 0 0>}
        cornerRadii=<0 0 0 0>
       }
       tintColor {colorUseRule=UseColor_Specified, specifiedColor=<rgba 1 1 1 1>}
       uVRegion {bIsValid=false, max=<0 0>, min=<0 0>}
      }
      verticalBottomSlotImage {
       bIsDynamicallyLoaded=false
       drawAs=NoDrawType
       imageType=NoImage
       mirroring=NoMirror
       resourceName=None
       resourceObject=null
       tiling=NoTile
       imageSize=<0 0>
       margin=<ltrb 0 0 0 0>
       outlineSettings {
        bUseBrushTransparency=false
        roundingType=HalfHeightRadius
        width=0
        color {colorUseRule=UseColor_Specified, specifiedColor=<rgba 0 0 0 0>}
        cornerRadii=<0 0 0 0>
       }
       tintColor {colorUseRule=UseColor_Specified, specifiedColor=<rgba 1 1 1 1>}
       uVRegion {bIsValid=false, max=<0 0>, min=<0 0>}
      }
      verticalTopSlotImage {
       bIsDynamicallyLoaded=false
       drawAs=NoDrawType
       imageType=NoImage
       mirroring=NoMirror
       resourceName=None
       resourceObject=null
       tiling=NoTile
       imageSize=<0 0>
       margin=<ltrb 0 0 0 0>
       outlineSettings {
        bUseBrushTransparency=false
        roundingType=HalfHeightRadius
        width=0
        color {colorUseRule=UseColor_Specified, specifiedColor=<rgba 0 0 0 0>}
        cornerRadii=<0 0 0 0>
       }
       tintColor {colorUseRule=UseColor_Specified, specifiedColor=<rgba 1 1 1 1>}
       uVRegion {bIsValid=false, max=<0 0>, min=<0 0>}
      }
     }
     textStyle {
      overflowPolicy=Clip
      transformPolicy=None
      colorAndOpacity {colorUseRule=UseColor_Foreground, specifiedColor=<rgba 1 0 1 1>}
      font {
       bForceMonospaced=false
       bMaterialIsStencil=false
       fontMaterial=null
       fontObject=null
       letterSpacing=0
       monospacedWidth=1
       size=18
       skewAmount=0
       typefaceFontName=Regular
       outlineSettings {
        bApplyOutlineToDropShadows=false
        bMiteredCorners=false
        bSeparateFillAlpha=false
        outlineMaterial=null
        outlineSize=0
        outlineColor=<rgba 0 0 0 1>
       }
      }
      highlightColor {colorUseRule=UseColor_ColorTable, specifiedColor=<rgba 0 0 0 0>}
      highlightShape {
       bIsDynamicallyLoaded=false
       drawAs=Image
       imageType=NoImage
       mirroring=NoMirror
       resourceName=None
       resourceObject=null
       tiling=NoTile
       imageSize=<0 0>
       margin=<ltrb 0 0 0 0>
       outlineSettings {
        bUseBrushTransparency=false
        roundingType=HalfHeightRadius
        width=0
        color {colorUseRule=UseColor_Specified, specifiedColor=<rgba 0 0 0 0>}
        cornerRadii=<0 0 0 0>
       }
       tintColor {colorUseRule=UseColor_ColorTable, specifiedColor=<rgba 0 0 0 0>}
       uVRegion {bIsValid=false, max=<0 0>, min=<0 0>}
      }
      selectedBackgroundColor {colorUseRule=UseColor_ColorTable, specifiedColor=<rgba 0 0 0 0>}
      shadowColorAndOpacity=<rgba 0 0 0 1>
      shadowOffset=<0 0>
      strikeBrush {
       bIsDynamicallyLoaded=false
       drawAs=Image
       imageType=NoImage
       mirroring=NoMirror
       resourceName=None
       resourceObject=null
       tiling=NoTile
       imageSize=<32 32>
       margin=<ltrb 0 0 0 0>
       outlineSettings {
        bUseBrushTransparency=false
        roundingType=HalfHeightRadius
        width=0
        color {colorUseRule=UseColor_Specified, specifiedColor=<rgba 0 0 0 0>}
        cornerRadii=<0 0 0 0>
       }
       tintColor {colorUseRule=UseColor_Specified, specifiedColor=<rgba 1 1 1 1>}
       uVRegion {bIsValid=false, max=<0 0>, min=<0 0>}
      }
      underlineBrush {
       bIsDynamicallyLoaded=false
       drawAs=Image
       imageType=NoImage
       mirroring=NoMirror
       resourceName=None
       resourceObject=null
       tiling=NoTile
       imageSize=<32 32>
       margin=<ltrb 0 0 0 0>
       outlineSettings {
        bUseBrushTransparency=false
        roundingType=HalfHeightRadius
        width=0
        color {colorUseRule=UseColor_Specified, specifiedColor=<rgba 0 0 0 0>}
        cornerRadii=<0 0 0 0>
       }
       tintColor {colorUseRule=UseColor_Specified, specifiedColor=<rgba 1 1 1 1>}
       uVRegion {bIsValid=false, max=<0 0>, min=<0 0>}
      }
     }
     vScrollBarPadding=<ltrb 0 0 0 0>
    }
   }
   slot {class=VerticalBoxSlot}
  }
  Spacer_2 {widgetClassPath=&Spacer, parent=VerticalBox_0, widget=Spacer_2, properties {size=<1 48>}, slot {class=VerticalBoxSlot}}
  HorizontalBox_102 {widgetClassPath=&HorizontalBox, parent=VerticalBox_0, widget=HorizontalBox_102, slot {class=VerticalBoxSlot}}
  ConverttoBlueprint {
   widgetClassPath=&EditorUtilityButton
   parent=HorizontalBox_102
   widget=ConverttoBlueprint
   slot {class=HorizontalBoxSlot, properties {horizontalAlignment=HAlign_Left}}
  }
  TextBlock_1 {
   widgetClassPath=&TextBlock
   parent=ConverttoBlueprint
   widget=TextBlock_1
   properties {text=Convert Selected SM to Blueprint}
   slot {class=ButtonSlot, properties {horizontalAlignment=HAlign_Left}}
  }
  Spacer_76 {widgetClassPath=&Spacer, parent=HorizontalBox_102, widget=Spacer_76, properties {size=<12 1>}, slot {class=HorizontalBoxSlot}}
  TextBlock_144 {
   widgetClassPath=&TextBlock
   parent=HorizontalBox_102
   widget=TextBlock_144
   properties {
    text=apply to all in current level:
    font {
     bForceMonospaced=false
     bMaterialIsStencil=false
     fontMaterial=null
     fontObject=&Roboto
     letterSpacing=0
     monospacedWidth=1
     size=12
     skewAmount=0
     typefaceFontName=Bold
     outlineSettings {
      bApplyOutlineToDropShadows=false
      bMiteredCorners=false
      bSeparateFillAlpha=false
      outlineMaterial=null
      outlineSize=0
      outlineColor=<rgba 0 0 0 1>
     }
    }
   }
   slot {class=HorizontalBoxSlot, properties {verticalAlignment=VAlign_Center}}
  }
  Spacer_1 {widgetClassPath=&Spacer, parent=HorizontalBox_102, widget=Spacer_1, properties {size=<4 1>}, slot {class=HorizontalBoxSlot}}
  SMConvertApplyToAll {
   widgetClassPath=&EditorUtilityCheckBox
   bIsVariable=true
   parent=HorizontalBox_102
   widget=SMConvertApplyToAll
   slot {class=HorizontalBoxSlot}
  }
  Spacer_3 {widgetClassPath=&Spacer, parent=VerticalBox_0, widget=Spacer_3, properties {size=<1 48>}, slot {class=VerticalBoxSlot}}
  CanvasPanel_84 {widgetClassPath=&CanvasPanel, parent=WidgetSwitcher_184, widget=CanvasPanel_84, slot {class=WidgetSwitcherSlot}}
  VerticalBox_59 {
   widgetClassPath=&VerticalBox
   parent=CanvasPanel_84
   widget=VerticalBox_59
   slot {class=CanvasPanelSlot, properties {bAutoSize=true, layoutData=~}}
  }
  CreateMatchingBP {
   widgetClassPath=&EditorUtilityButton
   parent=VerticalBox_59
   widget=CreateMatchingBP
   properties {bIsEnabled=false}
   slot {class=VerticalBoxSlot}
  }
  TextBlock_5 {
   widgetClassPath=&TextBlock
   parent=CreateMatchingBP
   widget=TextBlock_5
   properties {text=CreateMatchingBP}
   slot {class=ButtonSlot, properties {horizontalAlignment=HAlign_Left}}
  }
  ClearTagsInLevel {widgetClassPath=&EditorUtilityButton, parent=VerticalBox_59, widget=ClearTagsInLevel, slot {class=VerticalBoxSlot}}
  TextBlock_10 {
   widgetClassPath=&TextBlock
   parent=ClearTagsInLevel
   widget=TextBlock_10
   properties {text=ClearTagsInLevel}
   slot {class=ButtonSlot, properties {horizontalAlignment=HAlign_Left}}
  }
  DeleteFoliage {widgetClassPath=&EditorUtilityButton, parent=VerticalBox_59, widget=DeleteFoliage, slot {class=VerticalBoxSlot}}
  TextBlock_2 {
   widgetClassPath=&TextBlock
   parent=DeleteFoliage
   widget=TextBlock_2
   properties {text=DeleteFoliageActorsInLevel}
   slot {class=ButtonSlot, properties {horizontalAlignment=HAlign_Left}}
  }
  HorizontalBox_2 {widgetClassPath=&HorizontalBox, parent=VerticalBox_59, widget=HorizontalBox_2, slot {class=VerticalBoxSlot}}
  TextBlock_8 {
   widgetClassPath=&TextBlock
   parent=HorizontalBox_2
   widget=TextBlock_8
   properties {text=Source Level Instance Top Level Path:}
   slot {class=HorizontalBoxSlot, properties {padding=<ltrb 8 8 8 8>}}
  }
  VerticalBox_1 {
   widgetClassPath=&VerticalBox
   parent=HorizontalBox_2
   widget=VerticalBox_1
   slot {class=HorizontalBoxSlot, properties {verticalAlignment=VAlign_Center}}
  }
  SourceTextBox_pg2 {
   widgetClassPath=&EditorUtilityEditableTextBox
   bIsVariable=true
   parent=VerticalBox_1
   widget=SourceTextBox_pg2
   properties {
    text=/Figment_S04_LevelInstances/
    widgetStyle {
     backgroundColor {colorUseRule=UseColor_Specified, specifiedColor=<rgba 1 1 1 1>}
     backgroundImageFocused {
      bIsDynamicallyLoaded=false
      drawAs=RoundedBox
      imageType=NoImage
      mirroring=NoMirror
      resourceName=None
      resourceObject=null
      tiling=NoTile
      imageSize=<0 0>
      margin=<ltrb 0 0 0 0>
      outlineSettings {
       bUseBrushTransparency=false
       roundingType=FixedRadius
       width=1
       color {colorUseRule=UseColor_ColorTable, specifiedColor=<rgba 0 0 0 0>}
       cornerRadii=<4 4 4 4>
      }
      tintColor {colorUseRule=UseColor_Specified, specifiedColor=<rgba 0.0048 0.0048 0.0048 1>}
      uVRegion {bIsValid=false, max=<0 0>, min=<0 0>}
     }
     backgroundImageHovered {
      bIsDynamicallyLoaded=false
      drawAs=RoundedBox
      imageType=NoImage
      mirroring=NoMirror
      resourceName=None
      resourceObject=null
      tiling=NoTile
      imageSize=<0 0>
      margin=<ltrb 0 0 0 0>
      outlineSettings {
       bUseBrushTransparency=false
       roundingType=FixedRadius
       width=1
       color {colorUseRule=UseColor_ColorTable, specifiedColor=<rgba 0 0 0 0>}
       cornerRadii=<4 4 4 4>
      }
      tintColor {colorUseRule=UseColor_Specified, specifiedColor=<rgba 0.0048 0.0048 0.0048 1>}
      uVRegion {bIsValid=false, max=<0 0>, min=<0 0>}
     }
     backgroundImageNormal {
      bIsDynamicallyLoaded=false
      drawAs=RoundedBox
      imageType=NoImage
      mirroring=NoMirror
      resourceName=None
      resourceObject=null
      tiling=NoTile
      imageSize=<0 0>
      margin=<ltrb 0 0 0 0>
      outlineSettings {
       bUseBrushTransparency=false
       roundingType=FixedRadius
       width=1
       color {colorUseRule=UseColor_ColorTable, specifiedColor=<rgba 0 0 0 0>}
       cornerRadii=<4 4 4 4>
      }
      tintColor {colorUseRule=UseColor_Specified, specifiedColor=<rgba 0.0048 0.0048 0.0048 1>}
      uVRegion {bIsValid=false, max=<0 0>, min=<0 0>}
     }
     backgroundImageReadOnly {
      bIsDynamicallyLoaded=false
      drawAs=RoundedBox
      imageType=NoImage
      mirroring=NoMirror
      resourceName=None
      resourceObject=null
      tiling=NoTile
      imageSize=<0 0>
      margin=<ltrb 0 0 0 0>
      outlineSettings {
       bUseBrushTransparency=false
       roundingType=FixedRadius
       width=1
       color {colorUseRule=UseColor_ColorTable, specifiedColor=<rgba 0 0 0 0>}
       cornerRadii=<4 4 4 4>
      }
      tintColor {colorUseRule=UseColor_Specified, specifiedColor=<rgba 0.0284 0.0284 0.0284 1>}
      uVRegion {bIsValid=false, max=<0 0>, min=<0 0>}
     }
     focusedForegroundColor {colorUseRule=UseColor_Specified, specifiedColor=<rgba 1 1 1 1>}
     foregroundColor {colorUseRule=UseColor_Specified, specifiedColor=<rgba 0.5271 0.5271 0.5271 1>}
     hScrollBarPadding=<ltrb 0 0 0 0>
     padding=<ltrb 8 3 8 4>
     readOnlyForegroundColor {colorUseRule=UseColor_Foreground, specifiedColor=<rgba 1 0 1 1>}
     scrollBarStyle {
      thickness=8
      draggedThumbImage {
       bIsDynamicallyLoaded=false
       drawAs=RoundedBox
       imageType=NoImage
       mirroring=NoMirror
       resourceName=None
       resourceObject=null
       tiling=NoTile
       imageSize=<0 0>
       margin=<ltrb 0 0 0 0>
       outlineSettings {
        bUseBrushTransparency=false
        roundingType=FixedRadius
        width=0
        color {colorUseRule=UseColor_Specified, specifiedColor=<rgba 0 0 0 0>}
        cornerRadii=<4 4 4 4>
       }
       tintColor {colorUseRule=UseColor_Specified, specifiedColor=<rgba 0.2159 0.2159 0.2159 1>}
       uVRegion {bIsValid=false, max=<0 0>, min=<0 0>}
      }
      horizontalBackgroundImage {
       bIsDynamicallyLoaded=false
       drawAs=NoDrawType
       imageType=NoImage
       mirroring=NoMirror
       resourceName=None
       resourceObject=null
       tiling=NoTile
       imageSize=<0 0>
       margin=<ltrb 0 0 0 0>
       outlineSettings {
        bUseBrushTransparency=false
        roundingType=HalfHeightRadius
        width=0
        color {colorUseRule=UseColor_Specified, specifiedColor=<rgba 0 0 0 0>}
        cornerRadii=<0 0 0 0>
       }
       tintColor {colorUseRule=UseColor_Specified, specifiedColor=<rgba 1 1 1 1>}
       uVRegion {bIsValid=false, max=<0 0>, min=<0 0>}
      }
      horizontalBottomSlotImage {
       bIsDynamicallyLoaded=false
       drawAs=NoDrawType
       imageType=NoImage
       mirroring=NoMirror
       resourceName=None
       resourceObject=null
       tiling=NoTile
       imageSize=<0 0>
       margin=<ltrb 0 0 0 0>
       outlineSettings {
        bUseBrushTransparency=false
        roundingType=HalfHeightRadius
        width=0
        color {colorUseRule=UseColor_Specified, specifiedColor=<rgba 0 0 0 0>}
        cornerRadii=<0 0 0 0>
       }
       tintColor {colorUseRule=UseColor_Specified, specifiedColor=<rgba 1 1 1 1>}
       uVRegion {bIsValid=false, max=<0 0>, min=<0 0>}
      }
      horizontalTopSlotImage {
       bIsDynamicallyLoaded=false
       drawAs=NoDrawType
       imageType=NoImage
       mirroring=NoMirror
       resourceName=None
       resourceObject=null
       tiling=NoTile
       imageSize=<0 0>
       margin=<ltrb 0 0 0 0>
       outlineSettings {
        bUseBrushTransparency=false
        roundingType=HalfHeightRadius
        width=0
        color {colorUseRule=UseColor_Specified, specifiedColor=<rgba 0 0 0 0>}
        cornerRadii=<0 0 0 0>
       }
       tintColor {colorUseRule=UseColor_Specified, specifiedColor=<rgba 1 1 1 1>}
       uVRegion {bIsValid=false, max=<0 0>, min=<0 0>}
      }
      hoveredThumbImage {
       bIsDynamicallyLoaded=false
       drawAs=RoundedBox
       imageType=NoImage
       mirroring=NoMirror
       resourceName=None
       resourceObject=null
       tiling=NoTile
       imageSize=<0 0>
       margin=<ltrb 0 0 0 0>
       outlineSettings {
        bUseBrushTransparency=false
        roundingType=FixedRadius
        width=0
        color {colorUseRule=UseColor_Specified, specifiedColor=<rgba 0 0 0 0>}
        cornerRadii=<4 4 4 4>
       }
       tintColor {colorUseRule=UseColor_Specified, specifiedColor=<rgba 0.2159 0.2159 0.2159 1>}
       uVRegion {bIsValid=false, max=<0 0>, min=<0 0>}
      }
      normalThumbImage {
       bIsDynamicallyLoaded=false
       drawAs=RoundedBox
       imageType=NoImage
       mirroring=NoMirror
       resourceName=None
       resourceObject=null
       tiling=NoTile
       imageSize=<0 0>
       margin=<ltrb 0 0 0 0>
       outlineSettings {
        bUseBrushTransparency=false
        roundingType=FixedRadius
        width=0
        color {colorUseRule=UseColor_Specified, specifiedColor=<rgba 0 0 0 0>}
        cornerRadii=<4 4 4 4>
       }
       tintColor {colorUseRule=UseColor_Specified, specifiedColor=<rgba 0.0953 0.0953 0.0953 1>}
       uVRegion {bIsValid=false, max=<0 0>, min=<0 0>}
      }
      verticalBackgroundImage {
       bIsDynamicallyLoaded=false
       drawAs=NoDrawType
       imageType=NoImage
       mirroring=NoMirror
       resourceName=None
       resourceObject=null
       tiling=NoTile
       imageSize=<0 0>
       margin=<ltrb 0 0 0 0>
       outlineSettings {
        bUseBrushTransparency=false
        roundingType=HalfHeightRadius
        width=0
        color {colorUseRule=UseColor_Specified, specifiedColor=<rgba 0 0 0 0>}
        cornerRadii=<0 0 0 0>
       }
       tintColor {colorUseRule=UseColor_Specified, specifiedColor=<rgba 1 1 1 1>}
       uVRegion {bIsValid=false, max=<0 0>, min=<0 0>}
      }
      verticalBottomSlotImage {
       bIsDynamicallyLoaded=false
       drawAs=NoDrawType
       imageType=NoImage
       mirroring=NoMirror
       resourceName=None
       resourceObject=null
       tiling=NoTile
       imageSize=<0 0>
       margin=<ltrb 0 0 0 0>
       outlineSettings {
        bUseBrushTransparency=false
        roundingType=HalfHeightRadius
        width=0
        color {colorUseRule=UseColor_Specified, specifiedColor=<rgba 0 0 0 0>}
        cornerRadii=<0 0 0 0>
       }
       tintColor {colorUseRule=UseColor_Specified, specifiedColor=<rgba 1 1 1 1>}
       uVRegion {bIsValid=false, max=<0 0>, min=<0 0>}
      }
      verticalTopSlotImage {
       bIsDynamicallyLoaded=false
       drawAs=NoDrawType
       imageType=NoImage
       mirroring=NoMirror
       resourceName=None
       resourceObject=null
       tiling=NoTile
       imageSize=<0 0>
       margin=<ltrb 0 0 0 0>
       outlineSettings {
        bUseBrushTransparency=false
        roundingType=HalfHeightRadius
        width=0
        color {colorUseRule=UseColor_Specified, specifiedColor=<rgba 0 0 0 0>}
        cornerRadii=<0 0 0 0>
       }
       tintColor {colorUseRule=UseColor_Specified, specifiedColor=<rgba 1 1 1 1>}
       uVRegion {bIsValid=false, max=<0 0>, min=<0 0>}
      }
     }
     textStyle {
      overflowPolicy=Clip
      transformPolicy=None
      colorAndOpacity {colorUseRule=UseColor_Foreground, specifiedColor=<rgba 1 0 1 1>}
      font {
       bForceMonospaced=false
       bMaterialIsStencil=false
       fontMaterial=null
       fontObject=null
       letterSpacing=0
       monospacedWidth=1
       size=18
       skewAmount=0
       typefaceFontName=Regular
       outlineSettings {
        bApplyOutlineToDropShadows=false
        bMiteredCorners=false
        bSeparateFillAlpha=false
        outlineMaterial=null
        outlineSize=0
        outlineColor=<rgba 0 0 0 1>
       }
      }
      highlightColor {colorUseRule=UseColor_ColorTable, specifiedColor=<rgba 0 0 0 0>}
      highlightShape {
       bIsDynamicallyLoaded=false
       drawAs=Image
       imageType=NoImage
       mirroring=NoMirror
       resourceName=None
       resourceObject=null
       tiling=NoTile
       imageSize=<0 0>
       margin=<ltrb 0 0 0 0>
       outlineSettings {
        bUseBrushTransparency=false
        roundingType=HalfHeightRadius
        width=0
        color {colorUseRule=UseColor_Specified, specifiedColor=<rgba 0 0 0 0>}
        cornerRadii=<0 0 0 0>
       }
       tintColor {colorUseRule=UseColor_ColorTable, specifiedColor=<rgba 0 0 0 0>}
       uVRegion {bIsValid=false, max=<0 0>, min=<0 0>}
      }
      selectedBackgroundColor {colorUseRule=UseColor_ColorTable, specifiedColor=<rgba 0 0 0 0>}
      shadowColorAndOpacity=<rgba 0 0 0 1>
      shadowOffset=<0 0>
      strikeBrush {
       bIsDynamicallyLoaded=false
       drawAs=Image
       imageType=NoImage
       mirroring=NoMirror
       resourceName=None
       resourceObject=null
       tiling=NoTile
       imageSize=<32 32>
       margin=<ltrb 0 0 0 0>
       outlineSettings {
        bUseBrushTransparency=false
        roundingType=HalfHeightRadius
        width=0
        color {colorUseRule=UseColor_Specified, specifiedColor=<rgba 0 0 0 0>}
        cornerRadii=<0 0 0 0>
       }
       tintColor {colorUseRule=UseColor_Specified, specifiedColor=<rgba 1 1 1 1>}
       uVRegion {bIsValid=false, max=<0 0>, min=<0 0>}
      }
      underlineBrush {
       bIsDynamicallyLoaded=false
       drawAs=Image
       imageType=NoImage
       mirroring=NoMirror
       resourceName=None
       resourceObject=null
       tiling=NoTile
       imageSize=<32 32>
       margin=<ltrb 0 0 0 0>
       outlineSettings {
        bUseBrushTransparency=false
        roundingType=HalfHeightRadius
        width=0
        color {colorUseRule=UseColor_Specified, specifiedColor=<rgba 0 0 0 0>}
        cornerRadii=<0 0 0 0>
       }
       tintColor {colorUseRule=UseColor_Specified, specifiedColor=<rgba 1 1 1 1>}
       uVRegion {bIsValid=false, max=<0 0>, min=<0 0>}
      }
     }
     vScrollBarPadding=<ltrb 0 0 0 0>
    }
   }
   slot {class=VerticalBoxSlot}
  }
  DeleteFoliage_1 {widgetClassPath=&EditorUtilityButton, parent=VerticalBox_59, widget=DeleteFoliage_1, slot {class=VerticalBoxSlot}}
  TextBlock_9 {
   widgetClassPath=&TextBlock
   parent=DeleteFoliage_1
   widget=TextBlock_9
   properties {text=DeleteFoliageActorsInPath}
   slot {class=ButtonSlot, properties {horizontalAlignment=HAlign_Left}}
  }
  ClearTagsInPath {widgetClassPath=&EditorUtilityButton, parent=VerticalBox_59, widget=ClearTagsInPath, slot {class=VerticalBoxSlot}}
  TextBlock_11 {
   widgetClassPath=&TextBlock
   parent=ClearTagsInPath
   widget=TextBlock_11
   properties {text=ClearTagsInPath}
   slot {class=ButtonSlot, properties {horizontalAlignment=HAlign_Left}}
  }
  Border_47 {
   widgetClassPath=&Border
   parent=VerticalBox_59
   widget=Border_47
   properties {
    background {
     bIsDynamicallyLoaded=false
     drawAs=RoundedBox
     imageType=NoImage
     mirroring=NoMirror
     resourceName=None
     resourceObject=null
     tiling=NoTile
     imageSize=<32 32>
     margin=<ltrb 0 0 0 0>
     outlineSettings {
      bUseBrushTransparency=false
      roundingType=FixedRadius
      width=1.8204
      color {colorUseRule=UseColor_Specified, specifiedColor=<rgba 1 1 1 1>}
      cornerRadii=<20 20 20 20>
     }
     tintColor {colorUseRule=UseColor_Specified, specifiedColor=<rgba 0.0802 0.0808 0.0885 1>}
     uVRegion {bIsValid=false, max=<0 0>, min=<0 0>}
    }
    padding=<ltrb 20 20 20 20>
   }
   slot {class=VerticalBoxSlot}
  }
  VerticalBox_88 {
   widgetClassPath=&VerticalBox
   parent=Border_47
   widget=VerticalBox_88
   slot {class=BorderSlot, properties {padding=<ltrb 20 20 20 20>}}
  }
  HorizontalBox_16 {
   widgetClassPath=&HorizontalBox
   parent=VerticalBox_88
   widget=HorizontalBox_16
   slot {class=VerticalBoxSlot, properties {size {sizeRule=Fill, value=1}}}
  }
  SinglePropertyView_0 {
   widgetClassPath=&SinglePropertyView
   parent=HorizontalBox_16
   widget=SinglePropertyView_0
   properties {propertyName=MICFrom}
   slot {class=HorizontalBoxSlot}
  }
  TextBlock_59 {widgetClassPath=&TextBlock, parent=HorizontalBox_16, widget=TextBlock_59, properties {text="->"}, slot {class=HorizontalBoxSlot}}
  SinglePropertyView {
   widgetClassPath=&SinglePropertyView
   parent=HorizontalBox_16
   widget=SinglePropertyView
   properties {propertyName=MICTo}
   slot {class=HorizontalBoxSlot}
  }
  EditorUtilityButton_5 {
   widgetClassPath=&EditorUtilityButton
   parent=VerticalBox_88
   widget=EditorUtilityButton_5
   slot {class=VerticalBoxSlot, properties {horizontalAlignment=HAlign_Left}}
  }
  TextBlock_4 {
   widgetClassPath=&TextBlock
   parent=EditorUtilityButton_5
   widget=TextBlock_4
   properties {text=Load material from selected Actor}
   slot {class=ButtonSlot, properties {horizontalAlignment=HAlign_Left}}
  }
  EditorUtilityButton_4 {
   widgetClassPath=&EditorUtilityButton
   parent=VerticalBox_88
   widget=EditorUtilityButton_4
   slot {class=VerticalBoxSlot, properties {horizontalAlignment=HAlign_Left}}
  }
  TextBlock_3 {
   widgetClassPath=&TextBlock
   parent=EditorUtilityButton_4
   widget=TextBlock_3
   properties {text=Replace Actors using this material with another}
   slot {class=ButtonSlot, properties {horizontalAlignment=HAlign_Left}}
  }
  Spacer {widgetClassPath=&Spacer, parent=VerticalBox_59, widget=Spacer, properties {size=<1 24>}, slot {class=VerticalBoxSlot}}
  Border {
   widgetClassPath=&Border
   parent=VerticalBox_59
   widget=Border
   properties {
    background {
     bIsDynamicallyLoaded=false
     drawAs=RoundedBox
     imageType=NoImage
     mirroring=NoMirror
     resourceName=None
     resourceObject=null
     tiling=NoTile
     imageSize=<32 32>
     margin=<ltrb 0 0 0 0>
     outlineSettings {
      bUseBrushTransparency=false
      roundingType=FixedRadius
      width=1.8204
      color {colorUseRule=UseColor_Specified, specifiedColor=<rgba 1 1 1 1>}
      cornerRadii=<20 20 20 20>
     }
     tintColor {colorUseRule=UseColor_Specified, specifiedColor=<rgba 0.0802 0.0808 0.0885 1>}
     uVRegion {bIsValid=false, max=<0 0>, min=<0 0>}
    }
    padding=<ltrb 20 20 20 20>
   }
   slot {class=VerticalBoxSlot}
  }
  VerticalBox {widgetClassPath=&VerticalBox, parent=Border, widget=VerticalBox, slot {class=BorderSlot, properties {padding=<ltrb 20 20 20 20>}}}
  HorizontalBox {
   widgetClassPath=&HorizontalBox
   parent=VerticalBox
   widget=HorizontalBox
   slot {class=VerticalBoxSlot, properties {size {sizeRule=Fill, value=1}}}
  }
  SinglePropertyView_1 {
   widgetClassPath=&SinglePropertyView
   parent=HorizontalBox
   widget=SinglePropertyView_1
   properties {propertyName=TD Asset}
   slot {class=HorizontalBoxSlot}
  }
  EditorUtilityButton_7 {
   widgetClassPath=&EditorUtilityButton
   parent=VerticalBox
   widget=EditorUtilityButton_7
   slot {class=VerticalBoxSlot, properties {horizontalAlignment=HAlign_Left}}
  }
  TextBlock_7 {
   widgetClassPath=&TextBlock
   parent=EditorUtilityButton_7
   widget=TextBlock_7
   properties {text=Apply this Texture Data Asset to Ground Meshes with no TD}
   slot {class=ButtonSlot, properties {horizontalAlignment=HAlign_Left}}
  }
  FixUpInLevel {
   widgetClassPath=&EditorUtilityButton
   parent=VerticalBox_59
   widget=FixUpInLevel
   slot {class=VerticalBoxSlot, properties {horizontalAlignment=HAlign_Left}}
  }
  TextBlock_0 {
   widgetClassPath=&TextBlock
   parent=FixUpInLevel
   widget=TextBlock_0
   properties {text=Fix up actors in current level}
   slot {class=ButtonSlot, properties {horizontalAlignment=HAlign_Left}}
  }
  PaintConcreteVertexColorsRedInLevel {
   widgetClassPath=&EditorUtilityButton
   parent=VerticalBox_59
   widget=PaintConcreteVertexColorsRedInLevel
   slot {class=VerticalBoxSlot, properties {horizontalAlignment=HAlign_Left}}
  }
  TextBlock_12 {
   widgetClassPath=&TextBlock
   parent=PaintConcreteVertexColorsRedInLevel
   widget=TextBlock_12
   properties {text=Paint concrete vertex actors in Path}
   slot {class=ButtonSlot, properties {horizontalAlignment=HAlign_Left}}
  }
  SinglePropertyView_2 {
   widgetClassPath=&SinglePropertyView
   parent=CanvasPanel_84
   widget=SinglePropertyView_2
   properties {propertyName=FigmentParentMAt}
   slot {class=CanvasPanelSlot, properties=~}
  }
  Spacer_4 {
   widgetClassPath=&Spacer
   parent=CanvasPanel_84
   widget=Spacer_4
   properties {size=<1 48>}
   slot {class=CanvasPanelSlot, properties=~}
  }
 ]
}
