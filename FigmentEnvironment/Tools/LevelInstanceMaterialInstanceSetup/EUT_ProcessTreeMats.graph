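# Editor utility task that batch-fixes selected material instances, apparently for tree materials, so they use snow and the Athena terrain topo mask.
# It loops over the selected assets, casts each to MaterialInstanceConstant, then turns on the UseSnow static switch and assigns TopoMask wherever either differs.
# Edits set a Changed flag, but nothing saves the assets or reads that flag.
# Cancel requests set an unchecked flag, and a delayed integer-loop test sits disconnected.
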EditorUtilityBlueprint /FigmentEnvironment/Tools/LevelInstanceMaterialInstanceSetup/EUT_ProcessTreeMats
graphs [EventGraph {nodes [
   Delay!(execute, WorldContextObject:Object Reference, Duration=5) -> (then)
   > n1 = For Each Loop(Array:Array of Integers=Get intarray() -> (intarray:Array of Integers)) -> (Array Element:int)
   > PrintString(InString=Append(A="in loop: ", B="To String (Integer)"(InInt=@n1)))
   Event BeginExecution()
   > SetTaskNotificationText!(execute, self:Editor Utility Task Object Reference, Text:"Text (by ref)"=MakeLiteralText(Value="LOCTEXT(\"Processing Mats finished\")")) -> (then)
   > PrintString(InString=Started Processing)
   > n2 = GetSelectedAssets()
   > n3 = For Each Loop(Array:Array of Object References=@n2) -> (Array Element:Object Reference)
   FinishExecutingTask(execute=@n3.Completed)
   n4 = Cast To MaterialInstanceConstant(execute=@n3.LoopBody, Object=@n3)
   > PrintString(InString=GetDisplayName(Object=@n4))
   > n5 = Sequence()
   > n6 = Branch(Condition=GetMaterialInstanceStaticSwitchParameterValue!(
     Instance:Material Instance Constant Object Reference=@n4
     ParameterName:Name=UseSnow
     Association:EMaterialParameterAssociation Enum=GlobalParameter
    ) -> (ReturnValue:bool))
   SetMaterialInstanceStaticSwitchParameterValue!(
    execute=@n6.else
    Instance:Material Instance Constant Object Reference=@n4
    ParameterName:Name=UseSnow
    Value:bool=true
    Association:EMaterialParameterAssociation Enum=GlobalParameter
    bUpdateMaterialInstance:bool=true
   ) -> (then, ReturnValue:bool)
   > Set Changed(Changed:bool=true) -> (Output_Get:bool)
   n7 = Branch(
    execute=@n5.then_1
    Condition="Equal (Object)"(
     A=GetMaterialInstanceTextureParameterValue!(
      Instance:Material Instance Constant Object Reference=@n4
      ParameterName:Name=TopoMask
      Association:EMaterialParameterAssociation Enum=GlobalParameter
     ) -> (ReturnValue:Texture Object Reference)
     B=/Figment_S07/LM_7_00/Athena/Environments/Landscape/Textures/T_Athena_Terrain_Topo_Mask
    )
   )
   SetMaterialInstanceTextureParameterValue!(
    execute=@n7.else
    Instance:Material Instance Constant Object Reference=@n4
    ParameterName:Name=TopoMask
    Value:Texture Object Reference=&T_Athena_Terrain_Topo_Mask
    Association:EMaterialParameterAssociation Enum=GlobalParameter
   ) -> (then, ReturnValue:bool)
   > Set Changed(Changed=true)
   Event CancelRequested() {is=title K2Node_Event}
   > Set Cancel Pending(Cancel Pending:bool=true) -> (Output_Get:bool)
   Get Cancel Pending() -> (Cancel Pending:bool)
  ]}]
variables [
 Cancel Pending {type=boolean}
 Changed {type=boolean}
 intarray {type="array<integer>"}
]
