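# Editor utility task that saves a set of actors after a delay, then reloads building level instances, apparently for the LevelInstanceMaterialInstanceSetup tool.
# On BeginExecution it waits asynchronously, prints a debug message and force-saves the Actors array, including non-dirty assets.
# It then loops over level_instance_cleanup, calling LoadLevelInstance through LevelInstanceInterface and FinishExecutingTask inside the loop body.
# The ActorsToSaveMap loop, second delay and UnloadLevelInstance nodes have no incoming execution.
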EditorUtilityBlueprint /FigmentEnvironment/Tools/LevelInstanceMaterialInstanceSetup/EUT_SaveActorsAsync
graphs [EventGraph {nodes [
   n1 = AsyncEditorDelay!(execute:Exec, Seconds:float=0.2, MinimumFrames:int=30) -> (then:Exec, Complete:Exec) {is=title K2Node_AsyncAction}
   n2 = For Each Loop(
    Exec=[@n1.Complete, @n8.then]
    Array:Array of Building Level Instance Object References=Get level_instance_cleanup() -> (level_instance_cleanup:Array of Building Level Instance Object References)
   ) -> (Array Element:Building Level Instance Object Reference)
   > n3 = Cast To LevelInstanceInterface!(execute, Object=@n2) -> (then, CastFailed, AsLevel Instance Interface:Level Instance Interface Interface) {is=title K2Node_DynamicCast}
   > n4 = LoadLevelInstance(self:Level Instance Interface Interface=@n3)
   FinishExecutingTask(execute=[@n4.then, @n5.then])
   n5 = UnloadLevelInstance(self:Level Instance Interface Interface=@n3)
   Event BeginExecution()
   > n6 = AsyncEditorDelay(Seconds=1, MinimumFrames=30)
   n7 = PrintString(execute=@n6.Complete, InString="Done2!")
   n8 = SaveLoadedAssets(execute=[@n7.then, @n9.then], AssetsToSave=Get Actors() -> (Actors:Array of Object References), bOnlyIfIsDirty=false)
   n9 = "For Each Loop (Map)"!(
    execute:Exec
    MapPin:Map of Strings to Actor Array Structures=Get ActorsToSaveMap() -> (ActorsToSaveMap:Map of Strings to Actor Array Structures)
    BreakPin:Exec
   ) -> (then:Exec, KeyPin:String, ValuePin:Actor Array Structure, CompletedPin:Exec)
   Break Actor Array!(ActorArray:"Actor Array Structure (by ref)"=@n9.ValuePin) -> (AA_3_#1:Array of Object References) {is=title K2Node_BreakStruct}
  ]}]
variables [
 Actors {type="array<Object>", default []}
 ActorsToSaveMap {type="map<ActorArray>", default {}}
 level_instance_cleanup {type="array<BuildingLevelInstance>", default []}
]
