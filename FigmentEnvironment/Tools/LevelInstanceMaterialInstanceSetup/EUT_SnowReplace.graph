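# Editor utility that forces snow texture data onto BP__Cliff_Figment actors inside level instances located in the snow biome.
# It samples a terrain mask's blue channel at each cliff's world position, then converts matching instanced paths to source level paths.
# It unloads the affected BuildingLevelInstances and sets the first TextureData entry on matching source cliffs to ForceSnow.
# Edited actors are collected, but the SavePackages node is not wired.
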EditorUtilityBlueprint /FigmentEnvironment/Tools/LevelInstanceMaterialInstanceSetup/EUT_SnowReplace
graphs [
 EventGraph {nodes [
   Clear!(
    execute
    TargetMap:Map of Strings to Building Level Instance Object References=Get worldActors() -> (worldActors:Map of Strings to Building Level Instance Object References)
   ) -> (then) {is=title K2Node_CallFunction}
   > Clear(TargetArray:Array of Wildcards) {is=title K2Node_CallArrayFunction}
   n1 = GetEditorWorld(execute=[@n14.then_0, @n36.then_0], self=Get UnrealEditorSubsystem())
   n2 = GetLevels!(execute, World:World Object Reference=@n1) -> (then, ReturnValue:Array of Level Object References)
   > n3 = Reverse for Each Loop!(Exec:Exec, Array:Array of Level Object References=@n2) -> (LoopBody:Exec, ArrayIndex:int, ArrayElement:Level Object Reference, Completed:Exec) {is=title K2Node_MacroInstance}
   n4 = Get Object Path String(Object=@n3.ArrayElement)
   n5 = Branch(execute=[@n38.then, @n3.LoopBody], Condition="Equal, Case Insensitive (String)"(A=@n4, B=Get main level() -> (main level:String)))
   n6 = ParseIntoArray(SourceString=@n4, Delimiter=.)
   n7 = "Get (a ref)"(Array:Array of Strings=ParseIntoArray(SourceString="Get (a ref)"(Array=@n6, Dimension 1=1), Delimiter=:), Dimension 1=0) -> (Output:"String (by ref)")
   n8 = Find Item!(TargetArray=Get cleanLevels() -> (cleanLevels:Array of Strings), ItemToFind:"String (by ref)"=@n7) -> (ReturnValue)
   Branch(execute=@n5.else, Condition="Not Equal (Integer)"(A=@n8, B=-1))
   > UnloadLevelInstance!(
    execute:Exec
    self:Object Reference="Get (a copy)"(
     Array:Array of Building Level Instance Object References=Get level_instance_cleanup() -> (level_instance_cleanup:Array of Building Level Instance Object References)
     Dimension 1=@n8
    ) -> (Output:Building Level Instance Object Reference)
   ) -> (then:Exec) {is=title K2Node_Message}
   > PrintString(InString=Append(A=to clean, B=@n7))
   > Remove Level From World!(execute, InLevel:Level Object Reference=@n3.ArrayElement, bClearSelection:bool=true, bResetTransactionBuffer:bool=true) -> (then, ReturnValue:bool)
   n9 = Get levels() -> (levels:Array of Strings)
   n10 = For Each Loop(Exec=@n27.then, Array=@n9)
   > n11 = Get Asset By Object Path!(
    execute
    self:Asset Registry Interface=GetAssetRegistry()
    ObjectPath:"Soft Object Path Structure (by ref)"=MakeSoftObjectPath!(PathString:String=@n10) -> (ReturnValue:Soft Object Path Structure)
    bIncludeOnlyOnDiskAssets:bool=false
    bSkipARFilteredAssets:bool=true
   ) -> (then, ReturnValue:Asset Data Structure) {is=title K2Node_CallFunction}
   n12 = GetAsset(InAssetData=@n11)
   n13 = Cast To LevelInstanceInterface!(execute, Object=@n12) -> (then, CastFailed, AsLevel Instance Interface:Level Instance Interface Interface) {is=title K2Node_DynamicCast}
   > UnloadLevelInstance(self:Level Instance Interface Interface=@n13) {is=title K2Node_CallFunction}
   Event BeginExecution()
   > n14 = Sequence!(execute) -> (then_0, then_1, then_2:Exec, then_3:Exec, then_4:Exec, then_5:Exec)
   n15 = GetAllActorsOfClass!(
    execute=@n14.then_1
    WorldContextObject:Object Reference=@n1
    ActorClass:Actor Class Reference=/FigmentEnvironment/Blueprints/Cliffs/BP__Cliff_Figment.BP__Cliff_Figment_C
   ) -> (then, OutActors:"Array of BP  Cliff Figment Object References")
   > n16 = For Each Loop(Array:"Array of BP  Cliff Figment Object References"=@n15) -> (Array Element:"BP  Cliff Figment Object Reference")
   n17 = Get pathslocal() -> (pathslocal:Array of Strings)
   SortStringArray!(execute=@n16.Completed, TargetArray:Array of Strings=@n17, bStableSort:bool=true, SortOrder:EArraySortOrder Enum=Ascending) -> (then)
   > n18 = For Each Loop(Array:Array of Strings=@n17) -> (Array Element:String)
   n19 = For Each Loop(Exec=@n18.Completed, Array=Get levels())
   > Add!(
    execute
    TargetMap:Map of Strings to Actor Array Structures=Get ActorsToSaveMap() -> (ActorsToSaveMap:Map of Strings to Actor Array Structures)
    Key:"String (by ref)"=@n19
    Value:"Actor Array Structure (by ref)"=Make Actor Array!(AA_3_#1:Array of Object References) -> (ActorArray:Actor Array Structure) {is=title K2Node_MakeStruct}
   ) -> (then) {is=title K2Node_CallFunction}
   PrintString(execute=@n18.LoopBody, InString=@n18)
   > Add Unique(TargetArray=Get levels(), NewItem="Get (a ref)"(Array=ParseIntoArray(SourceString=@n18, Delimiter=:), Dimension 1=0))
   PrintString(execute=@n16.LoopBody, InString=GetDisplayName(Object=@n16))
   > n20 = IsInSnowBiome!(execute, self, actor:Actor Object Reference=@n16) -> (then, ReturnValue:bool)
   > Branch(Condition=@n20)
   n21 = RightChop!(SourceString:String=Get Object Path String(Object=@n16), Count:int=5) -> (ReturnValue:String)
   n22 = FindSubstring(SearchIn=@n21, Substring=_LevelInstance_)
   n23 = int + int(A=FindSubstring(SearchIn=@n21, Substring=_0.), B=2)
   > Add(
    TargetArray=Get pathslocal()
    NewItem=Append(A=Left!(SourceString:String=@n21, Count:int=@n22) -> (ReturnValue:String), B=RightChop(SourceString=@n21, Count=@n23))
   ) {is=title K2Node_CallArrayFunction}
   n24 = Get level_instance_cleanup()
   n25 = For Each Loop(Exec=@n14.then_4, Array:Array of Building Level Instance Object References=@n24) -> (Array Element:Building Level Instance Object Reference)
   > PrintString(InString=GetSoftPtrAssetName!(Asset:Soft Object Reference=GetWorldAsset!(self:Level Instance Interface Interface=@n25) -> (ReturnValue:"World Soft Object Reference (by ref)")) -> (ReturnValue:String))
   > n26 = Cast To LevelInstanceInterface(Object=@n25)
   > UnloadLevelInstance(self=@n26)
   FinishExecutingTask(execute=[@n14.then_5, @n36.then_1])
   PrintString(execute=@n14.then_2, InString=creating cleanup:)
   > n27 = SortStringArray(TargetArray=@n9, bStableSort=true, SortOrder=Ascending)
   n28 = For Each Loop(Exec=@n14.then_3, Array:Array of Strings=Get levels()) -> (Array Element:String)
   > n29 = Get Asset By Object Path(self=GetAssetRegistry(), ObjectPath=MakeSoftObjectPath(PathString=@n28), bIncludeOnlyOnDiskAssets=false, bSkipARFilteredAssets=true)
   > n30 = GetAllActorsOfClass(WorldContextObject=GetAsset(InAssetData=@n29), ActorClass=&BP__Cliff_Figment_C)
   > n31 = For Each Loop(Array:"Array of BP  Cliff Figment Object References"=@n30) -> (Array Element:"BP  Cliff Figment Object Reference")
   n32 = Get Object Path String(Object=@n31)
   n33 = Get pathslocal()
   > Branch(Condition="Not Equal (Integer)"(A=Find Item(TargetArray=@n33, ItemToFind=@n32), B=-1))
   > PrintString(InString=Append(A="    ", B=@n32))
   > n34 = GetParentFromTD!(execute, self, actor:Actor Object Reference=@n31) -> (
    then
    override:Material Interface Object Reference
    TDArrayOut:Array of Building Texture Data Object References
    Parent:Material Interface Object Reference
   )
   > Set Array Elem!(
    execute
    TargetArray:Array of Building Texture Data Object References=@n34.TDArrayOut
    Index:int=0
    Item:"Building Texture Data Object Reference (by ref)"=/Figment_S07/Athena/Environments/TextureData/CliffGround/TD_Athena_Terrain_Mat_Biome_ForceSnow
    bSizeToFit:bool=false
   ) -> (then)
   > n35 = SetEditorProperty(Object=@n31, PropertyName=TextureData, PropertyValue:Array of Building Texture Data Object References=@n34.TDArrayOut)
   > Branch(Condition=@n35)
   > Add(
    TargetArray:Array of Object References=Get ActorsToSave() -> (ActorsToSave:Array of Object References)
    NewItem:"Object Reference (by ref)"=@n31
   )
   n36 = Sequence!(execute) -> (then_0, then_1)
   n37 = Find(TargetMap:Map of Strings to Actor Array Structures=Get ActorsToSaveMap(), Key=@n28) -> (Value:Actor Array Structure)
   Branch(Condition=@n37.ReturnValue)
   > Add(
    TargetArray=Break Actor Array!(ActorArray:"Actor Array Structure (by ref)"=@n37) -> (AA_3_#1:Array of Object References) {is=title K2Node_BreakStruct}
    NewItem=@n31
   )
   For Each Loop(Array:Array of Level Object References=@n2) -> (Array Element:Level Object Reference)
   > n38 = PrintString(InString=@n4)
   Clear()
   n39 = GetDisplayName(Object=@n12)
   n40 = Find(TargetMap:Map of Strings to Building Level Instance Object References=Get worldActors(), Key=@n39) -> (Value:Building Level Instance Object Reference)
   Add(
    execute=@n49.then
    TargetArray:Array of Strings=Get cleanLevels()
    NewItem:"String (by ref)"=GetSoftPtrAssetName(Asset=GetWorldAsset(self=@n40))
   )
   GetDisplayName(Object=@n25)
   n41 = GetLevels(execute=@n47.then, World=@n1)
   > n42 = For Each Loop(Array=@n41)
   GetDisplayName(Object=@n42)
   GetDisplayName(Object=@n31)
   Get Object Path String(Object=@n16)
   n43 = GetAllActorsOfClass(execute=@n56.then, WorldContextObject=@n1, ActorClass=/Script/FortniteGame.BuildingLevelInstance) -> (OutActors:Array of Building Level Instance Object References)
   > n44 = For Each Loop(Array:Array of Building Level Instance Object References=@n43) -> (Array Element:Building Level Instance Object Reference)
   n45 = GetOwner(self:Actor Object Reference=@n44)
   PrintString(execute=@n54.then, InString=Append(A="owner: ", B=GetDisplayName(Object=@n45)))
   PrintString(InString=looping cleanup:)
   Keys!(execute, TargetMap:Map of Strings to Building Level Instance Object References=Get worldActors()) -> (then, Keys:Array of Strings)
   PrintString(execute=@n42.LoopBody, InString=Get Object Path String(Object=@n42))
   PrintString(InString=Hello)
   SavePackages!(
    execute
    PackagesToSave:Array of Package Object References=Make Array("[0]":Package Object Reference=GetPackageForObject!(Object:Object Reference=@n31) -> (ReturnValue:Package Object Reference)) -> (Array:Array of Package Object References)
    bOnlyDirty:bool=false
   ) -> (then, ReturnValue:bool)
   Append(A=YES, B=@n39)
   Get LevelInstanceSubsystem!(WorldContext:Object Reference) -> (ReturnValue:Level Instance Subsystem Object Reference) {is=title K2Node_GetSubsystem}
   For Each Loop(Array:Array of Level Object References=@n41) -> (Array Element:Level Object Reference)
   Reverse for Each Loop(Array:Array of Building Level Instance Object References=@n24) -> (ArrayElement:Building Level Instance Object Reference)
   Get level_instance_cleanup()
   n46 = Is Not Empty!(TargetArray:Array of Level Object References=@n41) -> (ReturnValue:bool)
   Clear(execute=@n48.then, TargetArray:Array of Object References=Get ActorsToSave())
   > n47 = Clear(TargetArray:Array of Strings=Get levels())
   Clear(execute=@n1.then, TargetArray=Get cleanLevels())
   > Clear(TargetArray=Get pathslocal())
   > n48 = Clear(TargetMap=Get worldActors()) {is=title K2Node_CallFunction}
   n49 = Add(
    execute=@n50.then
    TargetArray:Array of Building Level Instance Object References=Get level_instance_cleanup()
    NewItem:"Building Level Instance Object Reference (by ref)"=@n40
   )
   PrintString(execute=@n11.then, InString=@n39)
   > n50 = Branch(Condition=@n40.ReturnValue)
   n51 = GetWorldAsset(self=@n44)
   n52 = GetSoftPtrAssetName(Asset=@n51)
   n53 = GetWorldAsset(self=@n44)
   Add(
    execute=@n44.LoopBody
    TargetMap:Map of Strings to Building Level Instance Object References=Get worldActors()
    Key=@n52
    Value:"Building Level Instance Object Reference (by ref)"=@n44
   ) {is=title K2Node_CallFunction}
   > n54 = PrintString(InString=GetSoftPtrAssetName(Asset=@n53))
   n55 = Get Object Path String(Object="Get (a copy)"(Array:Array of Level Object References=@n41) -> (Output:Level Object Reference))
   Branch(execute=@n42.Completed, Condition=@n46)
   > n56 = Set main level(main level:String=@n55) -> (Output_Get:String)
  ]}
 GetParentFromTD {nodes [
   n1 = GetParentFromTD!() -> (then, actor:Actor Object Reference) {is=title K2Node_FunctionEntry}
   > n2 = GetEditorProperty!(execute:Exec, Object:Object Reference=@n1, PropertyName:Name=TextureData) -> (then:Exec, PropertyValue:Array of Building Texture Data Object References, ReturnValue:bool)
   > Branch(Condition=@n2.ReturnValue)
   > n3 = Set TDArray(TDArray:Array of Building Texture Data Object References=@n2) -> (Output_Get:Array of Building Texture Data Object References)
   n4 = "Get (a copy)"(Array:Array of Building Texture Data Object References=@n3) -> (Output:Building Texture Data Object Reference)
   > Is Valid(InputObject=@n4)
   n5 = Get OverrideMaterial!(self:Building Texture Data Object Reference=@n4) -> (OverrideMaterial:Material Interface Object Reference)
   > n6 = Cast To MaterialInstanceConstant(Object=@n5)
   > Return Node!(
    execute
    override:Material Interface Object Reference=@n5
    TDArrayOut:Array of Building Texture Data Object References=Get TDArray() -> (TDArray:Array of Building Texture Data Object References)
    Parent:Material Interface Object Reference=Get Parent!(self:Material Instance Object Reference=@n6) -> (Parent:Material Interface Object Reference)
   )
  ]}
 IsInSnowBiome {nodes [
   n1 = IsInSnowBiome!() -> (then, actor:Actor Object Reference) {is=title K2Node_FunctionEntry}
   n2 = "To Vector2D (Vector)"!(InVector:Vector=Get Actor Location(self=@n1)) -> (ReturnValue:Vector 2D Structure)
   n3 = Make Literal Float(Value=-153000)
   n4 = Texture2D Sample UV Editor Only!(
    Texture:Texture 2D Object Reference=/Figment_S07/LM_7_00/Athena/Environments/Landscape/Textures/T_Athena_Terrain_Topo_Mask_740
    UV:Vector 2D Structure=vector2d / vector2d(
     A:Vector 2D Structure=vector2d - vector2d(A:Vector 2D Structure=@n2, B=@n3) -> (ReturnValue:Vector 2D Structure) {is=title K2Node_PromotableOperator}
     B=Make Literal Float(Value=306200.7812)
    ) -> (ReturnValue:Vector 2D Structure) {is=title K2Node_PromotableOperator}
   ) -> (ReturnValue:Linear Color Structure)
   > Return Node!(execute, ReturnValue:bool="float > float"(A=BreakColor(InColor=@n4).B, B=0.5))
  ]}
]
variables [
 ActorsToSave {type="array<Object>", default []}
 ActorsToSaveMap {type="map<ActorArray>", default {}}
 cleanLevels {type="array<string>", default []}
 level_instance_cleanup {type="array<BuildingLevelInstance>", default []}
 levels {type="array<string>", default []}
 main level {type=string, default=""}
 pathslocal {type="array<string>", default []}
 worldActors {type="map<BuildingLevelInstance>", default {}}
]
