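# Editor utility widget with batch level-editing tools for Figment environments, centred on Fortnite BuildingTextureData and BP__Cliff_Figment actors.
# It creates or duplicates texture data and material instances, including S03 variants, then assigns them using editor properties and inline Python.
# It also maps static meshes to cliff Blueprints for creation and replacement, edits cliff tags, and deletes foliage and mesh-grass actors.
# Changed assets are collected in arrays for saving, often after checkout.
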EditorUtilityWidgetBlueprint /FigmentEnvironment/Tools/LevelInstanceMaterialInstanceSetup/EUW_ReplaceMaterialsWithInstances_TreeSwap
graphs [
 AppltTextureDataBiome {nodes [
   n1 = AppltTextureDataBiome!() -> (then, WorldContextObject:Object Reference) {is=title K2Node_FunctionEntry}
   > Clear(TargetArray:"Array of BP  Cliff Figment Object References"=Get Cliff Actors to Save() -> (Cliff Actors to Save:"Array of BP  Cliff Figment Object References"))
   > n2 = GetAllActorsOfClass!(
    execute
    WorldContextObject:Object Reference=@n1
    ActorClass:Actor Class Reference=/FigmentEnvironment/Blueprints/Cliffs/BP__Cliff_Figment.BP__Cliff_Figment_C
   ) -> (then, OutActors:"Array of BP  Cliff Figment Object References")
   > n3 = For Each Loop(Array:"Array of BP  Cliff Figment Object References"=@n2) -> (Array Element:"BP  Cliff Figment Object Reference")
   n4 = Get Cliff Actors to Save()
   CheckoutLoadedAssets(execute=@n3.Completed, AssetsToCheckout=@n4)
   > SaveLoadedAssets(AssetsToSave=@n4, bOnlyIfIsDirty=false)
   n5 = Cast To BP__Cliff_Figment!(execute=@n3.LoopBody, Object=@n3) -> (then, CastFailed, "AsBP  Cliff Figment":"BP  Cliff Figment Object Reference") {is=title K2Node_DynamicCast}
   > n6 = GetEditorProperty!(execute:Exec, Object:Object Reference=@n5, PropertyName:Name=TextureData) -> (then:Exec, PropertyValue:Array of Building Texture Data Object References, ReturnValue:bool)
   > SetEditorProperty(
    Object=@n3
    PropertyName=TextureData
    PropertyValue:Array of Building Texture Data Object References=Make Array!(
     "[0]":Building Texture Data Object Reference="Get (a copy)"(Array:Array of Building Texture Data Object References=@n6) -> (Output:Building Texture Data Object Reference)
     "[1]":Building Texture Data Object Reference="Get (a copy)"(Array=@n6, Dimension 1=1)
     "[2]":Building Texture Data Object Reference="Get (a copy)"(Array=@n6, Dimension 1=2)
     "[3]":Building Texture Data Object Reference=Get Color Map TD override() -> (Color Map TD override:Building Texture Data Object Reference)
    ) -> (Array:Array of Building Texture Data Object References)
   )
   > Add(
    TargetArray:"Array of BP  Cliff Figment Object References"=Get Cliff Actors to Save()
    NewItem:"BP  Cliff Figment Object Reference (by ref)"=@n5
   ) {is=title K2Node_CallArrayFunction}
   Get TD Asset() -> (TD Asset:Building Texture Data Object Reference)
  ]}
 AppltTextureDataSelected {nodes [
   n1 = GetSelectedActors(execute=@n5.then)
   > n2 = For Each Loop(Array:Array of Actor Object References=@n1) -> (Array Element:Actor Object Reference)
   n3 = Cast To FortStaticMeshActor!(execute, Object=@n2) -> (then, CastFailed, AsFort Static Mesh Actor:Fort Static Mesh Actor Object Reference) {is=title K2Node_DynamicCast}
   n4 = GetEditorProperty(execute=[@n3.then, @n12.then], Object=@n2, PropertyName=TextureData)
   > SetEditorProperty(
    Object=@n2
    PropertyName=TextureData
    PropertyValue=Make Array(
     "[0]"=Get TD Asset()
     "[1]"="Get (a copy)"(Array=@n4, Dimension 1=1)
     "[2]"="Get (a copy)"(Array=@n4, Dimension 1=2)
     "[3]"="Get (a copy)"(Array=@n4, Dimension 1=3)
    )
   )
   > Add(
    TargetArray:Array of Object References=Get ObjectsToSave() -> (ObjectsToSave:Array of Object References)
    NewItem:"Object Reference (by ref)"=@n2
   )
   AppltTextureDataSelected() {is=title K2Node_FunctionEntry}
   > n5 = Clear(TargetArray:Array of Object References=Get ObjectsToSave())
   PrintString(execute=@n12.else, InString="Unspported Type for TD Replacement, try Types that inerit from BuildingProp or FortStaticMeshActor")
   n6 = Get ObjectsToSave()
   SaveLoadedAssets(execute=@n10.then, AssetsToSave=@n6, bOnlyIfIsDirty=false)
   "Get (a copy)"(Array:Array of Actor Object References=@n1) -> (Output:Actor Object Reference)
   "Get (a copy)"(Array:Array of Building Texture Data Object References=@n4) -> (Output:Building Texture Data Object Reference)
   Get Cliff Actors to Save()
   Get Color Map TD override()
   n7 = Get Class(Object=@n2)
   n8 = ClassIsChildOf(TestClass=@n7, ParentClass=/Script/FortniteGame.FortStaticMeshActor)
   n9 = ClassIsChildOf(TestClass=@n7, ParentClass=&BuildingProp)
   n10 = CheckoutLoadedAssets(execute=@n2.Completed, AssetsToCheckout=@n6)
   n11 = OR Boolean(A=@n9, B=@n8)
   n12 = Branch(execute=@n2.LoopBody, Condition=@n11)
  ]}
 AssignTextureDataSlot4 {nodes [
   AssignTextureDataSlot4() {is=title K2Node_FunctionEntry}
   > n1 = GetEditorWorld(self=Get UnrealEditorSubsystem())
   > n2 = GetAllActorsOfClass(WorldContextObject=@n1, ActorClass=&BP__Cliff_Figment_C)
   > n3 = For Each Loop(Array:"Array of BP  Cliff Figment Object References"=@n2) -> (Array Element:"BP  Cliff Figment Object Reference")
   > n4 = Execute Python Script!(
    execute
    PythonScript="import unreal\r\nno_need_to_save = False\r\nTextureDataArray = geo.get_editor_property('TextureData')\r\nprint(TextureDataArray)\r\nif(TextureDataArray[3] == texture_data):\r\n    no_need_to_save = True\r\nelse:\r\n    TextureDataArray[3] = texture_data\r\n    geo.set_editor_property('TextureData', TextureDataArray)"
    geo:"BP  Cliff Figment Object Reference"=@n3
    texture_data:Building Texture Data Object Reference=Get data() -> (data:Building Texture Data Object Reference)
   ) -> (then, ReturnValue, no_need_to_save:bool=~) {inputs [geo,texture_data], outputs [no_need_to_save]}
   > n5 = Branch(Condition=@n4.no_need_to_save)
   Add(
    execute=@n5.else
    TargetArray:"Array of BP  Cliff Figment Object References"=Get Cliff Actors to Save()
    NewItem:"BP  Cliff Figment Object Reference (by ref)"=@n3
   )
  ]}
 AssignTextureDataToCliffActorCreateIfMissing {nodes [
   n1 = AssignTextureDataToCliffActorCreateIfMissing!() -> (then, CliffActor:"BP  Cliff Figment Object Reference") {is=title K2Node_FunctionEntry}
   > n2 = Execute Python Script!(execute, PythonScript="import unreal\r\ntexture_data = (geo.get_editor_property('TextureData')[0])", geo=@n1) -> (then, ReturnValue, texture_data:Object Reference) {inputs [geo], outputs [texture_data]}
   n3 = GetDisplayName(Object=GetPackageForObject!(Object:Object Reference=@n2.texture_data) -> (ReturnValue:Package Object Reference))
   > n4 = Branch(Condition=IsEmpty(InString=@n3))
   > PrintString(InString=Nothing Assigned)
   n5 = GetEditorWorld(execute=@n4.else, self=Get UnrealEditorSubsystem())
   > n6 = StripRightMostFolderFromPath_Util!(execute, Path:String=@n3, WorldContextObject:Object Reference=@n5, Strip Count:int=1, __WorldContext:Object Reference=@n5) -> (then, FinalPath1:String, AssetName:String)
   n7 = MakeLiteralString(Value=/TextureData/S03)
   n8 = MakeLiteralString(Value=/FigmentEnvironment/Blueprints/Cliffs/Materials)
   > n9 = Branch(Condition="Equal, Case Insensitive (String)"(A=Append(A=@n8, B=@n7), B=@n6))
   > PrintString(InString=already assigned)
   n10 = Append(A=@n6.AssetName, B=MakeLiteralString(Value=_S03))
   n11 = Append(A=@n6, B=MakeLiteralString(Value=/S03/))
   n12 = Append(A=@n11, B=@n10)
   n13 = DoesAssetExist(execute=@n9.else, AssetPath=@n12)
   > n14 = Branch(Condition=@n13)
   > PrintString(InString=TD asset existsi)
   > n15 = GetAssetByObjectPath!(
    execute
    self:Asset Registry Interface=GetAssetRegistry()
    ObjectPath:Name=String To Name(InString=Append!(A=@n12, B=., C:String=@n10) -> (ReturnValue))
    bIncludeOnlyOnDiskAssets:bool=false
   ) -> (then, ReturnValue:Asset Data Structure)
   > n16 = Cast To BuildingTextureData!(execute, Object=GetAsset(InAssetData=@n15)) -> (then, CastFailed, AsBuilding Texture Data:Building Texture Data Object Reference) {is=title K2Node_DynamicCast}
   > n17 = Execute Python Script!(
    execute
    PythonScript="import unreal\r\nmaterial_interface = tex_data.get_editor_property('OverrideMaterial')"
    tex_data:Building Texture Data Object Reference=@n16
   ) -> (then, ReturnValue, material_interface:Material Instance Constant Object Reference) {inputs [tex_data], outputs [material_interface]}
   > n18 = Set material_on_TD(material_on_TD:Material Instance Constant Object Reference=@n17.material_interface) -> (Output_Get:Material Instance Constant Object Reference)
   > PrintString(InString=GetDisplayName(Object=Get Parent!(self:Material Instance Object Reference=@n18) -> (Parent:Material Interface Object Reference)))
   > n19 = Execute Python Script!(
    execute
    PythonScript="import unreal\r\nno_need_to_save = False\r\nTextureDataArray = geo.get_editor_property('TextureData')\r\nprint(TextureDataArray)\r\nif(TextureDataArray[0] == texture_data):\r\n    no_need_to_save = True\r\nelse:\r\n    TextureDataArray[0] = texture_data\r\n    geo.set_editor_property('TextureData', TextureDataArray)"
    geo:"BP  Cliff Figment Object Reference"=@n1
    texture_data:Building Texture Data Object Reference=@n16
   ) -> (then, ReturnValue, no_need_to_save:bool=~) {inputs [geo,texture_data], outputs [no_need_to_save]}
   > n20 = Branch(Condition=@n19.no_need_to_save)
   Return Node!(execute=[@n39.then, @n20.then], ActorToSave:"BP  Cliff Figment Object Reference"=@n1)
   PrintString(execute=@n14.else, InString=Append!(A="TD asset doesnt exist, duplicating", B=@n12) -> (ReturnValue))
   > n21 = GetAssetTools()
   > n22 = DuplicateAsset!(execute, self:Asset Tools Interface=@n21, AssetName:String=@n10, PackagePath:String=@n11, OriginalObject:Object Reference=@n2.texture_data) -> (then, ReturnValue:Object Reference)
   > n23 = Execute Python Script!(
    execute
    PythonScript="import unreal\r\nmaterial_interface = tex_data.get_editor_property('OverrideMaterial')"
    tex_data:Object Reference=@n22
   ) -> (then, ReturnValue, material_interface:Material Instance Constant Object Reference) {inputs [tex_data], outputs [material_interface]}
   > n24 = Set material_on_TD(material_on_TD=@n23.material_interface)
   n25 = GetDisplayName(Object=GetPackageForObject(Object=@n24))
   > PrintString(InString=@n25)
   > n26 = GetEditorWorld(self=Get UnrealEditorSubsystem())
   > n27 = StripRightMostFolderFromPath_Util(Path=@n25, WorldContextObject=@n26, Strip Count=1, __WorldContext=@n26)
   n28 = Append(A=@n8, B=MakeLiteralString(Value=/S03/))
   n29 = Append(A=@n27.AssetName, B=MakeLiteralString(Value=_S03))
   > n30 = DoesAssetExist(AssetPath=Append(A=@n28, B=@n29))
   > n31 = Branch(Condition=@n30)
   n32 = GetAssetTools(execute=@n31.else)
   > n33 = DuplicateAsset(self=@n32, AssetName=@n29, PackagePath=@n28, OriginalObject=@n24)
   > n34 = Cast To MaterialInstanceConstant(Object=@n33)
   > n35 = Branch(Condition="Equal (Object)"(A=Get Parent(self=@n34), B=/Game/Athena/Environments/Landscape/Material/Athena_Terrain_Mat_Biome))
   > n36 = SetMaterialInstanceParent!(
    execute
    Instance:Material Instance Constant Object Reference=@n34
    NewParent:Material Interface Object Reference=Get material_orig() -> (material_orig:Material Interface Object Reference)
   ) -> (then)
   SaveLoadedAsset(execute=[@n36.then, @n35.else], AssetToSave=@n33, bOnlyIfIsDirty=true)
   > n37 = Cast To BuildingTextureData(Object=@n22)
   > Set OverrideMaterial!(execute, OverrideMaterial:Material Interface Object Reference=@n34, self:Building Texture Data Object Reference=@n37) -> (then, Output_Get:Material Interface Object Reference)
   > SaveLoadedAsset(AssetToSave=@n37, bOnlyIfIsDirty=false)
   > n38 = Execute Python Script!(
    execute
    PythonScript="import unreal\r\nno_need_to_save = False\r\nTextureDataArray = geo.get_editor_property('TextureData')\r\nprint(TextureDataArray)\r\nif(TextureDataArray[0] == texture_data):\r\n    no_need_to_save = True\r\nelse:\r\n    TextureDataArray[0] = texture_data\r\n    geo.set_editor_property('TextureData', TextureDataArray)"
    geo:"BP  Cliff Figment Object Reference"=@n1
    texture_data:Building Texture Data Object Reference=@n37
   ) -> (then, ReturnValue, no_need_to_save:bool=~) {inputs [geo,texture_data], outputs [no_need_to_save]}
   > n39 = Branch(Condition=@n38.no_need_to_save)
   Cast To MaterialInstanceConstant(Object:Wildcard)
  ]}
 BuildMeshMap {nodes [
   n1 = Keys!(
    execute
    TargetMap:Map of Static Mesh Object References to Strings=Get mapOfMeshesToBPs() -> (mapOfMeshesToBPs:Map of Static Mesh Object References to Strings)
   ) -> (then, Keys:Array of Static Mesh Object References)
   > n2 = For Each Loop(Array:Array of Static Mesh Object References=@n1) -> (Array Element:Static Mesh Object Reference)
   n3 = Values!(execute, TargetMap:Map of Static Mesh Object References to Strings=Get mapOfMeshesToBPs()) -> (then, Values:Array of Strings)
   PrintString(
    execute=@n2.LoopBody
    InString=Append!(
     A=GetDisplayName(Object=@n2)
     B=": "
     C:String="Get (a ref)"(Array:Array of Strings=@n3, Dimension 1=@n2.Array Index) -> (Output:"String (by ref)")
    ) -> (ReturnValue)
   )
   BuildMeshMap() {is=title K2Node_FunctionEntry}
   > n4 = Sequence!(execute) -> (then_0, then_1, then_2:Exec)
   > Clear!(execute, TargetMap:Map of Static Mesh Object References to Strings=Get mapOfMeshesToBPs()) -> (then) {is=title K2Node_CallFunction}
   > Clear(TargetMap:Map of Strings to Strings=Get mapOfMeshNamesToBp() -> (mapOfMeshNamesToBp:Map of Strings to Strings))
   > n5 = GetAssetsByPath!(
    execute
    self:Asset Registry Interface=GetAssetRegistry()
    PackagePath:Name=String To Name(InString=MakeLiteralString(Value=/FigmentEnvironment/Blueprints/Cliffs/))
    bRecursive:bool=false
    bIncludeOnlyOnDiskAssets:bool=true
   ) -> (then, OutAssetData:Array of Asset Data Structures, ReturnValue:bool)
   > n6 = For Each Loop(Array:Array of Asset Data Structures=@n5) -> (Array Element:Asset Data Structure)
   > n7 = GetBlueprintAsset(Object=GetAsset(InAssetData=@n6))
   > n8 = GeneratedClass!(execute, BlueprintObj:Blueprint Object Reference=@n7) -> (then, ReturnValue:Object Class Reference)
   > n9 = GetClassDefaultObject!(execute, Class:Object Class Reference=@n8) -> (then, ReturnValue:Object Reference)
   > n10 = Cast To BP__Cliff_Figment(Object=@n9)
   n11 = Get StaticMesh(self=Get StaticMeshComponent(self=@n10))
   n12 = GetDisplayName(Object=@n11)
   > n13 = Branch(Condition=IsEmpty(InString=@n12))
   n14 = "To String (Name)"(InName=Break Asset Data(AssetData=@n6))
   Add!(execute=@n13.else, TargetMap:Map of Strings to Strings=Get mapOfMeshNamesToBp(), Key:"String (by ref)"=@n12, Value:"String (by ref)"=@n14) -> (then) {is=title K2Node_CallFunction}
   > Add(
    TargetMap:Map of Static Mesh Object References to Strings=Get mapOfMeshesToBPs()
    Key:"Static Mesh Object Reference (by ref)"=@n11
    Value=@n14
   )
   n15 = Keys(execute=@n4.then_1, TargetMap=Get mapOfMeshesToBPs())
   > Clear!(
    execute
    TargetSet:Set of Static Mesh Object References=Get staticMeshSetCheckDuplicattes() -> (staticMeshSetCheckDuplicattes:Set of Static Mesh Object References)
   ) -> (then)
   n16 = Get staticMeshSetCheckDuplicattes()
   > Add Items!(execute, TargetSet:Set of Static Mesh Object References=@n16, NewItems:Array of Static Mesh Object References=@n15) -> (then)
   > n17 = Branch(Condition="Equal (Integer)"(
     A=Length!(TargetSet:Set of Static Mesh Object References=@n16) -> (ReturnValue:int=0) {is=title K2Node_CallFunction}
     B=Length(TargetArray:Array of Static Mesh Object References=@n15) {is=title K2Node_CallArrayFunction}
    ))
   > PrintString(InString=No static Mesh Duplicate blueprints found)
   PrintString(execute=@n17.else, InString="Ahhhhhh duplicated!\r\n######################################################")
   Return Node!(execute=@n4.then_2, mapOfMeshNamesToBp:Map of Strings to Strings=Get mapOfMeshNamesToBp())
   ClassIsChildOf()
   PrintString(InString=Hello)
   Remove!(execute, TargetMap:Map of Strings to Strings=Get mapOfMeshNamesToBp(), Key:"String (by ref)") -> (then, ReturnValue:bool)
   PrintString(InString=Append!(A="Map Build Empty: ", B=@n12, C="->", D:String=@n14) -> (ReturnValue))
  ]}
 BuildMultiLineMapOfMaterials {nodes [
   n1 = Clear!(
    execute
    TargetMap:Map of Strings to Material Interface Object References=Get MapStringToTargetMaterials() -> (MapStringToTargetMaterials:Map of Strings to Material Interface Object References)
   ) -> (then)
   n2 = "GetText (Multi-Line Editable Text)"!(self:"Editable Text (Multi- Line) Object Reference"=Get EditorUtilityMultiLineEditableText_101() -> (EditorUtilityMultiLineEditableText_101:Editor Utility Multi Line Editable Text Object Reference)) -> (ReturnValue:Text)
   n3 = For Each Loop(Exec=[@n1.then, @n9.then], Array:Array of Strings=ParseIntoArray(SourceString="To String (Text)"(InText=@n2), Delimiter="\r\n")) -> (Array Element:String)
   > n4 = GetAssetsByPath(self=GetAssetRegistry(), PackagePath=String To Name(InString=@n3), bRecursive=true, bIncludeOnlyOnDiskAssets=true)
   > n5 = For Each Loop(Array:Array of Asset Data Structures=@n4) -> (Array Element:Asset Data Structure)
   > n6 = Cast To MaterialInterface!(execute, Object=GetAsset(InAssetData=@n5)) -> (then, CastFailed, AsMaterial Interface:Material Interface Object Reference) {is=title K2Node_DynamicCast}
   > Add(
    TargetMap:Map of Strings to Material Interface Object References=Get MapStringToTargetMaterials()
    Key:"String (by ref)"=GetDisplayName(Object=@n6)
    Value:"Material Interface Object Reference (by ref)"=@n6
   )
   n7 = BuildMultiLineMapOfMaterials!() -> (then, Mesh Material:Material Interface Object Reference) {is=title K2Node_FunctionEntry}
   > Clear(TargetArray:Array of Building Texture Data Object References=Get BTD_To_Return() -> (BTD_To_Return:Array of Building Texture Data Object References)) {is=title K2Node_CallArrayFunction}
   > Clear(TargetArray=Get BTD_ToSave() -> (BTD_ToSave:Array of Building Texture Data Object References))
   > n8 = Sequence!(execute) -> (then_0, then_1)
   > n9 = Branch(Condition=Is Empty!(TargetMap:Map of Strings to Material Interface Object References=Get MapStringToTargetMaterials()) -> (ReturnValue:bool) {is=title K2Node_CallFunction})
   n10 = GetAssetTools(execute=@n8.then_1)
   > n11 = Branch(Condition=IsValid(Object=@n7))
   n12 = GetDisplayName(Object=@n7)
   > PrintString(InString=@n12)
   n13 = Find(TargetMap:Map of Strings to Material Interface Object References=Get MapStringToTargetMaterials(), Key=@n12) -> (Value:Material Interface Object Reference)
   > n14 = Branch(Condition=@n13.ReturnValue)
   > PrintString(InString=Append!(A=Found, B=Get Object Path String(Object=@n13)) -> (ReturnValue))
   > n15 = CreateNewTD!(execute, self, Asset Tools:Asset Tools Interface=@n10, Material Name:String=@n12, OverrideMaterial:Material Interface Object Reference=@n13) -> (then, AsBuilding Texture Data:Building Texture Data Object Reference, needs saving:bool)
   > n16 = Set SingleTD(SingleTD:Building Texture Data Object Reference=@n15) -> (Output_Get:Building Texture Data Object Reference)
   > n17 = Branch(Condition=@n15.needs saving)
   > n18 = SaveLoadedAsset(AssetToSave=@n16, bOnlyIfIsDirty=false)
   Return Node!(
    execute=[@n25.then, @n18.then, @n17.else, @n20.then]
    TDArray:Array of Building Texture Data Object References=Get BTD_To_Return()
    SingleTD_out:Building Texture Data Object Reference=Get SingleTD() -> (SingleTD:Building Texture Data Object Reference)
   )
   n19 = EnsureMaterialAndTextureDataExist!(execute=@n14.else, self, Mesh Material:Material Interface Object Reference=@n7) -> (then, " Texture Data To Assign":Building Texture Data Object Reference, Material:Material Interface Object Reference, bMaterialCreated1:bool)
   > n20 = Set SingleTD(SingleTD=@n19)
   n21 = "For Each Loop (Map)"!(execute:Exec=@n11.else, MapPin:Map of Strings to Material Interface Object References=Get MapStringToTargetMaterials(), BreakPin:Exec) -> (then:Exec, KeyPin:String, ValuePin:Material Interface Object Reference, CompletedPin:Exec)
   > n22 = CreateNewTD(Asset Tools=@n10, Material Name=@n21, OverrideMaterial=@n21.ValuePin)
   > n23 = Branch(Condition=@n22.needs saving)
   > n24 = Add(
    TargetArray:Array of Building Texture Data Object References=Get BTD_ToSave()
    NewItem:"Building Texture Data Object Reference (by ref)"=@n22
   ) {is=title K2Node_CallArrayFunction}
   Add(execute=[@n24.then, @n23.else], TargetArray=Get BTD_To_Return(), NewItem=@n22)
   n25 = SaveLoadedAssets(execute=@n21.CompletedPin, AssetsToSave=Get BTD_ToSave(), bOnlyIfIsDirty=false)
   PrintString(InString=Hello)
   GetDisplayName(Object=@n13)
   GetAssetTools()
  ]}
 ClearTags {nodes [
   n1 = GetEditorWorld(execute=@n2.then, self=Get UnrealEditorSubsystem())
   n2 = ClearTags!() -> (then, WorldContextObject:Object Reference) {is=title K2Node_FunctionEntry}
   n3 = GetAllActorsOfClass(
    execute=@n1.then
    WorldContextObject=SelectObject!(A:Object Reference=@n2, B:Object Reference=@n1, bSelectA:bool=IsValid(Object=@n2)) -> (ReturnValue:Object Reference)
    ActorClass=&BP__Cliff_Figment_C
   )
   n4 = For Each Loop(Exec=@n12.then, Array:Array of Actor Object References=@n3) -> (Array Element:Actor Object Reference)
   > n5 = Cast To BP__Cliff_Figment(Object=@n4)
   n6 = ResetEditorProperty!(execute, Object:Object Reference=@n5, PropertyName:Name=Tags, ChangeNotifyMode:EPropertyAccessChangeNotifyMode Enum=Never) -> (then, ReturnValue:bool)
   Add(
    execute=[@n6.then, @n10.then]
    TargetArray:"Array of BP  Cliff Figment Object References"=Get Cliff Actors to Save()
    NewItem:"BP  Cliff Figment Object Reference (by ref)"=@n5
   )
   n7 = PrintString(InString=GetDisplayName(Object=@n5))
   n8 = Branch(execute=[@n7.then, @n5.then], Condition=ActorHasTag!(self:Actor Object Reference=@n5, Tag:Name=PCG_MeshGrass) -> (ReturnValue:bool))
   Set tags(execute=@n8.else, tags:Array of Names=Get Tags!(self:Actor Object Reference=@n5) -> (Tags:Array of Names)) -> (Output_Get:Array of Names)
   n9 = Get tags() -> (tags:Array of Names)
   > Add(TargetArray:Array of Names=@n9, NewItem:"Name (by ref)"=PCG_MeshGrass)
   > n10 = Set Tags!(execute, Tags:Array of Names=@n9, self:Actor Object Reference=@n5) -> (then, Output_Get:Array of Names)
   n11 = Get Cliff Actors to Save()
   SaveLoadedAssets(execute=@n13.then, AssetsToSave=@n11, bOnlyIfIsDirty=false)
   n12 = Clear(execute=@n3.then, TargetArray:"Array of BP  Cliff Figment Object References"=Get Cliff Actors to Save())
   n13 = CheckoutLoadedAssets(execute=@n4.Completed, AssetsToCheckout=@n11)
  ]}
 Create BP Asset From Fort Static Mesh {nodes [
   n1 = Create BP Asset From Fort Static Mesh!() -> (then, Fort Static Mesh Actor:Static Mesh Actor Object Reference) {is=title K2Node_FunctionEntry}
   > GetAssetTools()
   n2 = Get StaticMesh(self=Get StaticMeshComponent(self=@n1))
   n3 = Append(A=MakeLiteralString(Value=&Cliffs/), B=Append(A=BP_, B=GetDisplayName(Object=@n2)))
   > n4 = CreateBlueprintAssetWithParent!(
    execute
    AssetPath:String=@n3
    ParentClass:Object Class Reference=&BP__Cliff_Figment_C
   ) -> (then, ReturnValue:Blueprint Object Reference)
   > n5 = LoadBlueprintClass!(execute, AssetPath:String=@n3) -> (then, ReturnValue:Object Class Reference)
   > n6 = GetClassDefaultObject(Class=@n5)
   > n7 = Cast To BP__Cliff_Figment(Object=@n6)
   > SetStaticMesh(self=Get StaticMeshComponent(self=@n7), NewMesh=@n2)
   > SaveAsset!(execute, AssetToSave:String=@n3, bOnlyIfIsDirty:bool=true) -> (then, ReturnValue:bool)
   > n8 = PrintString(InString=saved BP)
   Return Node!(execute=[@n9.then, @n8.then])
   n9 = PrintString(execute=@n7.CastFailed, InString=Nah)
   GetDisplayName()
   GetBlueprintAsset(Object=@n4)
   PrintString(InString=IS CLIFF)
   GetSelectionSet!(execute) -> (then, ReturnValue:Array of Actor Object References)
   Add!(
    execute
    ElementList:Typed Element List Structure=Create Element List!(
     execute
     Registry:Typed Element Registry Object Reference=Get Default Typed Element Registry!() -> (ReturnValue:Typed Element Registry Object Reference) {is=title K2Node_CallFunction}
    ) -> (then, ReturnValue:Typed Element List Structure)
    ElementHandle:"Script Typed Element Handle Structure (by ref)"
   ) -> (then, ReturnValue:bool) {is=title K2Node_CallFunction}
   GetClass!(InAssetData:"Asset Data Structure (by ref)") -> (ReturnValue:Object Class Reference)
   Get Element LIst() -> (Element LIst:Typed Element List Structure)
  ]}
 CreateCliffBpFromStaticMesh {nodes [
   CreateCliffBpFromStaticMesh() {is=title K2Node_FunctionEntry}
   > n1 = BuildMeshMap!(execute, self) -> (then, mapOfMeshNamesToBp:Map of Strings to Strings) {is=title K2Node_CallFunction}
   > n2 = GetSelectedLevelActors(self=Get EditorActorSubsystem())
   > n3 = Cast To FortStaticMeshActor(Object="Get (a copy)"(Array:Array of Actor Object References=@n2) -> (Output:Actor Object Reference))
   n4 = Get StaticMesh(self=Get StaticMeshComponent(self=@n3))
   > n5 = Branch(Condition=Find(TargetMap:Map of Strings to Strings=@n1, Key=GetDisplayName(Object=@n4)) -> (Value:String).ReturnValue)
   n6 = Switch on E_ApplyContext!(execute=[@n10.then, @n5.then], Selection:E_ApplyContext Enum=Get Apply Mode() -> (Apply Mode:E_ApplyContext Enum=NewEnumerator2)) -> (NewEnumerator2:Exec, NewEnumerator1:Exec, NewEnumerator0:Exec) {is=title K2Node_SwitchEnum}
   ReplaceAllStaticMeshesWithExistingBP!(execute=@n6.NewEnumerator1, self, WorldContextObject:Object Reference) -> (then)
   n7 = BuildMeshMap(execute=@n6.NewEnumerator2)
   > n8 = Replace Mesh Actor With Existing BP Single!(execute, self, Static Mesh Actor:Static Mesh Actor Object Reference=@n3, mapOfMeshNamesToBp:Map of Strings to Strings=@n7) -> (then, "AsBP  Cliff Figment":"BP  Cliff Figment Object Reference", Replaced:bool)
   > Branch(Condition=@n8.Replaced)
   n9 = Get ObjectsToSave()
   > Add(TargetArray:Array of Object References=@n9, NewItem:"Object Reference (by ref)"=@n8) {is=title K2Node_CallArrayFunction}
   > SaveLoadedAssets(AssetsToSave=@n9, bOnlyIfIsDirty=false)
   n10 = Create BP Asset From Fort Static Mesh!(execute=@n5.else, self, Fort Static Mesh Actor:Static Mesh Actor Object Reference=@n3) -> (then) {is=title K2Node_CallFunction}
  ]}
 CreateNewTD {nodes [
   n1 = "To String (Text)"(InText=Get Text(self=Get TargetTextBox_() -> (TargetTextBox_:Editor Utility Editable Text Box Object Reference)))
   n2 = CreateNewTD!() -> (then, Asset Tools:Asset Tools Interface, Material Name:String, OverrideMaterial:Material Interface Object Reference) {is=title K2Node_FunctionEntry}
   n3 = Append(A=TD_, B=@n2.Material Name)
   n4 = CreateAsset(execute=@n13.then, self=@n2, AssetName=@n3, PackagePath=@n1, AssetClass=/Script/FortniteGame.BuildingTextureData)
   > n5 = Cast To BuildingTextureData(Object=@n4)
   n6 = SaveLoadedAsset(AssetToSave=@n5, bOnlyIfIsDirty=false)
   n7 = Set ReturnTD(execute=[@n6.then, @n16.then], ReturnTD:Building Texture Data Object Reference=@n5) -> (Output_Get:Building Texture Data Object Reference)
   n8 = Append(A=@n1, B=@n3)
   n9 = DoesAssetExist(execute=@n2.then, AssetPath=@n8)
   Return Node!(
    execute=[@n7.then, @n12.then]
    AsBuilding Texture Data:Building Texture Data Object Reference=Get ReturnTD() -> (ReturnTD:Building Texture Data Object Reference)
    needs saving:bool=NOT Boolean(A=@n9)
   )
   n10 = LoadAsset(execute=@n14.then, AssetPath=@n8)
   > n11 = Cast To BuildingTextureData(Object=@n10)
   > n12 = Set ReturnTD(ReturnTD=@n11)
   n13 = PrintString(execute=@n15.else, InString=Append(A=Found Existing TD:, B=@n8))
   n14 = PrintString(execute=@n15.then, InString=Append(A=Found Existing TD:, B=@n8))
   n15 = Branch(execute=@n9.then, Condition=@n9)
   n16 = Set OverrideMaterial(execute=@n5.then, OverrideMaterial=@n2.OverrideMaterial, self=@n5)
  ]}
 DeleteFoliageActors {nodes [
   n1 = Get ObjectsToSave()
   SaveLoadedAssets!(
    execute
    self:Editor Asset Subsystem Object Reference=Get EditorAssetSubsystem() -> (ReturnValue:Editor Asset Subsystem Object Reference) {is=title K2Node_GetEditorSubsystem}
    AssetsToSave=@n1
    bOnlyIfIsDirty=false
   ) -> (then, ReturnValue)
   > EndTransaction()
   n2 = Get ExternalFoliageActorExternalPathsToKill() -> (ExternalFoliageActorExternalPathsToKill:Array of Strings)
   n3 = For Each Loop(Array:Array of Strings=@n2) -> (Array Element:String)
   n4 = GetAssetsByPackageName!(
    execute
    self:Asset Registry Interface=GetAssetRegistry()
    PackageName:Name=String To Name(InString=@n3)
    bIncludeOnlyOnDiskAssets:bool=false
    bSkipARFilteredAssets:bool=true
   ) -> (then, OutAssetData:Array of Asset Data Structures, ReturnValue:bool)
   n5 = "Get (a copy)"(Array:Array of Asset Data Structures=@n4) -> (Output:Asset Data Structure)
   > PrintString(InString="To String (Name)"(InName=Break Asset Data(AssetData=@n5)))
   n6 = Get ExternalFoliageActorsToKill() -> (ExternalFoliageActorsToKill:Array of Strings)
   n7 = CheckOutFiles!(execute, InFiles:Array of Strings=@n6, bSilent:bool=true) -> (then, ReturnValue:bool)
   n8 = Get FoliageActorsToKill() -> (FoliageActorsToKill:Array of Actor Object References)
   DestroyActors!(
    execute=[@n7.then, @n10.then_2]
    self:Editor Actor Subsystem Object Reference=Get EditorActorSubsystem()
    ActorsToDestroy:Array of Actor Object References=@n8
   ) -> (then, ReturnValue:bool)
   DeleteFoliageActors() {is=title K2Node_FunctionEntry}
   > n9 = GetEditorWorld(self=Get UnrealEditorSubsystem())
   > Clear(TargetArray:Array of Object References=Get ObjectsToSave())
   > BeginTransaction(Context=Mod, Description="LOCTEXT(\"Mod\")")
   > n10 = Sequence!(execute) -> (then_0, then_1, then_2:Exec, then_3:Exec)
   > n11 = GetAllActorsOfClass(WorldContextObject=@n9, ActorClass=/Script/Foliage.InstancedFoliageActor) -> (OutActors:Array of Instanced Foliage Actor Object References)
   > n12 = Execute Python Script!(
    execute
    PythonScript="import unreal\r\narr = []\r\nfor i in actors:\r\n    arr.append(i.get_package())"
    Actors:Array of Instanced Foliage Actor Object References=@n11
   ) -> (then, ReturnValue, arr:Array of Package Object References) {inputs [Actors], outputs [arr]}
   > Clear(TargetArray:Array of Actor Object References=Get FoliageActorsToKill())
   > Clear(TargetArray:Array of Strings=Get ExternalFoliageActorsToKill())
   > n13 = For Each Loop(Array:Array of Instanced Foliage Actor Object References=@n11) -> (Array Element:Instanced Foliage Actor Object Reference)
   n14 = Get Object Path String(Object=@n13)
   n15 = ParseIntoArray(SourceString=@n14, Delimiter=/)
   > n16 = Branch(Condition="Equal Exactly (String)"(A="Get (a copy)"(Array:Array of Strings=@n15) -> (Output:String), B=Temp))
   n17 = GetDisplayName(Object=GetPackageForObject(Object=@n13))
   Add(execute=@n16.else, TargetArray:Array of Strings=Get ExternalFoliageActorsToKill(), NewItem:"String (by ref)"=@n17)
   > Add(TargetArray=Get ExternalFoliageActorExternalPathsToKill(), NewItem=GetSystemPath!(Object:Object Reference=@n13) -> (ReturnValue:String))
   > Add(TargetArray:Array of Actor Object References=Get FoliageActorsToKill(), NewItem:"Actor Object Reference (by ref)"=@n13)
   n18 = GetPackageForObject(Object=@n13)
   > Add(TargetArray:Array of Object References=Get ObjectsToSave(), NewItem:"Object Reference (by ref)"=@n18)
   > PrintString(InString=@n17)
   SavePackages!(execute=@n10.then_3, PackagesToSave:Array of Package Object References=@n12.arr, bOnlyDirty:bool=false) -> (then, ReturnValue:bool)
   n19 = For Each Loop(Array:Array of Strings=@n6) -> (Array Element:String)
   > n20 = GetAssetsByPath(self=GetAssetRegistry(), PackagePath=String To Name(InString=@n19), bRecursive=false, bIncludeOnlyOnDiskAssets=false)
   n21 = "Get (a copy)"(Array:Array of Asset Data Structures=@n20) -> (Output:Asset Data Structure)
   Break Asset Data(AssetData=@n21)
   Add(
    TargetArray:Array of Package Object References=Make Array!("[0]":Package Object Reference) -> (Array:Array of Package Object References)
    NewItem:"Package Object Reference (by ref)"=GetPackageForObject(Object="Get (a copy)"(Array:Array of Instanced Foliage Actor Object References=@n11) -> (Output:Instanced Foliage Actor Object Reference))
   )
   Set Array Elem!(execute, TargetArray=@n2, Index:int=@n3.Array Index, Item:"String (by ref)", bSizeToFit:bool=false) -> (then)
   SaveLoadedAsset(bOnlyIfIsDirty=true)
   SavePackages(bOnlyDirty=false)
   GetAssetByObjectPath(ObjectPath=None, bIncludeOnlyOnDiskAssets=false)
   GetSystemPath(Object=GetAsset(InAssetData=@n21))
   SavePackages(bOnlyDirty=false)
   GetDisplayName(Object=@n13)
   CheckoutLoadedAssets(AssetsToCheckout=@n1)
   SavePackages(PackagesToSave=@n12.arr, bOnlyDirty=false)
   GetDirtyContentPackages!(execute) -> (then, OutDirtyPackages:Array of Package Object References)
   "To String (Name)"(InName=GetFolderPath!(self:Actor Object Reference=@n13) -> (ReturnValue:Name=None))
   SavePackages(bOnlyDirty=false)
   CheckOutFile!(execute=@n3.LoopBody, InFile:String=@n3, bSilent:bool=false) -> (then, ReturnValue:bool)
   CheckOutFiles(InFiles=@n2, bSilent=false)
   SaveLoadedAssets!(execute, AssetsToSave=@n8, bOnlyIfIsDirty=false) -> (then, ReturnValue)
   Return Node!(execute, Package Array To Save:Array of Package Object References=@n12.arr)
   Reroute Node(InputPin:String=@n14) -> (OutputPin:String)
   Make Array("[0]"=@n18)
   Get ObjectsToSave()
   Get ObjectsToSave()
   Get FoliageActorsToKill()
  ]}
 DeleteFoliageActorsClean {nodes [
   n1 = DeleteFoliageActorsClean!() -> (then, WorldContextObject:Object Reference) {is=title K2Node_FunctionEntry}
   > n2 = GetEditorWorld(self=Get UnrealEditorSubsystem())
   > Clear(TargetArray:Array of Object References=Get ObjectsToSave())
   > n3 = Sequence()
   > n4 = GetAllActorsOfClass(WorldContextObject=SelectObject(A=@n1, B=@n2, bSelectA=IsValid(Object=@n1)), ActorClass=&InstancedFoliageActor)
   > n5 = Execute Python Script(PythonScript="import unreal\r\narr = []\r\nfor i in actors:\r\n    arr.append(i.get_package())", Actors=@n4) {inputs [Actors], outputs [arr]}
   > Clear(TargetArray:Array of Actor Object References=Get FoliageActorsToKill())
   > Clear(TargetArray:Array of Strings=Get ExternalFoliageActorsToKill())
   > n6 = For Each Loop(Array:Array of Instanced Foliage Actor Object References=@n4) -> (Array Element:Instanced Foliage Actor Object Reference)
   n7 = ParseIntoArray(SourceString=Get Object Path String(Object=@n6), Delimiter=/)
   > n8 = Branch(Condition="Equal Exactly (String)"(A="Get (a copy)"(Array:Array of Strings=@n7) -> (Output:String), B=Temp))
   n9 = GetDisplayName(Object=GetPackageForObject(Object=@n6))
   Add(execute=@n8.else, TargetArray:Array of Strings=Get ExternalFoliageActorsToKill(), NewItem:"String (by ref)"=@n9)
   > Add(TargetArray=Get ExternalFoliageActorExternalPathsToKill(), NewItem=GetSystemPath(Object=@n6))
   > Add(TargetArray:Array of Actor Object References=Get FoliageActorsToKill(), NewItem:"Actor Object Reference (by ref)"=@n6)
   > Add(TargetArray:Array of Object References=Get ObjectsToSave(), NewItem:"Object Reference (by ref)"=GetPackageForObject(Object=@n6))
   > PrintString(InString=@n9)
   DestroyActors(execute=@n3.then_2, self=Get EditorActorSubsystem(), ActorsToDestroy=Get FoliageActorsToKill())
   SavePackages(execute=@n3.then_3, PackagesToSave=@n5.arr, bOnlyDirty=false)
   SavePackages(bOnlyDirty=false)
  ]}
 DeleteMeshGrass_BP {nodes [
   DeleteMeshGrass_BP() {is=title K2Node_FunctionEntry}
   > Clear(TargetArray:Array of Actor Object References=Get FoliageActorsToKill())
   > Clear(TargetArray:Array of Guid Structures=Get Guids() -> (Guids:Array of Guid Structures))
   > n1 = Sequence()
   > n2 = GetActorDescs!(execute) -> (then, OutActorDescs:Array of Actor Desc Structures, ReturnValue:bool)
   > n3 = For Each Loop(Array:Array of Actor Desc Structures=@n2) -> (Array Element:Actor Desc Structure)
   LoadActors!(execute=@n3.Completed, InActorsToLoad:Array of Guid Structures=Get Guids()) -> (then)
   n4 = Break Actor Desc!(ActorDesc:"Actor Desc Structure (by ref)"=@n3) -> (
    Guid:Guid Structure
    NativeClass:Object Class Reference
    Class:Soft Object Path Structure
    Name:Name
    Label:Name
    Bounds:Box Structure
    RuntimeGrid:Name
    bIsSpatiallyLoaded:bool=~
    bActorIsEditorOnly:bool=~
    ActorPackage:Name
    ActorPath:Name
    FolderPath:Name
    DataLayerAssets:Array of Soft Object Path Structures
   ) {is=title K2Node_BreakStruct}
   n5 = GetObjectFromSoftPath!(Path:"Soft Object Path Structure (by ref)"=@n4.Class) -> (ReturnValue:Object Reference)
   n6 = GetDisplayName(Object=@n5)
   Branch(execute=@n3.LoopBody, Condition="Equal, Case Insensitive (String)"(A=@n6, B=BP_MeshGrass_C))
   > PrintString(InString=@n6)
   > Add(TargetArray:Array of Guid Structures=Get Guids(), NewItem:"Guid Structure (by ref)"=@n4)
   n7 = GetEditorWorld(execute=@n1.then_1, self=Get UnrealEditorSubsystem())
   > n8 = GetAllActorsOfClass(WorldContextObject=@n7, ActorClass=/Game/Athena/Helios/Environments/Landscape/WorldBuildingTools/PCGGrass/BP_MeshGrass.BP_MeshGrass_C) -> (OutActors:Array of BP Mesh Grass Object References)
   > n9 = Execute Python Script(
    PythonScript="import unreal\r\narr = []\r\nfor i in actors:\r\n    arr.append(i.get_package())"
    Actors:Array of BP Mesh Grass Object References=@n8
   ) {inputs [Actors], outputs [arr]}
   > n10 = For Each Loop(Array:Array of BP Mesh Grass Object References=@n8) -> (Array Element:BP Mesh Grass Object Reference)
   > Add(TargetArray:Array of Actor Object References=Get FoliageActorsToKill(), NewItem:"Actor Object Reference (by ref)"=@n10)
   DestroyActors(execute=@n1.then_2, self=Get EditorActorSubsystem(), ActorsToDestroy=Get FoliageActorsToKill())
   SavePackages(execute=@n1.then_3, PackagesToSave=@n9.arr, bOnlyDirty=false)
   Get Class Display Name(Class=@n4.NativeClass)
   PrintString(InString="To String (Integer)"(InInt=Length(TargetArray:Array of BP Mesh Grass Object References=@n8)))
   Contains(SearchIn="To String (Name)"(InName=@n4.Name), Substring=grass)
   GetPackageForObject()
   Return Node!(execute, ReturnValue:Object Reference=@n5)
  ]}
 Dup Texture Data and Material Instances of All Cliff Actors {nodes [
   n1 = GetEditorWorld(execute=@n4.then, self=Get UnrealEditorSubsystem())
   n2 = GetAllActorsOfClass(WorldContextObject=@n1, ActorClass=/Script/FortniteGame.BuildingLevelInstance) -> (OutActors:Array of Building Level Instance Object References)
   > n3 = For Each Loop(Array:Array of Building Level Instance Object References=@n2) -> (Array Element:Building Level Instance Object Reference)
   > PrintString(InString=GetDisplayName(Object=@n3))
   > PrintString(InString="To String (Boolean)"(InBool=IsLoaded!(self:Level Instance Interface Interface=@n3) -> (ReturnValue:bool)))
   n4 = Dup Texture Data and Material Instances of All Cliff Actors() {is=title K2Node_FunctionEntry}
   n5 = GetAllActorsOfClass(execute=@n1.then, WorldContextObject=@n1, ActorClass=&BP__Cliff_Figment_C) -> (OutActors:"Array of BP  Cliff Figment Object References")
   n6 = For Each Loop(Exec=@n10.then, Array:Array of Actor Object References=@n5) -> (Array Element:Actor Object Reference)
   n7 = Cast To BP__Cliff_Figment(execute=@n11.then, Object=@n6)
   > n8 = AssignTextureDataToCliffActorCreateIfMissing!(execute, self, CliffActor:"BP  Cliff Figment Object Reference"=@n7) -> (then, ActorToSave:"BP  Cliff Figment Object Reference") {is=title K2Node_CallFunction}
   > Add(
    TargetArray:"Array of BP  Cliff Figment Object References"=Get Cliff Actors to Save()
    NewItem:"BP  Cliff Figment Object Reference (by ref)"=@n8
   )
   n9 = Get Cliff Actors to Save()
   SaveLoadedAssets(execute=@n12.then, AssetsToSave=@n9, bOnlyIfIsDirty=true)
   n10 = Clear(execute=@n5.then, TargetArray:"Array of BP  Cliff Figment Object References"=Get Cliff Actors to Save())
   n11 = PrintString(execute=@n6.LoopBody, InString=GetDisplayName(Object=@n6))
   n12 = CheckoutLoadedAssets(execute=@n6.Completed, AssetsToCheckout=@n9)
  ]}
 EnsureMaterialAndTextureDataExist {nodes [
   n1 = EnsureMaterialAndTextureDataExist!() -> (then, Mesh Material:Material Interface Object Reference) {is=title K2Node_FunctionEntry}
   n2 = GetDisplayName(Object=@n1)
   n3 = TextureAssetPathPrep!(execute=@n30.then, self, Material Asset Name:String=@n2) -> (then, "Material Full  Path":String, TD Asset Full Path:String, TD Asset Name:String, Material Folder:String, TD Asset Folder:String)
   n4 = GetAssetTools(execute=@n29.then)
   n5 = DuplicateAsset(execute=@n33.then, self=@n4, AssetName=@n2, PackagePath=@n3.Material Folder, OriginalObject=@n1)
   > n6 = Cast To MaterialInstanceConstant(Object:Object Reference=@n5)
   n7 = SetMaterialInstanceParent(Instance=@n6, NewParent=/FigmentEnvironment/Materials/Cliff/Athena_Terrain_Mat_Biome_Figment_Inst)
   SaveLoadedAsset(execute=[@n7.then, @n6.then, @n23.else, @n23.then], AssetToSave=@n6, bOnlyIfIsDirty=false)
   > n8 = Cast To MaterialInterface(Object=@n5)
   > n9 = Set CurrentMaterialForStaticMesh(CurrentMaterialForStaticMesh:Material Interface Object Reference=@n8) -> (Output_Get:Material Interface Object Reference)
   PrintString(execute=[@n9.then, @n28.then, @n34.then], InString=Append!(A="does ", B=@n3.TD Asset Full Path, C:String=" exist (b)") -> (ReturnValue))
   > n10 = DoesAssetExist(AssetPath=@n3.TD Asset Full Path)
   > n11 = Branch(Condition=@n10)
   > PrintString(InString=TD_Exists)
   > n12 = LoadAsset!(execute, self:Editor Asset Subsystem Object Reference=Get EditorAssetSubsystem(), AssetPath=@n3.TD Asset Full Path) -> (then, ReturnValue)
   > n13 = Cast To BuildingTextureData(Object=@n12)
   > n14 = Set CurrentTDForStaticMesh(CurrentTDForStaticMesh:Building Texture Data Object Reference=@n13) -> (Output_Get:Building Texture Data Object Reference)
   n15 = Get CurrentTDForStaticMesh() -> (CurrentTDForStaticMesh:Building Texture Data Object Reference)
   n16 = Get CurrentMaterialForStaticMesh() -> (CurrentMaterialForStaticMesh:Material Interface Object Reference)
   Return Node!(
    execute=[@n20.then, @n14.then]
    " Texture Data To Assign":Building Texture Data Object Reference=@n15
    Material:Material Interface Object Reference=@n16
    bMaterialCreated1:bool=Get bMaterialCreated() -> (bMaterialCreated:bool)
   )
   PrintString(execute=@n11.else, InString=TD does not Exist)
   > n17 = CreateAsset(self=@n4, AssetName=@n3.TD Asset Name, PackagePath=@n3.TD Asset Folder, AssetClass=&BuildingTextureData)
   > n18 = Cast To BuildingTextureData(Object=@n17)
   > Set CurrentTDForStaticMesh(CurrentTDForStaticMesh=@n18)
   > n19 = Set OverrideMaterial(OverrideMaterial=@n16, self=@n15)
   Add(execute=[@n21.then, @n19.then], TargetArray:Array of Object References=Get ObjectsToSave(), NewItem:"Object Reference (by ref)"=@n15)
   > n20 = CheckoutLoadedAsset!(execute, AssetToCheckout:Object Reference=@n15) -> (then, ReturnValue:bool)
   n21 = SaveLoadedAsset(AssetToSave=@n15, bOnlyIfIsDirty=false)
   n22 = Get Parent(self=@n6)
   n23 = Branch(Condition=OR Boolean!(
     A="Equal (Object)"(A=@n22, B=&Athena_Terrain_Mat_Biome)
     B="Equal (Object)"(A=@n22, B=/Game/Athena/Environments/Landscape/Material/Athena_Terrain_Mat_Biome_Inst)
     C:bool="Equal (Object)"(A=@n22, B=/Game/Packages/Fortnite_UniqueMaterials_1/DS_Fortnight/Terrain_Mat)
    ) -> (ReturnValue))
   "To String (Text)"(InText="GetText (Text Box)"!(self:Text Box Object Reference=Get TargetTextBox_()) -> (ReturnValue:Text))
   Get TargetTextBox_()
   PrintString(execute=@n3.then, InString=Append!(A="does ", B=@n3, C=" exist", D:String=(1)) -> (ReturnValue))
   > n24 = DoesAssetExist(AssetPath=@n3)
   > n25 = Branch(Condition=@n24)
   > PrintString(InString=exists)
   > n26 = LoadAsset(self=Get EditorAssetSubsystem(), AssetPath=@n3)
   > n27 = Cast To MaterialInterface(Object=@n26)
   > n28 = Set CurrentMaterialForStaticMesh(CurrentMaterialForStaticMesh=@n27)
   PrintString(execute=@n25.else, InString="Doesn't exist, iusing original")
   > n29 = Set bMaterialCreated(bMaterialCreated:bool=true) -> (Output_Get:bool)
   n30 = PrintString(execute=@n1.then, InString=@n2)
   n31 = Get Object Path String(Object=@n1)
   n32 = Contains(SearchIn=@n31, Substring=Figment_S09)
   n33 = Branch(execute=@n4.then, Condition=@n32)
   n34 = Set CurrentMaterialForStaticMesh(execute=@n33.else, CurrentMaterialForStaticMesh=@n1)
  ]}
 EventGraph {nodes [
   n1 = Construct Object from Class(execute=@n142.then, Class=&/EUT_SnowReplace.EUT_SnowReplace_C) -> (ReturnValue:EUT Snow Replace Object Reference)
   > n2 = RegisterAndExecuteTask!(execute:Exec, Task:Editor Utility Task Object Reference=@n1, OptionalParentTask:Editor Utility Task Object Reference) -> (then:Exec, OnFinished:Exec, Task:Editor Utility Task Object Reference) {is=title K2Node_AsyncAction}
   n3 = Cast To EUT_SnowReplace!(execute=@n214.then, Object=@n2) -> (then, CastFailed, AsEUT Snow Replace:EUT Snow Replace Object Reference) {is=title K2Node_DynamicCast}
   n4 = Get ActorsToSave!(self:EUT Snow Replace Object Reference=@n3) -> (ActorsToSave:Array of Object References)
   n5 = Clear(TargetArray:Array of Object References=@n4)
   CollectGarbage!(execute=[@n5.then, @n170.then]) -> (then)
   n6 = String To Name(InString=/Figment_S09_LevelInstances/Streaming/Sublevel_X3Y3)
   n7 = GetAssetRegistry()
   n8 = GetAssetsByPath(self=@n7, PackagePath=@n6, bRecursive=true, bIncludeOnlyOnDiskAssets=true)
   n9 = GetAssetsByPackageName(self=@n7, PackageName=@n6, bIncludeOnlyOnDiskAssets=false, bSkipARFilteredAssets=true)
   PrintString(execute=[@n8.then, @n9.then], InString="To String (Integer)"(InInt=Length(TargetArray:Array of Asset Data Structures=@n9)))
   n10 = GetEditorWorld(self=Get UnrealEditorSubsystem())
   > n11 = GetAllActorsOfClass(WorldContextObject=@n10, ActorClass=&BP__Cliff_Figment_C)
   > n12 = For Each Loop(Array:"Array of BP  Cliff Figment Object References"=@n11) -> (Array Element:"BP  Cliff Figment Object Reference")
   FreeAllMeshes(execute=@n12.Completed, self:Dynamic Mesh Pool Object Reference=Get DMP() -> (DMP:Dynamic Mesh Pool Object Reference))
   n13 = LoadAsset!(execute, AssetPath=MakeLiteralString(Value=/Figment_S07/LM_7_00/Athena/Environments/Landscape/MPC/MPC_Landscape)) -> (then, ReturnValue)
   > Is Valid(InputObject=@n13)
   > n14 = Cast To MaterialParameterCollection!(execute, Object=@n13) -> (then, CastFailed, AsMaterial Parameter Collection:Material Parameter Collection Object Reference) {is=title K2Node_DynamicCast}
   > Execute Python Script!(
    execute
    PythonScript="names = mpc.get_editor_property(\"scalar_parameters\")\r\nfor x in names:\r\n    print(x.default_value)"
    MPC:Material Parameter Collection Object Reference=@n14
   ) -> (then, ReturnValue) {inputs [MPC]}
   n15 = GetAssetsByPath(
    self=GetAssetRegistry()
    PackagePath=String To Name(InString=MakeLiteralString(Value=/Figment_S09_LevelInstances/__ExternalActors__/POIs/))
    bRecursive=true
    bIncludeOnlyOnDiskAssets=true
   )
   > n16 = For Each Loop(Array:Array of Asset Data Structures=@n15) -> (Array Element:Asset Data Structure)
   > PrintString(InString=Get Class Display Name(Class=FindAssetNativeClass!(AssetData:"Asset Data Structure (by ref)"=@n16) -> (ReturnValue:Object Class Reference)))
   n17 = GetEditorWorld(self=Get UnrealEditorSubsystem())
   > SetScalarParameterValue!(
    execute
    WorldContextObject=@n17
    Collection:Material Parameter Collection Object Reference=@n14
    ParameterName:Name=MakeLiteralName(Value=Holiday_SnowAmount)
    ParameterValue:float=0.333
   ) -> (then) {is=title K2Node_CallMaterialParameterCollectionFunction}
   n18 = Get SourceTextBox() -> (SourceTextBox:Editor Utility Editable Text Box Object Reference)
   n19 = For Each LI!(execute:Exec=@n118.else, InputPin:String="To String (Text)"(InText="GetText (Text Box)"(self=@n18))) -> (then:Exec, Output_Get:Object Reference) {is=title K2Node_MacroInstance}
   n20 = GetEditorWorld(execute=[@n118.then, @n19.then], self=Get UnrealEditorSubsystem())
   n21 = GetAllActorsOfClass(
    execute=@n209.then
    WorldContextObject=SelectObject(A=@n19, B=@n20, bSelectA=IsValid(Object=@n19))
    ActorClass=&BP__Cliff_Figment_C
   )
   n22 = For Each Loop(Exec=[@n21.then, @n81.then], Array:Array of Actor Object References=@n21) -> (Array Element:Actor Object Reference)
   n23 = Cast To BP__Cliff_Figment(execute=[@n71.then, @n22.LoopBody], Object=@n22)
   n24 = Get StaticMeshComponent(self=@n23)
   n25 = Cast To MaterialInstanceDynamic!(
    execute=@n221.then
    Object=GetMaterial!(self:Primitive Component Object Reference=@n24, ElementIndex:int=0) -> (ReturnValue:Material Interface Object Reference)
   ) -> (then, CastFailed, AsMaterial Instance Dynamic:Material Instance Dynamic Object Reference) {is=title K2Node_DynamicCast}
   n26 = Get Parent(self=@n25)
   > n27 = Cast To MaterialInstanceConstant(Object=@n26)
   n28 = GetDisplayName(Object=@n27)
   n29 = PrintString(InString=@n28)
   n30 = FindDominantVertexColor!(execute=[@n29.then, @n215.then_0], self, Cliff BP:"BP  Cliff Figment Object Reference"=@n23) -> (then, FinallWeightedColor:Linear Color Structure)
   n31 = BreakColor(InColor=@n30)
   > n32 = TechArt_EUW_FloatBubbleSort!(
    execute
    Input Value Array:"Array of Float (double-precision)s"=Make Array!("[0]":double=@n31, "[1]":double=@n31.G, "[2]":double=@n31.B) -> (Array:"Array of Float (double-precision)s")
    ReverseOrder:"Boolean (by ref)"=true
    __WorldContext:Object Reference
   ) -> (then, Sorted Values:"Array of Float (double-precision)s", Sorted Index Array:Array of Integers)
   > n33 = Branch(Condition=Contains(SearchIn=@n28, Substring=Arid))
   n34 = "Get (a ref)"(Array:Array of Integers=@n32.Sorted Index Array, Dimension 1=0) -> (Output:"Integer (by ref)")
   > n35 = Switch on Int!(execute:Exec, Selection:int=@n34) -> (Default:Exec, 0:Exec, 1:Exec, 2:Exec)
   n36 = "Get (a ref)"(Array:"Array of Float (double-precision)s"=@n32, Dimension 1=0) -> (Output:"Float (double-precision) (by ref)")
   n37 = "float > float"(A=@n36, B=0.3333)
   n38 = Branch(execute=@n35.0, Condition=@n37)
   > Set PhysToSet(PhysToSet:Physical Material Object Reference=/Game/FortressPhysicalMaterials/Asphalt) -> (Output_Get:Physical Material Object Reference)
   Set PhysToSet(execute=[@n38.else, @n35.Default], PhysToSet=/Game/FortressPhysicalMaterials/Sand)
   n39 = Branch(
    execute=@n35.1
    Condition=XOR Boolean!(
     A:bool=@n37
     B:bool=GetMaterialInstanceStaticSwitchParameterValue!(
      Instance:Material Instance Constant Object Reference=@n27
      ParameterName:Name=UseGrassAsBase
      Association:EMaterialParameterAssociation Enum=GlobalParameter
     ) -> (ReturnValue:bool)
    ) -> (ReturnValue:bool)
   )
   > Set PhysToSet(PhysToSet=/Game/FortressPhysicalMaterials/Grass)
   Set PhysToSet(execute=@n39.else, PhysToSet=&Sand)
   n40 = Branch(
    execute=@n33.else
    Condition=GetMaterialInstanceStaticSwitchParameterValue(Instance=@n27, ParameterName=UseConcrete_Exclusive, Association=GlobalParameter)
   )
   Set PhysToSet(execute=[@n43.then, @n40.then], PhysToSet=&Asphalt)
   n41 = Branch(execute=@n40.else, Condition=@n37)
   > n42 = Switch on Int(Selection=@n34)
   n43 = Branch(
    execute=@n42.0
    Condition=GetMaterialInstanceStaticSwitchParameterValue(Instance=@n27, ParameterName=UseConcrete, Association=GlobalParameter)
   )
   Set PhysToSet(execute=@n43.else, PhysToSet=/Game/FortressPhysicalMaterials/Generic)
   n44 = Make Array(
    "[0]":bool=Contains(SearchIn=@n28, Substring=Farm)
    "[1]":bool=Contains(SearchIn=@n28, Substring=Forest)
    "[2]":bool=Contains(SearchIn=@n28, Substring=Dirt)
   ) -> (Array:Array of Booleans)
   n45 = Switch on Int!(
    execute=[@n47.then, @n42.1]
    Selection=int + int(A=Find Item!(TargetArray:Array of Booleans=@n44, ItemToFind:"Boolean (by ref)"=true) -> (ReturnValue))
   ) -> (Default, 0, 1, 2, 3:Exec)
   Set PhysToSet(execute=[@n48.then, @n45.0, @n45.1, @n45.2, @n45.3], PhysToSet=/Game/FortressPhysicalMaterials/Dirt)
   Set PhysToSet(execute=@n45.Default, PhysToSet=&Generic)
   n46 = "Get (a ref)"(Array=@n32, Dimension 1=1)
   n47 = Branch(
    execute=@n42.2
    Condition=AND Boolean(
     A="float > float"(A=@n46, B=0.3333)
     B=GetMaterialInstanceStaticSwitchParameterValue(Instance=@n27, ParameterName="UseAdditionalTopLayerBlend(ControlledByGreenVertColor)", Association=GlobalParameter)
    )
   )
   n48 = Branch(
    execute=@n47.else
    Condition=GetMaterialInstanceStaticSwitchParameterValue(Instance=@n27, ParameterName=UseDirt, Association=GlobalParameter)
   )
   Set PhysToSet(execute=@n48.else, PhysToSet=&Grass)
   n49 = Branch(
    execute=@n41.else
    Condition=GetMaterialInstanceStaticSwitchParameterValue(Instance=@n27, ParameterName="Use Grass(true) or a BaseTilingTexture(False)", Association=GlobalParameter)
   )
   > n50 = Branch(Condition="float > float"(A=@n31.A, B=0.575))
   Set PhysToSet(execute=[@n50.then, @n49.else], PhysToSet=&Dirt)
   Set PhysToSet(execute=@n50.else, PhysToSet=&Grass)
   n51 = GetAssetsByPath(
    self=GetAssetRegistry()
    PackagePath=String To Name(InString=MakeLiteralString(Value=&Cliffs/))
    bRecursive=false
    bIncludeOnlyOnDiskAssets=true
   )
   > n52 = For Each Loop(Array:Array of Asset Data Structures=@n51) -> (Array Element:Asset Data Structure)
   n53 = GetSelectedAssetsOfClass(execute=@n136.then_0, AssetClass=/Script/Engine.MaterialInstanceConstant)
   > n54 = For Each Loop(Array:Array of Object References=@n53) -> (Array Element:Object Reference)
   > n55 = Cast To MaterialInstanceConstant(Object=@n54)
   n56 = SaveLoadedAsset(AssetToSave=@n55, bOnlyIfIsDirty=false)
   n57 = Branch(
    execute=[@n176.then, @n56.then, @n174.then]
    Condition=GetMaterialInstanceStaticSwitchParameterValue(Instance=@n55, ParameterName=UseColorTest, Association=GlobalParameter)
   )
   n58 = SetMaterialInstanceStaticSwitchParameterValue!(
    execute=@n57.else
    Instance:Material Instance Constant Object Reference=@n55
    ParameterName:Name=UseColorTest
    Value:bool=true
    Association:EMaterialParameterAssociation Enum=GlobalParameter
    bUpdateMaterialInstance:bool=true
   ) -> (then, ReturnValue:bool)
   > Branch(Condition=@n58)
   > Add(TargetArray=Get ObjectsToSave())
   n59 = Keys(TargetMap:Map of Strings to Building Level Instance Object References=Get worldActors!(self:EUT Snow Replace Object Reference=@n3) -> (worldActors:Map of Strings to Building Level Instance Object References)) -> (Keys:Array of Strings)
   > n60 = For Each Loop(Array:Array of Strings=@n59) -> (Array Element:String)
   > PrintString(InString=@n60)
   CollectGarbage()
   > SaveLoadedAssets(bOnlyIfIsDirty=true)
   n61 = GetEditorWorld(self=Get UnrealEditorSubsystem())
   n62 = Clear(execute=[@n61.then, @n64.then], TargetArray=Get ObjectsToSave())
   n63 = Get SourceTextBox()
   n64 = For Each LI(execute=@n143.then, InputPin="To String (Text)"(InText="GetText (Text Box)"(self=@n63)))
   n65 = GetAllActorsOfClass(execute=@n62.then, WorldContextObject=@n64, ActorClass=&FortStaticMeshActor) -> (OutActors:Array of Fort Static Mesh Actor Object References)
   > n66 = For Each Loop(Array:Array of Fort Static Mesh Actor Object References=@n65) -> (Array Element:Fort Static Mesh Actor Object Reference)
   n67 = Get ObjectsToSave()
   CheckoutLoadedAssets(execute=@n66.Completed, AssetsToCheckout=@n67)
   > SaveLoadedAssets(AssetsToSave=@n67, bOnlyIfIsDirty=false)
   n68 = Get StaticMeshComponent(self=@n66)
   n69 = GetEditorMaterial!(self:Primitive Component Object Reference=@n68, ElementIndex:int=0) -> (ReturnValue:Material Interface Object Reference)
   n70 = Cast To MaterialInstanceConstant(execute=@n66.LoopBody, Object=@n69)
   > Branch(Condition="Equal (Object)"(A=Get Parent(self=@n70), B=/FigmentEnvironment/LM_8_1_1/Athena/WinterArt/Materials/M_Ice_Snow_Inst))
   n71 = GetSelectedLevelActors(self=Get EditorActorSubsystem())
   n72 = GetActorDescs()
   > n73 = For Each Loop(Array:Array of Actor Desc Structures=@n72) -> (Array Element:Actor Desc Structure)
   n74 = Break Actor Desc(ActorDesc=@n73) -> (bIsSpatiallyLoaded=~, bActorIsEditorOnly=~)
   > Branch(Condition="Equal (Class)"(A=@n74.NativeClass, B=&BuildingLevelInstance))
   > PrintString(InString="To String (Name)"(InName=@n74.Name))
   n75 = GetEditorWorld(self=Get UnrealEditorSubsystem())
   n76 = GetSelectedLevelActors(execute=@n167.then, self=Get EditorActorSubsystem())
   n77 = "Get (a copy)"(Array:Array of Actor Object References=@n76) -> (Output:Actor Object Reference)
   > n78 = Cast To Landscape!(execute, Object=@n77) -> (then, CastFailed, AsLandscape:Landscape Object Reference) {is=title K2Node_DynamicCast}
   n79 = Landscape Export Heightmap to RenderTarget!(
    execute=[@n75.then, @n169.then]
    self:Landscape Proxy Object Reference=@n78
    InRenderTarget:Texture Render Target 2D Object Reference=&/RT_WterBody
    InExportHeightIntoRGChannel:bool=true
    InExportLandscapeProxies:bool=true
   ) -> (then, ReturnValue:bool)
   > n80 = Branch(Condition=@n79)
   > PrintString(InString=RT)
   PrintString(execute=@n80.else, InString="Failed RT\r\n")
   n81 = GetSelectedLevelActors(self=Get EditorActorSubsystem())
   n82 = GetSelectedLevelActors(self=Get EditorActorSubsystem())
   n83 = GetEditorWorld(execute=[@n120.then, @n186.then], self=Get UnrealEditorSubsystem())
   n84 = GetAllActorsOfClass(execute=@n210.then_1, WorldContextObject=@n83, ActorClass=&FortStaticMeshActor)
   n85 = For Each Loop(Exec=[@n84.then, @n82.then], Array:Array of Actor Object References=@n84) -> (Array Element:Actor Object Reference)
   n86 = Get ObjectsToSave()
   CheckoutLoadedAssets(execute=@n85.Completed, AssetsToCheckout=@n86)
   > SaveLoadedAssets(AssetsToSave=@n86, bOnlyIfIsDirty=false)
   n87 = Cast To FortStaticMeshActor(execute=@n85.LoopBody, Object=@n85)
   n88 = "GetText (Text Box)"(self=Get TargetTextBox() -> (TargetTextBox:Editor Utility Editable Text Box Object Reference))
   n89 = Get StaticMeshComponent(self=@n87)
   n90 = GetMaterial(self=@n89, ElementIndex=0)
   n91 = Append!(A=Split!(InPath:String=Get Object Path String(Object=@n90)) -> (PathPart:String, FilenamePart:String, ExtensionPart:String), B=/) -> (ReturnValue)
   > n92 = Branch(Condition="Not Equal, Case Insenstive (String)"!(A:String=@n91, B:String="To String (Text)"(InText=@n88)) -> (ReturnValue:bool))
   n93 = GetDisplayName(Object=@n90)
   n94 = Find(TargetMap:Map of Strings to Material Interface Object References=Get MapStringToTargetMaterials(), Key=@n93) -> (Value:Material Interface Object Reference)
   > n95 = Branch(Condition=@n94.ReturnValue)
   > SetMaterial(self=@n89, ElementIndex=0, Material=@n94)
   > Add(TargetArray=Get ObjectsToSave(), NewItem=@n87)
   PrintString(
    execute=@n95.else
    InString=Append!(A=MakeLiteralString(Value=Couldn't find material for:), B=GetDisplayName(Object=@n87), C:String=" -> ", D:String=@n93) -> (ReturnValue)
   )
   PrintString(execute=@n92.else, InString=in path)
   CheckoutLoadedAssets()
   > SaveLoadedAssets(bOnlyIfIsDirty=true)
   "On Clicked (FixUpInLevel)"() {is=title K2Node_ComponentBoundEvent}
   > PaintConcreteVertex1000!(execute, self, WorldContextObject:Object Reference) -> (then)
   "On Clicked (CreateCliffBPfrommesh)"() {is=title K2Node_ComponentBoundEvent}
   > CreateCliffBpFromStaticMesh() {is=title K2Node_CallFunction}
   "On Clicked (EditorUtilityButton_34)"() {is=title K2Node_ComponentBoundEvent}
   > SetActiveWidgetIndex(self=Get WidgetSwitcher_184() -> (WidgetSwitcher_184:Widget Switcher Object Reference), Index=0)
   n96 = Get HorizontalBox_174() -> (HorizontalBox_174:Horizontal Box Object Reference)
   > n97 = For Each Loop(Array:Array of Widget Object References=GetAllChildren(self=@n96)) -> (Array Element:Widget Object Reference)
   > n98 = Cast To EditorUtilityButton!(execute, Object=@n97) -> (then, CastFailed, AsEditor Utility Button:Editor Utility Button Object Reference) {is=title K2Node_DynamicCast}
   > SetButtonOutlineWidth!(execute, self, InputPin:double=Make Literal Float(Value=0), InputPin2:Editor Utility Button Object Reference=@n98) -> (then)
   > SetButtonOutlineWidth(InputPin=3, InputPin2=Get EditorUtilityButton_34() -> (EditorUtilityButton_34:Editor Utility Button Object Reference))
   "On Clicked (EditorUtilityButton_60)"() {is=title K2Node_ComponentBoundEvent}
   > SetActiveWidgetIndex(self=Get WidgetSwitcher_184(), Index=1)
   > n99 = For Each Loop(Array=GetAllChildren(self=@n96))
   > n100 = Cast To EditorUtilityButton(Object=@n99)
   > SetButtonOutlineWidth(InputPin=Make Literal Float(Value=0), InputPin2=@n100)
   > SetButtonOutlineWidth(InputPin=3, InputPin2=Get EditorUtilityButton_60() -> (EditorUtilityButton_60:Editor Utility Button Object Reference))
   "On Clicked (DeleteFoliage_1)"() {is=title K2Node_ComponentBoundEvent}
   n101 = "To String (Text)"(InText="GetText (Text Box)"(self=Get SourceTextBox()))
   > n102 = For Each LI(InputPin=@n101)
   DeleteFoliageActorsClean!(execute=[@n102.then, @n194.then], self, WorldContextObject:Object Reference=@n102) -> (then) {is=title K2Node_CallFunction}
   "On Clicked (ClearTagsInPath)"() {is=title K2Node_ComponentBoundEvent}
   n103 = "To String (Text)"(InText="GetText (Text Box)"(self=Get SourceTextBox()))
   > n104 = For Each LI(InputPin=@n103)
   ClearTags!(execute=[@n104.then, @n195.then], self, WorldContextObject:Object Reference=@n104) -> (then) {is=title K2Node_CallFunction}
   "On Clicked (ClearTagsInLevel)"() {is=title K2Node_ComponentBoundEvent}
   > ClearTags()
   "On Clicked (PaintConcreteVertexColorsRedInLevel)"() {is=title K2Node_ComponentBoundEvent}
   n105 = Get SourceTextBox()
   > n106 = For Each LI(InputPin="To String (Text)"(InText="GetText (Text Box)"(self=@n105)))
   > PaintConcreteVertex1000(WorldContextObject=@n106)
   "On Clicked (FactoryWallWPOInPath)"() {is=title K2Node_ComponentBoundEvent}
   n107 = Get SourceTextBox()
   > n108 = For Each LI(InputPin="To String (Text)"(InText="GetText (Text Box)"(self=@n107)))
   > n109 = GetEditorWorld(self=Get UnrealEditorSubsystem())
   > n110 = GetAllActorsOfClass(
    WorldContextObject=SelectObject(A=@n108, B=@n109, bSelectA=IsValid(Object=@n108))
    ActorClass=/Game/Building/ActorBlueprints/Wall/Parent_BuildingWall.Parent_BuildingWall_C
   ) -> (OutActors:Array of Parent Building Wall Object References)
   > Clear(TargetArray:Array of Parent Building Wall Object References=Get ParentBuildingWallArray() -> (ParentBuildingWallArray:Array of Parent Building Wall Object References))
   > n111 = For Each Loop(Array:Array of Object References=@n110) -> (Array Element:Object Reference)
   n112 = Get ParentBuildingWallArray()
   CheckoutLoadedAssets(execute=@n111.Completed, AssetsToCheckout=@n112)
   > SaveLoadedAssets(AssetsToSave=@n112, bOnlyIfIsDirty=false)
   n113 = Cast To Parent_BuildingWall!(execute=@n111.LoopBody, Object=@n111) -> (then, CastFailed, AsParent Building Wall:Parent Building Wall Object Reference) {is=title K2Node_DynamicCast}
   > ResetEditorProperty(
    Object=Get StaticMeshComponent(self:Building Static Mesh Object Reference=@n113) -> (StaticMeshComponent:Base Building Static Mesh Component Object Reference)
    PropertyName=bEvaluateWorldPositionOffsetInRayTracing
   )
   > Add(
    TargetArray:Array of Parent Building Wall Object References=Get ParentBuildingWallArray()
    NewItem:"Parent Building Wall Object Reference (by ref)"=@n113
   )
   "On Clicked (DetectPhysMatInLevel)"() {is=title K2Node_ComponentBoundEvent}
   > n114 = CreateDynamicMeshPool!(execute) -> (then, ReturnValue:Dynamic Mesh Pool Object Reference)
   > n115 = Set DMP(DMP:Dynamic Mesh Pool Object Reference=@n114) -> (Output_Get:Dynamic Mesh Pool Object Reference)
   > n116 = RequestMesh!(execute, self:Dynamic Mesh Pool Object Reference=@n115) -> (then, ReturnValue:Dynamic Mesh Object Reference)
   > Set DM(DM:Dynamic Mesh Object Reference=@n116) -> (Output_Get:Dynamic Mesh Object Reference)
   > n117 = Sequence!(execute) -> (then_0, then_1)
   FreeAllMeshes(execute=@n117.then_1, self=Get DMP())
   n118 = Branch(execute=@n117.then_0, Condition=Get SM_to_BP_all_exp() -> (SM_to_BP_all_exp:bool))
   "On Clicked (AppltTD)"() {is=title K2Node_ComponentBoundEvent}
   > AppltTextureDataSelected() {is=title K2Node_CallFunction}
   "On Clicked (ConverttoBlueprint)"() {is=title K2Node_ComponentBoundEvent}
   > ReplaceStaticMeshesCombined()
   n119 = "On Check State Changed (SMConvertApplyToAll_1)"!() -> (OutputDelegate, then, bIsChecked:bool) {is=title K2Node_ComponentBoundEvent}
   > Set SM_to_BP_all_exp(SM_to_BP_all_exp:bool=@n119.bIsChecked) -> (Output_Get:bool)
   n120 = "On Clicked (ReplaceMaterialsWithTarget)"() {is=title K2Node_ComponentBoundEvent}
   "On Clicked (ReplaceTreeMaterials)"() {is=title K2Node_ComponentBoundEvent}
   > TreeSwap()
   "On Clicked (mpc_change)"() {is=title K2Node_ComponentBoundEvent}
   > n121 = Construct Object from Class(Class=&/EUT_ProcessTreeMats.EUT_ProcessTreeMats_C) -> (ReturnValue:EUT Process Tree Mats Object Reference)
   > Set intarray!(
    execute
    intarray:Array of Integers=Make Array("[0]":int=13, "[1]":int=6, "[2]":int=2) -> (Array:Array of Integers)
    self:EUT Process Tree Mats Object Reference=@n121
   ) -> (then, Output_Get:Array of Integers)
   > n122 = RegisterAndExecuteTask(Task=@n121)
   PrintString(execute=@n122.OnFinished, InString=Done)
   "On Clicked (add_unmarked)"() {is=title K2Node_ComponentBoundEvent}
   > Clear(TargetArray:Array of Strings=Get FilesToAdd() -> (FilesToAdd:Array of Strings))
   > Clear(TargetArray=Get FilesStatus() -> (FilesStatus:Array of Strings))
   n123 = GetAssetRegistry()
   > n124 = GetAssetsByPath(
    self=@n123
    PackagePath=String To Name(InString=MakeLiteralString(Value=/Figment_S08/LM_8_1_1/))
    bRecursive=true
    bIncludeOnlyOnDiskAssets=true
   )
   > n125 = Set Assets to add(Assets to add:Array of Asset Data Structures=@n124) -> (Output_Get:Array of Asset Data Structures)
   > n126 = GetAssetsByPath(
    self=@n123
    PackagePath=String To Name(InString=MakeLiteralString(Value=/Figment_S07/LM_7_00/))
    bRecursive=true
    bIncludeOnlyOnDiskAssets=true
   )
   > Append Array!(execute, TargetArray:Array of Asset Data Structures=@n125, SourceArray:Array of Asset Data Structures=@n126) -> (then)
   > n127 = For Each Loop(Array:Array of Asset Data Structures=Get Assets to add() -> (Assets to add:Array of Asset Data Structures)) -> (Array Element:Asset Data Structure)
   n128 = Sequence(execute=@n127.Completed)
   n129 = Get FilesStatus()
   > n130 = QueryFileStates!(execute, InFiles:Array of Strings=@n129, bSilent:bool=false, bUseSourceControlStateCache:bool=false) -> (then, ReturnValue:Array of Source Control State Structures)
   > n131 = For Each Loop(Array:Array of Source Control State Structures=@n130) -> (Array Element:Source Control State Structure)
   n132 = Break Source Control State!(SourceControlState:"Source Control State Structure (by ref)"=@n131) -> (
    Filename:String
    bIsValid:bool=~
    bIsUnknown:bool=~
    bCanCheckIn:bool=~
    bCanCheckOut:bool=~
    bIsCheckedOut:bool=~
    bIsCurrent:bool=~
    bIsSourceControlled:bool=~
    bIsAdded:bool=~
    bIsDeleted:bool=~
    bIsIgnored:bool=~
    bCanEdit:bool=~
    bCanDelete:bool=~
    bIsModified:bool=~
    bCanAdd:bool=~
    bIsConflicted:bool=~
    bCanRevert:bool=~
    bIsCheckedOutOther:bool=~
    CheckedOutOther:String
    bIsCheckedOutInOtherBranch:bool=~
    bIsModifiedInOtherBranch:bool=~
    PreviousUser:String
   ) {is=title K2Node_BreakStruct}
   > n133 = Branch(Condition=@n132.bIsSourceControlled)
   Add(execute=@n133.else, TargetArray:Array of Strings=Get FilesToAdd(), NewItem:"String (by ref)"=@n132)
   n134 = MarkFilesForAdd!(execute=@n128.then_1, InFiles:Array of Strings=Get FilesToAdd(), bSilent:bool=false) -> (then, ReturnValue:bool)
   > n135 = Branch(Condition=@n134)
   > PrintString(InString=Success: Adding Files)
   PrintString(execute=@n135.else, InString=Failed: Adding Files)
   Add(execute=@n127.LoopBody, TargetArray=Get FilesStatus(), NewItem="To String (Name)"(InName=Break Asset Data(AssetData=@n127)))
   "On Clicked (MigrateTDsfromSearchPAth)"() {is=title K2Node_ComponentBoundEvent}
   > MigrateTDsAndMats()
   "On Clicked (DeleteMG)"() {is=title K2Node_ComponentBoundEvent}
   > DeleteMeshGrass_BP!(execute, self) -> (then, ReturnValue:Object Reference) {is=title K2Node_CallFunction}
   "On Clicked (DeleteFoliage)"() {is=title K2Node_ComponentBoundEvent}
   > DeleteFoliageActorsClean()
   "On Clicked (SetTreeTopo)"() {is=title K2Node_ComponentBoundEvent}
   > Clear(TargetArray:Array of Object References=Get ObjectsToSave())
   > n136 = Sequence()
   SaveLoadedAssets(execute=@n136.then_1, AssetsToSave=Get ObjectsToSave(), bOnlyIfIsDirty=false)
   "On Clicked (ReparentMIs)"() {is=title K2Node_ComponentBoundEvent}
   > Clear(TargetArray=Get ObjectsToSave())
   > n137 = Sequence()
   > n138 = GetSelectedAssetsOfClass(AssetClass=&MaterialInstanceConstant)
   > n139 = For Each Loop(Array:Array of Object References=@n138) -> (Array Element:Object Reference)
   > n140 = Cast To MaterialInstanceConstant(Object=@n139)
   > n141 = Branch(Condition="Equal (Object)"(A=@n140, B=/Figment_S07/LM_7_00/Athena/Environments/Landscape/Material/Athena_Terrain_Mat_Biome_Inst))
   Branch(
    execute=@n141.else
    Condition="Equal (Object)"(A=Get Parent(self=@n140), B=/Figment_S07/LM_7_00/Athena/Environments/Landscape/Material/Athena_Terrain_Mat_Biome)
   )
   > SetMaterialInstanceParent(Instance=@n140, NewParent=&Athena_Terrain_Mat_Biome_Inst)
   > Add(TargetArray:Array of Object References=Get ObjectsToSave(), NewItem:"Object Reference (by ref)"=@n140)
   SaveLoadedAssets(execute=@n137.then_1, AssetsToSave=Get ObjectsToSave(), bOnlyIfIsDirty=false)
   n142 = "On Clicked (SwapSnowMeshes)"() {is=title K2Node_ComponentBoundEvent}
   n143 = "On Clicked (SwapSnowMats)"() {is=title K2Node_ComponentBoundEvent}
   "On Clicked (ASsignIceHLOD)"() {is=title K2Node_ComponentBoundEvent}
   > n144 = LoadAsset(AssetPath=/Figment_S08/Athena/HLODSetup/Figment_IceEnv_HLOD0)
   > n145 = Cast To HLODLayer!(execute, Object=@n144) -> (then, CastFailed, AsHLODLayer:HLODLayer Object Reference) {is=title K2Node_DynamicCast}
   n146 = Get SourceTextBox()
   > n147 = For Each LI(InputPin="To String (Text)"(InText="GetText (Text Box)"(self=@n146)))
   > Clear(TargetArray=Get ObjectsToSave())
   > n148 = GetAllActorsOfClass(WorldContextObject=@n147, ActorClass=&FortStaticMeshActor) -> (OutActors:Array of Fort Static Mesh Actor Object References)
   > n149 = For Each Loop(Array:Array of Fort Static Mesh Actor Object References=@n148) -> (Array Element:Fort Static Mesh Actor Object Reference)
   n150 = Get ObjectsToSave()
   CheckoutLoadedAssets(execute=@n149.Completed, AssetsToCheckout=@n150)
   > SaveLoadedAssets(AssetsToSave=@n150, bOnlyIfIsDirty=false)
   n151 = Cast To MaterialInstanceConstant(
    execute=@n149.LoopBody
    Object=GetEditorMaterial(
     self=Get StaticMeshComponent(self:Static Mesh Actor Object Reference=@n149) -> (StaticMeshComponent:Static Mesh Component Object Reference)
     ElementIndex=0
    )
   )
   n152 = Get Parent(self=@n151)
   > Branch(Condition=OR Boolean!(
     A="Equal (Object)"(A=@n152, B=&M_Ice_Snow_Inst)
     B="Equal (Object)"(A=@n151, B=&M_Ice_Snow_Inst)
    ) -> (ReturnValue))
   > SetEditorProperty(Object=@n149, PropertyName=HLODLayer, PropertyValue:HLODLayer Object Reference=@n145)
   > Add(TargetArray=Get ObjectsToSave(), NewItem=@n149)
   "On Clicked (FixTDDependencies)"() {is=title K2Node_ComponentBoundEvent}
   > Fix_TD_Dependencies()
   "On Property Changed (SinglePropertyView_0)"!() -> (OutputDelegate, then, PropertyName:Name=None) {is=title K2Node_ComponentBoundEvent}
   > PrintString(InString=GetDisplayName(Object=Get MICFrom() -> (MICFrom:Material Instance Constant Object Reference)))
   "On Property Changed (SinglePropertyView)"!() -> (OutputDelegate, then, PropertyName:Name=None) {is=title K2Node_ComponentBoundEvent}
   > PrintString(InString=GetDisplayName(Object=Get MICTo() -> (MICTo:Material Instance Constant Object Reference)))
   "On Clicked (EditorUtilityButton_4)"() {is=title K2Node_ComponentBoundEvent}
   > n153 = GetEditorWorld(self=Get UnrealEditorSubsystem())
   > n154 = GetAllActorsOfClass(WorldContextObject=@n153, ActorClass=&FortStaticMeshActor)
   > n155 = For Each Loop(Array=@n154)
   n156 = Get StaticMeshComponent(self=@n155)
   > Branch(Condition="Equal (Object)"(A=GetMaterial(self=@n156, ElementIndex=0), B=Get MICFrom()))
   > n157 = Branch(Condition=IsValid(Object=Get MICTo()))
   > SetMaterial(self=@n156, ElementIndex=0, Material=Get MICTo())
   PrintString(execute=@n157.else, InString=target material not set in UI)
   "On Clicked (EditorUtilityButton_5)"() {is=title K2Node_ComponentBoundEvent}
   > n158 = GetSelectedActors()
   > n159 = Cast To FortStaticMeshActor(Object="Get (a copy)"(Array=@n158))
   > n160 = Cast To MaterialInstanceConstant(Object=GetMaterial(self=Get StaticMeshComponent(self=@n159), ElementIndex=0))
   > Set MICFrom(MICFrom:Material Instance Constant Object Reference=@n160) -> (Output_Get:Material Instance Constant Object Reference)
   "On Clicked (EditorUtilityButton_7)"() {is=title K2Node_ComponentBoundEvent}
   > n161 = GetEditorWorld(self=Get UnrealEditorSubsystem())
   > n162 = GetAllActorsOfClass(WorldContextObject=@n161, ActorClass=&BP__Cliff_Figment_C) -> (OutActors:"Array of BP  Cliff Figment Object References")
   > n163 = For Each Loop(Array:"Array of BP  Cliff Figment Object References"=@n162) -> (Array Element:"BP  Cliff Figment Object Reference")
   > n164 = GetEditorProperty(Object=@n163, PropertyName=TextureData)
   n165 = GetDisplayName(Object=GetMaterial(self=Get StaticMeshComponent(self=@n163), ElementIndex=0))
   n166 = IsValid(Object="Get (a copy)"(Array:Array of Building Texture Data Object References=@n164) -> (Output:Building Texture Data Object Reference))
   > Branch(Condition=AND Boolean(A=Contains(SearchIn=@n165, Substring=Terrain_Mat), B=NOT Boolean(A=@n166)))
   > SetEditorProperty(
    Object=@n163
    PropertyName=TextureData
    PropertyValue:Array of Building Texture Data Object References=Make Array!(
     "[0]":Building Texture Data Object Reference=Get TD Asset()
     "[1]":Building Texture Data Object Reference="Get (a copy)"(Array=@n164, Dimension 1=1)
     "[2]":Building Texture Data Object Reference="Get (a copy)"(Array=@n164, Dimension 1=2)
     "[3]":Building Texture Data Object Reference="Get (a copy)"(Array=@n164, Dimension 1=3)
    ) -> (Array:Array of Building Texture Data Object References)
   )
   > PrintString(InString=GetDisplayName(Object=@n163))
   n167 = "On Clicked (ConverttoBlueprint2)"() {is=title K2Node_ComponentBoundEvent}
   n168 = Cast To LandscapeStreamingProxy!(execute, Object=@n77) -> (then, CastFailed, AsLandscape Streaming Proxy:Landscape Streaming Proxy Object Reference) {is=title K2Node_DynamicCast}
   > n169 = PrintString(InString=GetDisplayName(Object=@n168))
   PrintString(execute=@n168.CastFailed, InString=Failed)
   Event PreConstruct()
   > SetObject(self=Get SinglePropertyView_0() -> (SinglePropertyView_0:Single Property View Object Reference), NewObject=Self-Reference())
   > SetObject(self=Get SinglePropertyView() -> (SinglePropertyView:Single Property View Object Reference), NewObject=Self-Reference())
   > SetObject(self=Get SinglePropertyView_1() -> (SinglePropertyView_1:Single Property View Object Reference), NewObject=Self-Reference())
   > SetObject(self=Get SinglePropertyView_Apply() -> (SinglePropertyView_Apply:Single Property View Object Reference), NewObject=Self-Reference())
   n170 = Execute Python Script!(
    execute
    PythonScript="import unreal\r\nimport gc as py_gc\r\n\r\narr = []\r\nfor i in actors:\r\n    arr.append(i.get_package())\r\n    py_gc.collect()\r\n\r\n"
    Actors:Array of Object References=@n4
   ) -> (then, ReturnValue, arr:Array of Package Object References) {inputs [Actors], outputs [arr]}
   n171 = GetEditorProperty(Object=@n14, PropertyName="scalar_parameters ") -> (PropertyValue:String)
   PrintString(execute=[@n171.then, @n193.LoopBody], InString=@n171)
   n172 = Branch(Condition=GetMaterialInstanceStaticSwitchParameterValue(Instance=@n55, ParameterName=UseSnowLegacy, Association=GlobalParameter))
   n173 = Get Texture Parameter Value!(execute=[@n55.then, @n172.then], self:Material Instance Constant Object Reference=@n55, ParameterName:Name=TopoMask) -> (then, ReturnValue:Texture Object Reference) {is=title K2Node_CallFunction}
   > n174 = Branch(Condition="Equal (Object)"(A=@n173, B=/FigmentEnvironment/Athena/Environments/Landscape/Textures/T_Athena_Terrain_Topo_Mask_S08))
   n175 = Set MIC Texture Parameter Editor Only!(
    execute=@n174.else
    Material:Material Instance Constant Object Reference=@n55
    ParamName:String=TopoMask
    Texture:Texture 2D Object Reference=&T_Athena_Terrain_Topo_Mask_S08
   ) -> (then, ReturnValue:bool) {is=title K2Node_CallFunction}
   > Branch(Condition=@n175)
   > n176 = Add(TargetArray=Get ObjectsToSave(), NewItem=@n55)
   n177 = Branch(Condition="Equal (Object)"(A=@n69, B=/Figment_S07/LM_7_00/Athena/WinterArt/Materials/M_Ice_Snow))
   n178 = Branch(execute=@n177.else, Condition="Equal (Object)"(A=@n69, B=/Figment_S07/LM_7_00/Athena/WinterArt/Materials/M_Ice_Snow_Cliff_02))
   Branch(execute=@n178.else, Condition="Equal (Object)"(A=@n69, B=/Figment_S07/Athena/WinterArt/Materials/M_Ice_Snow_Cliff_02))
   n179 = Branch(Condition="Equal (Object)"(A=@n69, B=&M_Ice_Snow))
   > n180 = SetMaterial(self=@n68, ElementIndex=0, Material=/FigmentEnvironment/LM_8_1_1/Athena/WinterArt/Materials/M_Ice_Snow)
   Add(execute=[@n180.then, @n185.then, @n182.then, @n184.then], TargetArray=Get ObjectsToSave(), NewItem=@n66)
   n181 = Branch(
    execute=@n179.else
    Condition=OR Boolean(
     A="Equal (Object)"(A=@n69, B=/Figment_S07/LM_7_00/Athena/WinterArt/Materials/M_Ice_Snow_Cliff_02)
     B="Equal (Object)"(A=@n69, B=/Figment_S07/Athena/WinterArt/Materials/M_Ice_Snow_Cliff_02)
    )
   )
   > n182 = SetMaterial(self=@n68, ElementIndex=0, Material=/FigmentEnvironment/LM_8_1_1/Athena/WinterArt/Materials/M_Ice_Snow_Cliff_02)
   n183 = Branch(
    execute=@n181.else
    Condition=OR Boolean(
     A="Equal (Object)"(A=@n69, B=/Figment_S07/LM_7_00/Athena/WinterArt/Materials/M_Ice_Snow_Proto_Pile)
     B="Equal (Object)"(A=@n69, B=/Figment_S07/Athena/WinterArt/Materials/M_Ice_Snow_Proto_Pile)
    )
   )
   > n184 = SetMaterial(self=@n68, ElementIndex=0, Material=/FigmentEnvironment/LM_8_1_1/Athena/WinterArt/Materials/M_Ice_Snow_Proto_Pile)
   Branch(execute=@n183.else, Condition="Equal (Object)"(A=@n69, B=&M_Ice_Snow))
   > n185 = SetMaterial(self=@n68, ElementIndex=0, Material=&M_Ice_Snow_Inst)
   n186 = Branch()
   n187 = "To String (Text)"(InText="GetText (Text Box)"(self=Get TargetTextBox()))
   For Each LI(execute=@n186.else, InputPin=@n187)
   n188 = For Each LI(InputPin="To String (Text)"(InText="GetText (Text Box)"()))
   > AppltTextureDataBiome!(execute, self, WorldContextObject:Object Reference=@n188) -> (then) {is=title K2Node_CallFunction}
   n189 = "To String (Text)"(InText="GetText (Text Box)"(self=Get SourceTextBox()))
   n190 = For Each Context!(execute:Exec, InputPin:String=@n189) -> (then:Exec, Output_Get:Object Reference, ResolvedExec:Exec, ReturnValue:Array of Level Object References) {is=title K2Node_MacroInstance}
   n191 = For Each Loop(Exec=@n190.ResolvedExec, Array:Array of Level Object References=@n190.ReturnValue) -> (Array Element:Level Object Reference)
   > PrintString(InString=GetDisplayName(Object=@n191))
   n192 = Get MPC() -> (MPC:Material Parameter Collection Object Reference)
   n193 = For Each Loop(Array:Array of Names=GetScalarParameterNames!(self:Material Parameter Collection Object Reference=@n192) -> (ReturnValue:Array of Names)) -> (Array Element:Name)
   n194 = For Each Context(InputPin=@n101)
   n195 = For Each Context(InputPin=@n103)
   Set count(count:int=0) -> (Output_Get:int=0)
   > n196 = For Each LI(InputPin=@n189)
   > n197 = GetAllActorsOfClass(WorldContextObject=@n196, ActorClass=/FigmentEnvironment/Blueprints/Cliffs/BP_S_Sidewalk_1x1_A.BP_S_Sidewalk_1x1_A_C) -> (OutActors:Array of BP S Sidewalk 1x 1 A Object References)
   n198 = Length(TargetArray:Array of BP S Sidewalk 1x 1 A Object References=@n197)
   > n199 = Set count(count=int + int(A=Get count() -> (count:int=0), B=@n198)) -> (Output_Get=0)
   > PrintString(InString=Append!(A="To String (Integer)"(InInt=@n198), B=":    ", C="To String (Integer)"(InInt=@n199)) -> (ReturnValue))
   n200 = Construct Object from Class(execute=@n3.then, Class=&/EUT_SaveActorsAsync.EUT_SaveActorsAsync_C) -> (ReturnValue:EUT Save Actors Async Object Reference)
   RegisterAndExecuteTask(execute=@n219.then, Task=@n200)
   Break Asset Data(AssetData=@n52)
   Break Asset Data()
   Add(execute=@n213.then, TargetArray=Get ObjectsToSave(), NewItem=@n23)
   Get Asset By Object Path!(
    execute
    self:Asset Registry Interface=@n7
    ObjectPath:"Soft Object Path Structure (by ref)"
    bIncludeOnlyOnDiskAssets:bool=false
    bSkipARFilteredAssets:bool=true
   ) -> (then, ReturnValue:Asset Data Structure) {is=title K2Node_CallFunction}
   "To String (Text)"(InText="GetText (Text Box)"(self=Get SourceTextBox()))
   PrintString(InString=Append(A=GetDisplayName(Object=@n26), B=" -> ", C=stone))
   GetSystemPath(Object=@n90)
   n201 = GetAssetsByPath(execute=@n211.then, self=GetAssetRegistry(), PackagePath=String To Name(InString=@n187), bRecursive=true, bIncludeOnlyOnDiskAssets=true)
   > n202 = For Each Loop(Array:Array of Asset Data Structures=@n201) -> (Array Element:Asset Data Structure)
   > n203 = Cast To MaterialInterface(Object=GetAsset(InAssetData=@n202))
   PrintString(InString=Get Object Path String(Object=@n203))
   "To String (Name)"(InName=@n193)
   PrintString(InString=@n91)
   GetAsset()
   n204 = Get StaticMeshComponent(self=@n23)
   Set PhysicalMaterial Override!(
    execute
    self:Primitive Component Object Reference=@n204
    NewPhysMaterial:Physical Material Object Reference=&Dirt
   ) -> (then) {is=title K2Node_CallFunction}
   PrintString(InString=Concrete)
   BreakTopLevelAssetPath!(TopLevelAssetPath:"Top Level Asset Path Structure (by ref)"=Break Asset Data(AssetData=@n16).AssetClassPath) -> (PathString:String)
   MakeTopLevelAssetPath!(PackageName:String, AssetName:String) -> (ReturnValue:Top Level Asset Path Structure)
   PrintString(InString=GetDisplayName())
   PrintString(InString=Grass)
   PrintString(InString=Concrete)
   Dup Texture Data and Material Instances of All Cliff Actors() {is=title K2Node_CallFunction}
   Reassign TextureData with existing!(execute, self, Level Object Reference:Object Reference) -> (then)
   MakeLiteralInt!(Value:int=0) -> (ReturnValue:int=0)
   "To String (Float)"(InDouble=@n36)
   "To String (Integer)"(InInt=Length(TargetArray:Array of Asset Data Structures=@n15))
   PrintString(InString=Sand)
   PrintString(InString=Sand)
   PrintString(InString=Grass)
   ClassIsChildOf()
   PrintString(InString=done1)
   "SetText (Text Box)"!(
    execute
    self:Text Box Object Reference=Get SourceTextBox()
    InText:Text="To Text (String)"(InString=Get TargetPAth() -> (TargetPAth:String))
   ) -> (then)
   n205 = Get ObjectsToSave()
   SaveLoadedAssets(execute=@n212.then, AssetsToSave=@n205, bOnlyIfIsDirty=false)
   Set PhysicalMaterial Override(self=@n204, NewPhysMaterial=&Dirt)
   Set PhysicalMaterial Override(self=@n204, NewPhysMaterial=&Generic)
   Set PhysicalMaterial Override(self=@n204, NewPhysMaterial=&Generic)
   Set PhysicalMaterial Override(self=@n204, NewPhysMaterial=&Asphalt)
   ForEachLevelInstanceContext!(execute, self, InString:String) -> (then, Output_Get:Object Reference)
   Set PhysicalMaterial Override(self=@n204, NewPhysMaterial=&Asphalt)
   Set PhysicalMaterial Override(self=@n204, NewPhysMaterial=&Grass)
   Set PhysicalMaterial Override(self=@n204, NewPhysMaterial=&Sand)
   Set PhysicalMaterial Override(self=@n204, NewPhysMaterial=&Grass)
   AsyncQueryFileStates!(
    execute
    FileStateCallback:Delegate=FileStatus!() -> (OutputDelegate, then, FileStateOut:Source Control State Structure) {is=title K2Node_CustomEvent}
    InFiles:Array of Strings=@n129
    bSilent:bool=false
   ) -> (then)
   Set PhysicalMaterial Override(self=@n204, NewPhysMaterial=&Sand)
   MigrateTDsAndMatsS08!(execute, self, Apply Mode:E_ApplyContext Enum=NewEnumerator2) -> (then)
   MigrateTDsAndMatsS08(Apply Mode=NewEnumerator1)
   BuildMultiLineMapOfMaterials!(execute, self, Mesh Material:Material Interface Object Reference) -> (then, TDArray:Array of Building Texture Data Object References, SingleTD_out:Building Texture Data Object Reference) {is=title K2Node_CallFunction}
   BuildMeshMap()
   TakeHighResScreenshot!(
    execute
    ResX:int=0
    ResY:int=0
    Filename:String
    Camera:Camera Actor Object Reference
    bMaskEnabled:bool=false
    bCaptureHDR:bool=false
    ComparisonTolerance:EComparisonTolerance Enum=Low
    ComparisonNotes:String
    Delay:float=0
    bForceGameView:bool=true
   ) -> (then, ReturnValue:Automation Editor Task Object Reference)
   PrintString(InString=GetDisplayName(Object=@n55))
   BugItGo!(execute, Location:Vector="0, 0, 0", Rotation:Rotator="0, 0, 0", WorldContextObject:Object Reference, __WorldContext:Object Reference) -> (then)
   SaveLoadedAsset(AssetToSave=@n55, bOnlyIfIsDirty=false)
   Get Texture Parameter Value(ParameterName=None)
   GetCameraPose!(CameraData:"Camera Evaluation Data Ref Structure (by ref)") -> (ReturnValue:Camera Pose Structure)
   Landscape Import Heightmap from RenderTarget!(
    execute
    self:Landscape Proxy Object Reference
    InRenderTarget:Texture Render Target 2D Object Reference
    InImportHeightFromRGChannel:bool=false
    InEditLayerIndex:int=0
   ) -> (then, ReturnValue:bool)
   Get Editor Camera Info!(execute) -> (then, CameraLocation:Vector, ForwardDirection:Vector, FOV:float, ReturnValue:bool) {is=title K2Node_CallFunction}
   GetClass(InAssetData=@n52)
   Add!(
    execute=@n203.then
    TargetMap:Map of Strings to Material Interface Object References=Get MapStringToTargetMaterials()
    Key:"String (by ref)"=GetDisplayName(Object=@n203)
    Value:"Material Interface Object Reference (by ref)"=@n203
   ) -> (then) {is=title K2Node_CallFunction}
   GetAssetTools()
   PrintString(execute=@n216.else, InString=No new assignment)
   ReplaceSnowMaterials()
   PrintString(InString=Yay)
   GetVectorParameterNames!(self:Material Parameter Collection Object Reference=@n192) -> (ReturnValue:Array of Names)
   RemoveChild(self=Get CanvasPanel_X() -> (CanvasPanel_X:Canvas Panel Object Reference), Content=Get viewport() -> (viewport:Viewport Object Reference))
   FixupTextureData_S03!(execute, self, Level Object Reference:Object Reference) -> (then)
   Remove_appended_TDs!(execute, self, Level Object:Object Reference) -> (then)
   PrintString(WorldContextObject=@n195, InString=GetDisplayName())
   "Get Class Path Name (Top Level Asset Path)"!(Class:Object Class Reference) -> (ReturnValue:Top Level Asset Path Structure) {is=title K2Node_CallFunction}
   SavePackages(PackagesToSave=@n170.arr, bOnlyDirty=false)
   PrintString(InString=GetDisplayName(Object=@n22))
   GetActorOfClass!(execute, WorldContextObject:Object Reference=@n83, ActorClass:Actor Class Reference) -> (then, ReturnValue:Actor Object Reference)
   n206 = For Each Loop(
    Exec=@n78.then
    Array:Array of Landscape Edit Layer Base Object References=GetEditLayers!(self:Landscape Object Reference=@n78) -> (ReturnValue:Array of Landscape Edit Layer Base Object References)
   ) -> (Array Element:Landscape Edit Layer Base Object Reference)
   PrintString(InString=GetDisplayName(Object=@n206))
   SetLayerVisibility!(
    execute
    self:Layers Subsystem Object Reference=Get LayersSubsystem() -> (ReturnValue:Layers Subsystem Object Reference) {is=title K2Node_GetEditorSubsystem}
    LayerName:"Name (by ref)"=String To Name(InString=MakeLiteralString(Value=Layer1))
    bIsVisible:bool=false
   ) -> (then)
   GetAsset()
   Set PhysicalMaterial Override(self=@n204, NewPhysMaterial=&Grass)
   DeleteFoliageActors!(execute, self) -> (then, Package Array To Save:Array of Package Object References) {is=title K2Node_CallFunction}
   SetVectorParameterValue!(
    execute
    WorldContextObject=@n192
    Collection:Material Parameter Collection Object Reference
    ParameterName:Name=None
    ParameterValue:"Linear Color Structure (by ref)"
   ) -> (then) {is=title K2Node_CallMaterialParameterCollectionFunction}
   Cast To HLODLayer(Object:Wildcard)
   Event Construct()
   Event Tick()
   Sequence()
   "Get (a copy)"(Array:Array of Asset Data Structures=@n8) -> (Output:Asset Data Structure)
   "Get (a copy)"(Array:Array of Actor Object References=GetSelectedLevelActors(self=Get EditorActorSubsystem())) -> (Output:Actor Object Reference)
   "Get (a copy)"(Array=GetSelectedActors())
   "Get (a copy)"(Array=@n71)
   Get ContentBrowserDataSubsystem() -> (ReturnValue:Content Browser Data Subsystem Object Reference) {is=title K2Node_GetEditorSubsystem}
   Get UnrealEditorSubsystem()
   Get LevelEditorSubsystem()
   Get MeshPaintingSubsystem() -> (ReturnValue:Mesh Painting Subsystem Object Reference) {is=title K2Node_GetEngineSubsystem}
   Get LandscapeSubsystem!(WorldContext:Object Reference) -> (ReturnValue:Landscape Subsystem Object Reference) {is=title K2Node_GetSubsystem}
   Reroute Node(InputPin:Wildcard) -> (OutputPin:Wildcard)
   Reroute Node(InputPin:Object Reference=@n111) -> (OutputPin:Object Reference)
   Reroute Node(InputPin:"Integer (by ref)"=@n34) -> (OutputPin:"Integer (by ref)")
   Reroute Node(InputPin:Wildcard) -> (OutputPin:Wildcard)
   For Each Loop(Array:Array of Wildcards) -> (Array Element:Wildcard)
   For Each Context(InputPin="To String (Text)"(InText="GetText (Text Box)"()))
   n207 = GetMaterialInstanceTextureParameterValue!(
    Instance:Material Instance Constant Object Reference=@n27
    ParameterName:Name=Base Tiling Texture
    Association:EMaterialParameterAssociation Enum=GlobalParameter
   ) -> (ReturnValue:Texture Object Reference)
   Make Array!("[0]":bool=Contains(SearchIn=GetDisplayName(Object=@n207), Substring=Cave), "[1]":bool=false) -> (Array:Array of Booleans)
   "Equal (Object)"(B=/Figment_S07/LM_7_00/Athena/WinterArt/Materials/M_Ice_Snow_Proto_Pile)
   "Equal (Object)"(B=/Figment_S07/LM_7_00/Athena/WinterArt/Materials/M_Ice_Snow_Cliff_02)
   SetEditorProperty(execute=@n206.LoopBody, Object=@n206, PropertyName=bVisible, PropertyValue:bool=MakeLiteralBool(Value=true))
   Get StaticMesh(self:Building Static Mesh Object Reference)
   Get Cliff Actors to Save()
   Get ParameterCollectionParameterValues!(self:Material Instance Object Reference) -> (ParameterCollectionParameterValues:Array of Parameter Collection Parameter Value Structures)
   Get TD Asset()
   Get level_instance_cleanup!(self:EUT Snow Replace Object Reference=@n3) -> (level_instance_cleanup:Array of Building Level Instance Object References)
   Set SM_to_BP_all(SM_to_BP_all:bool=false) -> (Output_Get:bool)
   n208 = Break Body Instance!(BodyInstance:"Body Instance Structure (by ref)"=Get BodyInstance!(self:Primitive Component Object Reference=@n24) -> (BodyInstance:Body Instance Structure)) -> (
    PositionSolverIterationCount:Byte
    VelocitySolverIterationCount:Byte
    ProjectionSolverIterationCount:Byte
    SleepFamily:ESleepFamily Enum
    bUseCCD:bool=~
    bIgnoreAnalyticCollisions:bool=~
    bNotifyRigidBodyCollision:bool=~
    bSmoothEdgeCollisions:bool=~
    bInertiaConditioning:bool=~
    bOneWayInteraction:bool=~
    bOverrideSolverAsyncDeltaTime:bool=~
    SolverAsyncDeltaTime:float=~
    MaxDepenetrationVelocity:float=~
    MassInKgOverride:float=~
    LinearDamping:float=~
    AngularDamping:float=~
    COMNudge:Vector=~
    MassScale:float=~
    GravityGroupIndex:Byte
    InertiaTensorScale:Vector=~
    WalkableSlopeOverride:Walkable Slope Override Structure
    PhysMaterialOverride:Physical Material Object Reference
    MaxAngularVelocity:float=~
    CustomSleepThresholdMultiplier:float=~
    StabilizationThresholdMultiplier:float=~
    bSimulatePhysics:bool=~
    bEnableGravity:bool=~
    bUpdateKinematicFromSimulation:bool=~
    bGyroscopicTorqueEnabled:bool=~
    bAutoWeld:bool=~
    bStartAwake:bool=~
    bGenerateWakeEvents:bool=~
    bUpdateMassWhenScaleChanges:bool=~
   ) {is=title K2Node_BreakStruct}
   n209 = Clear(execute=@n20.then, TargetArray=Get ObjectsToSave())
   Clear(execute=@n83.then, TargetArray=Get ObjectsToSave())
   > n210 = Sequence()
   > n211 = Clear(TargetMap=Get MapStringToTargetMaterials()) {is=title K2Node_CallFunction}
   n212 = CheckoutLoadedAssets(execute=@n22.Completed, AssetsToCheckout=@n205)
   n213 = Set PhysicalMaterial Override(execute=@n216.then, self=@n24, NewPhysMaterial=Get PhysToSet() -> (PhysToSet:Physical Material Object Reference))
   n214 = PrintString(execute=@n2.OnFinished, InString=done1)
   n215 = Sequence!(execute=@n27.then) -> (then_0, then_1, then_2:Exec)
   n216 = Branch(
    execute=@n215.then_1
    Condition="Not Equal (Object)"(
     A:Object Reference=Get PhysToSet()
     B:Object Reference=Get Phys Material Override() -> (Phys Material Override:Physical Material Object Reference)
    ) -> (ReturnValue:bool=~) {is=title K2Node_PromotableOperator}
   )
   n217 = Get level_instance_cleanup(self=@n3)
   n218 = Get ActorsToSaveMap!(self:EUT Snow Replace Object Reference=@n3) -> (ActorsToSaveMap:Map of Strings to Actor Array Structures)
   n219 = Set level_instance_cleanup!(
    execute=@n220.then
    level_instance_cleanup:Array of Building Level Instance Object References=@n217
    self:EUT Save Actors Async Object Reference=@n200
   ) -> (then, Output_Get:Array of Building Level Instance Object References)
   Set Actors!(execute=@n200.then, Actors:Array of Object References=@n4, self:EUT Save Actors Async Object Reference=@n200) -> (then, Output_Get:Array of Object References)
   > n220 = Set ActorsToSaveMap!(execute, ActorsToSaveMap:Map of Strings to Actor Array Structures=@n218, self:EUT Save Actors Async Object Reference=@n200) -> (then, Output_Get:Map of Strings to Actor Array Structures)
   n221 = Set Phys Material Override(execute=@n23.then, Phys Material Override:Physical Material Object Reference=@n208.PhysMaterialOverride) -> (Output_Get:Physical Material Object Reference)
  ]}
 FigmetizeName {nodes [
   n1 = FigmetizeName!() -> (then, SourceString:String) {is=title K2Node_FunctionEntry}
   n2 = ParseIntoArray(SourceString=@n1, Delimiter=_)
   > n3 = Set CopyNameArray(CopyNameArray:Array of Strings=@n2) -> (Output_Get:Array of Strings)
   > Set Orig(Orig:Array of Strings=@n2) -> (Output_Get:Array of Strings)
   n4 = Get Orig() -> (Orig:Array of Strings)
   n5 = "Equal Exactly (String)"(
    A="Get (a ref)"(Array:Array of Strings=@n4, Dimension 1=Last Index!(TargetArray=@n4) -> (ReturnValue=0)) -> (Output:"String (by ref)")
    B=Inst
   )
   > n6 = Branch(Condition=@n5)
   > n7 = Remove Index!(
    execute
    TargetArray=@n3
    IndexToRemove:int=Last Index(TargetArray=Get CopyNameArray() -> (CopyNameArray:Array of Strings)) -> (ReturnValue=0)
   ) -> (then)
   n8 = Get CopyNameArray()
   Add(execute=[@n7.then, @n6.else], TargetArray:Array of Strings=@n8, NewItem:"String (by ref)"=Figment) {is=title K2Node_CallArrayFunction}
   > Add(TargetArray=@n8, NewItem=Inst)
   n9 = Select(
    Option 0:String=Append!(A=@n1, B=_Figment) -> (ReturnValue)
    Option 1:String=JoinStringArray(SourceArray=Get CopyNameArray(), Separator=_)
    Index=@n5
   ) -> (ReturnValue:String)
   > Return Node!(execute, FigmentName:String=@n9)
   For Each Loop(Array:Array of Strings=Get Orig()) -> (Array Element:String)
   > PrintString(InString=@n9)
  ]}
 FindDominantVertexColor {nodes [
   n1 = FindDominantVertexColor!() -> (then, Cliff BP:"BP  Cliff Figment Object Reference") {is=title K2Node_FunctionEntry}
   > n2 = CopyMeshFromComponent!(
    execute
    Component:Scene Component Object Reference=Get StaticMeshComponent(self=@n1)
    ToDynamicMesh:Dynamic Mesh Object Reference=Get DM() -> (DM:Dynamic Mesh Object Reference)
    Options:Geometry Script Copy Mesh From Component Options Structure=Make Geometry Script Copy Mesh from Component Options!(
     bWantNormals:bool=True
     bWantTangents:bool=false
     bWantInstanceColors:bool=true
     RequestedLOD:Geometry Script Mesh Read LOD Structure=Make Geometry Script Mesh Read LOD!(LODType:EGeometryScriptLODType Enum=RenderData, LODIndex:int=0) -> (GeometryScriptMeshReadLOD:Geometry Script Mesh Read LOD Structure) {is=title K2Node_MakeStruct}
    ) -> (GeometryScriptCopyMeshFromComponentOptions:Geometry Script Copy Mesh From Component Options Structure) {is=title K2Node_MakeStruct}
    bTransformToWorld:bool=true
   ) -> (Failure:Exec, Success:Exec, LocalToWorld:Transform, ReturnValue:Dynamic Mesh Object Reference)
   > PrintString(InString=not got mesh)
   Set totalColorLocal(execute=@n2.Success, totalColorLocal:Linear Color Structure=(R=0,G=0,B=0,A=0)) -> (Output_Get:Linear Color Structure)
   > Set totalArea(totalArea:double=0) -> (Output_Get:double)
   > n3 = For Loop(LastIndex=int - int(A=GetTriangleCount!(self:Dynamic Mesh Object Reference=@n2.ReturnValue) -> (ReturnValue:int=0)))
   n4 = Get totalArea() -> (totalArea:double)
   n5 = LinearColor / LinearColor(
    A:Linear Color Structure=Get totalColorLocal() -> (totalColorLocal:Linear Color Structure)
    B:Linear Color Structure="New Opacity (LinearColor)"!(InColor:Linear Color Structure="To LinearColor (Float)"!(InDouble:double=@n4) -> (ReturnValue:Linear Color Structure), InOpacity:float=@n4) -> (ReturnValue:Linear Color Structure)
   ) -> (ReturnValue:Linear Color Structure) {is=title K2Node_PromotableOperator}
   PrintString(execute=@n3.Completed, InString="To String (LinearColor)"!(InColor:Linear Color Structure=@n5) -> (ReturnValue:String))
   > Return Node!(execute, FinallWeightedColor:Linear Color Structure=@n5)
   n6 = GetTriangleNormals!(execute=@n3.LoopBody, TargetMesh:Dynamic Mesh Object Reference=@n2.ReturnValue, TriangleID:int=@n3) -> (then, Normal1:Vector, Normal2:Vector, Normal3:Vector, bTriHasValidNormals:bool, ReturnValue:Dynamic Mesh Object Reference)
   > n7 = GetTriangleVertexColors!(execute, TargetMesh:Dynamic Mesh Object Reference=@n2.ReturnValue, TriangleID:int=@n3) -> (
    then
    Color1:Linear Color Structure
    Color2:Linear Color Structure
    Color3:Linear Color Structure
    bTriHasValidVertexColors:bool
    ReturnValue:Dynamic Mesh Object Reference
   )
   n8 = Dot Product!(
    A:Vector=Normalize!(
     A:Vector=GetTriangleFaceNormal!(TargetMesh:Dynamic Mesh Object Reference=@n2.ReturnValue, TriangleID:int=@n3) -> (bIsValidTriangle:bool, ReturnValue:Vector).ReturnValue
     Tolerance:float=0.0001
    ) -> (ReturnValue:Vector) {is=title K2Node_CallFunction}
    B:Vector="0,0,1"
   ) -> (ReturnValue:double)
   n9 = Break Int Vector!(IntVector:"Int Vector Structure (by ref)"=GetTriangleIndices!(TargetMesh:Dynamic Mesh Object Reference=@n2.ReturnValue, TriangleID:int=@n3) -> (bIsValidTriangle:bool, ReturnValue:Int Vector Structure).ReturnValue) -> (X:int, Y:int, Z:int) {is=title K2Node_BreakStruct}
   n10 = GetVertexPosition!(TargetMesh:Dynamic Mesh Object Reference=@n2.ReturnValue, VertexID:int=@n9) -> (bIsValidVertex:bool, ReturnValue:Vector)
   n11 = Cross Product!(
    A:Vector=vector - vector(A=GetVertexPosition(TargetMesh=@n2.ReturnValue, VertexID=@n9.Y).ReturnValue, B=@n10.ReturnValue)
    B:Vector=vector - vector(A=GetVertexPosition(TargetMesh=@n2.ReturnValue, VertexID=@n9.Z).ReturnValue, B=@n10.ReturnValue)
   ) -> (ReturnValue:Vector)
   n12 = float * float(
    A=float * float(A=Vector Length(A=@n11), B=0.5)
    B="Clamp (Float)"(Value=Power!(Base:double=@n8, Exp:double=1) -> (ReturnValue:double) {is=title K2Node_CallFunction}, Min=0, Max=1)
   ) {is=title K2Node_PromotableOperator}
   n13 = float * float(A=@n12, B=0.3333)
   n14 = "New Opacity (LinearColor)"(InColor="To LinearColor (Float)"(InDouble=@n13), InOpacity=@n13)
   n15 = Normalize(A=@n6.Normal3, Tolerance=0.0001)
   n16 = "Clamp (Float)"(Value=Power(Base=Dot Product(A=@n15, B="0,0,1"), Exp=1), Min=0, Max=1)
   n17 = "To LinearColor (Float)"(InDouble=@n16)
   n18 = Normalize(A=@n6, Tolerance=0.0001)
   n19 = "Clamp (Float)"(Value=Power(Base=Dot Product(A=@n18, B="0,0,1"), Exp=1), Min=0, Max=1)
   n20 = "To LinearColor (Float)"(InDouble=@n19)
   n21 = Normalize(A=@n6.Normal2, Tolerance=0.0001)
   n22 = "Clamp (Float)"(Value=Power(Base=Dot Product(A=@n21, B="0,0,1"), Exp=1), Min=0, Max=1)
   n23 = "To LinearColor (Float)"(InDouble=@n22)
   n24 = LinearColor + LinearColor!(
    A:Linear Color Structure=LinearColor * LinearColor(A:Linear Color Structure=@n7, B:Linear Color Structure="New Opacity (LinearColor)"(InColor=@n20, InOpacity=@n19)) -> (ReturnValue:Linear Color Structure) {is=title K2Node_PromotableOperator}
    B:Linear Color Structure=LinearColor * LinearColor(A=@n7.Color2, B="New Opacity (LinearColor)"(InColor=@n23, InOpacity=@n22))
    C:Linear Color Structure=LinearColor * LinearColor(A=@n7.Color3, B="New Opacity (LinearColor)"(InColor=@n17, InOpacity=@n16))
   ) -> (ReturnValue:Linear Color Structure) {is=title K2Node_PromotableOperator}
   > Set totalColorLocal(totalColorLocal=LinearColor + LinearColor!(A=Get totalColorLocal(), B=LinearColor * LinearColor(A=@n24, B=@n14)) -> (ReturnValue))
   > Set totalArea(totalArea=float + float(A=Get totalArea(), B=float * float(A=@n12, B=1)))
   PrintString(InString="To String (LinearColor)"(InColor=@n7))
   n25 = BreakColor(InColor=@n5)
   LinearColor / LinearColor(A=@n5, B:double="Max (Float)"!(A=@n25, B=@n25.G, C:double=@n25.B) -> (ReturnValue))
   float * float(A=float + float!(A=@n19, B=@n22, C:double=@n16) -> (ReturnValue), B=0.3333)
   Get Dominant_Color() -> (Dominant_Color:Linear Color Structure)
   Get TotalTriangleArea() -> (TotalTriangleArea:double)
  ]}
 Fix_TD_Dependencies {nodes [
   n1 = GetSelectedLevelActors(self=Get EditorActorSubsystem())
   > Cast To BP__Cliff_Figment(Object="Get (a copy)"(Array=@n1))
   n2 = GetEditorWorld(execute=@n14.NewEnumerator1, self=Get UnrealEditorSubsystem())
   n3 = GetAllActorsOfClass(WorldContextObject=@n2, ActorClass=&BP__Cliff_Figment_C) -> (OutActors:"Array of BP  Cliff Figment Object References")
   > Set Cliff Actors(Cliff Actors:"Array of BP  Cliff Figment Object References"=@n3) -> (Output_Get:"Array of BP  Cliff Figment Object References")
   > n4 = Branch(Condition=Is Empty!(TargetArray:"Array of BP  Cliff Figment Object References"=@n3) -> (ReturnValue:bool) {is=title K2Node_CallArrayFunction})
   > PrintString(InString=No Cliff Actors in level)
   Clear(execute=[@n23.then, @n4.else, @n18.then], TargetArray=Get ObjectsToSave()) {is=title K2Node_CallArrayFunction}
   > n5 = For Each Loop(Array:"Array of BP  Cliff Figment Object References"=Get Cliff Actors() -> (Cliff Actors:"Array of BP  Cliff Figment Object References")) -> (Array Element:"BP  Cliff Figment Object Reference")
   n6 = Get ObjectsToSave()
   SaveLoadedAssets(execute=@n5.Completed, AssetsToSave=@n6, bOnlyIfIsDirty=false)
   > CheckoutLoadedAssets(AssetsToCheckout=@n6)
   n7 = Set CliffBP(execute=[@n13.then, @n5.LoopBody], CliffBP:"BP  Cliff Figment Object Reference"=@n5) -> (Output_Get:"BP  Cliff Figment Object Reference")
   n8 = GetPackageForObject(Object=@n7)
   > n9 = Get Dependencies!(
    execute:Exec
    self:Object Reference="To Object (Interface)"!(Interface:"Interface Interface (by ref)"=GetAssetRegistry()) -> (ReturnValue:Object Reference)
    PackageName:Name=String To Name(InString=GetDisplayName(Object=@n8))
    DependencyOptions:"Asset Registry Dependency Options Structure (by ref)"=Make Asset Registry Dependency Options(bIncludeSoftPackageReferences=True)
   ) -> (then:Exec, OutDependencies:Array of Names, ReturnValue:bool) {is=title K2Node_Message}
   > n10 = For Each Loop(Array:Array of Names=@n9) -> (Array Element:Name)
   n11 = "To String (Name)"(InName=@n10)
   n12 = StartsWith(SourceString=@n11, InPrefix=/Figment_S08)
   > Branch(Condition=@n12)
   > PrintString(InString=Append!(A=Get Object Path String(Object=@n7), B="\r\n", C:String="To String (Name)"(InName=@n10)) -> (ReturnValue))
   n13 = PrintString(InString=GetDisplayName(Object=@n5))
   Fix_TD_Dependencies() {is=title K2Node_FunctionEntry}
   > n14 = Switch on E_ApplyContext(Selection=Get Apply Mode() -> (Apply Mode=NewEnumerator2))
   n15 = Get SourceTextBox()
   n16 = For Each LI(execute=@n14.NewEnumerator0, InputPin="To String (Text)"(InText="GetText (Text Box)"(self=@n15)))
   > n17 = GetAllActorsOfClass(WorldContextObject=@n16, ActorClass=&BP__Cliff_Figment_C)
   > n18 = Set Cliff Actors(Cliff Actors=@n17)
   n19 = GetSelectedLevelActors(execute=@n14.NewEnumerator2, self=Get EditorActorSubsystem())
   > n20 = Branch(Condition=Is Empty(TargetArray:Array of Actor Object References=@n19))
   > PrintString(InString=Nothing Selected)
   n21 = For Each Loop(Exec=@n20.else, Array:Array of Actor Object References=@n19) -> (Array Element:Actor Object Reference)
   > n22 = Cast To BP__Cliff_Figment(Object=@n21)
   > n23 = Add(TargetArray:"Array of BP  Cliff Figment Object References"=Get Cliff Actors(), NewItem:"BP  Cliff Figment Object Reference (by ref)"=@n22)
   Branch(Condition=AND Boolean(A=Contains(SearchIn=@n11, Substring=TD), B=@n12))
   > PrintString(InString=@n11)
   n24 = Replace(SourceString=@n11, From=/Figment_S08, To=/FigmentEnvironment)
   > n25 = DoesAssetExist(AssetPath=@n24)
   > Branch(Condition=@n25)
   > PrintString(InString=Append!(A=@n24, B=" Exists!") -> (ReturnValue))
   > n26 = LoadAsset(AssetPath=@n24)
   > n27 = GetEditorProperty(Object=Get CliffBP() -> (CliffBP:"BP  Cliff Figment Object Reference"), PropertyName=TextureData) -> (PropertyValue:Array of Building Texture Data Object References)
   > Set TDARRAY(TDARRAY:Array of Building Texture Data Object References=@n27) -> (Output_Get:Array of Building Texture Data Object References)
   > n28 = Cast To BuildingTextureData(Object=@n26)
   n29 = Get TDARRAY() -> (TDARRAY:Array of Building Texture Data Object References)
   > Set Array Elem(
    TargetArray:Array of Building Texture Data Object References=@n29
    Index=0
    Item:"Building Texture Data Object Reference (by ref)"=@n28
    bSizeToFit=true
   )
   n30 = Get CliffBP()
   > n31 = SetEditorProperty(Object=@n30, PropertyName=TextureData, PropertyValue:Array of Building Texture Data Object References=@n29)
   > Branch(Condition=@n31)
   > Add(TargetArray:Array of Object References=Get ObjectsToSave(), NewItem:"Object Reference (by ref)"=@n30)
   n32 = Get SourceTextBox()
   For Each LI(InputPin="To String (Text)"(InText="GetText (Text Box)"(self=@n32)))
   > Clear(TargetArray=Get ObjectsToSave())
   > GetAllActorsOfClass(ActorClass=&BP__Cliff_Figment_C)
   n33 = GetAllActorsOfClass(execute=@n2.then, WorldContextObject=@n2, ActorClass=/Game/ArtTools/BP_LandscapeSplineMeshActorCopy.BP_LandscapeSplineMeshActorCopy_C) -> (OutActors:Array of BP Landscape Spline Mesh Actor Copy Object References)
   n34 = For Each Loop(Exec=@n44.then, Array:Array of BP Landscape Spline Mesh Actor Copy Object References=@n33) -> (Array Element:BP Landscape Spline Mesh Actor Copy Object Reference)
   Get Object Path String(Object=@n34)
   GetAllActorsOfClass(ActorClass=&BP__Cliff_Figment_C) -> (OutActors:"Array of BP  Cliff Figment Object References")
   n35 = GetPackageForObject(Object=@n34)
   n36 = Get Dependencies(
    execute=@n56.then
    self="To Object (Interface)"(Interface=GetAssetRegistry())
    PackageName=String To Name(InString=GetDisplayName(Object=@n35))
    DependencyOptions=Make Asset Registry Dependency Options(bIncludeSoftPackageReferences=True)
   )
   > n37 = For Each Loop(Array:Array of Names=@n36) -> (Array Element:Name)
   n38 = "To String (Name)"(InName=@n37)
   PrintString(InString=@n38)
   IsEditorPropertyOverridden!(execute, Object:Object Reference=@n34, PropertyName:Name=None) -> (then, ReturnValue:EEditorPropertyValueState Enum=Default)
   GetClass()
   GetClass(InAssetData=CreateAssetData(InAsset=@n34, bAllowBlueprintClass=false))
   GetDisplayName()
   IsEditorPropertyOverridden(PropertyName=None) -> (ReturnValue=Default)
   n39 = Get SplineMeshComponent!(self:Spline Mesh Actor Object Reference=@n34) -> (SplineMeshComponent:Spline Mesh Component Object Reference)
   n40 = GetEditorProperty(execute=@n34.LoopBody, Object=@n39, PropertyName=OverrideMaterials) -> (PropertyValue:Array of Material Interface Object References)
   > n41 = Set materialinterfacearray(materialinterfacearray:Array of Material Interface Object References=@n40) -> (Output_Get:Array of Material Interface Object References)
   n42 = For Each Loop with Break!(Exec:Exec=@n57.then, Array:Array of Material Interface Object References=@n41, Break:Exec=@n55.then) -> (LoopBody:Exec, Array Element:Material Interface Object Reference, Array Index:int, Completed:Exec)
   n43 = Get Object Path String(Object=@n42)
   Append!(A="To String (Integer)"(InInt=0), B=": ", C:String=@n43) -> (ReturnValue)
   Cast To BP_LandscapeSplineMeshActorCopy!(execute, Object=@n34) -> (then, CastFailed, AsBP Landscape Spline Mesh Actor Copy:BP Landscape Spline Mesh Actor Copy Object Reference) {is=title K2Node_DynamicCast}
   Execute Python Script!(
    execute
    PythonScript="import unreal\r\n\r\n## Get a reference to your object (example: a selected actor)\r\n##actor = unreal.EditorLevelLibrary.get_selected_level_actors()[0]\r\n\r\nuclass = actor.get_class()\r\n\r\nprint(dir(uclass))"
    actor:BP Landscape Spline Mesh Actor Copy Object Reference=@n34
   ) -> (then, ReturnValue) {inputs [actor]}
   Execute Python Script!(
    execute
    PythonScript="import unreal\r\n\r\n\r\nprop = comp.get_editor_property(\"override_materials\")\r\nprint(prop)"
    comp:Spline Mesh Component Object Reference=@n39
   ) -> (then, ReturnValue) {inputs [comp]}
   Reroute Node(InputPin:Array of Material Interface Object References=@n41) -> (OutputPin:Array of Material Interface Object References)
   "Equal (GameplayTagContainer)"(A:Wildcard, B:Wildcard) -> (ReturnValue:bool=~) {is=title K2Node_PromotableOperator}
   SetEditorProperty(
    execute=@n48.then
    Object=@n39
    PropertyName=OverrideMaterials
    PropertyValue:Array of Material Interface Object References=Get overridearray() -> (overridearray:Array of Material Interface Object References)
   )
   Get materialinterfacearray() -> (materialinterfacearray:Array of Material Interface Object References)
   n44 = Clear(execute=@n33.then, TargetArray=Get ObjectsToSave())
   n45 = Replace(SourceString=@n38, From=/Figment_S08, To=/FigmentEnvironment)
   n46 = LoadAsset(execute=@n50.then, AssetPath=@n45)
   > n47 = Cast To MaterialInterface(Object=@n46)
   > n48 = Set Array Elem(
    TargetArray:Array of Material Interface Object References=Get overridearray()
    Index=Get materialindexslot() -> (materialindexslot:int=0)
    Item:"Material Interface Object Reference (by ref)"=@n47
    bSizeToFit=false
   )
   n49 = DoesAssetExist(execute=@n54.then, AssetPath=@n45)
   > n50 = Branch(Condition=@n49)
   n51 = "Equal Exactly (String)"(A=@n43, B=None)
   n52 = StartsWith(SourceString=@n38, InPrefix=/Figment_S08)
   n53 = AND Boolean(A=@n52, B=@n51)
   n54 = Branch(execute=@n37.LoopBody, Condition=@n53)
   n55 = Branch(execute=@n42.LoopBody, Condition=@n51)
   n56 = Set materialindexslot(execute=@n42.Completed, materialindexslot:int=@n42.Array Index) -> (Output_Get:int=0)
   n57 = Set overridearray(execute=@n41.then, overridearray:Array of Material Interface Object References=@n41) -> (Output_Get:Array of Material Interface Object References)
  ]}
 FixupTextureData_S03 {nodes [
   n1 = GetSelectedLevelActors(self=Get EditorActorSubsystem())
   > n2 = For Each Loop(Array:Array of Actor Object References=@n1) -> (Array Element:Actor Object Reference)
   > Cast To BP__Cliff_Figment(Object=@n2)
   n3 = GetEditorWorld(execute=@n4.then, self=Get UnrealEditorSubsystem())
   n4 = FixupTextureData_S03!() -> (then, Level Object Reference:Object Reference) {is=title K2Node_FunctionEntry}
   n5 = GetAllActorsOfClass(
    execute=@n3.then
    WorldContextObject=SelectObject(A=@n4, B=@n3, bSelectA=IsValid(Object=@n4))
    ActorClass=&BP__Cliff_Figment_C
   )
   n6 = For Each Loop(Exec=@n58.then, Array=@n5)
   PrintString(InString=GetDisplayName(Object=@n6))
   > n7 = Cast To BP__Cliff_Figment(Object=@n6)
   > n8 = Execute Python Script!(
    execute
    PythonScript="import unreal\r\ntexture_data = (geo.get_editor_property('TextureData')[0])"
    geo:"BP  Cliff Figment Object Reference"=@n7
   ) -> (then, ReturnValue, texture_data:Object Reference) {inputs [geo], outputs [texture_data]}
   n9 = GetDisplayName(Object=GetPackageForObject(Object=@n8.texture_data))
   > n10 = Branch(Condition=IsEmpty(InString=@n9))
   > PrintString(InString=Nothing Assigned)
   n11 = GetEditorWorld(execute=@n10.else, self=Get UnrealEditorSubsystem())
   > n12 = StripRightMostFolderFromPath_Util(Path=@n9, WorldContextObject=@n11, Strip Count=1, __WorldContext=@n11)
   n13 = MakeLiteralString(Value=/Figment_S03/Athena/Environments/TextureData/CliffGround/)
   > n14 = Branch(Condition="Equal, Case Insensitive (String)"(A=@n13, B=@n12))
   > PrintString(InString=already assigned)
   n15 = MakeLiteralString(Value=&CliffGround/)
   n16 = Append!(A=@n12.AssetName, B=_S03) -> (ReturnValue)
   n17 = Append(A=@n15, B=@n16)
   n18 = DoesAssetExist(execute=@n14.else, AssetPath=@n17)
   > n19 = Branch(Condition=@n18)
   > n20 = PrintString(InString=TD asset existsi)
   n21 = GetAssetByObjectPath(
    execute=[@n20.then, @n50.then]
    self=GetAssetRegistry()
    ObjectPath=String To Name(InString=Append!(A=@n17, B=., C:String=@n16) -> (ReturnValue))
    bIncludeOnlyOnDiskAssets=false
   )
   n22 = Cast To BuildingTextureData(execute=[@n21.then, @n49.then], Object=GetAsset(InAssetData=@n21))
   > n23 = Execute Python Script!(
    execute
    PythonScript="import unreal\r\nmaterial_interface = tex_data.get_editor_property('OverrideMaterial')"
    tex_data:Building Texture Data Object Reference=@n22
   ) -> (then, ReturnValue, material_interface:Material Instance Constant Object Reference) {inputs [tex_data], outputs [material_interface]}
   > n24 = Set material_on_TD(material_on_TD=@n23.material_interface)
   n25 = Get Parent(self=@n24)
   > n26 = PrintString(InString=GetDisplayName(Object=@n25))
   n27 = Execute Python Script!(
    execute=[@n26.then, @n51.then]
    PythonScript="import unreal\r\nno_need_to_save = False\r\nTextureDataArray = geo.get_editor_property('TextureData')\r\nprint(TextureDataArray)\r\nif(TextureDataArray[0] == texture_data):\r\n    no_need_to_save = True\r\nelse:\r\n    TextureDataArray[0] = texture_data\r\n    geo.set_editor_property('TextureData', TextureDataArray)"
    geo:"BP  Cliff Figment Object Reference"=@n7
    texture_data:Building Texture Data Object Reference=@n22
   ) -> (then, ReturnValue, no_need_to_save:bool=~) {inputs [geo,texture_data], outputs [no_need_to_save]}
   > n28 = Branch(Condition=@n27.no_need_to_save)
   Add(
    execute=@n28.else
    TargetArray:"Array of BP  Cliff Figment Object References"=Get Cliff Actors to Save()
    NewItem:"BP  Cliff Figment Object Reference (by ref)"=@n7
   )
   PrintString(execute=@n19.else, InString=Append!(A="TD asset doesnt exist, duplicating", B) -> (ReturnValue))
   > n29 = GetAssetTools()
   > n30 = DuplicateAsset(self=@n29, AssetName=@n16, PackagePath=@n15, OriginalObject=@n8.texture_data)
   > n31 = Execute Python Script!(
    execute
    PythonScript="import unreal\r\nmaterial_interface = tex_data.get_editor_property('OverrideMaterial')"
    tex_data:Object Reference=@n30
   ) -> (then, ReturnValue, material_interface:Material Instance Constant Object Reference) {inputs [tex_data], outputs [material_interface]}
   > n32 = Set material_on_TD(material_on_TD=@n31.material_interface)
   n33 = GetDisplayName(Object=GetPackageForObject(Object=@n32))
   > PrintString(InString=@n33)
   > n34 = GetEditorWorld(self=Get UnrealEditorSubsystem())
   > n35 = StripRightMostFolderFromPath_Util(Path=@n33, WorldContextObject=@n34, Strip Count=1, __WorldContext=@n34)
   n36 = MakeLiteralString(Value=/Figment_S03/Athena/Environments/Masterials/CliffGround/)
   n37 = Append(A=@n35.AssetName, B=MakeLiteralString(Value=_S03))
   > n38 = DoesAssetExist(AssetPath=Append(A=@n36, B=@n37))
   > n39 = Branch(Condition=@n38)
   n40 = GetAssetTools(execute=@n39.else)
   > n41 = DuplicateAsset(self=@n40, AssetName=@n37, PackagePath=@n36, OriginalObject=@n32)
   > n42 = Cast To MaterialInstanceConstant(Object=@n41)
   > n43 = Branch(Condition="Equal (Object)"(A=Get Parent(self=@n42), B=/Game/Athena/Environments/Landscape/Material/Athena_Terrain_Mat_Biome))
   n44 = Get FigmentParentMAt() -> (FigmentParentMAt:Material Instance Constant Object Reference)
   > Cast To MaterialInterface(Object=SelectObject(A=Get FigmentParentMAt(), B=@n42, bSelectA=IsValid(Object=@n44)))
   > SetMaterialInstanceParent(Instance=@n42, NewParent=@n42)
   > SaveLoadedAsset(AssetToSave=@n41, bOnlyIfIsDirty=true)
   > n45 = Cast To BuildingTextureData(Object=@n30)
   > Set OverrideMaterial(OverrideMaterial=@n42, self=@n45)
   > SaveLoadedAsset(AssetToSave=@n45, bOnlyIfIsDirty=false)
   > n46 = Execute Python Script!(
    execute
    PythonScript="import unreal\r\nno_need_to_save = False\r\nTextureDataArray = geo.get_editor_property('TextureData')\r\nprint(TextureDataArray)\r\nif(TextureDataArray[0] == texture_data):\r\n    no_need_to_save = True\r\nelse:\r\n    TextureDataArray[0] = texture_data\r\n    geo.set_editor_property('TextureData', TextureDataArray)"
    geo:"BP  Cliff Figment Object Reference"=@n7
    texture_data:Building Texture Data Object Reference=@n45
   ) -> (then, ReturnValue, no_need_to_save:bool=~) {inputs [geo,texture_data], outputs [no_need_to_save]}
   > n47 = Branch(Condition=@n46.no_need_to_save)
   Add(execute=@n47.else, TargetArray=Get Cliff Actors to Save(), NewItem=@n7)
   n48 = GetPackageForObject(Object=@n42)
   PrintString(
    execute=@n43.else
    InString=Append(A=GetDisplayName(Object=@n48), B="\r\n########################\r\n is not a child of Athena Material\r\n########################################")
   )
   n49 = PrintString(InString=Hello)
   n50 = PrintString(InString=Hello)
   n51 = Branch(Condition="Equal (Object)"(A=@n25, B=Get material_orig()))
   PrintString(execute=@n51.else, InString=Wrong parent)
   n52 = Is Valid(InputObject=Get Color Map TD override())
   PrintString(execute=@n52.Is Not Valid, InString=Please Assign TD Asset)
   n53 = GetSelectedLevelActors(execute=@n52.Is Valid, self=Get EditorActorSubsystem())
   n54 = Cast To BP__Cliff_Figment(execute=[@n53.then, @n6.LoopBody], Object=@n6)
   > n55 = GetEditorProperty(Object=@n54, PropertyName=TextureData) -> (PropertyValue:Array of Building Texture Data Object References)
   > n56 = Branch(Condition="Equal (Object)"(
     A="Get (a copy)"(Array:Array of Building Texture Data Object References=@n55, Dimension 1=3) -> (Output:Building Texture Data Object Reference)
     B=Get Color Map TD override()
    ))
   SetEditorProperty(
    execute=@n56.else
    Object=@n54
    PropertyName=TextureData
    PropertyValue:Array of Building Texture Data Object References=Make Array!(
     "[0]":Building Texture Data Object Reference="Get (a copy)"(Array=@n55)
     "[1]":Building Texture Data Object Reference="Get (a copy)"(Array=@n55, Dimension 1=1)
     "[2]":Building Texture Data Object Reference="Get (a copy)"(Array=@n55, Dimension 1=2)
     "[3]":Building Texture Data Object Reference=Get Color Map TD override()
    ) -> (Array:Array of Building Texture Data Object References)
   )
   PrintString(InString="TD asset doesnt exist, duplicating")
   GetAssetByObjectPath(ObjectPath=None, bIncludeOnlyOnDiskAssets=false)
   Get Dependencies!(execute, self:Asset Registry Interface, PackageName:Name=None, DependencyOptions:"Asset Registry Dependency Options Structure (by ref)") -> (then, OutDependencies:Array of Names, ReturnValue:bool) {is=title K2Node_CallFunction}
   GetAsset()
   CollectGarbage()
   n57 = Get Cliff Actors to Save()
   SaveLoadedAssets(execute=@n60.then, AssetsToSave=@n57, bOnlyIfIsDirty=true)
   GetSystemPath(Object=@n8.texture_data)
   PrintString(InString=@n9)
   Append!(A, B=., C:String) -> (ReturnValue)
   Append!(A=@n13, B=MakeLiteralString(Value=&S03)) -> (ReturnValue)
   Cast To MaterialInstanceConstant(Object:Wildcard)
   Execute Python Script!(
    execute
    PythonScript="import unreal\r\nTextureData = geo.get_editor_property('TextureData')[0]\r\nOverrideMaterial = TextureData.get_editor_property('OverrideMaterial')\r\nprint(TextureData)"
    Geo:"BP  Cliff Figment Object Reference"=@n7
   ) -> (then, ReturnValue) {inputs [Geo]}
   "Get (a copy)"(Array:Array of Actor Object References=@n53) -> (Output:Actor Object Reference)
   Reroute Node(InputPin:Object Reference=@n41) -> (OutputPin:Object Reference)
   Get data()
   Get material_orig()
   n58 = Clear(
    execute=@n59.then
    TargetArray:"Array of BP  Cliff Figment Object References"=Get Actors to Revert() -> (Actors to Revert:"Array of BP  Cliff Figment Object References")
   )
   n59 = Clear(execute=@n5.then, TargetArray=Get Cliff Actors to Save())
   n60 = CheckoutLoadedAssets(execute=@n6.Completed, AssetsToCheckout=@n57)
  ]}
 ForEachContext {nodes [
   n1 = Get SourceTextBox()
   n2 = Inputs!() -> (execute, InputPin:String)
   n3 = Select(Option 0=@n2, Option 1="To String (Text)"(InText="GetText (Text Box)"(self=@n1)), Index=IsEmpty(InString=@n2))
   n4 = GetAssetsByPath(execute=@n14.then, self=GetAssetRegistry(), PackagePath=String To Name(InString=@n3), bRecursive=true, bIncludeOnlyOnDiskAssets=true)
   n5 = For Each Loop(Exec=@n15.then_0, Array:Array of Asset Data Structures=@n4) -> (Array Element:Asset Data Structure)
   n6 = Break Asset Data(AssetData=@n5)
   n7 = Branch(Condition=Contains(SearchIn="To String (Name)"(InName=@n6), Substring=ExternalObjects, bUseCase=true))
   n8 = Get Class Display Name(Class=GetClass(InAssetData=@n5))
   PrintString(execute=@n7.else, InString=@n8)
   > n9 = Branch(Condition="Equal Exactly (String)"(A=@n8, B=World))
   > PrintString(InString="To String (Name)"(InName=@n6))
   > Clear(TargetArray=Get Cliff Actors to Save())
   > Sequence()
   > Sequence!(execute) -> (then_0, then_1, then_2, then_3:Exec, then_4:Exec)
   n10 = GetAsset(InAssetData=@n5)
   > n11 = Set Current Level To Process(Current Level To Process:Object Reference=@n10) -> (Output_Get:Object Reference)
   n12 = Resolve World Soft Object Path!(execute=@n16.then, Path:"Soft Object Path Structure (by ref)"=GetSoftObjectPath(Object=@n10)) -> (Success:Exec, NotAWorldObject:Exec, InvalidPath:Exec, WorldObject:World Soft Object Reference)
   > n13 = GetLevels!(execute, World:World Object Reference=Resolve Soft Reference!(Input:World Soft Object Reference=@n12) -> (Output:World Object Reference)) -> (then, ReturnValue:Array of Level Object References)
   Outputs!(then:Exec=@n11.then, Output_Get:Object Reference=@n11, ResolvedExec:Exec=@n13.then, ReturnValue:Array of Level Object References=@n13)
   PrintString(execute=@n9.else, InString="Found Object other than World SKIPPING\r\n", TextColor=(R=1,G=0.4929,B=0,A=1))
   n14 = PrintString(execute=@n2.execute, InString=@n3)
   n15 = Sequence!(execute=@n4.then) -> (then_0, then_1)
   n16 = Set Current Level To Process(execute=@n5.LoopBody, Current Level To Process=@n10)
  ]}
 ForEachLI {nodes [
   n1 = Get SourceTextBox()
   n2 = Inputs()
   n3 = Select(Option 0=@n2, Option 1="To String (Text)"(InText="GetText (Text Box)"(self=@n1)), Index=IsEmpty(InString=@n2))
   n4 = GetAssetsByPath(execute=@n22.then, self=GetAssetRegistry(), PackagePath=String To Name(InString=@n3), bRecursive=true, bIncludeOnlyOnDiskAssets=true)
   n5 = For Each Loop(Exec=@n23.then_0, Array=@n4)
   n6 = Break Asset Data(AssetData=@n5)
   n7 = PrintString(InString="To String (Name)"(InName=@n6))
   n8 = GetEditorWorld(execute=[@n7.then, @n20.then], self=Get UnrealEditorSubsystem())
   > n9 = Sequence!(execute) -> (then_0, then_1, then_2:Exec, then_3:Exec)
   > n10 = Branch(Condition=false)
   n11 = GetAsset(InAssetData=@n5)
   n12 = Set Current Level To Process(execute=[@n16.then_0, @n10.else], Current Level To Process=@n11)
   > Outputs!(then, Output_Get=@n12)
   n13 = Make Soft Reference!(Input:Object Reference=@n11) -> (Output:Soft Object Reference)
   UnloadLevelInstance!(
    execute:Exec=[@n15.then_1, @n9.then_1]
    self:Object Reference=Resolve Soft Reference(Input:Soft Object Reference=@n13) -> (Output:Object Reference)
   ) -> (then:Exec) {is=title K2Node_Message}
   > CollectGarbage()
   n14 = SpawnActorFromClass!(
    execute
    self:Editor Actor Subsystem Object Reference=Get EditorActorSubsystem()
    ActorClass:Actor Class Reference=/Script/Engine.LevelInstance
    Location:Vector="0, 0, 0"
    Rotation:Rotator
    bTransient:bool=true
   ) -> (then, ReturnValue:Level Instance Object Reference)
   > Set LevelInstanceActor(LevelInstanceActor:Level Instance Object Reference=@n14) -> (Output_Get:Level Instance Object Reference)
   > SetEditorProperty(Object=@n14, PropertyName=WorldAsset, PropertyValue:Soft Object Reference=@n13, ChangeNotifyMode=Never)
   > Clear(TargetArray=Get Cliff Actors to Save())
   > n15 = Sequence()
   > LevelInstanceEnterEditMode!(execute, LevelInstance:Level Instance Object Reference=Get LevelInstanceActor() -> (LevelInstanceActor:Level Instance Object Reference)) -> (then, ReturnValue:bool)
   > n16 = Sequence!(execute) -> (then_0, then_1, then_2, then_3, then_4:Exec)
   n17 = Get EditorActorSubsystem()
   SelectNothing(execute=[@n15.then_2, @n15.then_3], self:Editor Actor Subsystem Object Reference=@n17)
   > DestroyActor!(execute, self:Editor Actor Subsystem Object Reference=@n17, ActorToDestroy:Actor Object Reference=Get LevelInstanceActor()) -> (then, ReturnValue:bool)
   > CollectGarbage()
   n18 = Get Class Display Name(Class=GetClass(InAssetData=@n5))
   n19 = PrintString(InString=@n18)
   n20 = Branch(execute=[@n19.then, @n24.else], Condition="Equal Exactly (String)"(A=@n18, B=World))
   GetAssetsByPath(
    PackagePath=String To Name(InString=MakeLiteralString(Value="\\Figment_S09_LevelInstances\\Content\\__ExternalActors__\\POIs\\New"))
    bRecursive=true
    bIncludeOnlyOnDiskAssets=true
   )
   PrintString(InString="Found Object other than World SKIPPING\r\n", TextColor=(R=1,G=0.4929,B=0,A=1))
   "To String (Text)"(InText="GetText (Text Box)"(self=Get SourceTextBox()))
   GetActorDescs()
   GetClass(InAssetData=@n5)
   GetActorDescsForActors!(execute, InActors:Array of Actor Object References) -> (then, OutActorDescs:Array of Actor Desc Structures, ReturnValue:bool)
   LevelInstanceCommit!(execute, LevelInstance:Level Instance Object Reference=Get LevelInstanceActor(), bDiscardChanges:bool=true) -> (then, ReturnValue:bool)
   GetActorDescs()
   Make Soft Reference(Input=@n11)
   Resolve Soft Reference(Input=Make Soft Reference(Input=@n11))
   Get LevelInstanceSubsystem!(WorldContext:Object Reference) -> (ReturnValue:Level Instance Subsystem Object Reference) {is=title K2Node_GetSubsystem}
   Get WorldPartitionSubsystem!(WorldContext:Object Reference) -> (ReturnValue:World Partition Subsystem Object Reference) {is=title K2Node_GetSubsystem}
   Reroute Node(InputPin:World Object Reference=@n8) -> (OutputPin:World Object Reference)
   n21 = Contains(SearchIn="To String (Name)"(InName=@n6), Substring=ExternalObjects, bUseCase=true)
   n22 = PrintString(execute=@n2.execute, InString=@n3)
   n23 = Sequence!(execute=@n4.then) -> (then_0, then_1)
   n24 = Branch(execute=@n25.then, Condition=@n21)
   n25 = Set Current Level To Process(execute=@n5.LoopBody, Current Level To Process=@n11)
  ]}
 ForEachLI_Full {nodes [
   n1 = Get SourceTextBox()
   n2 = Inputs()
   n3 = Select(Option 0=@n2, Option 1="To String (Text)"(InText="GetText (Text Box)"(self=@n1)), Index=IsEmpty(InString=@n2))
   n4 = GetAssetsByPath(execute=@n21.then, self=GetAssetRegistry(), PackagePath=String To Name(InString=@n3), bRecursive=true, bIncludeOnlyOnDiskAssets=true)
   n5 = For Each Loop(Exec=@n22.then_0, Array=@n4)
   n6 = Break Asset Data(AssetData=@n5)
   n7 = PrintString(InString="To String (Name)"(InName=@n6))
   n8 = GetEditorWorld(execute=[@n7.then, @n18.then], self=Get UnrealEditorSubsystem())
   > n9 = SpawnActorFromClass(self=Get EditorActorSubsystem(), ActorClass=&LevelInstance, Location="0, 0, 0", bTransient=true)
   > n10 = Set LevelInstanceActor(LevelInstanceActor=@n9)
   n11 = GetAsset(InAssetData=@n5)
   n12 = Make Soft Reference(Input=@n11)
   > SetEditorProperty(Object=@n10, PropertyName=WorldAsset, PropertyValue=@n12, ChangeNotifyMode=Never)
   > Clear(TargetArray=Get Cliff Actors to Save())
   > n13 = Sequence!(execute) -> (then_0, then_1, then_2:Exec, then_3:Exec, then_4:Exec)
   n14 = Get LevelInstanceActor()
   > LevelInstanceEnterEditMode(LevelInstance=@n14)
   > Outputs!(then)
   LevelInstanceCommit(execute=@n13.then_1, LevelInstance=@n14, bDiscardChanges=true)
   > CollectGarbage()
   ResetEditorProperty(execute=@n13.then_2, Object=@n14, PropertyName=WorldAsset, ChangeNotifyMode=Never)
   > CollectGarbage()
   n15 = Get EditorActorSubsystem()
   SelectNothing(execute=@n13.then_3, self=@n15)
   > DestroyActor(self=@n15, ActorToDestroy=Get LevelInstanceActor())
   > CollectGarbage()
   n16 = Get Class Display Name(Class=GetClass(InAssetData=@n5))
   n17 = PrintString(InString=@n16)
   n18 = Branch(execute=[@n17.then, @n23.else], Condition="Equal Exactly (String)"(A=@n16, B=World))
   Sequence!(execute) -> (then_0, then_1, then_2, then_3)
   > n19 = Branch(Condition=false)
   Set Current Level To Process(execute=@n19.else, Current Level To Process=@n11)
   GetAssetsByPath(
    PackagePath=String To Name(InString=MakeLiteralString(Value="\\Figment_S09_LevelInstances\\Content\\__ExternalActors__\\POIs\\New"))
    bRecursive=true
    bIncludeOnlyOnDiskAssets=true
   )
   UnloadLevelInstance(self:Level Instance Interface Interface=@n14) {is=title K2Node_CallFunction}
   "To String (Text)"(InText="GetText (Text Box)"(self=Get SourceTextBox()))
   GetActorDescs()
   GetClass(InAssetData=@n5)
   GetActorDescsForActors()
   GetActorDescs()
   Resolve Soft Reference(Input=@n12)
   Sequence!(execute) -> (then_0, then_1, then_2, then_3, then_4:Exec)
   Get LevelInstanceSubsystem()
   Get WorldPartitionSubsystem()
   Reroute Node(InputPin=@n8)
   n20 = Contains(SearchIn="To String (Name)"(InName=@n6), Substring=ExternalObjects, bUseCase=true)
   n21 = PrintString(execute=@n2.execute, InString=@n3)
   n22 = Sequence!(execute=@n4.then) -> (then_0, then_1)
   n23 = Branch(execute=@n24.then, Condition=@n20)
   n24 = Set Current Level To Process(execute=@n5.LoopBody, Current Level To Process=@n11)
  ]}
 ForEachLevelInstanceActor {nodes [
   n1 = ForEachLevelInstanceActor!() -> (then, InputPin:String) {is=title K2Node_FunctionEntry}
   > For Each LI(InputPin=@n1)
   > Return Node!(execute)
  ]}
 ForEachLevelInstanceContext {nodes [
   n1 = ForEachLevelInstanceContext!() -> (then, InString:String) {is=title K2Node_FunctionEntry}
   > n2 = For Each Context(InputPin=@n1)
   > Return Node!(execute, Output_Get:Object Reference=@n2)
  ]}
 GetParentFromTD {nodes [
   n1 = GetParentFromTD!() -> (then, actor:Actor Object Reference) {is=title K2Node_FunctionEntry}
   > n2 = GetEditorProperty(Object=@n1, PropertyName=TextureData)
   > Branch(Condition=@n2.ReturnValue)
   > n3 = Set TDArray(TDArray:Array of Building Texture Data Object References=@n2) -> (Output_Get:Array of Building Texture Data Object References)
   n4 = "Get (a copy)"(Array:Array of Building Texture Data Object References=@n3) -> (Output:Building Texture Data Object Reference)
   > Is Valid(InputObject=@n4)
   n5 = Get OverrideMaterial!(self:Building Texture Data Object Reference=@n4) -> (OverrideMaterial:Material Interface Object Reference)
   > n6 = Cast To MaterialInstanceConstant(Object:Object Reference=@n5)
   > Return Node!(
    execute
    override:Material Interface Object Reference=@n5
    TDArrayOut:Array of Building Texture Data Object References=Get TDArray() -> (TDArray:Array of Building Texture Data Object References)
    Parent:Material Interface Object Reference=Get Parent(self=@n6)
   )
   Get current actor() -> (current actor:Actor Object Reference)
  ]}
 GetStaticMeshActorsExclusive {nodes [
   n1 = GetStaticMeshActorsExclusive!() -> (then, WorldContextObject:Object Reference) {is=title K2Node_FunctionEntry}
   > n2 = GetAllActorsOfClass(WorldContextObject=@n1, ActorClass=&FortStaticMeshActor) -> (OutActors:Array of Fort Static Mesh Actor Object References)
   > Add Items(
    TargetSet:Set of Actor Object References=Get FortStaticMeshes() -> (FortStaticMeshes:Set of Actor Object References)
    NewItems:Array of Actor Object References=@n2
   )
   > n3 = GetAllActorsOfClass(WorldContextObject=@n1, ActorClass=&BP__Cliff_Figment_C) -> (OutActors:"Array of BP  Cliff Figment Object References")
   n4 = Get FortStaticMeshes()
   > Remove Items!(execute, TargetSet:Set of Actor Object References=@n4, Items:Array of Actor Object References=@n3) -> (then)
   > n5 = To Array!(execute, A:Set of Actor Object References=@n4) -> (then, Result:Array of Actor Object References)
   > Return Node!(execute, Result:Array of Actor Object References=@n5)
  ]}
 GetWorlds {nodes [
   n1 = GetWorlds!() -> (then, Level Actors:Array of Building Level Instance Object References) {is=title K2Node_FunctionEntry}
   > n2 = For Each Loop(Array:Array of Building Level Instance Object References=@n1) -> (Array Element:Building Level Instance Object Reference)
   n3 = Get worldslocal() -> (worldslocal:Array of Object References)
   Return Node!(execute=@n2.Completed, WorldArray2:Array of Object References=@n3)
   > n4 = GetEditorProperty(Object=@n2, PropertyName=WorldAsset) -> (PropertyValue:Soft Object Reference)
   n5 = GetObjectFromSoftPath(Path=To Soft Object Path!(SoftObjectReference:Soft Object Reference=@n4) -> (ReturnValue:Soft Object Path Structure))
   > Add(TargetArray:Array of Object References=@n3, NewItem:"Object Reference (by ref)"=@n5)
   GetWorldAsset!(self:Level Instance Interface Interface=Cast To LevelInstanceInterface!(execute, Object=@n2) -> (then, CastFailed, AsLevel Instance Interface:Level Instance Interface Interface) {is=title K2Node_DynamicCast}) -> (ReturnValue:"World Soft Object Reference (by ref)")
   PrintString(InString=GetDisplayName(Object=@n5))
   Append(A="To String (Integer)"(InInt=@n2.Array Index), B=GetSoftPtrAssetName!(Asset:Soft Object Reference) -> (ReturnValue:String))
  ]}
 IsInSnowBiome {nodes [
   n1 = IsInSnowBiome!() -> (then, actor:Actor Object Reference) {is=title K2Node_FunctionEntry}
   n2 = "To Vector2D (Vector)"!(InVector:Vector=Get Actor Location(self=@n1)) -> (ReturnValue:Vector 2D Structure)
   n3 = Make Literal Float(Value=-153000)
   n4 = Texture2D Sample UV Editor Only!(
    Texture:Texture 2D Object Reference=/Figment_S07/LM_7_00/Athena/Environments/Landscape/Textures/T_Athena_Terrain_Topo_Mask_740
    UV:Vector 2D Structure=vector2d / vector2d(
     A:Vector 2D Structure=vector2d - vector2d(A:Vector 2D Structure=@n2, B=@n3) -> (ReturnValue:Vector 2D Structure) {is=title K2Node_PromotableOperator}
     B=Make Literal Float(Value=306200.7812)
    ) -> (ReturnValue:Vector 2D Structure) {is=title K2Node_PromotableOperator}
   ) -> (ReturnValue:Linear Color Structure)
   > Return Node!(execute, ReturnValue:bool="float > float"(A=BreakColor(InColor=@n4).B, B=0.5))
  ]}
 MigrateTDImpl {nodes [
   n1 = MigrateTDImpl!() -> (then, InputPin:"Array of BP  Cliff Figment Object References") {is=title K2Node_FunctionEntry}
   > Clear(TargetArray:Array of Object References=Get ObjectsToSave())
   > n2 = For Each Loop(Array:"Array of BP  Cliff Figment Object References"=@n1) -> (Array Element:"BP  Cliff Figment Object Reference")
   n3 = Get ObjectsToSave()
   CheckoutLoadedAssets(execute=@n2.Completed, AssetsToCheckout=@n3)
   > SaveLoadedAssets(AssetsToSave=@n3, bOnlyIfIsDirty=false)
   n4 = GetEditorProperty(execute=@n2.LoopBody, Object=@n2, PropertyName=TextureData) -> (PropertyValue:Array of Building Texture Data Object References)
   n5 = "Get (a copy)"(Array=@n4)
   > Is Valid(InputObject=@n5)
   n6 = ParseIntoArray(SourceString=Get Object Path String(Object=@n5), Delimiter=/)
   n7 = "Get (a copy)"(Array:Array of Strings=@n6) -> (Output:String)
   n8 = "Get (a copy)"(Array=@n6, Dimension 1=1)
   n9 = "To String (Text)"(InText="GetText (Text Box)"(self=Get TargetTextBox_()))
   n10 = ParseIntoArray(SourceString=@n9, Delimiter=/)
   n11 = "Get (a copy)"(Array=@n10)
   n12 = "Get (a copy)"(Array=@n10, Dimension 1=1)
   > n13 = Branch(Condition="Equal, Case Insensitive (String)"(A=Append!(A=@n7, B=/, C:String=@n8) -> (ReturnValue), B=Append(A=@n11, B=/, C=@n12)))
   > Reroute Node(InputPin:Exec) -> (OutputPin:Exec)
   n14 = GetDisplayName(Object=Get OverrideMaterial(self=@n5))
   n15 = Append!(A=@n9, B=@n14) -> (ReturnValue)
   DoesAssetExist(execute=@n13.else, AssetPath=@n15)
   n16 = Find(TargetMap=Get MapStringToTargetMaterials(), Key=@n14)
   > n17 = Branch(Condition=@n16.ReturnValue)
   > PrintString(InString=Append(A=" Exist", B=@n15))
   > n18 = Cast To MaterialInterface(Object=@n16)
   > Set CurrentMaterialForStaticMesh(CurrentMaterialForStaticMesh=@n18)
   n19 = GetDisplayName(Object=@n5)
   n20 = "To String (Text)"(InText=Get Text(self=Get TargetTextBox_()))
   n21 = Append(A=@n20, B=@n19)
   > n22 = DoesAssetExist(AssetPath=@n21)
   > n23 = Branch(Condition=@n22)
   > n24 = LoadAsset(AssetPath=@n21)
   > n25 = Cast To BuildingTextureData(Object=@n24)
   n26 = "Get (a copy)"(Array:Array of Building Texture Data Object References=@n4, Dimension 1=1) -> (Output:Building Texture Data Object Reference)
   n27 = "Get (a copy)"(Array=@n4, Dimension 1=2)
   n28 = "Get (a copy)"(Array=@n4, Dimension 1=3)
   > SetEditorProperty(
    Object=@n2
    PropertyName=TextureData
    PropertyValue:Array of Building Texture Data Object References=Make Array("[0]"=@n25, "[1]"=@n26, "[2]"=@n27, "[3]"=@n28)
   )
   > Add(TargetArray=Get ObjectsToSave(), NewItem=@n2)
   PrintString(execute=@n23.else, InString=Doesn't Exist)
   > n29 = GetAssetTools()
   > n30 = DuplicateAsset(self=@n29, AssetName=@n19, PackagePath=@n20, OriginalObject=@n5)
   > n31 = Cast To BuildingTextureData(Object=@n30)
   > n32 = Set TD Asset(TD Asset:Building Texture Data Object Reference=@n31) -> (Output_Get:Building Texture Data Object Reference)
   > Set OverrideMaterial(OverrideMaterial=Get CurrentMaterialForStaticMesh(), self=@n32)
   > Add(TargetArray=Get ObjectsToSave(), NewItem=@n32)
   > SetEditorProperty(Object=@n2, PropertyName=TextureData, PropertyValue=Make Array("[0]"=@n32, "[1]"=@n26, "[2]"=@n27, "[3]"=@n28))
   > Add(TargetArray=Get ObjectsToSave(), NewItem=@n2)
   PrintString(execute=@n17.else, InString=Append(A=Doesn't Exist, B=@n15))
   Get Object Path String(Object=@n5)
  ]}
 MigrateTDsAndMats {nodes [
   MigrateTDsAndMats() {is=title K2Node_FunctionEntry}
   > n1 = Sequence()
   > BuildMultiLineMapOfMaterials()
   Clear(
    execute=@n1.then_1
    TargetArray:"Array of BP  Cliff Figment Object References"=Get CliffActors() -> (CliffActors:"Array of BP  Cliff Figment Object References")
   )
   > n2 = Switch on E_ApplyContext(Selection=Get Apply Mode() -> (Apply Mode=NewEnumerator2))
   n3 = Get SourceTextBox()
   n4 = For Each LI(execute=@n2.NewEnumerator0, InputPin="To String (Text)"(InText="GetText (Text Box)"(self=@n3)))
   > n5 = GetAllActorsOfClass(WorldContextObject=@n4, ActorClass=&BP__Cliff_Figment_C)
   > n6 = Set CliffActors(CliffActors:"Array of BP  Cliff Figment Object References"=@n5) -> (Output_Get:"Array of BP  Cliff Figment Object References")
   MigrateTDImpl!(execute=[@n13.then, @n6.then, @n9.then], self, InputPin:"Array of BP  Cliff Figment Object References"=Get CliffActors()) -> (then) {is=title K2Node_CallFunction}
   n7 = GetEditorWorld(execute=@n2.NewEnumerator1, self=Get UnrealEditorSubsystem())
   > n8 = GetAllActorsOfClass(WorldContextObject=@n7, ActorClass=&BP__Cliff_Figment_C)
   > n9 = Set CliffActors(CliffActors=@n8)
   n10 = GetSelectedLevelActors(execute=@n2.NewEnumerator2, self=Get EditorActorSubsystem())
   > n11 = For Each Loop(Array:Array of Actor Object References=@n10) -> (Array Element:Actor Object Reference)
   > n12 = Cast To BP__Cliff_Figment(Object=@n11)
   > n13 = Add(TargetArray:"Array of BP  Cliff Figment Object References"=Get CliffActors(), NewItem:"BP  Cliff Figment Object Reference (by ref)"=@n12)
  ]}
 MigrateTDsAndMatsS08 {nodes [
   n1 = For Each Loop(Exec=@n38.then, Array:"Array of BP  Cliff Figment Object References"=Get CliffActors()) -> (Array Element:"BP  Cliff Figment Object Reference")
   > n2 = GetEditorProperty(Object=@n1, PropertyName=TextureData)
   n3 = "Get (a copy)"(Array=@n2)
   n4 = GetDisplayName(Object=Get OverrideMaterial(self=@n3))
   n5 = "To String (Text)"(InText="GetText (Text Box)"(self=Get TargetTextBox_()))
   n6 = Append(A=@n5, B=@n4)
   n7 = LoadAsset(AssetPath=@n6)
   n8 = Find(TargetMap=Get MapStringToTargetMaterials(), Key=@n4)
   n9 = Cast To MaterialInterface(execute=[@n41.then, @n7.then], Object=@n8)
   > Set CurrentMaterialForStaticMesh(CurrentMaterialForStaticMesh=@n9)
   n10 = GetDisplayName(Object=@n3)
   n11 = "To String (Text)"(InText=Get Text(self=Get TargetTextBox_()))
   n12 = Append(A=@n11, B=@n10)
   > n13 = DoesAssetExist(AssetPath=@n12)
   > n14 = Branch(Condition=@n13)
   > n15 = LoadAsset(AssetPath=@n12)
   > n16 = Cast To BuildingTextureData(Object=@n15)
   n17 = "Get (a copy)"(Array=@n2, Dimension 1=1)
   n18 = "Get (a copy)"(Array=@n2, Dimension 1=2)
   n19 = "Get (a copy)"(Array=@n2, Dimension 1=3)
   > SetEditorProperty(Object=@n1, PropertyName=TextureData, PropertyValue=Make Array("[0]"=@n16, "[1]"=@n17, "[2]"=@n18, "[3]"=@n19))
   > Add(TargetArray:Array of Object References=Get ObjectsToSave(), NewItem:"Object Reference (by ref)"=@n1)
   PrintString(execute=@n14.else, InString=Doesn't Exist)
   > n20 = GetAssetTools()
   > n21 = DuplicateAsset(self=@n20, AssetName=@n10, PackagePath=@n11, OriginalObject=@n3)
   > n22 = Cast To BuildingTextureData(Object=@n21)
   > n23 = Set TD Asset(TD Asset=@n22)
   > Set OverrideMaterial(OverrideMaterial=Get CurrentMaterialForStaticMesh(), self=@n23)
   > Add(TargetArray=Get ObjectsToSave(), NewItem=@n23)
   > SetEditorProperty(Object=@n1, PropertyName=TextureData, PropertyValue=Make Array("[0]"=@n23, "[1]"=@n17, "[2]"=@n18, "[3]"=@n19))
   > Add(TargetArray=Get ObjectsToSave(), NewItem=@n1)
   n24 = MigrateTDsAndMatsS08!() -> (then, Apply Mode:E_ApplyContext Enum=NewEnumerator2) {is=title K2Node_FunctionEntry}
   > n25 = Sequence()
   > BuildMultiLineMapOfMaterials()
   PrintString(execute=@n25.then_1, InString=HelloIn)
   > Clear(TargetArray=Get CliffActors())
   > n26 = Sequence()
   > n27 = Switch on E_ApplyContext(Selection=@n24)
   n28 = GetEditorWorld(execute=@n27.NewEnumerator1, self=Get UnrealEditorSubsystem())
   > n29 = GetAllActorsOfClass(WorldContextObject=@n28, ActorClass=&BP__Cliff_Figment_C)
   > Set CliffActors(CliffActors=@n29)
   n30 = GetSelectedLevelActors(execute=@n27.NewEnumerator2, self=Get EditorActorSubsystem())
   > n31 = For Each Loop(Array:Array of Actor Object References=@n30) -> (Array Element:Actor Object Reference)
   > n32 = Cast To BP__Cliff_Figment(Object=@n31)
   > Add(TargetArray:"Array of BP  Cliff Figment Object References"=Get CliffActors(), NewItem:"BP  Cliff Figment Object Reference (by ref)"=@n32)
   n33 = Switch on E_ApplyContext(execute=@n26.then_1, Selection=@n24)
   n34 = Get SourceTextBox()
   n35 = For Each LI(execute=@n33.NewEnumerator0, InputPin="To String (Text)"(InText="GetText (Text Box)"(self=@n34)))
   > n36 = GetAllActorsOfClass(WorldContextObject=@n35, ActorClass=&BP__Cliff_Figment_C)
   > n37 = Set CliffActors(CliffActors=@n36)
   n38 = Clear(execute=[@n33.NewEnumerator1, @n33.NewEnumerator2, @n37.then], TargetArray:Array of Object References=Get ObjectsToSave())
   PrintString(execute=@n43.else, InString=Append(A=Doesn't Exist, B=@n6))
   n39 = Get ObjectsToSave()
   SaveLoadedAssets(execute=@n44.then, AssetsToSave=@n39, bOnlyIfIsDirty=false)
   Get Object Path String(Object=@n3)
   Reroute Node(InputPin=@n45.then)
   n40 = Get Object Path String(Object=@n3)
   n41 = PrintString(execute=@n43.then, InString=Append(A=" Exist", B=@n6))
   n42 = "Equal, Case Insensitive (String)"(
    A="Get (a copy)"(Array:Array of Strings=ParseIntoArray(SourceString=@n40, Delimiter=/)) -> (Output:String)
    B="Get (a copy)"(Array=ParseIntoArray(SourceString=@n5, Delimiter=/))
   )
   DoesAssetExist(execute=@n45.else, AssetPath=@n6)
   > n43 = Branch(Condition=@n8.ReturnValue)
   n44 = CheckoutLoadedAssets(execute=@n1.Completed, AssetsToCheckout=@n39)
   Sequence!(execute=@n46.Is Valid) -> (then_0, then_1, then_2:Exec)
   > n45 = Branch(Condition=@n42)
   n46 = Is Valid(exec=@n2.then, InputObject=@n3)
  ]}
 PaintConcreteVertex1000 {nodes [
   n1 = PaintConcreteVertex1000!() -> (then, WorldContextObject:Object Reference) {is=title K2Node_FunctionEntry}
   > n2 = GetEditorWorld(self=Get UnrealEditorSubsystem())
   > n3 = GetAllActorsOfClass(
    WorldContextObject=SelectObject(A=@n1, B=@n2, bSelectA=IsValid(Object=@n1))
    ActorClass=&BP__Cliff_Figment_C
   )
   > Clear(TargetArray:"Array of BP  Cliff Figment Object References"=Get Cliff Actors to Save())
   > n4 = For Each Loop(Array:"Array of BP  Cliff Figment Object References"=@n3) -> (Array Element:"BP  Cliff Figment Object Reference")
   n5 = Get Cliff Actors to Save()
   CheckoutLoadedAssets(execute=@n4.Completed, AssetsToCheckout=@n5)
   > SaveLoadedAssets(AssetsToSave=@n5, bOnlyIfIsDirty=false)
   n6 = Cast To BP__Cliff_Figment(execute=@n4.LoopBody, Object=@n4)
   n7 = Get StaticMeshComponent(self=@n6)
   > n8 = Cast To MaterialInstanceDynamic(Object=GetMaterial(self=@n7, ElementIndex=0))
   n9 = Get Parent(self=@n8)
   n10 = GetDisplayName(Object=@n9)
   > Branch(Condition=OR Boolean(
     A="Equal (Object)"(A=@n9, B=/Figment_S07/LM_7_00/Athena/Environments/Landscape/Material/Athena_Terrain_Mat_Concerte_Inst)
     B=Contains(SearchIn=@n10, Substring=Athena_Terrain_Mat_Concerte_Inst)
    ))
   > PaintVerticesSingleColor!(
    execute
    StaticMeshComponent:Static Mesh Component Object Reference=@n7
    FillColor:"Linear Color Structure (by ref)"=Make Linear Color!(R:float=1, G:float=0, B:float=0, A:float=0) -> (LinearColor:Linear Color Structure) {is=title K2Node_MakeStruct}
    bConvertToSRGB:bool=false
   ) -> (then)
   > Add(TargetArray=Get Cliff Actors to Save(), NewItem=@n6)
   Contains(Substring=concrete)
   PrintString(InString=mid)
  ]}
 Reassign TextureData with existing {nodes [
   n1 = GetSelectedLevelActors(self=Get EditorActorSubsystem())
   > n2 = For Each Loop(Array:Array of Actor Object References=@n1) -> (Array Element:Actor Object Reference)
   > Cast To BP__Cliff_Figment(Object=@n2)
   n3 = GetEditorWorld(execute=@n4.then, self=Get UnrealEditorSubsystem())
   n4 = Reassign TextureData with existing!() -> (then, Level Object Reference:Object Reference) {is=title K2Node_FunctionEntry}
   n5 = GetAllActorsOfClass(
    execute=@n3.then
    WorldContextObject=SelectObject(A=@n4, B=@n3, bSelectA=IsValid(Object=@n4))
    ActorClass=&BP__Cliff_Figment_C
   )
   n6 = For Each Loop(Exec=@n53.then, Array=@n5)
   n7 = PrintString(InString=GetDisplayName(Object=@n6))
   n8 = Cast To BP__Cliff_Figment(execute=[@n7.then, @n6.LoopBody], Object=@n6)
   > n9 = Execute Python Script!(
    execute
    PythonScript="import unreal\r\ntexture_data = (geo.get_editor_property('TextureData')[0])"
    geo:"BP  Cliff Figment Object Reference"=@n8
   ) -> (then, ReturnValue, texture_data:Object Reference) {inputs [geo], outputs [texture_data]}
   n10 = GetDisplayName(Object=GetPackageForObject(Object=@n9.texture_data))
   > n11 = Branch(Condition=IsEmpty(InString=@n10))
   > PrintString(InString=Nothing Assigned)
   n12 = GetEditorWorld(execute=@n11.else, self=Get UnrealEditorSubsystem())
   > n13 = StripRightMostFolderFromPath_Util(Path=@n10, WorldContextObject=@n12, Strip Count=1, __WorldContext=@n12)
   n14 = MakeLiteralString(Value=/Figment_S03/Athena/Environments/TextureData/CliffGround/)
   > n15 = Branch(Condition="Equal, Case Insensitive (String)"(A=@n14, B=@n13))
   > PrintString(InString=already assigned)
   n16 = MakeLiteralString(Value=&CliffGround/)
   n17 = Append(A=@n13.AssetName, B=_S03)
   n18 = Append(A=@n16, B=@n17)
   n19 = DoesAssetExist(execute=@n15.else, AssetPath=@n18)
   > n20 = Branch(Condition=@n19)
   > n21 = PrintString(InString=TD asset existsi)
   n22 = GetAssetByObjectPath(
    execute=[@n21.then, @n51.then]
    self=GetAssetRegistry()
    ObjectPath=String To Name(InString=Append!(A=@n18, B=., C:String=@n17) -> (ReturnValue))
    bIncludeOnlyOnDiskAssets=false
   )
   n23 = Cast To BuildingTextureData(execute=[@n22.then, @n50.then], Object=GetAsset(InAssetData=@n22))
   > n24 = Execute Python Script!(
    execute
    PythonScript="import unreal\r\nmaterial_interface = tex_data.get_editor_property('OverrideMaterial')"
    tex_data:Building Texture Data Object Reference=@n23
   ) -> (then, ReturnValue, material_interface:Material Instance Constant Object Reference) {inputs [tex_data], outputs [material_interface]}
   > n25 = Set material_on_TD(material_on_TD=@n24.material_interface)
   n26 = Get Parent(self=@n25)
   > n27 = PrintString(InString=GetDisplayName(Object=@n26))
   n28 = Execute Python Script!(
    execute=[@n27.then, @n52.then]
    PythonScript="import unreal\r\nno_need_to_save = False\r\nTextureDataArray = geo.get_editor_property('TextureData')\r\nprint(TextureDataArray)\r\nif(TextureDataArray[0] == texture_data):\r\n    no_need_to_save = True\r\nelse:\r\n    TextureDataArray[0] = texture_data\r\n    geo.set_editor_property('TextureData', TextureDataArray)"
    geo:"BP  Cliff Figment Object Reference"=@n8
    texture_data:Building Texture Data Object Reference=@n23
   ) -> (then, ReturnValue, no_need_to_save:bool=~) {inputs [geo,texture_data], outputs [no_need_to_save]}
   > n29 = Branch(Condition=@n28.no_need_to_save)
   Add(execute=@n29.else, TargetArray=Get Cliff Actors to Save(), NewItem=@n8)
   PrintString(execute=@n20.else, InString=Append!(A="TD asset doesnt exist, duplicating", B) -> (ReturnValue))
   > n30 = GetAssetTools()
   > n31 = DuplicateAsset(self=@n30, AssetName=@n17, PackagePath=@n16, OriginalObject=@n9.texture_data)
   > n32 = Execute Python Script!(
    execute
    PythonScript="import unreal\r\nmaterial_interface = tex_data.get_editor_property('OverrideMaterial')"
    tex_data:Object Reference=@n31
   ) -> (then, ReturnValue, material_interface:Material Instance Constant Object Reference) {inputs [tex_data], outputs [material_interface]}
   > n33 = Set material_on_TD(material_on_TD=@n32.material_interface)
   n34 = GetDisplayName(Object=GetPackageForObject(Object=@n33))
   > PrintString(InString=@n34)
   > n35 = GetEditorWorld(self=Get UnrealEditorSubsystem())
   > n36 = StripRightMostFolderFromPath_Util(Path=@n34, WorldContextObject=@n35, Strip Count=1, __WorldContext=@n35)
   n37 = MakeLiteralString(Value=/Figment_S03/Athena/Environments/Masterials/CliffGround/)
   n38 = Append(A=@n36.AssetName, B=MakeLiteralString(Value=_S03))
   > n39 = DoesAssetExist(AssetPath=Append(A=@n37, B=@n38))
   > n40 = Branch(Condition=@n39)
   n41 = GetAssetTools(execute=@n40.else)
   > n42 = DuplicateAsset(self=@n41, AssetName=@n38, PackagePath=@n37, OriginalObject=@n33)
   > n43 = Cast To MaterialInstanceConstant(Object=@n42)
   > n44 = Branch(Condition="Equal (Object)"(A=Get Parent(self=@n43), B=&Athena_Terrain_Mat_Biome))
   n45 = Get FigmentParentMAt()
   > Cast To MaterialInterface(Object=SelectObject(A=Get FigmentParentMAt(), B=@n43, bSelectA=IsValid(Object=@n45)))
   > SetMaterialInstanceParent(Instance=@n43, NewParent=@n43)
   > SaveLoadedAsset(AssetToSave=@n42, bOnlyIfIsDirty=true)
   > n46 = Cast To BuildingTextureData(Object=@n31)
   > Set OverrideMaterial(OverrideMaterial=@n43, self=@n46)
   > SaveLoadedAsset(AssetToSave=@n46, bOnlyIfIsDirty=false)
   > n47 = Execute Python Script!(
    execute
    PythonScript="import unreal\r\nno_need_to_save = False\r\nTextureDataArray = geo.get_editor_property('TextureData')\r\nprint(TextureDataArray)\r\nif(TextureDataArray[0] == texture_data):\r\n    no_need_to_save = True\r\nelse:\r\n    TextureDataArray[0] = texture_data\r\n    geo.set_editor_property('TextureData', TextureDataArray)"
    geo:"BP  Cliff Figment Object Reference"=@n8
    texture_data:Building Texture Data Object Reference=@n46
   ) -> (then, ReturnValue, no_need_to_save:bool=~) {inputs [geo,texture_data], outputs [no_need_to_save]}
   > n48 = Branch(Condition=@n47.no_need_to_save)
   Add(execute=@n48.else, TargetArray=Get Cliff Actors to Save(), NewItem=@n8)
   n49 = GetPackageForObject(Object=@n43)
   PrintString(
    execute=@n44.else
    InString=Append(A=GetDisplayName(Object=@n49), B="\r\n########################\r\n is not a child of Athena Material\r\n########################################")
   )
   n50 = PrintString(InString=Hello)
   n51 = PrintString(InString=Hello)
   n52 = Branch(Condition="Equal (Object)"(A=@n26, B=Get material_orig()))
   PrintString(execute=@n52.else, InString=Wrong parent)
   PrintString(InString="TD asset doesnt exist, duplicating")
   GetAssetByObjectPath(ObjectPath=None, bIncludeOnlyOnDiskAssets=false)
   Get Dependencies(PackageName=None)
   GetAsset()
   CollectGarbage(execute=@n56.then)
   GetSystemPath(Object=@n9.texture_data)
   PrintString(InString=@n10)
   Append!(A, B=., C:String) -> (ReturnValue)
   Append!(A=@n14, B=MakeLiteralString(Value=&S03)) -> (ReturnValue)
   Cast To MaterialInstanceConstant(Object:Wildcard)
   Execute Python Script!(
    execute
    PythonScript="import unreal\r\nTextureData = geo.get_editor_property('TextureData')[0]\r\nOverrideMaterial = TextureData.get_editor_property('OverrideMaterial')\r\nprint(TextureData)"
    Geo:"BP  Cliff Figment Object Reference"=@n8
   ) -> (then, ReturnValue) {inputs [Geo]}
   Reroute Node(InputPin:Object Reference=@n42) -> (OutputPin:Object Reference)
   Get material_orig()
   n53 = Clear(execute=@n54.then, TargetArray=Get Actors to Revert())
   n54 = Clear(execute=@n5.then, TargetArray=Get Cliff Actors to Save())
   n55 = Get Cliff Actors to Save()
   n56 = SaveLoadedAssets(execute=@n57.then, AssetsToSave=@n55, bOnlyIfIsDirty=true)
   n57 = CheckoutLoadedAssets(execute=@n6.Completed, AssetsToCheckout=@n55)
  ]}
 RebuildTreeMaps {nodes [
   RebuildTreeMaps() {is=title K2Node_FunctionEntry}
   > Clear(TargetMap:Map of Strings to Object References=Get MapStringToTreeBPs() -> (MapStringToTreeBPs:Map of Strings to Object References)) {is=title K2Node_CallFunction}
   > Clear(TargetMap:Map of Strings to Material Interface Object References=Get MapStringToTargetMaterials())
   > n1 = Sequence()
   n2 = "To String (Text)"(InText="GetText (Text Box)"(self=Get TargetTextBox_1() -> (TargetTextBox_1:Editor Utility Editable Text Box Object Reference)))
   n3 = GetAssetRegistry()
   > n4 = GetAssetsByPath(self=@n3, PackagePath=String To Name(InString=Append(A=@n2, B=Materials/)), bRecursive=true, bIncludeOnlyOnDiskAssets=true)
   > n5 = For Each Loop(Array:Array of Asset Data Structures=@n4) -> (Array Element:Asset Data Structure)
   > n6 = Cast To MaterialInterface(Object=GetAsset(InAssetData=@n5))
   > Add(TargetMap=Get MapStringToTargetMaterials(), Key=GetDisplayName(Object=@n6), Value=@n6) {is=title K2Node_CallFunction}
   n7 = GetAssetsByPath(
    execute=@n1.then_1
    self=@n3
    PackagePath=String To Name(InString=Append(A=@n2, B=Blueprints/))
    bRecursive=true
    bIncludeOnlyOnDiskAssets=true
   )
   > n8 = For Each Loop(Array=@n7)
   n9 = GetAsset(InAssetData=@n8)
   > Add(
    TargetMap:Map of Strings to Object References=Get MapStringToTreeBPs()
    Key=GetDisplayName(Object=@n9)
    Value:"Object Reference (by ref)"=@n9
   )
   Return Node!(execute=@n1.then_2)
  ]}
 RecursivelyGetLevels {nodes [
   n1 = RecursivelyGetLevels!() -> (then, WorldArray1:Array of Object References) {is=title K2Node_FunctionEntry}
   > n2 = For Each Loop(Array:Array of Object References=@n1) -> (Array Element:Object Reference)
   > n3 = GetAllActorsOfClass(WorldContextObject=@n2, ActorClass=&BuildingLevelInstance) -> (OutActors:Array of Building Level Instance Object References)
   > n4 = GetWorlds!(execute, self, Level Actors:Array of Building Level Instance Object References=@n3) -> (then, WorldArray2:Array of Object References) {is=title K2Node_CallFunction}
   > n5 = Branch(Condition=Is Empty(TargetArray:Array of Object References=@n4))
   Return Node!(execute=[@n7.then, @n5.then], WorldArray2:Array of Object References=@n1)
   n6 = RecursivelyGetLevels!(execute=@n5.else, self, WorldArray1:Array of Object References=@n4) -> (then, WorldArray2:Array of Object References) {is=title K2Node_CallFunction}
   > n7 = Append Array(TargetArray:Array of Object References=@n1, SourceArray:Array of Object References=@n6)
  ]}
 Remove_appended_TDs {nodes [
   n1 = GetSelectedLevelActors(self=Get EditorActorSubsystem())
   n2 = GetEditorWorld(execute=@n3.then, self=Get UnrealEditorSubsystem())
   n3 = Remove_appended_TDs!() -> (then, Level Object:Object Reference) {is=title K2Node_FunctionEntry}
   n4 = GetAllActorsOfClass(
    execute=@n2.then
    WorldContextObject=SelectObject(A=@n3, B=@n2, bSelectA=IsValid(Object=@n3))
    ActorClass=&BP__Cliff_Figment_C
   ) -> (OutActors:"Array of BP  Cliff Figment Object References")
   n5 = For Each Loop(Exec=@n28.then, Array:"Array of BP  Cliff Figment Object References"=@n4) -> (Array Element:"BP  Cliff Figment Object Reference")
   n6 = Cast To BP__Cliff_Figment(execute=[@n1.then, @n5.LoopBody], Object=@n5)
   > n7 = GetEditorProperty(Object=@n6, PropertyName=TextureData)
   n8 = "Get (a copy)"(Array:Array of Building Texture Data Object References=@n7) -> (Output:Building Texture Data Object Reference)
   n9 = GetDisplayName(Object=GetPackageForObject(Object=@n8))
   > n10 = StripRightMostFolderFromPath_Util(Path=@n9, Strip Count=1)
   > n11 = Set Folder(Folder:String=@n10) -> (Output_Get:String)
   PrintString(execute=[@n26.then, @n11.then], InString=@n9)
   n12 = GetDisplayName(Object=@n8)
   > PrintString(InString=@n12)
   n13 = ParseIntoArray(SourceString=@n12, Delimiter=_)
   > Set Path(Path:Array of Strings=@n13) -> (Output_Get:Array of Strings)
   > Set AppendCount(AppendCount:int=0) -> (Output_Get:int=0)
   n14 = Length(TargetArray:Array of Strings=@n13)
   n15 = int - int(A=@n14)
   > n16 = For Loop with Break!(Execute:Exec, FirstIndex:int, LastIndex:int=@n15, Break:Exec=@n25.else) -> (LoopBody:Exec, Index:int, Completed:Exec) {is=title K2Node_MacroInstance}
   n17 = Get AppendCount() -> (AppendCount:int=0)
   Branch(execute=@n16.Completed, Condition="integer > integer"(A=@n17, B=1))
   > Resize!(execute, TargetArray=Get Path() -> (Path:Array of Strings), Size:int=int - int(A=@n14, B=int - int(A=@n17, B=~))) -> (then) {is=title K2Node_CallArrayFunction}
   n18 = Append(A=JoinStringArray(SourceArray=Get Path(), Separator=_), B=_S03)
   n19 = Append!(A=Get Folder() -> (Folder:String), B=/, C:String=@n18) -> (ReturnValue)
   > PrintString(InString=@n19)
   > n20 = DoesAssetExist(AssetPath=@n19)
   > Branch(Condition=@n20)
   > n21 = GetAssetByObjectPath(self=GetAssetRegistry(), ObjectPath=String To Name(InString=Append(A=@n19, B=., C=@n18)), bIncludeOnlyOnDiskAssets=false)
   > n22 = Cast To BuildingTextureData(Object=GetAsset(InAssetData=@n21))
   > PrintString(InString="exists!")
   > SetEditorProperty(
    Object=@n6
    PropertyName=TextureData
    PropertyValue=Make Array(
     "[0]"=@n22
     "[1]"="Get (a copy)"(Array=@n7, Dimension 1=1)
     "[2]"="Get (a copy)"(Array=@n7, Dimension 1=2)
     "[3]"="Get (a copy)"(Array=@n7, Dimension 1=3)
    )
   )
   > Add(TargetArray=Get Cliff Actors to Save(), NewItem=@n6) {is=title K2Node_CallArrayFunction}
   n23 = int - int(A=@n15, B=@n16)
   n24 = "Get (a copy)"(Array:Array of Strings=@n13, Dimension 1=@n23) -> (Output:String)
   n25 = Branch(execute=@n16.LoopBody, Condition=OR Boolean(A="Equal Exactly (String)"(A=@n24, B=S03), B="Equal Exactly (String)"(A=@n24, B=S02)))
   > Increment Int(Value=Get AppendCount() -> (AppendCount=0))
   n26 = GetPathNameForLoadedAsset(LoadedAsset=@n8)
   "To String (Integer)"(InInt=@n23)
   String To Name()
   JoinStringArray(Separator=" ")
   n27 = Get Cliff Actors to Save()
   SaveLoadedAssets(execute=@n30.then, AssetsToSave=@n27, bOnlyIfIsDirty=true)
   "To String (Integer)"(InInt=@n17)
   Append(B=.)
   Append(B=.)
   "Get (a copy)"(Array:Array of Actor Object References=@n1) -> (Output:Actor Object Reference)
   Get AssetEditorSubsystem()
   n28 = Clear(execute=@n29.then, TargetArray=Get Actors to Revert()) {is=title K2Node_CallArrayFunction}
   n29 = Clear(execute=@n4.then, TargetArray=Get Cliff Actors to Save())
   n30 = CheckoutLoadedAssets(execute=@n5.Completed, AssetsToCheckout=@n27)
  ]}
 Replace Mesh Actor With Existing BP Single {nodes [
   n1 = Replace Mesh Actor With Existing BP Single!() -> (then, Static Mesh Actor:Static Mesh Actor Object Reference, mapOfMeshNamesToBp:Map of Strings to Strings) {is=title K2Node_FunctionEntry}
   > n2 = Replace Static Mesh Actor Impl!(execute, self, Static Mesh Actor:Static Mesh Actor Object Reference=@n1, mapOfMeshNamesToBp:Map of Strings to Strings=@n1.mapOfMeshNamesToBp) -> (then, Replaced:bool, "AsBP  Cliff Figment":"BP  Cliff Figment Object Reference")
   > Return Node!(execute, "AsBP  Cliff Figment":"BP  Cliff Figment Object Reference"=@n2."AsBP  Cliff Figment", Replaced:bool=@n2)
  ]}
 Replace Static Mesh Actor Impl {nodes [
   n1 = Replace Static Mesh Actor Impl!() -> (then, Static Mesh Actor:Static Mesh Actor Object Reference, mapOfMeshNamesToBp:Map of Strings to Strings) {is=title K2Node_FunctionEntry}
   n2 = BreakTransform(InTransform=Get Actor Transform(self=@n1))
   n3 = Get StaticMeshComponent(self=@n1)
   n4 = GetDisplayName(Object=Get StaticMesh(self:Static Mesh Component Object Reference=@n3))
   n5 = Find(TargetMap:Map of Strings to Strings=@n1.mapOfMeshNamesToBp, Key=@n4) -> (Value:String)
   n6 = LoadAsset(AssetPath=@n5)
   n7 = SpawnActorFromObject!(execute, ObjectToUse:Object Reference=@n6, Location:Vector=@n2, Rotation:Rotator=@n2.Rotation, bTransient:bool=false) -> (then, ReturnValue:Actor Object Reference)
   > SetActorScale3D!(execute, self:Actor Object Reference=@n7, NewScale3D:Vector=@n2.Scale) -> (then)
   > n8 = Cast To BP__Cliff_Figment(Object=@n7)
   n9 = GetSelectedLevelActors(execute=@n17.then, self=Get EditorActorSubsystem())
   > n10 = Cast To BP__Cliff_Figment(Object="Get (a copy)"(Array=@n9))
   n11 = GetEditorMaterial(self=@n3, ElementIndex=0)
   n12 = BuildMultiLineMapOfMaterials(execute=@n19.then, Mesh Material=@n11)
   n13 = ResetTextureData!(execute=[@n18.then, @n8.then], self, Actor:Object Reference=@n10, TD asset:Building Texture Data Object Reference=@n12.SingleTD_out) -> (then)
   n14 = DoesAssetExist(execute=@n12.then, AssetPath=@n5)
   Return Node!(execute=[@n15.then, @n13.then], Replaced=@n14, "AsBP  Cliff Figment"=@n10)
   SetActorLabel(self=[@n7, @n8], NewActorLabel=Append!(A=GetDisplayName(Object=@n1), B=_BP) -> (ReturnValue), bMarkDirty=false)
   > n15 = Add(TargetArray:Array of Object References=Get ObjectsToSave(), NewItem:"Object Reference (by ref)"=@n10)
   Add(TargetArray:Array of Actor Object References=Get FoliageActorsToKill(), NewItem:"Actor Object Reference (by ref)"=@n1)
   Add(
    TargetArray:Array of Package Object References=Make Array!("[0]":Package Object Reference) -> (Array:Array of Package Object References)
    NewItem:"Package Object Reference (by ref)"
   )
   EnsureMaterialAndTextureDataExist(Mesh Material=@n11) {is=title K2Node_CallFunction}
   PrintString(execute=@n21.else, InString=Append(A="NO FOUND !!!!!!!!!!!!!!!", B=@n4))
   SpawnActorFromClass!(execute, ActorClass, Location="0, 0, 0", Rotation, bTransient) -> (then, ReturnValue:Actor Object Reference)
   GetPackageForObject()
   Execute Python Script!(execute, PythonScript="import unreal\r\narr = []\r\nfor i in actors:\r\n    arr.append(i.get_package())", Actors:Wildcard) -> (then, ReturnValue, arr:Wildcard) {inputs [Actors], outputs [arr]}
   Branch(Condition=Is Empty(TargetMap=Get MapStringToTargetMaterials()) {is=title K2Node_CallFunction})
   Reroute Node(InputPin:Wildcard) -> (OutputPin:Wildcard)
   Reroute Node()
   Reroute Node(InputPin:Static Mesh Actor Object Reference=@n1) -> (OutputPin:Static Mesh Actor Object Reference)
   n16 = Get EditorActorSubsystem()
   SetSelectedLevelActors(execute=@n20.then, self=@n16, ActorsToSelect=Make Array("[0]":Actor Object Reference=@n1) -> (Array:Array of Actor Object References))
   > n17 = ReplaceSelectedActors!(execute, InAssetPath:String=@n5) -> (then)
   n18 = SelectNothing(execute=@n10.then, self=@n16)
   n19 = PrintString(execute=@n21.then, InString=Append!(A="Found BP for:  ", B=@n4, C:String=@n5) -> (ReturnValue))
   n20 = SelectNothing(execute=[@n6.then, @n22.then], self=@n16)
   n21 = Branch(execute=@n1.then, Condition=@n5.ReturnValue)
   n22 = Branch(execute=@n14.then, Condition=@n14)
  ]}
 ReplaceAllStaticMeshesWithExistingBP {nodes [
   n1 = ReplaceAllStaticMeshesWithExistingBP!() -> (then, WorldContextObject:Object Reference) {is=title K2Node_FunctionEntry}
   > n2 = GetEditorWorld(self=Get UnrealEditorSubsystem())
   > Clear(TargetArray:Array of Object References=Get ObjectsToSave())
   > n3 = Sequence()
   > n4 = BuildMeshMap()
   n5 = GetStaticMeshActorsExclusive!(execute=@n3.then_1, self, WorldContextObject:Object Reference=SelectObject(A=@n1, B=@n2, bSelectA=IsValid(Object=@n1))) -> (then, Result:Array of Actor Object References) {is=title K2Node_CallFunction}
   > n6 = For Each Loop(Array:Array of Actor Object References=@n5) -> (Array Element:Actor Object Reference)
   > n7 = Cast To FortStaticMeshActor(Object=@n6)
   > n8 = Replace Static Mesh Actor Impl(Static Mesh Actor=@n7, mapOfMeshNamesToBp=@n4) {is=title K2Node_CallFunction}
   > Branch(Condition=@n8)
   > Add(TargetArray:Array of Object References=Get ObjectsToSave(), NewItem:"Object Reference (by ref)"=@n8."AsBP  Cliff Figment")
   n9 = Get ObjectsToSave()
   CheckoutLoadedAssets(execute=@n3.then_2, AssetsToCheckout=@n9)
   > SaveLoadedAssets(AssetsToSave=@n9, bOnlyIfIsDirty=false)
  ]}
 ReplaceSelectedStaticMeshesWithExistingBP {nodes [
   n1 = Get ActorsToDestroy() -> (ActorsToDestroy:Array of Static Mesh Actor Object References)
   n2 = Append Array(TargetArray:Array of Static Mesh Actor Object References=@n1, SourceArray:Array of Static Mesh Actor Object References)
   n3 = Execute Python Script(
    execute=[@n2.then, @n17.Completed]
    PythonScript="import unreal\r\narr = []\r\nfor i in actors:\r\n    arr.append(i.get_package())"
    Actors:Array of Static Mesh Actor Object References=@n1
   ) -> (arr:Array of Package Object References) {inputs [Actors], outputs [arr]}
   > DestroyActors(self=Get EditorActorSubsystem(), ActorsToDestroy=@n1)
   > n4 = SavePackages(PackagesToSave=@n3.arr, bOnlyDirty=false)
   n5 = Get ObjectsToSave()
   SaveLoadedAssets(execute=[@n6.then, @n4.then], AssetsToSave=@n5, bOnlyIfIsDirty=false)
   n6 = CheckoutLoadedAssets(AssetsToCheckout=@n5)
   PrintString(InString=print new)
   > n7 = For Each Loop(Array:Array of Object References=Get ObjectsToSave()) -> (Array Element:Object Reference)
   PrintString(execute=@n7.Completed, InString=print des)
   > n8 = For Each Loop(Array:Array of Static Mesh Actor Object References=Get ActorsToDestroy()) -> (Array Element:Static Mesh Actor Object Reference)
   > PrintString(InString=GetDisplayName(Object=@n8))
   PrintString(execute=@n7.LoopBody, InString=GetDisplayName(Object=@n7))
   n9 = Execute Python Script(
    PythonScript="import unreal\r\narr = []\r\nfor i in actors:\r\n    arr.append(i.get_package())"
    Actors:Array of Object References=Get ObjectsToSave()
   ) {inputs [Actors], outputs [arr]}
   > SavePackages(PackagesToSave=@n9.arr, bOnlyDirty=false)
   n10 = ReplaceSelectedStaticMeshesWithExistingBP!() -> (then, WorldContextObject:Object Reference, Array:Array of Actor Object References) {is=title K2Node_FunctionEntry}
   > n11 = GetEditorWorld(self=Get UnrealEditorSubsystem())
   > Clear(TargetArray:Array of Static Mesh Actor Object References=Get ActorsToDestroy())
   > Clear(TargetArray:Array of Object References=Get ObjectsToSave())
   > n12 = Sequence()
   > n13 = BuildMeshMap()
   n14 = Branch(execute=@n12.then_1, Condition=Is Empty(TargetArray:Array of Actor Object References=@n10.Array) {is=title K2Node_CallArrayFunction})
   > n15 = GetStaticMeshActorsExclusive(WorldContextObject=SelectObject(A=@n10, B=@n11, bSelectA=IsValid(Object=@n10)))
   > n16 = Set Mesh actors(Mesh actors:Array of Actor Object References=@n15) -> (Output_Get:Array of Actor Object References)
   n17 = For Each Loop(Exec=[@n20.then, @n16.then], Array:Array of Actor Object References=Get Mesh actors() -> (Mesh actors:Array of Actor Object References)) -> (Array Element:Actor Object Reference)
   > n18 = Cast To FortStaticMeshActor(Object=@n17)
   > n19 = Replace Static Mesh Actor Impl(Static Mesh Actor=@n18, mapOfMeshNamesToBp=@n13)
   > Branch(Condition=@n19)
   n20 = Set Mesh actors(execute=@n14.else, Mesh actors=@n10.Array)
   Add(TargetArray:Array of Package Object References, NewItem:"Package Object Reference (by ref)"=GetPackageForObject(Object=@n18))
   Add(TargetArray:Array of Static Mesh Actor Object References=Get ActorsToDestroy(), NewItem:"Static Mesh Actor Object Reference (by ref)"=@n18)
   SaveLoadedAssets(bOnlyIfIsDirty=false)
   "Get (a copy)"(Array=@n16)
  ]}
 ReplaceSnowMaterials {nodes [
   n1 = GetEditorWorld(self=Get UnrealEditorSubsystem())
   n2 = GetAllActorsOfClass(execute=[@n1.then, @n7.then], WorldContextObject=@n1, ActorClass=&BuildingLevelInstance) -> (OutActors:Array of Building Level Instance Object References)
   > n3 = GetWorlds(Level Actors=@n2)
   n4 = Get worlds() -> (worlds:Array of Object References)
   > Append Array(TargetArray:Array of Object References=@n4, SourceArray:Array of Object References=@n3)
   > n5 = RecursivelyGetLevels(WorldArray1=@n4)
   > n6 = For Each Loop(Array:Array of Object References=@n5) -> (Array Element:Object Reference)
   PrintString(execute=[@n6.LoopBody, @n105.LoopBody], InString=GetDisplayName(Object=@n6))
   n7 = PrintString(InString=Level Actors)
   n8 = Get Paths() -> (Paths:Array of Strings)
   n9 = For Each Loop(Exec=@n46.then, Array:Array of Strings=@n8) -> (Array Element:String)
   n10 = ParseIntoArray(SourceString=@n9, Delimiter=:)
   n11 = "Get (a ref)"(Array=ParseIntoArray(SourceString="Get (a ref)"(Array=@n10, Dimension 1=0), Delimiter=.), Dimension 1=0)
   Add(
    TargetMap:Map of Strings to Integers=Get PathsCount() -> (PathsCount:Map of Strings to Integers)
    Key=@n11
    Value:"Integer (by ref)"=@n9.Array Index
   ) {is=title K2Node_CallFunction}
   > PrintString(InString=@n11)
   n12 = GetAllActorsOfClass(execute=@n23.then, WorldContextObject=@n1, ActorClass=&BP__Cliff_Figment_C) -> (OutActors:"Array of BP  Cliff Figment Object References")
   n13 = For Each Loop(Exec=[@n12.then, @n50.then], Array:"Array of BP  Cliff Figment Object References"=@n12) -> (Array Element:"BP  Cliff Figment Object Reference")
   n14 = Set current actor(execute=@n83.then, current actor:Actor Object Reference=@n13) -> (Output_Get:Actor Object Reference)
   n15 = GetParentFromTD!(execute, self, actor:Actor Object Reference=@n14) -> (
    then
    override:Material Interface Object Reference
    TDArrayOut:Array of Building Texture Data Object References
    Parent:Material Interface Object Reference
   ) {is=title K2Node_CallFunction}
   > Set TDArray(TDArray=@n15.TDArrayOut)
   > Set current parent(current parent:Material Interface Object Reference=@n15.Parent) -> (Output_Get:Material Interface Object Reference)
   > n16 = Set CurrentOverride(CurrentOverride:Material Interface Object Reference=@n15) -> (Output_Get:Material Interface Object Reference)
   n17 = Branch(
    execute=[@n86.then, @n16.then]
    Condition="Equal (Object)"(
     A=Get CurrentOverride() -> (CurrentOverride:Material Interface Object Reference)
     B=/Figment_S07/Athena/Environments/TextureData/CliffGround/TD_Athena_Terrain_Mat_Biome_ForceSnow
    )
   )
   Set Array Elem(
    execute=[@n17.else, @n85.then]
    TargetArray:Array of Building Texture Data Object References=Get TDArray()
    Index=0
    Item:"Building Texture Data Object Reference (by ref)"=&TD_Athena_Terrain_Mat_Biome_ForceSnow
    bSizeToFit=false
   )
   > n18 = SetEditorProperty(Object=Get current actor(), PropertyName=TextureData, PropertyValue=Get TDArray())
   > Branch(Condition=@n18)
   n19 = Get current actor()
   > Add(TargetArray:Array of Object References=Get ObjectsToSave(), NewItem:"Object Reference (by ref)"=@n19) {is=title K2Node_CallArrayFunction}
   > PrintString(InString=GetDisplayName(Object=@n19))
   n20 = For Each Loop(Array:Array of Building Level Instance Object References=@n2) -> (Array Element:Building Level Instance Object Reference)
   n21 = GetDisplayName(Object=Resolve Soft Reference(Input:World Soft Object Reference=GetWorldAsset(self=@n20)) -> (Output:World Object Reference))
   Add(
    TargetMap:Map of Strings to Level Instance Interface Interfaces=Get LevelInstanceActors() -> (LevelInstanceActors:Map of Strings to Level Instance Interface Interfaces)
    Key=@n21
    Value:"Level Instance Interface Interface (by ref)"=@n20
   ) {is=title K2Node_CallFunction}
   > PrintString(InString=@n21)
   Clear(TargetMap:Map of Strings to Level Instance Interface Interfaces=Get LevelInstanceActors()) {is=title K2Node_CallFunction}
   > Clear!(execute, TargetSet:Set of Wildcards) -> (then)
   > n22 = Clear(TargetSet:Set of Actor Object References=Get NewCliffSet() -> (NewCliffSet:Set of Actor Object References))
   n23 = Clear(execute=[@n69.then, @n22.then], TargetArray:Array of Strings=Get Paths()) {is=title K2Node_CallArrayFunction}
   n24 = Get level_instance_cleanup!() -> (level_instance_cleanup)
   n25 = GetActorDescsForActors(InActors=@n24)
   > n26 = For Each Loop(Array:Array of Actor Desc Structures=@n25) -> (Array Element:Actor Desc Structure)
   n27 = Get levelguids() -> (levelguids:Array of Guid Structures)
   UnloadActors!(execute=@n26.Completed, InActorsToUnload:Array of Guid Structures=@n27) -> (then)
   Add(
    execute=@n26.LoopBody
    TargetArray:Array of Guid Structures=@n27
    NewItem:"Guid Structure (by ref)"=Break Actor Desc(ActorDesc=@n26) -> (bIsSpatiallyLoaded=~, bActorIsEditorOnly=~)
   ) {is=title K2Node_CallArrayFunction}
   n28 = GetSelectedLevelActors(self=Get EditorActorSubsystem())
   n29 = "Get (a copy)"(Array=@n28)
   > PrintString(InString=Get Object Path String(Object=@n29))
   n30 = GetAssetRegistry()
   > ScanPathsSynchronous!(
    execute
    self:Asset Registry Interface=@n30
    InPaths:Array of Strings=Make Array("[0]":String=/Figment_S07) -> (Array:Array of Strings)
    bForceRescan:bool=true
    bIgnoreDenyListScanFilters:bool=false
   ) -> (then)
   n31 = MakeLiteralString(Value=/Figment_S07/dummy/dummy)
   > LoadAsset(AssetPath=@n31)
   n32 = GetEditorWorld(execute=@n84.then, self=Get UnrealEditorSubsystem())
   n33 = GetLevels(World=@n32)
   > n34 = For Each Loop(Array:Array of Level Object References=@n33) -> (Array Element:Level Object Reference)
   n35 = CollectGarbage(execute=@n34.Completed)
   FinishLoadingBeforeScreenshot!(execute=[@n35.then, @n121.Completed]) -> (then)
   > ForceGarbageCollection!(execute, bFullPurge:bool=true) -> (then)
   > TakeHighResScreenshot(ResX=0, ResY=0, bMaskEnabled=false, bCaptureHDR=false, ComparisonTolerance=Low, Delay=0, bForceGameView=true)
   PrintString(execute=@n34.LoopBody, InString=GetDisplayName(Object=@n34))
   n36 = GetAssetsByPackageName(self=@n30, PackageName=String To Name(InString=@n31), bIncludeOnlyOnDiskAssets=false, bSkipARFilteredAssets=true)
   n37 = Get Asset By Object Path!(
    execute:Exec
    self:Object Reference=Self-Reference()
    ObjectPath:"Soft Object Path Structure (by ref)"=MakeSoftObjectPath!(PathString:String=@n31) -> (ReturnValue:Soft Object Path Structure)
    bIncludeOnlyOnDiskAssets:bool=false
    bSkipARFilteredAssets:bool=true
   ) -> (then:Exec, ReturnValue:Asset Data Structure) {is=title K2Node_Message}
   n38 = GetAsset(InAssetData=@n37)
   PrintString(execute=[@n36.then, @n72.then, @n37.then], InString=GetDisplayName(Object=@n38))
   n39 = GetSelectedLevelActors(self=Get EditorActorSubsystem())
   > n40 = Cast To LevelInstanceInterface(Object="Get (a copy)"(Array=@n39))
   n41 = Cast To LevelInstance!(Object:Level Instance Interface Interface=@n40) -> (AsLevel Instance:Level Instance Object Reference, bSuccess:bool=~) {is=title K2Node_DynamicCast}
   > LevelInstanceEnterEditMode(LevelInstance=@n41)
   n42 = Get FoliageActorsToKill()
   n43 = GetActorDescsForActors(InActors=@n42)
   n44 = GetActorDescs()
   n45 = For Each Loop(Exec=[@n43.then, @n44.then], Array:Array of Actor Desc Structures=@n44) -> (Array Element:Actor Desc Structure)
   n46 = SortStringArray!(execute=[@n13.Completed, @n45.Completed], TargetArray:Array of Strings=@n8, bStableSort:bool=true, SortOrder:EArraySortOrder Enum=Ascending) -> (then)
   n47 = Break Actor Desc(ActorDesc=@n45) -> (bIsSpatiallyLoaded=~, bActorIsEditorOnly=~)
   n48 = "To String (Name)"(InName=@n47.ActorPath)
   n49 = Get Paths()
   Branch(
    execute=@n45.LoopBody
    Condition="Not Equal (Integer)"(A=Find Item(TargetArray:Array of Strings=@n49, ItemToFind:"String (by ref)"=@n48), B=-1)
   )
   > PrintString(InString=Append!(A="Found: ", B="To String (Name)"(InName=@n47.ActorPackage)) -> (ReturnValue))
   > Add(
    TargetMap:Map of Strings to Guid Structures=Get stringToGuids() -> (stringToGuids:Map of Strings to Guid Structures)
    Key=@n48
    Value:"Guid Structure (by ref)"=@n47
   ) {is=title K2Node_CallFunction}
   SaveLoadedAssets(AssetsToSave=@n42, bOnlyIfIsDirty=false)
   > CheckoutLoadedAssets(AssetsToCheckout=@n42)
   n50 = Clear!(execute, TargetMap:Map of Strings to Guid Structures=Get stringToGuids()) -> (then) {is=title K2Node_CallFunction}
   n51 = Get stringToGuids()
   Values(TargetMap:Map of Strings to Guid Structures=@n51) -> (Values:Array of Guid Structures)
   > n52 = "For Each Loop (Map)"(MapPin:Map of Strings to Guid Structures=@n51) -> (ValuePin:Guid Structure)
   n53 = For Loop(execute=@n81.then, FirstIndex=Get LoopStart() -> (LoopStart:int=0), LastIndex=Get LoopEnd() -> (LoopEnd:int=0))
   n54 = LoadAsset(execute=[@n53.LoopBody, @n52.then], AssetPath="Get (a ref)"(Array=Get Paths(), Dimension 1=@n53))
   > n55 = Cast To BP__Cliff_Figment(Object=@n54)
   > n56 = GetParentFromTD(actor=@n55)
   > Set TDArray(TDArray=@n56.TDArrayOut)
   > n57 = Branch(Condition="Equal (Object)"(A=@n56, B=&TD_Athena_Terrain_Mat_Biome_ForceSnow))
   Set Array Elem(
    execute=@n57.else
    TargetArray=Get TDArray()
    Index=0
    Item=&TD_Athena_Terrain_Mat_Biome_ForceSnow
    bSizeToFit=false
   )
   > SetEditorProperty(Object=@n55, PropertyName=TextureData, PropertyValue=Get TDArray())
   > n58 = SaveLoadedAsset(AssetToSave=@n55, bOnlyIfIsDirty=false)
   n59 = Get Paths_SetArray() -> (Paths_SetArray:Array of Strings)
   n60 = Last Index(TargetArray=@n59) -> (ReturnValue=0)
   n61 = For Loop(execute=[@n93.Completed, @n9.Completed], LastIndex=@n60)
   n62 = "Get (a ref)"(Array=@n59, Dimension 1=@n61)
   n63 = ParseIntoArray(SourceString=@n62, Delimiter=/)
   n64 = "Get (a ref)"(Array=@n63, Dimension 1=Last Index(TargetArray=@n63) -> (ReturnValue=0))
   n65 = Get Asset By Object Path(
    execute=@n146.then
    self=GetAssetRegistry()
    ObjectPath=MakeSoftObjectPath(PathString=Append!(A=@n62, B=., C:String=@n64) -> (ReturnValue))
    bIncludeOnlyOnDiskAssets=false
    bSkipARFilteredAssets=true
   ) {is=title K2Node_CallFunction}
   n66 = GetAsset(InAssetData=@n65)
   n67 = GetDisplayName(Object=@n66)
   n68 = Cast To LevelInstance(Object=Find(TargetMap:Map of Strings to Level Instance Interface Interfaces=Get LevelInstanceActors(), Key=@n67) -> (Value:Level Instance Interface Interface)) -> (bSuccess=~)
   LevelInstanceCommit(execute=@n58.then, LevelInstance=@n68, bDiscardChanges=true)
   Add Items(TargetSet:"Set of BP  Cliff Figment Object References", NewItems:"Array of BP  Cliff Figment Object References"=@n12)
   > Clear(TargetArray:Array of Integers=Get PathsCountInt() -> (PathsCountInt:Array of Integers)) {is=title K2Node_CallArrayFunction}
   > Clear(TargetMap:Map of Strings to Integers=Get PathsCount()) {is=title K2Node_CallFunction}
   > Clear(TargetArray:Array of Strings=Get Paths_SetArray()) {is=title K2Node_CallArrayFunction}
   > Clear(TargetArray=Get Paths_KeepCount() -> (Paths_KeepCount:Array of Strings))
   > Clear!(execute, TargetSet:Set of Strings=Get LevelInstances() -> (LevelInstances:Set of Strings)) -> (then) {is=title K2Node_CallFunction}
   > n69 = Clear(TargetArray:Array of Actor Object References=Get FoliageActorsToKill()) {is=title K2Node_CallArrayFunction}
   n70 = GetActorDescs()
   > n71 = For Each Loop(Array=@n70)
   n72 = Cast To Car_DEFAULT!(execute, Object=@n38) -> (then, CastFailed, AsCar DEFAULT:Car DEFAULT Object Reference) {is=title K2Node_DynamicCast}
   n73 = GetAssetRegistry()
   n74 = Get levels() -> (levels:Array of Strings)
   n75 = For Each Loop(Exec=@n143.then, Array:Array of Strings=@n74) -> (Array Element:String)
   > n76 = Get Asset By Object Path(self=@n73, ObjectPath=MakeSoftObjectPath(PathString=@n75), bIncludeOnlyOnDiskAssets=false, bSkipARFilteredAssets=true)
   n77 = GetDisplayName(Object=GetAsset(InAssetData=@n76))
   n78 = Find(TargetMap:Map of Strings to Building Level Instance Object References=Get worldActors!() -> (worldActors), Key=@n77) -> (Value:Building Level Instance Object Reference)
   n79 = Cast To LevelInstanceInterface(Object=@n78)
   > UnloadLevelInstance(self=[@n78, @n79])
   > CollectGarbage()
   n80 = Cast To BP__Cliff_Figment(Object=@n66)
   PrintString(execute=[@n65.then, @n80.then], InString=@n67)
   > n81 = LevelInstanceEnterEditMode(LevelInstance=@n68)
   n82 = Sequence!(execute) -> (then_0, then_1)
   n83 = Cast To BP__Cliff_Figment(execute=[@n106.then, @n82.then_0, @n13.LoopBody], Object=@n13)
   n84 = ReplaceSnowMaterials() {is=title K2Node_FunctionEntry}
   n85 = Branch(Condition="Equal (Object)"(
     A=Get current parent() -> (current parent:Material Interface Object Reference)
     B=&Athena_Terrain_Mat_Biome_Inst
    ))
   n86 = Branch(Condition=Get bIsSnow() -> (bIsSnow:bool))
   n87 = "Equal (Integer)"(A=@n60, B=@n61)
   n88 = "Min (Integer)"(A=int + int(A=@n61), B=@n60) -> (ReturnValue=0)
   n89 = Get PathsCountInt()
   n90 = SelectInt(
    A=int - int(A=Last Index(TargetArray=@n8) -> (ReturnValue=0))
    B=int - int(A="Get (a ref)"(Array:Array of Integers=@n89, Dimension 1=@n88) -> (Output:"Integer (by ref)"))
    bPickA=@n87
   )
   n91 = "Get (a ref)"(Array=@n89, Dimension 1=@n61)
   n92 = For Loop(FirstIndex=@n91, LastIndex=@n90)
   > PrintString(InString="Get (a ref)"(Array:Array of Strings=Get Paths(), Dimension 1=@n92) -> (Output:"String (by ref)"))
   n93 = For Each Loop(Array=@n59)
   n94 = "Get (a ref)"(Array:Array of Integers=Get PathsCountInt(), Dimension 1=@n93.Array Index) -> (Output:"Integer (by ref)")
   > PrintString(InString=Append(A=@n93, B=:, C="To String (Integer)"(InInt=@n94)))
   n95 = For Each Loop(Array:Array of Building Level Instance Object References=@n2) -> (Array Element:Building Level Instance Object Reference)
   > PrintString(InString=GetDisplayName(Object=@n95))
   n96 = For Each Loop(Array:Array of Strings=Get pathslocal() -> (pathslocal:Array of Strings)) -> (Array Element:String)
   CollectGarbage(execute=@n96.Completed)
   > SaveLoadedAssets(AssetsToSave=Get ObjectsToSave(), bOnlyIfIsDirty=false)
   n97 = LoadAsset(execute=@n96.LoopBody, AssetPath=@n96)
   > n98 = Cast To BP__Cliff_Figment(Object=@n97)
   > n99 = GetParentFromTD(actor=@n98)
   > Set Array Elem(
    TargetArray=@n99.TDArrayOut
    Index=0
    Item=&TD_Athena_Terrain_Mat_Biome_ForceSnow
    bSizeToFit=false
   )
   > SetEditorProperty(Object=@n98, PropertyName=TextureData, PropertyValue=@n99.TDArrayOut)
   > Add(TargetArray:Array of Object References=Get ObjectsToSave(), NewItem:"Object Reference (by ref)"=@n98) {is=title K2Node_CallArrayFunction}
   n100 = For Each Loop(Array:Array of Building Level Instance Object References=@n24) -> (Array Element:Building Level Instance Object Reference)
   > n101 = Cast To LevelInstanceInterface(Object=@n100)
   > LoadLevelInstance(self:Level Instance Interface Interface=@n101)
   n102 = GetLevels(execute=@n32.then, World=@n32)
   n103 = For Each Loop(Array:Array of Level Object References=@n102) -> (Array Element:Level Object Reference)
   > PrintString(InString=Get Object Path String(Object=@n103))
   n104 = For Each Loop(Array:Array of Object References) -> (Array Element:Object Reference)
   > PrintString(InString=GetDisplayName(Object=@n104))
   n105 = For Each Loop(Array:Array of World Soft Object References) -> (Array Element:World Soft Object Reference)
   Set bIsSnow(bIsSnow:bool=false) -> (Output_Get:bool)
   > Clear(TargetArray:Array of Object References=Get ObjectsToSave())
   > n106 = GetSelectedLevelActors(self=Get EditorActorSubsystem())
   Break Actor Desc(ActorDesc=@n71) -> (bIsSpatiallyLoaded=~, bActorIsEditorOnly=~)
   Break Asset Data(AssetData=@n65)
   n107 = GetAllActorsOfClass(execute=@n119.Completed, ActorClass=&BP__Cliff_Figment_C)
   > n108 = For Each Loop(Array:"Array of BP  Cliff Figment Object References"=@n107) -> (Array Element:"BP  Cliff Figment Object Reference")
   n109 = RightChop!(SourceString:String=Get Object Path String(Object=@n108), Count:int=5) -> (ReturnValue:String)
   n110 = FindSubstring(SearchIn=@n109, Substring=_LevelInstance_)
   n111 = int + int(A=FindSubstring(SearchIn=@n109, Substring=_0.), B=2)
   Add(
    execute=@n152.then
    TargetArray:Array of Strings=Get pathslocal()
    NewItem:"String (by ref)"=Append!(A=Left!(SourceString:String=@n109, Count:int=@n110) -> (ReturnValue:String), B=RightChop(SourceString=@n109, Count=@n111)) -> (ReturnValue)
   )
   Add(execute=@n127.then, TargetArray:Array of Integers=Get PathsCountInt(), NewItem:"Integer (by ref)"=@n9.Array Index)
   n112 = Get pathslocal()
   n113 = For Each Loop(Exec=@n144.then, Array:Array of Strings=@n112) -> (Array Element:String)
   Add Unique(
    execute=@n142.then
    TargetArray=Get levels()
    NewItem="Get (a ref)"(Array:Array of Strings=ParseIntoArray(SourceString=@n113, Delimiter=:), Dimension 1=0) -> (Output:"String (by ref)")
   )
   n114 = For Each Loop(Exec=@n75.Completed, Array=@n74)
   > n115 = Get Asset By Object Path(self=@n73, ObjectPath=MakeSoftObjectPath(PathString=@n114), bIncludeOnlyOnDiskAssets=false, bSkipARFilteredAssets=true)
   > n116 = GetAllActorsOfClass(WorldContextObject=GetAsset(InAssetData=@n115), ActorClass=&BP__Cliff_Figment_C)
   > n117 = For Each Loop(Array:"Array of BP  Cliff Figment Object References"=@n116) -> (Array Element:"BP  Cliff Figment Object Reference")
   Add(execute=@n137.then, TargetArray:Array of Object References=Get ObjectsToSave(), NewItem:"Object Reference (by ref)"=@n117)
   Add(
    execute=@n141.then
    TargetArray:Array of Building Level Instance Object References=Get level_instance_cleanup()
    NewItem:"Building Level Instance Object Reference (by ref)"=@n78
   )
   Add(TargetArray:Array of Actor Object References=Get FoliageActorsToKill(), NewItem:"Actor Object Reference (by ref)")
   Add(TargetArray=Get FoliageActorsToKill(), NewItem=@n55)
   MakeLiteralString(Value=../../../FortniteGame/Plugins/GameFeatures/Figment/Figment_S07/Content/tmp/tmp.uasset)
   LevelInstanceEnterEditMode()
   GetDisplayName(Object=@n117)
   Get Object Path String(Object=@n108)
   GetSystemPath(Object=@n29)
   n118 = GetAllActorsOfClass(execute=@n134.then, ActorClass=&BuildingLevelInstance) -> (OutActors:Array of Building Level Instance Object References)
   > n119 = For Each Loop(Array:Array of Building Level Instance Object References=@n118) -> (Array Element:Building Level Instance Object Reference)
   > Add(
    TargetMap:Map of Strings to Building Level Instance Object References=Get worldActors()
    Key=GetSoftPtrAssetName(Asset=GetWorldAsset(self=@n119))
    Value:"Building Level Instance Object Reference (by ref)"=@n119
   ) {is=title K2Node_CallFunction}
   MakeLiteralString(Value=D:/daniel.elliott_FN40_MS/FortniteGame/Plugins/GameFeatures/Figment/Figment_S07_LevelInstances/Content/__ExternalActors__/Streaming/New/Sublevel_X3Y4/D/OL/7IR0ETBPUZ2WNLSSKSF03M.uasset)
   SaveLoadedAsset(AssetToSave=@n117, bOnlyIfIsDirty=false)
   CollectGarbage()
   n120 = Get ObjectsToSave()
   SaveLoadedAssets(AssetsToSave=@n120, bOnlyIfIsDirty=false)
   Get Actor Transform(self=@n78)
   UnloadLevelInstance(self=@n101)
   n121 = For Each Loop(Exec=@n114.Completed, Array=@n24)
   > n122 = Cast To LevelInstanceInterface(Object=@n121)
   LoadLevelInstance(self=@n122)
   UnloadLevelInstance(execute=@n122.then, self=@n122)
   UnloadPrimaryAsset!(
    execute
    PrimaryAssetId:Primary Asset Id Structure=GetPrimaryAssetIdFromObject!(Object:Object Reference=@n121) -> (ReturnValue:Primary Asset Id Structure)
   ) -> (then)
   Remove Level From World!(execute, InLevel:Level Object Reference, bClearSelection:bool=true, bResetTransactionBuffer:bool=true) -> (then, ReturnValue:bool)
   LoadAsset(AssetPath=/Figment_S07_Map/Athena_Terrain_S07)
   GetDisplayName(Object=@n103)
   ResetLevelAsset!(execute, World:World Object Reference=@n32) -> (then)
   PrintString(InString=Finished unloading')
   PrintString(InString=@n48)
   FlushLevelStreaming!(execute, WorldContextObject:Object Reference) -> (then)
   IsLoaded(self=@n122)
   n123 = Break Asset Data(AssetData="Get (a copy)"(Array:Array of Asset Data Structures=@n36) -> (Output:Asset Data Structure))
   "To String (Name)"(InName=@n123.PackagePath)
   "To String (Name)"(InName=@n123)
   ReloadPackages!(execute, PackagesToReload:Array of Package Object References, InteractionMode:EReloadPackagesInteractionMode Enum=Interactive) -> (then, bOutAnyPackagesReloaded:bool, OutErrorMessage:Text)
   GetDisplayName(Object=@n72)
   "To Object (Interface)"(Interface=@n30)
   LoadLevelInstance(self=@n40)
   LevelInstanceEnterEditMode(LevelInstance=@n41)
   PrintString(InString=Hello)
   NewLevel!(execute, self:Level Editor Subsystem Object Reference, AssetPath:String=/Figment_S07/dummy, bIsPartitionedWorld:bool=false) -> (then, ReturnValue:bool)
   SaveAsset(AssetToSave=Get Object Path String(), bOnlyIfIsDirty=false)
   Get Class Display Name(Class=GetClass(InAssetData=@n65))
   SavePackages(
    PackagesToSave=Execute Python Script(PythonScript="import unreal\r\narr = []\r\nfor i in actors:\r\n    arr.append(i.get_package())", Actors=@n120) {inputs [Actors], outputs [arr]}.arr
    bOnlyDirty=false
   )
   PrintString(execute=@n20.LoopBody, InString=GetDisplayName(Object=@n20))
   RightChop(SourceString=Get Object Path String(Object=@n14), Count=5)
   CollapseRelativeDirectories!(InPath:String=ConvertToRelativePath!(Filename:String=MakeLiteralString(Value=D:/daniel.elliott_FN40_MS/FortniteGame/Plugins/GameFeatures/Figment/Figment_S07/Content/dummy/dummy)) -> (ReturnValue:String)) -> (OutPath:String, ReturnValue:bool)
   LogString(execute=@n131.then, InString=GetDisplayName(Object=GetPackageForObject(Object=@n14)))
   n124 = Get current parent()
   AND Boolean(A=Contains(SearchIn=GetDisplayName(Object=@n124), Substring=S7Snow), B=NOT Boolean(A=Get bIsSnow()))
   Append(A=YES, B=@n77)
   Resolve Soft Reference(Input=@n105)
   Execute Python Script!(
    execute
    PythonScript="import unreal\r\n\r\nar = unreal.AssetRegistryHelpers.get_asset_registry()\r\nar.scan_paths_synchronous([\"/Figment_S07\"], force_rescan=True)\r\n\r\nasset = unreal.EditorAssetLibrary.load_asset(\"/TempHistory/Foo/MyAsset\")\r\nunreal.log(asset)"
    None:Wildcard
   ) -> (then, ReturnValue) {inputs [None]}
   Get LevelEditorSubsystem()
   Get LevelInstanceSubsystem()
   Get WorldPartitionSubsystem()
   Branch()
   Branch(Condition=@n87)
   Reroute Node(InputPin:Wildcard) -> (OutputPin:Wildcard)
   Reroute Node(InputPin:Level Object Reference="Get (a copy)"(Array:Array of Level Object References=@n102) -> (Output:Level Object Reference)) -> (OutputPin:Level Object Reference)
   n125 = Get SourceTextBox()
   For Each LI(InputPin="To String (Text)"(InText="GetText (Text Box)"(self=@n125)))
   While Loop!(execute:Exec, Condition:bool=Get bBreakRecursion() -> (bBreakRecursion:bool)) -> (LoopBody:Exec, Completed:Exec) {is=title K2Node_MacroInstance}
   Make Array!(
    "[0]":Actor Class Reference=/Game/Building/ActorBlueprints/Prop/Car_DEFAULT.Car_DEFAULT_C
    "[1]":Actor Class Reference=&BP__Cliff_Figment_C
   ) -> (Array:Array of Actor Class References)
   "Equal (Class)"(A=Get Class() -> (ReturnValue:Object Class Reference), B=&Car_DEFAULT_C)
   Select(Option 0:Wildcard, Option 1:Wildcard, Index:Index Wildcard) -> (ReturnValue:Wildcard)
   Get ObjectsToSave()
   Get CurrentOverride()
   Get current parent()
   Get contexts() -> (contexts:Array of Object References)
   Get Paths()
   Get contexts()
   Get worlds()
   Get TargetPAth()
   Set bIsSnow(bIsSnow=false)
   n126 = Add Unique(execute=@n9.LoopBody, TargetArray=Get Paths_SetArray(), NewItem=@n11)
   > n127 = Branch(Condition="Not Equal (Integer)"(A=@n126, B=-1))
   n128 = RightChop(SourceString=Get Object Path String(Object=@n14), Count=5)
   n129 = FindSubstring(SearchIn=@n128, Substring=_0.)
   n130 = Append(
    A=Left(SourceString=@n128, Count=FindSubstring(SearchIn=@n128, Substring=_LevelInstance_))
    B=RightChop(SourceString=@n128, Count=int + int(A=@n129, B=2))
   )
   Add(execute=@n140.then, TargetArray:Array of Strings=Get Paths(), NewItem:"String (by ref)"=@n130) {is=title K2Node_CallArrayFunction}
   > n131 = LogString(InString=@n130)
   n132 = Get Object Path String(Object=@n117)
   n133 = Find Item(TargetArray=Get pathslocal(), ItemToFind=@n132)
   n134 = Clear(execute=@n102.then, TargetArray=Get ObjectsToSave())
   n135 = GetParentFromTD(execute=@n145.then, actor=@n117)
   > Set Array Elem(
    TargetArray=@n135.TDArrayOut
    Index=0
    Item=&TD_Athena_Terrain_Mat_Biome_ForceSnow
    bSizeToFit=false
   )
   > n136 = SetEditorProperty(Object=@n117, PropertyName=TextureData, PropertyValue=@n135.TDArrayOut)
   > n137 = Branch(Condition=@n136)
   n138 = GetParentFromTD(execute=@n150.then, actor=@n14)
   > n139 = Branch(Condition="Equal (Object)"(A=@n138, B=&TD_Athena_Terrain_Mat_Biome_ForceSnow))
   n140 = Branch(
    execute=@n139.else
    Condition="Equal (Object)"(A=@n138.Parent, B=&Athena_Terrain_Mat_Biome_Inst)
   )
   PrintString(execute=@n76.then, InString=@n77)
   > n141 = Branch(Condition=@n78.ReturnValue)
   n142 = PrintString(execute=@n113.LoopBody, InString=@n113)
   n143 = SortStringArray(execute=@n147.then, TargetArray=@n74, bStableSort=true, SortOrder=Ascending)
   n144 = SortStringArray(execute=@n108.Completed, TargetArray=@n112, bStableSort=true, SortOrder=Ascending)
   n145 = PrintString(execute=@n155.then, InString=Append(A="    ", B=@n132))
   n146 = PrintString(execute=@n156.then, InString=Append!(A="To String (Integer)"(InInt=@n91), B=-, C:String="To String (Integer)"(InInt=@n90)) -> (ReturnValue))
   n147 = PrintString(execute=@n113.Completed, InString=levls)
   n148 = GetDisplayName(Object=@n108)
   n149 = IsInSnowBiome!(execute=@n154.then_0, self, actor:Actor Object Reference=@n14) -> (then, ReturnValue:bool) {is=title K2Node_CallFunction}
   > Set bIsSnow(bIsSnow=true)
   > n150 = Branch(Condition=@n149)
   n151 = IsInSnowBiome(execute=@n153.then, actor=@n108)
   > n152 = Branch(Condition=@n151)
   n153 = PrintString(execute=@n108.LoopBody, InString=@n148)
   n154 = Sequence(execute=@n14.then)
   n155 = Branch(execute=@n117.LoopBody, Condition="Not Equal (Integer)"(A=@n133, B=-1))
   Set LoopStart(execute=@n61.LoopBody, LoopStart:int=@n91) -> (Output_Get:int=0)
   > n156 = Set LoopEnd(LoopEnd:int=@n90) -> (Output_Get:int=0)
  ]}
 ReplaceStaticMeshesCombined {nodes [
   DestroyActors(self=Get EditorActorSubsystem(), ActorsToDestroy=Get FoliageActorsToKill())
   > SavePackages(
    PackagesToSave=Execute Python Script!(
     execute
     PythonScript="import unreal\r\narr = []\r\nfor i in actors:\r\n    arr.append(i.get_package())"
     Actors:Array of Actor Object References=Get FoliageActorsToKill()
    ) -> (then, ReturnValue, arr:Array of Package Object References) {inputs [Actors], outputs [arr]}.arr
    bOnlyDirty=false
   )
   ReplaceStaticMeshesCombined() {is=title K2Node_FunctionEntry}
   > Clear(TargetArray=Get ObjectsToSave())
   > n1 = BuildMeshMap()
   > Set MeshMap(MeshMap:Map of Strings to Strings=@n1) -> (Output_Get:Map of Strings to Strings)
   > n2 = Switch on E_ApplyContext(Selection=Get Apply Mode() -> (Apply Mode=NewEnumerator2))
   n3 = "To String (Text)"(InText=Get Text(self=Get SourceTextBox()))
   For Each LI Full!(execute:Exec=@n2.NewEnumerator0, InputPin:String=@n3) -> (then:Exec) {is=title K2Node_MacroInstance}
   > n4 = GetEditorWorld(self=Get UnrealEditorSubsystem())
   > n5 = GetStaticMeshActorsExclusive(WorldContextObject=@n4)
   > n6 = Set Actors!(execute, Actors:Array of Actor Object References=@n5) -> (then, Output_Get:Array of Actor Object References)
   Clear(execute=[@n6.then, @n15.then, @n13.then], TargetArray=Get ObjectsToSave())
   > Clear(TargetArray:Array of Actor Object References=Get FoliageActorsToKill())
   > n7 = For Each Loop(Array:Array of Actor Object References=Get Actors() -> (Actors:Array of Actor Object References)) -> (Array Element:Actor Object Reference)
   n8 = Get ObjectsToSave()
   SaveLoadedAssets(execute=@n7.Completed, AssetsToSave=@n8, bOnlyIfIsDirty=false)
   > CheckoutLoadedAssets(AssetsToCheckout=@n8)
   n9 = Cast To FortStaticMeshActor(execute=@n7.LoopBody, Object=@n7)
   > n10 = Replace Static Mesh Actor Impl(Static Mesh Actor=@n9, mapOfMeshNamesToBp=Get MeshMap() -> (MeshMap:Map of Strings to Strings))
   > Branch(Condition=@n10)
   > Add(TargetArray:Array of Object References=Get ObjectsToSave(), NewItem:"Object Reference (by ref)"=@n7)
   n11 = GetEditorWorld(execute=@n2.NewEnumerator1, self=Get UnrealEditorSubsystem())
   > n12 = GetStaticMeshActorsExclusive(WorldContextObject=@n11)
   > n13 = Set Actors(Actors=@n12)
   n14 = GetSelectedLevelActors(execute=@n2.NewEnumerator2, self=Get EditorActorSubsystem())
   > n15 = Set Actors(Actors=@n14)
   Sequence()
   Reroute Node(InputPin:"BP  Cliff Figment Object Reference"=@n10."AsBP  Cliff Figment") -> (OutputPin:"BP  Cliff Figment Object Reference")
   For Each LI(InputPin=@n3)
  ]}
 ResetTextureData {nodes [
   n1 = ResetTextureData!() -> (then, Actor:Object Reference, TD asset:Building Texture Data Object Reference) {is=title K2Node_FunctionEntry}
   > n2 = Cast To BP__Cliff_Figment(Object=@n1)
   > n3 = GetEditorProperty(Object=@n2, PropertyName=TextureData)
   > n4 = Branch(Condition=@n3.ReturnValue)
   > SetEditorProperty(
    Object=@n1
    PropertyName=TextureData
    PropertyValue=Make Array!(
     "[0]":Building Texture Data Object Reference=@n1.TD asset
     "[1]":Building Texture Data Object Reference="Get (a copy)"(Array:Array of Building Texture Data Object References=@n3, Dimension 1=1) -> (Output:Building Texture Data Object Reference)
     "[2]":Building Texture Data Object Reference="Get (a copy)"(Array=@n3, Dimension 1=2)
     "[3]":Building Texture Data Object Reference="Get (a copy)"(Array=@n3, Dimension 1=3)
    ) -> (Array:Array of Building Texture Data Object References)
   )
   n5 = Get NewLocalVar_0() -> (NewLocalVar_0:Building Texture Data Object Reference)
   SetEditorProperty(execute=@n4.else, Object=@n1, PropertyName=TextureData, PropertyValue=Make Array("[0]"=@n1.TD asset, "[1]"=@n5, "[2]"=@n5, "[3]"=@n5))
   PrintString(execute=@n2.CastFailed, InString="not cliff figment!")
   Get data()
   Get CurrentTDForStaticMesh()
  ]}
 SetButtonOutlineWidth {nodes [
   n1 = SetButtonOutlineWidth!() -> (then, InputPin:double=~, InputPin2:Editor Utility Button Object Reference) {is=title K2Node_FunctionEntry}
   n2 = Break Button Style!(ButtonStyle:"Button Style Structure (by ref)"=Get WidgetStyle!(self:Button Object Reference=@n1.InputPin2) -> (WidgetStyle:Button Style Structure)) -> (
    Normal:Slate Brush Structure
    Hovered:Slate Brush Structure
    Pressed:Slate Brush Structure
    Disabled:Slate Brush Structure
    NormalForeground:Slate Color Structure
    HoveredForeground:Slate Color Structure
    PressedForeground:Slate Color Structure
    DisabledForeground:Slate Color Structure
    NormalPadding:Margin Structure
    PressedPadding:Margin Structure
    PressedSlateSound:Slate Sound Structure
    ClickedSlateSound:Slate Sound Structure
    HoveredSlateSound:Slate Sound Structure
   ) {is=title K2Node_BreakStruct}
   n3 = Break Slate Brush!(SlateBrush:"Slate Brush Structure (by ref)"=@n2) -> (
    TintColor:Slate Color Structure
    DrawAs:ESlateBrushDrawType Enum
    Tiling:ESlateBrushTileType Enum
    Mirroring:ESlateBrushMirrorType Enum
    ImageSize:Vector 2D Structure
    Margin:Margin Structure
    ResourceObject:Object Reference
    OutlineSettings:Slate Brush Outline Settings Structure
   ) {is=title K2Node_BreakStruct}
   n4 = Break Slate Brush Outline Settings!(SlateBrushOutlineSettings:"Slate Brush Outline Settings Structure (by ref)"=@n3.OutlineSettings) -> (
    CornerRadii:Vector 4 Structure
    Color:Slate Color Structure
    Width:float=~
    RoundingType:ESlateBrushRoundingType Enum
    bUseBrushTransparency:bool=~
   ) {is=title K2Node_BreakStruct}
   n5 = Make Slate Brush Outline Settings!(
    CornerRadii:Vector 4 Structure=@n4
    Color:Slate Color Structure=@n4.Color
    Width:float=@n1
    RoundingType:ESlateBrushRoundingType Enum=@n4.RoundingType
    bUseBrushTransparency:bool=@n4.bUseBrushTransparency
   ) -> (SlateBrushOutlineSettings:Slate Brush Outline Settings Structure) {is=title K2Node_MakeStruct}
   > SetStyle!(
    execute
    self:Button Object Reference=@n1.InputPin2
    InStyle:"Button Style Structure (by ref)"=Make Button Style!(
     Normal:Slate Brush Structure=Make Slate Brush!(
      TintColor:Slate Color Structure=@n3
      DrawAs:ESlateBrushDrawType Enum=@n3.DrawAs
      Tiling:ESlateBrushTileType Enum=@n3.Tiling
      Mirroring:ESlateBrushMirrorType Enum=@n3.Mirroring
      ImageSize:Vector 2D Structure=@n3.ImageSize
      Margin:Margin Structure=@n3.Margin
      ResourceObject:Object Reference=@n3.ResourceObject
      OutlineSettings:Slate Brush Outline Settings Structure=@n5
     ) -> (SlateBrush:Slate Brush Structure) {is=title K2Node_MakeStruct}
     Hovered:Slate Brush Structure=@n2.Hovered
     Pressed:Slate Brush Structure=@n2.Pressed
     Disabled:Slate Brush Structure=@n2.Disabled
     NormalForeground:Slate Color Structure=@n2.NormalForeground
     HoveredForeground:Slate Color Structure=@n2.HoveredForeground
     PressedForeground:Slate Color Structure=@n2.PressedForeground
     DisabledForeground:Slate Color Structure=@n2.DisabledForeground
     NormalPadding:Margin Structure=@n2.NormalPadding
     PressedPadding:Margin Structure=@n2.PressedPadding
     PressedSlateSound:Slate Sound Structure=@n2.PressedSlateSound
     ClickedSlateSound:Slate Sound Structure=@n2.ClickedSlateSound
     HoveredSlateSound:Slate Sound Structure=@n2.HoveredSlateSound
    ) -> (ButtonStyle:Button Style Structure) {is=title K2Node_MakeStruct}
   ) -> (then)
  ]}
 StripForwardSlash {nodes [
   n1 = StripForwardSlash!() -> (then, SourceString:String) {is=title K2Node_FunctionEntry}
   n2 = Append!(A=/, B=JoinStringArray(SourceArray=ParseIntoArray(SourceString=@n1, Delimiter=/), Separator=/)) -> (ReturnValue)
   > PrintString(InString=@n2)
   > Return Node!(execute, FigmentName:String=@n2)
  ]}
 TextureAssetPathPrep {nodes [
   n1 = TextureAssetPathPrep!() -> (then, Material Asset Name:String) {is=title K2Node_FunctionEntry}
   > n2 = GetEditorWorld(self=Get UnrealEditorSubsystem())
   n3 = "To String (Text)"(InText="GetText (Text Box)"(self=Get TargetTextBox_()))
   > n4 = StripRightMostFolderFromPath_Util(Path=@n3, WorldContextObject=@n2, Strip Count=2, __WorldContext=@n2)
   n5 = "To String (Text)"(InText=Get Text(self=Get TargetTextBox_()))
   n6 = Append(A=TD_, B=@n1)
   n7 = "GetText (Multi-Line Editable Text)"(self=Get EditorUtilityMultiLineEditableText_101())
   n8 = "Get (a ref)"(Array=ParseIntoArray(SourceString="To String (Text)"(InText=@n7), Delimiter="\r\n"), Dimension 1=0)
   n9 = Select(Option 0:String=Append(A=@n8, B=/), Option 1:String=@n8, Index:bool=EndsWith(SourceString=@n8, InSuffix=/)) -> (ReturnValue:String)
   > Return Node!(
    execute
    "Material Full  Path":String=Append(A=@n9, B=@n1)
    TD Asset Full Path:String=Append(A=@n5, B=@n6)
    TD Asset Name:String=@n6
    Material Folder:String=@n9
    TD Asset Folder:String=@n5
   )
   MakeLiteralString(Value=/FigmentEnvironment/Blueprints/Cliffs/TextureData)
   "To String (Text)"(InText=Get Text(self=Get TargetTextBox_()))
   MakeLiteralString(Value=/FigmentEnvironment/Blueprints/Cliffs/TextureData/CliffGround/)
   MakeLiteralString(Value=/FigmentEnvironment/Blueprints/Cliffs/Materials/CliffGround/)
   Reroute Node(InputPin:String=Append(A=@n4, B=/TextureData/CliffGround/)) -> (OutputPin:String)
   Reroute Node(InputPin=@n3)
  ]}
 TreeSwap {nodes [
   n1 = SetSelectedLevelActors(self=Get EditorActorSubsystem(), ActorsToSelect=Make Array!("[0]":Actor Object Reference) -> (Array:Array of Actor Object References))
   ReplaceSelectedActors(execute=[@n1.then, @n47.then], InAssetPath=Get ReplaceBp() -> (ReplaceBp:String))
   > SelectNothing(self=Get EditorActorSubsystem())
   n2 = GetSelectedLevelActors(self=Get EditorActorSubsystem())
   > n3 = For Each Loop(Array=@n2)
   n4 = Get ObjectsToSave()
   CheckoutLoadedAssets(execute=[@n3.Completed, @n51.Completed], AssetsToCheckout=@n4)
   > SaveLoadedAssets(AssetsToSave=@n4, bOnlyIfIsDirty=false)
   n5 = Cast To Parent_Tree!(execute=@n3.LoopBody, Object=@n3) -> (then, CastFailed, AsParent Tree:Parent Tree Object Reference) {is=title K2Node_DynamicCast}
   > n6 = Set currentActor(currentActor:Parent Tree Object Reference=@n5) -> (Output_Get:Parent Tree Object Reference)
   n7 = Sequence!(execute=[@n71.then, @n6.then, @n52.then]) -> (then_0, then_1, then_2:Exec)
   n8 = Get MaterialArray() -> (MaterialArray:Array of Material Interface Object References)
   > Clear(TargetArray:Array of Material Interface Object References=@n8)
   n9 = Get StaticMeshComponent(self:Building Static Mesh Object Reference=Get currentActor() -> (currentActor:Parent Tree Object Reference)) -> (StaticMeshComponent:Base Building Static Mesh Component Object Reference)
   n10 = Length(TargetArray:Array of Material Interface Object References=GetMaterials!(self:Mesh Component Object Reference=@n9) -> (ReturnValue:Array of Material Interface Object References))
   > Resize(TargetArray:Array of Material Interface Object References=@n8, Size=@n10)
   > n11 = For Loop(LastIndex=int - int(A=@n10))
   > Set currentMatId(currentMatId:int=@n11) -> (Output_Get:int=0)
   n12 = GetEditorMaterial(self=@n9, ElementIndex=@n11)
   n13 = Get Object Path String(Object=@n12)
   n14 = "To String (Text)"(InText="GetText (Text Box)"(self=Get TargetTextBox_1()))
   > n15 = Branch(Condition="Equal Exactly (String)"(A=Split(InPath=@n13), B=Append(A=@n14, B=Materials)))
   > PrintString(InString=already fig mat)
   Set changed(execute=@n15.else, changed:bool=true) -> (Output_Get:bool)
   n16 = GetDisplayName(Object=@n12)
   > n17 = FigmetizeName!(execute, self, SourceString:String=@n16) -> (then, FigmentName:String) {is=title K2Node_CallFunction}
   n18 = Find(TargetMap:Map of Strings to Material Interface Object References=Get MapStringToTargetMaterials(), Key=@n17) -> (Value:Material Interface Object Reference)
   > n19 = Branch(Condition=@n18.ReturnValue)
   > PrintString(InString="found existing material (but new BP)")
   > Set Array Elem(
    TargetArray:Array of Material Interface Object References=Get MaterialArray()
    Index=Get currentMatId() -> (currentMatId:int=0)
    Item:"Material Interface Object Reference (by ref)"=@n18
    bSizeToFit=false
   )
   n20 = GetNaniteOverideMaterial!(self:Material Interface Object Reference=@n12) -> (ReturnValue:Material Interface Object Reference)
   n21 = Set HasNaniteOverride(execute=@n19.else, HasNaniteOverride:bool=IsValid(Object=@n20)) -> (Output_Get:bool)
   > PrintString(InString=Append(A=couldnt find, B=@n17))
   > n22 = Branch(Condition=@n21)
   > n23 = FigmetizeName(SourceString=GetDisplayName(Object=@n20))
   n24 = Find(TargetMap=Get MapStringToTargetMaterials(), Key=@n23)
   > n25 = Branch(Condition=@n24.ReturnValue)
   > n26 = Set naniteOverride(naniteOverride:Material Interface Object Reference=@n24) -> (Output_Get:Material Interface Object Reference)
   n27 = Find(TargetMap=Get MapStringToTargetMaterials(), Key=@n17)
   n28 = Branch(execute=[@n22.else, @n26.then], Condition=@n27.ReturnValue)
   > PrintString(InString=Append(A=just been created?, B=@n17))
   > n29 = Set mainMat(mainMat:Material Interface Object Reference=@n27) -> (Output_Get:Material Interface Object Reference)
   n30 = Set Array Elem(
    execute=[@n88.then, @n29.then]
    TargetArray=Get MaterialArray()
    Index=Get currentMatId() -> (currentMatId=0)
    Item=Get mainMat() -> (mainMat:Material Interface Object Reference)
    bSizeToFit=false
   )
   n31 = Get TargetTextBox_1()
   n32 = Append(A="To String (Text)"(InText="GetText (Text Box)"(self=@n31)), B=Materials)
   n33 = GetAssetTools(execute=@n83.then)
   n34 = CreateAsset(execute=[@n42.then, @n28.else], self=@n33, AssetName=@n17, PackagePath=@n32, AssetClass=&MaterialInstanceConstant)
   > n35 = Cast To MaterialInterface(Object=@n34)
   n36 = Cast To MaterialInstanceConstant(execute=@n30.then, Object:Object Reference=@n35)
   > SetMaterialInstanceParent(Instance=@n36, NewParent=@n12)
   > n37 = Branch(Condition=Get HasNaniteOverride() -> (HasNaniteOverride:bool))
   > n38 = SetNaniteOverrideMaterial!(
    execute
    self:Material Instance Constant Object Reference=@n36
    bInEnableOverride:bool=true
    InOverrideMaterial:Material Interface Object Reference=Get naniteOverride() -> (naniteOverride:Material Interface Object Reference)
   ) -> (then)
   SaveLoadedAsset(execute=[@n38.then, @n37.else], AssetToSave=@n36, bOnlyIfIsDirty=false)
   > MarkFileForAdd!(execute, InFile:String=Append!(A=@n32, B=/, C:String=@n17) -> (ReturnValue), bSilent:bool=true) -> (then, ReturnValue:bool)
   n39 = CreateAsset(execute=@n25.else, self=@n33, AssetName=@n23, PackagePath=@n32, AssetClass=&MaterialInstanceConstant)
   > n40 = Cast To MaterialInterface(Object=@n39)
   > Add(
    TargetMap:Map of Strings to Material Interface Object References=Get MapStringToTargetMaterials()
    Key=@n23
    Value:"Material Interface Object Reference (by ref)"=@n40
   ) {is=title K2Node_CallFunction}
   > n41 = Cast To MaterialInstanceConstant(Object=@n39)
   > SetMaterialInstanceParent(Instance=@n41, NewParent=@n20)
   > SaveLoadedAsset(AssetToSave=@n41, bOnlyIfIsDirty=false)
   > n42 = MarkFileForAdd(InFile=Append(A=@n32, B=/, C=@n23), bSilent=true)
   Branch(execute=@n7.then_1, Condition=Get changed() -> (changed:bool))
   > n43 = Sequence!(execute) -> (then_0, then_1)
   n44 = Get StaticMeshComponent(self=Get currentActor())
   n45 = GetMaterials(self=@n44)
   > n46 = For Loop(LastIndex=int - int(A=Length(TargetArray=@n45)))
   > SetMaterial(
    self=@n44
    ElementIndex=@n46
    Material="Get (a copy)"(Array:Array of Material Interface Object References=Get MaterialArray(), Dimension 1=@n46) -> (Output:Material Interface Object Reference)
   )
   Add(execute=@n43.then_1, TargetArray=Get ObjectsToSave(), NewItem=Get currentActor()) {is=title K2Node_CallArrayFunction}
   > Set changed(changed=false)
   n47 = SetMaterial(ElementIndex=1, Material=Get MatC_1() -> (MatC_1:Material Instance Constant Object Reference))
   n48 = Get SourceTextBox()
   n49 = For Each LI(execute=@n33.then, InputPin="To String (Text)"(InText="GetText (Text Box)"(self=@n48)))
   n50 = GetAllActorsOfClass(execute=@n85.then, WorldContextObject=@n49, ActorClass=/Game/Building/ActorBlueprints/Prop/Parent_Tree.Parent_Tree_C) -> (OutActors:Array of Parent Tree Object References)
   > n51 = For Each Loop(Array:Array of Parent Tree Object References=@n50) -> (Array Element:Parent Tree Object Reference)
   > n52 = Set currentActor(currentActor=@n51)
   n53 = GetClass(InAssetData=CreateAssetData(InAsset=@n52, bAllowBlueprintClass=false))
   PrintString(InString=Get Class Display Name(Class=@n53))
   > n54 = GetBlueprintForClass!(execute, Class:Object Class Reference=@n53) -> (True:Exec, False:Exec, ReturnValue:Blueprint Object Reference)
   n55 = Get Object Path String(Object=@n54)
   n56 = Get TargetTextBox_1()
   n57 = Append!(A="To String (Text)"(InText="GetText (Text Box)"(self=@n56)), B=Blueprints/) -> (ReturnValue)
   > n58 = Branch(Condition="Equal Exactly (String)"(A=Split(InPath=@n55), B=Split(InPath=@n57)))
   > PrintString(InString=is already converted)
   Sequence!(execute=@n58.else) -> (then_0, then_1, then_2:Exec)
   n59 = Append(A=GetDisplayName(Object=@n54), B=_Figment)
   > PrintString(InString=Append(A="Finding: ", B=@n59))
   n60 = Find(TargetMap:Map of Strings to Object References=Get MapStringToTreeBPs(), Key=@n59) -> (Value:Object Reference)
   > n61 = Branch(Condition=@n60.ReturnValue)
   > PrintString(InString=Found BP.)
   > Set ReplaceBp(ReplaceBp:String=Split!(SourceString:String=Get Object Path String(Object=@n60), InStr:String=., SearchCase:ESearchCase Enum, SearchDir:ESearchDir Enum=FromStart) -> (LeftS:String, RightS:String, ReturnValue:bool)) -> (Output_Get:String)
   n62 = Get TargetTextBox_1()
   n63 = Append!(A="To String (Text)"(InText="GetText (Text Box)"(self=@n62)), B=Blueprints/, C:String=@n59) -> (ReturnValue)
   > n64 = GetClassDefaultObjectFromPath!(execute, ClassAssetPath:String=@n63) -> (then, ReturnValue:Object Reference)
   > n65 = Cast To Parent_Tree(Object=@n64)
   n66 = Get MaterialArray()
   > Clear(TargetArray=@n66)
   n67 = Get StaticMeshComponent(self=@n65)
   n68 = Length(TargetArray=GetMaterials(self=@n67))
   > Resize(TargetArray=@n66, Size=@n68)
   > n69 = For Loop(LastIndex=int - int(A=@n68))
   > Set Array Elem(TargetArray=Get MaterialArray(), Index=@n69, Item=GetEditorMaterial(self=@n67, ElementIndex=@n69), bSizeToFit=false)
   PrintString(execute=@n61.else, InString=Append!(A="Couldnt find: ", B=@n59) -> (ReturnValue))
   > n70 = CreateBlueprintAssetWithParent(AssetPath=@n63, ParentClass=@n53)
   > Add(TargetMap:Map of Strings to Object References=Get MapStringToTreeBPs(), Key=@n59, Value:"Object Reference (by ref)"=@n70) {is=title K2Node_CallFunction}
   > n71 = PrintString(InString=Append(A="added: ", B=@n63))
   ReplaceSelectedActors()
   > n72 = GetSelectedActors()
   > Add(TargetArray=Get ObjectsToSave(), NewItem="Get (a copy)"(Array:Array of Actor Object References=@n72) -> (Output:Actor Object Reference)) {is=title K2Node_CallArrayFunction}
   n73 = GetSelectedActors()
   > n74 = Gather Subobject Data For Instance(self=Get SubobjectDataSubsystem(), Context="Get (a copy)"(Array=@n73))
   > n75 = For Each Loop(Array:Array of Subobject Data Handle Structures=@n74) -> (Array Element:Subobject Data Handle Structure)
   n76 = GetData(DataHandle=@n75)
   > n77 = Cast To StaticMeshComponent(Object=GetAssociatedObject!(Data:"Subobject Data Structure (by ref)"=@n76) -> (ReturnValue:Object Reference))
   n78 = GetMaterials(self=@n77)
   > SetEditorProperty(
    Object=@n77
    PropertyName=Materials
    PropertyValue:Array of Material Interface Object References=Make Array!(
     "[0]":Material Interface Object Reference="Get (a copy)"(Array:Array of Material Interface Object References=@n78) -> (Output:Material Interface Object Reference)
     "[1]":Material Interface Object Reference=Get MatC_1()
     "[2]":Material Interface Object Reference="Get (a copy)"(Array=@n78, Dimension 1=2)
    ) -> (Array:Array of Material Interface Object References)
   )
   Sequence!(execute) -> (then_0, then_1)
   > n79 = Get Blueprint Component!(
    execute
    Blueprint:Blueprint Object Reference=@n70
    ComponentClass:Object Class Reference=&StaticMeshComponent
    bCreateIfMissing:bool=false
   ) -> (then, ReturnValue:Actor Component Object Reference) {is=title K2Node_CallFunction}
   > n80 = Cast To StaticMeshComponent(Object=@n79)
   n81 = GetMaterials(self=@n80)
   > n82 = For Loop(LastIndex=int - int(A=Length(TargetArray=@n81)))
   > SetMaterial(self=@n80, ElementIndex=@n82, Material="Get (a copy)"(Array=Get MaterialArray(), Dimension 1=@n82))
   n83 = TreeSwap() {is=title K2Node_FunctionEntry}
   n84 = Set ReplaceBp(ReplaceBp=@n63)
   > SaveAsset(AssetToSave=@n84, bOnlyIfIsDirty=false)
   GetClassDefaultObject(Class=@n53)
   GetAssociatedObject()
   PrintString(InString=GetDisplayName(Object=@n77))
   GetEditorMaterial(self=@n77, ElementIndex=0)
   SetMaterial(self=@n77, ElementIndex=1)
   Get Object Path String(Object=@n52)
   PrintString(InString=found Mat)
   GetDisplayName()
   PrintString(InString=Hello)
   Contains(SearchIn=@n16, Substring=figment)
   Get Object Path String(Object=@n54)
   RebuildTreeMaps() {is=title K2Node_CallFunction}
   Get Class()
   GetEditorWorld(self=Get UnrealEditorSubsystem())
   Sequence()
   Sequence()
   "Get (a copy)"(Array:Array of Actor Object References=GetSelectedActors()) -> (Output:Actor Object Reference)
   Get EditorActorSubsystem()
   Get EditorActorSubsystem()
   Branch(Condition=IsRootComponent!(Data:"Subobject Data Structure (by ref)"=@n76) -> (ReturnValue:bool))
   Branch()
   Get StaticMeshComponent()
   Set changed(changed=false)
   n85 = Clear(execute=@n86.then, TargetArray:Array of Object References=Get ObjectsToSave())
   n86 = RebuildTreeMaps(execute=@n49.then)
   n87 = Set mainMat(execute=@n35.then, mainMat=@n35)
   > n88 = Add(
    TargetMap:Map of Strings to Material Interface Object References=Get MapStringToTargetMaterials()
    Key=@n17
    Value:"Material Interface Object Reference (by ref)"=@n87
   ) {is=title K2Node_CallFunction}
  ]}
]
variables [
 Actors to Revert {type="array<BP__Cliff_Figment_C>", default []}
 ActorsToDestroy {type="array<StaticMeshActor>", default []}
 ActorsToTD {type="map<BuildingTextureData>", default {}}
 Apply Mode {type=E_ApplyContext, default=Apply to Selection}
 Assets to add {type="array<AssetData>", default []}
 BTD_To_Return {type="array<BuildingTextureData>", default []}
 BTD_ToSave {type="array<BuildingTextureData>", default []}
 Cliff Actors to Save {type="array<BP__Cliff_Figment_C>", default []}
 Color Map TD override {type=BuildingTextureData, default=/FigmentEnvironment/Blueprints/Cliffs/Biome/TD_BiomeMap_S05}
 count {type=integer}
 Current Level To Process {type=Object, default=null}
 CurrentMaterialForStaticMesh {type=MaterialInterface, default=null}
 CurrentTDForStaticMesh {type=BuildingTextureData, default=null}
 data {type=BuildingTextureData, default=/FigmentEnvironment/Blueprints/Cliffs/Biome/TD_BiomeMap_S04}
 DM {type=DynamicMesh, default=null}
 DMP {type=DynamicMeshPool, default=null}
 Dominant_Color {type=LinearColor, default=<rgba 0 0 0 0>}
 Element LIst {type=ScriptTypedElementListProxy, default {}}
 ExternalFoliageActorExternalPathsToKill {type="array<string>", default []}
 ExternalFoliageActorsToKill {type="array<string>", default []}
 FigmentParentMAt {type=MaterialInstanceConstant, default=null}
 FilesStatus {type="array<string>", default []}
 FilesToAdd {type="array<string>", default []}
 FoliageActorsToKill {type="array<Actor>"}
 FortStaticMeshesToSave {type="array<FortStaticMeshActor>", default []}
 Level Instance Actor {type=LevelInstance, default=null}
 LevelInstanceActor {type=LevelInstance, default=null}
 LevelInstanceActors {type="map<object>", default {}}
 LevelInstances {type="array<string>", default []}
 MapOfMaterials {default {None=null}}
 mapOfMeshesToBPs {default {}}
 mapOfMeshNamesToBp {type="map<string>", default {}}
 MapStringToTargetMaterials {type="map<MaterialInterface>", default {}}
 MapStringToTreeBPs {type="map<Object>", default {}}
 material_on_TD {type=MaterialInstanceConstant, default=null}
 material_orig {type=MaterialInterface, default=&Athena_Terrain_Mat_Biome}
 MICFrom {type=MaterialInstanceConstant, default=null}
 MICTo {type=MaterialInstanceConstant, default=null}
 MPC {type=MaterialParameterCollection, default=/Game/Athena/Environments/Landscape/MPC/MPC_Landscape}
 NewCliffSet {type="array<Actor>"}
 NewVar {type=string, default=""}
 ObjectsToSave {type="array<Object>", default []}
 OriginalMIs {type="array<MaterialInstanceConstant>", default []}
 packages {type=boolean}
 ParentBuildingWallArray {type="array<Parent_BuildingWall_C>", default []}
 Paths {type="array<string>", default []}
 Paths_KeepCount {type="array<string>", default []}
 Paths_SetArray {type="array<string>", default []}
 PathsCount {type="map<integer>", default {}}
 PathsCountInt {type="array<integer>"}
 Phys Material Override {type=PhysicalMaterial, default=null}
 PhysToSet {type=PhysicalMaterial, default=null}
 SM_to_BP_all {type=boolean}
 SM_to_BP_all_exp {type=boolean}
 staticMeshSetCheckDuplicattes {type="array<StaticMesh>", default []}
 stringToGuids {type="map<Guid>", default {}}
 TargetPAth {type=string, default=/Figment_S05_LevelInstances/}
 TD Asset {type=BuildingTextureData, default=null}
 TotalTriangleArea {type=number}
 TreeMaterialFigmentName {type="array<string>", default []}
 viewport {type=Viewport, default=null}
]
widgets {
 info {rootWidgetClass=&CanvasPanel, widgetCount=112}
 widgets [
  CanvasPanel_X {
   widgetClassPath=&CanvasPanel
   bIsVariable=true
   widget=CanvasPanel_X
   properties {accessibleBehavior=Auto, bOverrideAccessibleDefaults=true, visibility=Visible}
  }
  VerticalBox_146 {
   widgetClassPath=&VerticalBox
   parent=CanvasPanel_X
   widget=VerticalBox_146
   slot {class=CanvasPanelSlot, properties {bAutoSize=true, layoutData=~}}
  }
  HorizontalBox_174 {widgetClassPath=&HorizontalBox, bIsVariable=true, parent=VerticalBox_146, widget=HorizontalBox_174, slot {class=VerticalBoxSlot}}
  EditorUtilityButton_34 {
   widgetClassPath=&EditorUtilityButton
   parent=HorizontalBox_174
   widget=EditorUtilityButton_34
   properties {widgetStyle {
     clickedSlateSound {resourceObject=null}
     disabled {
      bIsDynamicallyLoaded=false
      drawAs=RoundedBox
      imageType=NoImage
      mirroring=NoMirror
      resourceName=None
      resourceObject=null
      tiling=NoTile
      imageSize=<0 0>
      margin=<ltrb 0 0 0 0>
      outlineSettings {
       bUseBrushTransparency=false
       roundingType=FixedRadius
       width=1
       color {colorUseRule=UseColor_ColorTable, specifiedColor=<rgba 0 0 0 0>}
       cornerRadii=<4 4 4 4>
      }
      tintColor {colorUseRule=UseColor_Specified, specifiedColor=<rgba 0.0395 0.0395 0.0395 1>}
      uVRegion {bIsValid=false, max=<0 0>, min=<0 0>}
     }
     disabledForeground {colorUseRule=UseColor_ColorTable, specifiedColor=<rgba 0 0 0 0>}
     hovered {
      bIsDynamicallyLoaded=false
      drawAs=RoundedBox
      imageType=NoImage
      mirroring=NoMirror
      resourceName=None
      resourceObject=null
      tiling=NoTile
      imageSize=<32 32>
      margin=<ltrb 0 0 0 0>
      outlineSettings {
       bUseBrushTransparency=false
       roundingType=FixedRadius
       width=1
       color {colorUseRule=UseColor_Specified, specifiedColor=<rgba 0.0912 0.1979 0.1294 1>}
       cornerRadii=<20 20 4 4>
      }
      tintColor {colorUseRule=UseColor_Specified, specifiedColor=<rgba 0.0953 0.0953 0.0953 1>}
      uVRegion {bIsValid=false, max=<0 0>, min=<0 0>}
     }
     hoveredForeground {colorUseRule=UseColor_ColorTable, specifiedColor=<rgba 0 0 0 0>}
     hoveredSlateSound {resourceObject=null}
     normal {
      bIsDynamicallyLoaded=false
      drawAs=RoundedBox
      imageType=NoImage
      mirroring=NoMirror
      resourceName=None
      resourceObject=null
      tiling=NoTile
      imageSize=<32 32>
      margin=<ltrb 0 0 0 0>
      outlineSettings {
       bUseBrushTransparency=false
       roundingType=FixedRadius
       width=3
       color {colorUseRule=UseColor_Specified, specifiedColor=<rgba 0.1354 0.1354 0.1354 1>}
       cornerRadii=<20 20 4 4>
      }
      tintColor {colorUseRule=UseColor_Specified, specifiedColor=<rgba 0.0395 0.0395 0.0395 1>}
      uVRegion {bIsValid=false, max=<0 0>, min=<0 0>}
     }
     normalForeground {colorUseRule=UseColor_ColorTable, specifiedColor=<rgba 0 0 0 0>}
     normalPadding=<ltrb 12 1.5 12 1.5>
     pressed {
      bIsDynamicallyLoaded=false
      drawAs=RoundedBox
      imageType=NoImage
      mirroring=NoMirror
      resourceName=None
      resourceObject=null
      tiling=NoTile
      imageSize=<32 32>
      margin=<ltrb 0 0 0 0>
      outlineSettings {
       bUseBrushTransparency=false
       roundingType=FixedRadius
       width=1
       color {colorUseRule=UseColor_ColorTable, specifiedColor=<rgba 0 0 0 0>}
       cornerRadii=<4 4 4 4>
      }
      tintColor {colorUseRule=UseColor_Specified, specifiedColor=<rgba 0.0284 0.0284 0.0284 1>}
      uVRegion {bIsValid=false, max=<0 0>, min=<0 0>}
     }
     pressedForeground {colorUseRule=UseColor_ColorTable, specifiedColor=<rgba 0 0 0 0>}
     pressedPadding=<ltrb 12 2.5 12 0.5>
     pressedSlateSound {resourceObject=null}
    }}
   slot {class=HorizontalBoxSlot}
  }
  TextBlock_178 {
   widgetClassPath=&TextBlock
   parent=EditorUtilityButton_34
   widget=TextBlock_178
   properties {
    text=Figment Processing Tools
    font {
     bForceMonospaced=false
     bMaterialIsStencil=false
     fontMaterial=null
     fontObject=/Engine/EngineFonts/RobotoTiny
     letterSpacing=0
     monospacedWidth=1
     size=31.5
     skewAmount=0
     typefaceFontName=Light
     outlineSettings {
      bApplyOutlineToDropShadows=false
      bMiteredCorners=false
      bSeparateFillAlpha=false
      outlineMaterial=null
      outlineSize=0
      outlineColor=<rgba 0 0 0 1>
     }
    }
   }
   slot {class=ButtonSlot}
  }
  EditorUtilityButton_60 {
   widgetClassPath=&EditorUtilityButton
   parent=HorizontalBox_174
   widget=EditorUtilityButton_60
   properties {widgetStyle {
     clickedSlateSound {resourceObject=null}
     disabled {
      bIsDynamicallyLoaded=false
      drawAs=RoundedBox
      imageType=NoImage
      mirroring=NoMirror
      resourceName=None
      resourceObject=null
      tiling=NoTile
      imageSize=<0 0>
      margin=<ltrb 0 0 0 0>
      outlineSettings {
       bUseBrushTransparency=false
       roundingType=FixedRadius
       width=1
       color {colorUseRule=UseColor_ColorTable, specifiedColor=<rgba 0 0 0 0>}
       cornerRadii=<4 4 4 4>
      }
      tintColor {colorUseRule=UseColor_Specified, specifiedColor=<rgba 0.0395 0.0395 0.0395 1>}
      uVRegion {bIsValid=false, max=<0 0>, min=<0 0>}
     }
     disabledForeground {colorUseRule=UseColor_ColorTable, specifiedColor=<rgba 0 0 0 0>}
     hovered {
      bIsDynamicallyLoaded=false
      drawAs=RoundedBox
      imageType=NoImage
      mirroring=NoMirror
      resourceName=None
      resourceObject=null
      tiling=NoTile
      imageSize=<32 32>
      margin=<ltrb 0 0 0 0>
      outlineSettings {
       bUseBrushTransparency=false
       roundingType=FixedRadius
       width=1
       color {colorUseRule=UseColor_ColorTable, specifiedColor=<rgba 0 0 0 0>}
       cornerRadii=<20 20 4 4>
      }
      tintColor {colorUseRule=UseColor_Specified, specifiedColor=<rgba 0.0953 0.0953 0.0953 1>}
      uVRegion {bIsValid=false, max=<0 0>, min=<0 0>}
     }
     hoveredForeground {colorUseRule=UseColor_ColorTable, specifiedColor=<rgba 0 0 0 0>}
     hoveredSlateSound {resourceObject=null}
     normal {
      bIsDynamicallyLoaded=false
      drawAs=RoundedBox
      imageType=NoImage
      mirroring=NoMirror
      resourceName=None
      resourceObject=null
      tiling=NoTile
      imageSize=<32 32>
      margin=<ltrb 0 0 0 0>
      outlineSettings {
       bUseBrushTransparency=false
       roundingType=FixedRadius
       width=0
       color {colorUseRule=UseColor_Specified, specifiedColor=<rgba 0.1354 0.1354 0.1354 1>}
       cornerRadii=<20 20 4 4>
      }
      tintColor {colorUseRule=UseColor_Specified, specifiedColor=<rgba 0.0395 0.0395 0.0395 1>}
      uVRegion {bIsValid=false, max=<0 0>, min=<0 0>}
     }
     normalForeground {colorUseRule=UseColor_ColorTable, specifiedColor=<rgba 0 0 0 0>}
     normalPadding=<ltrb 12 1.5 12 1.5>
     pressed {
      bIsDynamicallyLoaded=false
      drawAs=RoundedBox
      imageType=NoImage
      mirroring=NoMirror
      resourceName=None
      resourceObject=null
      tiling=NoTile
      imageSize=<32 32>
      margin=<ltrb 0 0 0 0>
      outlineSettings {
       bUseBrushTransparency=false
       roundingType=FixedRadius
       width=1
       color {colorUseRule=UseColor_ColorTable, specifiedColor=<rgba 0 0 0 0>}
       cornerRadii=<20 20 4 4>
      }
      tintColor {colorUseRule=UseColor_Specified, specifiedColor=<rgba 0.0284 0.0284 0.0284 1>}
      uVRegion {bIsValid=false, max=<0 0>, min=<0 0>}
     }
     pressedForeground {colorUseRule=UseColor_ColorTable, specifiedColor=<rgba 0 0 0 0>}
     pressedPadding=<ltrb 12 2.5 12 0.5>
     pressedSlateSound {resourceObject=null}
    }}
   slot {class=HorizontalBoxSlot}
  }
  TextBlock_228 {
   widgetClassPath=&TextBlock
   parent=EditorUtilityButton_60
   widget=TextBlock_228
   properties {
    text=Experimental
    font {
     bForceMonospaced=false
     bMaterialIsStencil=false
     fontMaterial=null
     fontObject=&RobotoTiny
     letterSpacing=0
     monospacedWidth=1
     size=31.5
     skewAmount=0
     typefaceFontName=Light
     outlineSettings {
      bApplyOutlineToDropShadows=false
      bMiteredCorners=false
      bSeparateFillAlpha=false
      outlineMaterial=null
      outlineSize=0
      outlineColor=<rgba 0 0 0 1>
     }
    }
   }
   slot {class=ButtonSlot}
  }
  ScaleBox_216 {
   widgetClassPath=&ScaleBox
   parent=VerticalBox_146
   widget=ScaleBox_216
   properties {stretch=UserSpecified, userSpecifiedScale=2}
   slot {class=VerticalBoxSlot, properties {padding=<ltrb 40 10 0 10>, size {sizeRule=Fill, value=1}}}
  }
  SinglePropertyView_Apply {
   widgetClassPath=&SinglePropertyView
   parent=ScaleBox_216
   widget=SinglePropertyView_Apply
   properties {propertyName=Apply Mode}
   slot {class=ScaleBoxSlot, properties {horizontalAlignment=HAlign_Left, verticalAlignment=VAlign_Fill}}
  }
  HorizontalBox_376 {
   widgetClassPath=&HorizontalBox
   parent=VerticalBox_146
   widget=HorizontalBox_376
   slot {class=VerticalBoxSlot, properties {padding=<ltrb 40 0 0 0>}}
  }
  TextBlock_106 {
   widgetClassPath=&TextBlock
   parent=HorizontalBox_376
   widget=TextBlock_106
   properties {
    text=Source Level Instance Top Level Path:
    font {
     bForceMonospaced=false
     bMaterialIsStencil=false
     fontMaterial=null
     fontObject=&RobotoTiny
     letterSpacing=0
     monospacedWidth=1
     size=24
     skewAmount=0
     typefaceFontName=Light
     outlineSettings {
      bApplyOutlineToDropShadows=false
      bMiteredCorners=false
      bSeparateFillAlpha=false
      outlineMaterial=null
      outlineSize=0
      outlineColor=<rgba 0 0 0 1>
     }
    }
   }
   slot {class=HorizontalBoxSlot, properties {padding=<ltrb 8 8 8 8>}}
  }
  VerticalBox_163 {
   widgetClassPath=&VerticalBox
   parent=HorizontalBox_376
   widget=VerticalBox_163
   slot {class=HorizontalBoxSlot, properties {verticalAlignment=VAlign_Center}}
  }
  SourceTextBox {
   widgetClassPath=&EditorUtilityEditableTextBox
   bIsVariable=true
   parent=VerticalBox_163
   widget=SourceTextBox
   properties {
    text=/Figment_S09_LevelInstances/
    widgetStyle {
     backgroundColor {colorUseRule=UseColor_Specified, specifiedColor=<rgba 1 1 1 1>}
     backgroundImageFocused {
      bIsDynamicallyLoaded=false
      drawAs=RoundedBox
      imageType=NoImage
      mirroring=NoMirror
      resourceName=None
      resourceObject=null
      tiling=NoTile
      imageSize=<0 0>
      margin=<ltrb 0 0 0 0>
      outlineSettings {
       bUseBrushTransparency=false
       roundingType=FixedRadius
       width=1
       color {colorUseRule=UseColor_ColorTable, specifiedColor=<rgba 0 0 0 0>}
       cornerRadii=<4 4 4 4>
      }
      tintColor {colorUseRule=UseColor_Specified, specifiedColor=<rgba 0.0048 0.0048 0.0048 1>}
      uVRegion {bIsValid=false, max=<0 0>, min=<0 0>}
     }
     backgroundImageHovered {
      bIsDynamicallyLoaded=false
      drawAs=RoundedBox
      imageType=NoImage
      mirroring=NoMirror
      resourceName=None
      resourceObject=null
      tiling=NoTile
      imageSize=<0 0>
      margin=<ltrb 0 0 0 0>
      outlineSettings {
       bUseBrushTransparency=false
       roundingType=FixedRadius
       width=1
       color {colorUseRule=UseColor_ColorTable, specifiedColor=<rgba 0 0 0 0>}
       cornerRadii=<4 4 4 4>
      }
      tintColor {colorUseRule=UseColor_Specified, specifiedColor=<rgba 0.0048 0.0048 0.0048 1>}
      uVRegion {bIsValid=false, max=<0 0>, min=<0 0>}
     }
     backgroundImageNormal {
      bIsDynamicallyLoaded=false
      drawAs=RoundedBox
      imageType=NoImage
      mirroring=NoMirror
      resourceName=None
      resourceObject=null
      tiling=NoTile
      imageSize=<0 0>
      margin=<ltrb 0 0 0 0>
      outlineSettings {
       bUseBrushTransparency=false
       roundingType=FixedRadius
       width=1
       color {colorUseRule=UseColor_ColorTable, specifiedColor=<rgba 0 0 0 0>}
       cornerRadii=<4 4 4 4>
      }
      tintColor {colorUseRule=UseColor_Specified, specifiedColor=<rgba 0.0048 0.0048 0.0048 1>}
      uVRegion {bIsValid=false, max=<0 0>, min=<0 0>}
     }
     backgroundImageReadOnly {
      bIsDynamicallyLoaded=false
      drawAs=RoundedBox
      imageType=NoImage
      mirroring=NoMirror
      resourceName=None
      resourceObject=null
      tiling=NoTile
      imageSize=<0 0>
      margin=<ltrb 0 0 0 0>
      outlineSettings {
       bUseBrushTransparency=false
       roundingType=FixedRadius
       width=1
       color {colorUseRule=UseColor_ColorTable, specifiedColor=<rgba 0 0 0 0>}
       cornerRadii=<4 4 4 4>
      }
      tintColor {colorUseRule=UseColor_Specified, specifiedColor=<rgba 0.0284 0.0284 0.0284 1>}
      uVRegion {bIsValid=false, max=<0 0>, min=<0 0>}
     }
     focusedForegroundColor {colorUseRule=UseColor_Specified, specifiedColor=<rgba 1 1 1 1>}
     foregroundColor {colorUseRule=UseColor_Specified, specifiedColor=<rgba 0.5271 0.5271 0.5271 1>}
     hScrollBarPadding=<ltrb 0 0 0 0>
     padding=<ltrb 8 3 8 4>
     readOnlyForegroundColor {colorUseRule=UseColor_Foreground, specifiedColor=<rgba 1 0 1 1>}
     scrollBarStyle {
      thickness=8
      draggedThumbImage {
       bIsDynamicallyLoaded=false
       drawAs=RoundedBox
       imageType=NoImage
       mirroring=NoMirror
       resourceName=None
       resourceObject=null
       tiling=NoTile
       imageSize=<0 0>
       margin=<ltrb 0 0 0 0>
       outlineSettings {
        bUseBrushTransparency=false
        roundingType=FixedRadius
        width=0
        color {colorUseRule=UseColor_Specified, specifiedColor=<rgba 0 0 0 0>}
        cornerRadii=<4 4 4 4>
       }
       tintColor {colorUseRule=UseColor_Specified, specifiedColor=<rgba 0.2159 0.2159 0.2159 1>}
       uVRegion {bIsValid=false, max=<0 0>, min=<0 0>}
      }
      horizontalBackgroundImage {
       bIsDynamicallyLoaded=false
       drawAs=NoDrawType
       imageType=NoImage
       mirroring=NoMirror
       resourceName=None
       resourceObject=null
       tiling=NoTile
       imageSize=<0 0>
       margin=<ltrb 0 0 0 0>
       outlineSettings {
        bUseBrushTransparency=false
        roundingType=HalfHeightRadius
        width=0
        color {colorUseRule=UseColor_Specified, specifiedColor=<rgba 0 0 0 0>}
        cornerRadii=<0 0 0 0>
       }
       tintColor {colorUseRule=UseColor_Specified, specifiedColor=<rgba 1 1 1 1>}
       uVRegion {bIsValid=false, max=<0 0>, min=<0 0>}
      }
      horizontalBottomSlotImage {
       bIsDynamicallyLoaded=false
       drawAs=NoDrawType
       imageType=NoImage
       mirroring=NoMirror
       resourceName=None
       resourceObject=null
       tiling=NoTile
       imageSize=<0 0>
       margin=<ltrb 0 0 0 0>
       outlineSettings {
        bUseBrushTransparency=false
        roundingType=HalfHeightRadius
        width=0
        color {colorUseRule=UseColor_Specified, specifiedColor=<rgba 0 0 0 0>}
        cornerRadii=<0 0 0 0>
       }
       tintColor {colorUseRule=UseColor_Specified, specifiedColor=<rgba 1 1 1 1>}
       uVRegion {bIsValid=false, max=<0 0>, min=<0 0>}
      }
      horizontalTopSlotImage {
       bIsDynamicallyLoaded=false
       drawAs=NoDrawType
       imageType=NoImage
       mirroring=NoMirror
       resourceName=None
       resourceObject=null
       tiling=NoTile
       imageSize=<0 0>
       margin=<ltrb 0 0 0 0>
       outlineSettings {
        bUseBrushTransparency=false
        roundingType=HalfHeightRadius
        width=0
        color {colorUseRule=UseColor_Specified, specifiedColor=<rgba 0 0 0 0>}
        cornerRadii=<0 0 0 0>
       }
       tintColor {colorUseRule=UseColor_Specified, specifiedColor=<rgba 1 1 1 1>}
       uVRegion {bIsValid=false, max=<0 0>, min=<0 0>}
      }
      hoveredThumbImage {
       bIsDynamicallyLoaded=false
       drawAs=RoundedBox
       imageType=NoImage
       mirroring=NoMirror
       resourceName=None
       resourceObject=null
       tiling=NoTile
       imageSize=<0 0>
       margin=<ltrb 0 0 0 0>
       outlineSettings {
        bUseBrushTransparency=false
        roundingType=FixedRadius
        width=0
        color {colorUseRule=UseColor_Specified, specifiedColor=<rgba 0 0 0 0>}
        cornerRadii=<4 4 4 4>
       }
       tintColor {colorUseRule=UseColor_Specified, specifiedColor=<rgba 0.2159 0.2159 0.2159 1>}
       uVRegion {bIsValid=false, max=<0 0>, min=<0 0>}
      }
      normalThumbImage {
       bIsDynamicallyLoaded=false
       drawAs=RoundedBox
       imageType=NoImage
       mirroring=NoMirror
       resourceName=None
       resourceObject=null
       tiling=NoTile
       imageSize=<0 0>
       margin=<ltrb 0 0 0 0>
       outlineSettings {
        bUseBrushTransparency=false
        roundingType=FixedRadius
        width=0
        color {colorUseRule=UseColor_Specified, specifiedColor=<rgba 0 0 0 0>}
        cornerRadii=<4 4 4 4>
       }
       tintColor {colorUseRule=UseColor_Specified, specifiedColor=<rgba 0.0953 0.0953 0.0953 1>}
       uVRegion {bIsValid=false, max=<0 0>, min=<0 0>}
      }
      verticalBackgroundImage {
       bIsDynamicallyLoaded=false
       drawAs=NoDrawType
       imageType=NoImage
       mirroring=NoMirror
       resourceName=None
       resourceObject=null
       tiling=NoTile
       imageSize=<0 0>
       margin=<ltrb 0 0 0 0>
       outlineSettings {
        bUseBrushTransparency=false
        roundingType=HalfHeightRadius
        width=0
        color {colorUseRule=UseColor_Specified, specifiedColor=<rgba 0 0 0 0>}
        cornerRadii=<0 0 0 0>
       }
       tintColor {colorUseRule=UseColor_Specified, specifiedColor=<rgba 1 1 1 1>}
       uVRegion {bIsValid=false, max=<0 0>, min=<0 0>}
      }
      verticalBottomSlotImage {
       bIsDynamicallyLoaded=false
       drawAs=NoDrawType
       imageType=NoImage
       mirroring=NoMirror
       resourceName=None
       resourceObject=null
       tiling=NoTile
       imageSize=<0 0>
       margin=<ltrb 0 0 0 0>
       outlineSettings {
        bUseBrushTransparency=false
        roundingType=HalfHeightRadius
        width=0
        color {colorUseRule=UseColor_Specified, specifiedColor=<rgba 0 0 0 0>}
        cornerRadii=<0 0 0 0>
       }
       tintColor {colorUseRule=UseColor_Specified, specifiedColor=<rgba 1 1 1 1>}
       uVRegion {bIsValid=false, max=<0 0>, min=<0 0>}
      }
      verticalTopSlotImage {
       bIsDynamicallyLoaded=false
       drawAs=NoDrawType
       imageType=NoImage
       mirroring=NoMirror
       resourceName=None
       resourceObject=null
       tiling=NoTile
       imageSize=<0 0>
       margin=<ltrb 0 0 0 0>
       outlineSettings {
        bUseBrushTransparency=false
        roundingType=HalfHeightRadius
        width=0
        color {colorUseRule=UseColor_Specified, specifiedColor=<rgba 0 0 0 0>}
        cornerRadii=<0 0 0 0>
       }
       tintColor {colorUseRule=UseColor_Specified, specifiedColor=<rgba 1 1 1 1>}
       uVRegion {bIsValid=false, max=<0 0>, min=<0 0>}
      }
     }
     textStyle {
      overflowPolicy=Clip
      transformPolicy=None
      colorAndOpacity {colorUseRule=UseColor_Foreground, specifiedColor=<rgba 1 0 1 1>}
      font {
       bForceMonospaced=false
       bMaterialIsStencil=false
       fontMaterial=null
       fontObject=&RobotoTiny
       letterSpacing=0
       monospacedWidth=1
       size=18
       skewAmount=0
       typefaceFontName=Light
       outlineSettings {
        bApplyOutlineToDropShadows=false
        bMiteredCorners=false
        bSeparateFillAlpha=false
        outlineMaterial=null
        outlineSize=0
        outlineColor=<rgba 0 0 0 1>
       }
      }
      highlightColor {colorUseRule=UseColor_ColorTable, specifiedColor=<rgba 0 0 0 0>}
      highlightShape {
       bIsDynamicallyLoaded=false
       drawAs=Image
       imageType=NoImage
       mirroring=NoMirror
       resourceName=None
       resourceObject=null
       tiling=NoTile
       imageSize=<0 0>
       margin=<ltrb 0 0 0 0>
       outlineSettings {
        bUseBrushTransparency=false
        roundingType=HalfHeightRadius
        width=0
        color {colorUseRule=UseColor_Specified, specifiedColor=<rgba 0 0 0 0>}
        cornerRadii=<0 0 0 0>
       }
       tintColor {colorUseRule=UseColor_ColorTable, specifiedColor=<rgba 0 0 0 0>}
       uVRegion {bIsValid=false, max=<0 0>, min=<0 0>}
      }
      selectedBackgroundColor {colorUseRule=UseColor_ColorTable, specifiedColor=<rgba 0 0 0 0>}
      shadowColorAndOpacity=<rgba 0 0 0 1>
      shadowOffset=<0 0>
      strikeBrush {
       bIsDynamicallyLoaded=false
       drawAs=Image
       imageType=NoImage
       mirroring=NoMirror
       resourceName=None
       resourceObject=null
       tiling=NoTile
       imageSize=<32 32>
       margin=<ltrb 0 0 0 0>
       outlineSettings {
        bUseBrushTransparency=false
        roundingType=HalfHeightRadius
        width=0
        color {colorUseRule=UseColor_Specified, specifiedColor=<rgba 0 0 0 0>}
        cornerRadii=<0 0 0 0>
       }
       tintColor {colorUseRule=UseColor_Specified, specifiedColor=<rgba 1 1 1 1>}
       uVRegion {bIsValid=false, max=<0 0>, min=<0 0>}
      }
      underlineBrush {
       bIsDynamicallyLoaded=false
       drawAs=Image
       imageType=NoImage
       mirroring=NoMirror
       resourceName=None
       resourceObject=null
       tiling=NoTile
       imageSize=<32 32>
       margin=<ltrb 0 0 0 0>
       outlineSettings {
        bUseBrushTransparency=false
        roundingType=HalfHeightRadius
        width=0
        color {colorUseRule=UseColor_Specified, specifiedColor=<rgba 0 0 0 0>}
        cornerRadii=<0 0 0 0>
       }
       tintColor {colorUseRule=UseColor_Specified, specifiedColor=<rgba 1 1 1 1>}
       uVRegion {bIsValid=false, max=<0 0>, min=<0 0>}
      }
     }
     vScrollBarPadding=<ltrb 0 0 0 0>
    }
   }
   slot {class=VerticalBoxSlot}
  }
  WidgetSwitcher_184 {
   widgetClassPath=&WidgetSwitcher
   bIsVariable=true
   parent=VerticalBox_146
   widget=WidgetSwitcher_184
   slot {class=VerticalBoxSlot}
  }
  CanvasPanel_36 {widgetClassPath=&CanvasPanel, parent=WidgetSwitcher_184, widget=CanvasPanel_36, slot {class=WidgetSwitcherSlot}}
  EditorUtilityScrollBox_597 {
   widgetClassPath=&EditorUtilityScrollBox
   parent=CanvasPanel_36
   widget=EditorUtilityScrollBox_597
   properties {alwaysShowScrollbar=true, alwaysShowScrollbarTrack=true, frontPadScrolling=true}
   slot {class=CanvasPanelSlot, properties {layoutData {alignment=<0 0>, anchors {maximum=<0 0>, minimum=<0 0>}, offsets=<ltrb 0 0 1280 1000>}}}
  }
  VerticalBox_0 {
   widgetClassPath=&VerticalBox
   parent=EditorUtilityScrollBox_597
   widget=VerticalBox_0
   slot {class=ScrollBoxSlot, properties {horizontalAlignment=HAlign_Left, padding=<ltrb 40 40 0 0>}}
  }
  ConverttoBlueprint {
   widgetClassPath=&EditorUtilityButton
   parent=VerticalBox_0
   widget=ConverttoBlueprint
   properties {toolTipText="Takes all  static meshes, finds the Cliff BP that owns that mesh in the Figment Environment plugin and replaces it wil it."}
   slot {class=VerticalBoxSlot, properties {horizontalAlignment=HAlign_Left, padding=<ltrb 4 4 4 4>}}
  }
  TextBlock_1 {
   widgetClassPath=&TextBlock
   parent=ConverttoBlueprint
   widget=TextBlock_1
   properties {
    text=Convert Static Mesh to Blueprint
    toolTipText="Takes all  static meshes, finds the Cliff BP that owns that mesh in the Figment Environment plugin and replaces it with the found one (use the button below to create from from an existitng mesh)"
    font {
     bForceMonospaced=false
     bMaterialIsStencil=false
     fontMaterial=null
     fontObject=&RobotoTiny
     letterSpacing=0
     monospacedWidth=1
     size=24
     skewAmount=0
     typefaceFontName=Light
     outlineSettings {
      bApplyOutlineToDropShadows=false
      bMiteredCorners=false
      bSeparateFillAlpha=false
      outlineMaterial=null
      outlineSize=0
      outlineColor=<rgba 0 0 0 1>
     }
    }
   }
   slot {class=ButtonSlot, properties {horizontalAlignment=HAlign_Left}}
  }
  ConverttoBlueprint2 {
   widgetClassPath=&EditorUtilityButton
   parent=VerticalBox_0
   widget=ConverttoBlueprint2
   properties {
    bIsEnabled=false
    toolTipText="Takes all  static meshes, finds the Cliff BP that owns that mesh in the Figment Environment plugin and replaces it wil it."
    visibility=Collapsed
   }
   slot {class=VerticalBoxSlot, properties {horizontalAlignment=HAlign_Left, padding=<ltrb 4 4 4 4>}}
  }
  TextBlock_6 {
   widgetClassPath=&TextBlock
   parent=ConverttoBlueprint2
   widget=TextBlock_6
   properties {
    text=Convert Static Mesh to Blueprint 2
    toolTipText="Takes all  static meshes, finds the Cliff BP that owns that mesh in the Figment Environment plugin and replaces it with the found one (use the button below to create from from an existitng mesh)"
    font {
     bForceMonospaced=false
     bMaterialIsStencil=false
     fontMaterial=null
     fontObject=&RobotoTiny
     letterSpacing=0
     monospacedWidth=1
     size=24
     skewAmount=0
     typefaceFontName=Light
     outlineSettings {
      bApplyOutlineToDropShadows=false
      bMiteredCorners=false
      bSeparateFillAlpha=false
      outlineMaterial=null
      outlineSize=0
      outlineColor=<rgba 0 0 0 1>
     }
    }
   }
   slot {class=ButtonSlot, properties {horizontalAlignment=HAlign_Left}}
  }
  CreateCliffBPfrommesh {
   widgetClassPath=&EditorUtilityButton
   parent=VerticalBox_0
   widget=CreateCliffBPfrommesh
   properties {toolTipText=This will take a selected mesh and make a BP asset from it.}
   slot {class=VerticalBoxSlot, properties {horizontalAlignment=HAlign_Left, padding=<ltrb 4 4 4 4>}}
  }
  TextBlock {
   widgetClassPath=&TextBlock
   parent=CreateCliffBPfrommesh
   widget=TextBlock
   properties {
    text=Create Cliff BP from static mesh
    toolTipText=This will take a selected mesh and make a BP asset from it.
    font {
     bForceMonospaced=false
     bMaterialIsStencil=false
     fontMaterial=null
     fontObject=&RobotoTiny
     letterSpacing=0
     monospacedWidth=1
     size=24
     skewAmount=0
     typefaceFontName=Light
     outlineSettings {
      bApplyOutlineToDropShadows=false
      bMiteredCorners=false
      bSeparateFillAlpha=false
      outlineMaterial=null
      outlineSize=0
      outlineColor=<rgba 0 0 0 1>
     }
    }
   }
   slot {class=ButtonSlot, properties {horizontalAlignment=HAlign_Left}}
  }
  MigrateTDsfromSearchPAth {
   widgetClassPath=&EditorUtilityButton
   parent=VerticalBox_0
   widget=MigrateTDsfromSearchPAth
   properties {toolTipText=Assumes that materials have been migrated friom UE4 original release. this will use the search names below and find the materials by name and replace them on the actors}
   slot {class=VerticalBoxSlot, properties {horizontalAlignment=HAlign_Left, padding=<ltrb 4 4 4 4>}}
  }
  TextBlock_21 {
   widgetClassPath=&TextBlock
   parent=MigrateTDsfromSearchPAth
   widget=TextBlock_21
   properties {
    text=Migrate TDs from Search Path and Apply
    toolTipText="Gets existing Cliff/Ground meshes and checks if the Texture Data are in correct ones in the Target. If not it will look for the correct TD/material (based on name) in the 'Search Material Path' list and hook them up. This assumes that old for static meshes have already been converted."
    font {
     bForceMonospaced=false
     bMaterialIsStencil=false
     fontMaterial=null
     fontObject=&RobotoTiny
     letterSpacing=0
     monospacedWidth=1
     size=24
     skewAmount=0
     typefaceFontName=Light
     outlineSettings {
      bApplyOutlineToDropShadows=false
      bMiteredCorners=false
      bSeparateFillAlpha=false
      outlineMaterial=null
      outlineSize=0
      outlineColor=<rgba 0 0 0 1>
     }
    }
   }
   slot {class=ButtonSlot, properties {horizontalAlignment=HAlign_Left}}
  }
  add_unmarked {
   widgetClassPath=&EditorUtilityButton
   parent=VerticalBox_0
   widget=add_unmarked
   slot {class=VerticalBoxSlot, properties {horizontalAlignment=HAlign_Left, padding=<ltrb 4 4 4 4>}}
  }
  ChangeMPC_1 {
   widgetClassPath=&TextBlock
   parent=add_unmarked
   widget=ChangeMPC_1
   properties {
    text=Add Unmarked
    toolTipText="When migrating assets from UE4 to UE5 this will add the new files to revision control. "
    font {
     bForceMonospaced=false
     bMaterialIsStencil=false
     fontMaterial=null
     fontObject=&RobotoTiny
     letterSpacing=0
     monospacedWidth=1
     size=24
     skewAmount=0
     typefaceFontName=Light
     outlineSettings {
      bApplyOutlineToDropShadows=false
      bMiteredCorners=false
      bSeparateFillAlpha=false
      outlineMaterial=null
      outlineSize=0
      outlineColor=<rgba 0 0 0 1>
     }
    }
   }
   slot {class=ButtonSlot, properties {horizontalAlignment=HAlign_Left}}
  }
  DeleteMG {
   widgetClassPath=&EditorUtilityButton
   parent=VerticalBox_0
   widget=DeleteMG
   slot {class=VerticalBoxSlot, properties {horizontalAlignment=HAlign_Left, padding=<ltrb 4 4 4 4>}}
  }
  DeleteMGT {
   widgetClassPath=&TextBlock
   parent=DeleteMG
   widget=DeleteMGT
   properties {
    text=Delete Mesh Grass
    toolTipText="This Deletes all mesh grass actors from the Map. Useful when going to the next season and we have stale actors left over from the last season (and there are duplicate BP_MeshGrass actors"
    font {
     bForceMonospaced=false
     bMaterialIsStencil=false
     fontMaterial=null
     fontObject=&RobotoTiny
     letterSpacing=0
     monospacedWidth=1
     size=24
     skewAmount=0
     typefaceFontName=Light
     outlineSettings {
      bApplyOutlineToDropShadows=false
      bMiteredCorners=false
      bSeparateFillAlpha=false
      outlineMaterial=null
      outlineSize=0
      outlineColor=<rgba 0 0 0 1>
     }
    }
   }
   slot {class=ButtonSlot, properties {horizontalAlignment=HAlign_Left}}
  }
  HorizontalBox_1 {
   widgetClassPath=&HorizontalBox
   parent=VerticalBox_0
   widget=HorizontalBox_1
   slot {class=VerticalBoxSlot, properties {padding=<ltrb 4 4 4 4>}}
  }
  VerticalBox_339 {
   widgetClassPath=&VerticalBox
   parent=HorizontalBox_1
   widget=VerticalBox_339
   slot {class=HorizontalBoxSlot, properties {verticalAlignment=VAlign_Center}}
  }
  HorizontalBox_136 {
   widgetClassPath=&HorizontalBox
   parent=VerticalBox_339
   widget=HorizontalBox_136
   slot {class=VerticalBoxSlot, properties {padding=<ltrb 4 4 4 4>}}
  }
  TextBlock_20 {
   widgetClassPath=&TextBlock
   parent=HorizontalBox_136
   widget=TextBlock_20
   properties {
    text=Search Material Path:
    font {
     bForceMonospaced=false
     bMaterialIsStencil=false
     fontMaterial=null
     fontObject=&RobotoTiny
     letterSpacing=0
     monospacedWidth=1
     size=24
     skewAmount=0
     typefaceFontName=Light
     outlineSettings {
      bApplyOutlineToDropShadows=false
      bMiteredCorners=false
      bSeparateFillAlpha=false
      outlineMaterial=null
      outlineSize=0
      outlineColor=<rgba 0 0 0 1>
     }
    }
   }
   slot {class=HorizontalBoxSlot, properties {padding=<ltrb 12 12 12 12>}}
  }
  Border_182 {
   widgetClassPath=&Border
   parent=HorizontalBox_136
   widget=Border_182
   properties {
    background {
     bIsDynamicallyLoaded=false
     drawAs=RoundedBox
     imageType=NoImage
     mirroring=NoMirror
     resourceName=None
     resourceObject=null
     tiling=NoTile
     imageSize=<32 32>
     margin=<ltrb 0 0 0 0>
     outlineSettings {
      bUseBrushTransparency=false
      roundingType=FixedRadius
      width=0
      color {colorUseRule=UseColor_Specified, specifiedColor=<rgba 0 0 0 0>}
      cornerRadii=<12 12 12 12>
     }
     tintColor {colorUseRule=UseColor_Specified, specifiedColor=<rgba 1 1 1 1>}
     uVRegion {bIsValid=false, max=<0 0>, min=<0 0>}
    }
    brushColor=<rgba 0.0469 0.0625 0.0469 1>
    padding=<ltrb 6 6 6 6>
   }
   slot {class=HorizontalBoxSlot}
  }
  EditorUtilityMultiLineEditableText_101 {
   widgetClassPath=/Script/Blutility.EditorUtilityMultiLineEditableText
   bIsVariable=true
   parent=Border_182
   widget=EditorUtilityMultiLineEditableText_101
   properties {
    autoWrapText=false
    lineHeightPercentage=1.2
    text="/FigmentEnvironment/LM_8_1_1/Athena/Environments/Landscape/Material/\r\n/FigmentEnvironment/LM_8_1_1/Environments/World/Sidewalks/Materials/\r\n/FigmentEnvironment/LM_8_1_1/Environments/Sets/JungleTemple/"
    margin=<ltrb 2 2 2 2>
    widgetStyle {
     overflowPolicy=Clip
     transformPolicy=None
     colorAndOpacity {colorUseRule=UseColor_Specified, specifiedColor=<rgba 1 0.5542 0.2916 1>}
     font {
      bForceMonospaced=false
      bMaterialIsStencil=false
      fontMaterial=null
      fontObject=&RobotoTiny
      letterSpacing=0
      monospacedWidth=1
      size=18
      skewAmount=0
      typefaceFontName=Light
      outlineSettings {
       bApplyOutlineToDropShadows=false
       bMiteredCorners=false
       bSeparateFillAlpha=false
       outlineMaterial=null
       outlineSize=0
       outlineColor=<rgba 0.0365 0.0365 0.0365 1>
      }
     }
     highlightColor {colorUseRule=UseColor_ColorTable, specifiedColor=<rgba 0 0 0 0>}
     highlightShape {
      bIsDynamicallyLoaded=false
      drawAs=Image
      imageType=NoImage
      mirroring=NoMirror
      resourceName=None
      resourceObject=null
      tiling=NoTile
      imageSize=<0 0>
      margin=<ltrb 0 0 0 0>
      outlineSettings {
       bUseBrushTransparency=false
       roundingType=HalfHeightRadius
       width=0
       color {colorUseRule=UseColor_Specified, specifiedColor=<rgba 0 0 0 0>}
       cornerRadii=<0 0 0 0>
      }
      tintColor {colorUseRule=UseColor_Specified, specifiedColor=<rgba 0.2582 0.5395 0.0685 1>}
      uVRegion {bIsValid=false, max=<0 0>, min=<0 0>}
     }
     selectedBackgroundColor {colorUseRule=UseColor_Specified, specifiedColor=<rgba 0 0.162 0.7454 1>}
     shadowColorAndOpacity=<rgba 0 0 0 1>
     shadowOffset=<0 0>
     strikeBrush {
      bIsDynamicallyLoaded=false
      drawAs=Image
      imageType=NoImage
      mirroring=NoMirror
      resourceName=None
      resourceObject=null
      tiling=NoTile
      imageSize=<32 32>
      margin=<ltrb 0 0 0 0>
      outlineSettings {
       bUseBrushTransparency=false
       roundingType=HalfHeightRadius
       width=0
       color {colorUseRule=UseColor_Specified, specifiedColor=<rgba 0 0 0 0>}
       cornerRadii=<0 0 0 0>
      }
      tintColor {colorUseRule=UseColor_Specified, specifiedColor=<rgba 1 1 1 1>}
      uVRegion {bIsValid=false, max=<0 0>, min=<0 0>}
     }
     underlineBrush {
      bIsDynamicallyLoaded=false
      drawAs=Image
      imageType=NoImage
      mirroring=NoMirror
      resourceName=None
      resourceObject=null
      tiling=NoTile
      imageSize=<32 32>
      margin=<ltrb 0 0 0 0>
      outlineSettings {
       bUseBrushTransparency=false
       roundingType=HalfHeightRadius
       width=0
       color {colorUseRule=UseColor_Specified, specifiedColor=<rgba 0 0 0 0>}
       cornerRadii=<0 0 0 0>
      }
      tintColor {colorUseRule=UseColor_Specified, specifiedColor=<rgba 1 1 1 1>}
      uVRegion {bIsValid=false, max=<0 0>, min=<0 0>}
     }
    }
   }
   slot {class=BorderSlot, properties {padding=<ltrb 6 6 6 6>}}
  }
  HorizontalBox_57 {
   widgetClassPath=&HorizontalBox
   parent=VerticalBox_339
   widget=HorizontalBox_57
   slot {class=VerticalBoxSlot, properties {padding=<ltrb 4 4 4 4>}}
  }
  TextBlock_192 {
   widgetClassPath=&TextBlock
   parent=HorizontalBox_57
   widget=TextBlock_192
   properties {
    text=Target TD Path:
    font {
     bForceMonospaced=false
     bMaterialIsStencil=false
     fontMaterial=null
     fontObject=&RobotoTiny
     letterSpacing=0
     monospacedWidth=1
     size=24
     skewAmount=0
     typefaceFontName=Light
     outlineSettings {
      bApplyOutlineToDropShadows=false
      bMiteredCorners=false
      bSeparateFillAlpha=false
      outlineMaterial=null
      outlineSize=0
      outlineColor=<rgba 0 0 0 1>
     }
    }
   }
   slot {class=HorizontalBoxSlot, properties {padding=<ltrb 12 12 12 12>}}
  }
  TargetTextBox_ {
   widgetClassPath=&EditorUtilityEditableTextBox
   bIsVariable=true
   parent=HorizontalBox_57
   widget=TargetTextBox_
   properties {
    text=/FigmentEnvironment/LM_8_1_1/Athena/Environments/Landscape/TextureData/
    widgetStyle {
     backgroundColor {colorUseRule=UseColor_Specified, specifiedColor=<rgba 1 1 1 1>}
     backgroundImageFocused {
      bIsDynamicallyLoaded=false
      drawAs=RoundedBox
      imageType=NoImage
      mirroring=NoMirror
      resourceName=None
      resourceObject=null
      tiling=NoTile
      imageSize=<0 0>
      margin=<ltrb 0 0 0 0>
      outlineSettings {
       bUseBrushTransparency=false
       roundingType=FixedRadius
       width=1
       color {colorUseRule=UseColor_ColorTable, specifiedColor=<rgba 0 0 0 0>}
       cornerRadii=<4 4 4 4>
      }
      tintColor {colorUseRule=UseColor_Specified, specifiedColor=<rgba 0.0048 0.0048 0.0048 1>}
      uVRegion {bIsValid=false, max=<0 0>, min=<0 0>}
     }
     backgroundImageHovered {
      bIsDynamicallyLoaded=false
      drawAs=RoundedBox
      imageType=NoImage
      mirroring=NoMirror
      resourceName=None
      resourceObject=null
      tiling=NoTile
      imageSize=<0 0>
      margin=<ltrb 0 0 0 0>
      outlineSettings {
       bUseBrushTransparency=false
       roundingType=FixedRadius
       width=1
       color {colorUseRule=UseColor_ColorTable, specifiedColor=<rgba 0 0 0 0>}
       cornerRadii=<4 4 4 4>
      }
      tintColor {colorUseRule=UseColor_Specified, specifiedColor=<rgba 0.0048 0.0048 0.0048 1>}
      uVRegion {bIsValid=false, max=<0 0>, min=<0 0>}
     }
     backgroundImageNormal {
      bIsDynamicallyLoaded=false
      drawAs=RoundedBox
      imageType=NoImage
      mirroring=NoMirror
      resourceName=None
      resourceObject=null
      tiling=NoTile
      imageSize=<0 0>
      margin=<ltrb 0 0 0 0>
      outlineSettings {
       bUseBrushTransparency=false
       roundingType=FixedRadius
       width=1
       color {colorUseRule=UseColor_ColorTable, specifiedColor=<rgba 0 0 0 0>}
       cornerRadii=<4 4 4 4>
      }
      tintColor {colorUseRule=UseColor_Specified, specifiedColor=<rgba 0.0048 0.0048 0.0048 1>}
      uVRegion {bIsValid=false, max=<0 0>, min=<0 0>}
     }
     backgroundImageReadOnly {
      bIsDynamicallyLoaded=false
      drawAs=RoundedBox
      imageType=NoImage
      mirroring=NoMirror
      resourceName=None
      resourceObject=null
      tiling=NoTile
      imageSize=<0 0>
      margin=<ltrb 0 0 0 0>
      outlineSettings {
       bUseBrushTransparency=false
       roundingType=FixedRadius
       width=1
       color {colorUseRule=UseColor_ColorTable, specifiedColor=<rgba 0 0 0 0>}
       cornerRadii=<4 4 4 4>
      }
      tintColor {colorUseRule=UseColor_Specified, specifiedColor=<rgba 0.0284 0.0284 0.0284 1>}
      uVRegion {bIsValid=false, max=<0 0>, min=<0 0>}
     }
     focusedForegroundColor {colorUseRule=UseColor_Specified, specifiedColor=<rgba 1 1 1 1>}
     foregroundColor {colorUseRule=UseColor_Specified, specifiedColor=<rgba 0.5271 0.5271 0.5271 1>}
     hScrollBarPadding=<ltrb 0 0 0 0>
     padding=<ltrb 8 3 8 4>
     readOnlyForegroundColor {colorUseRule=UseColor_Foreground, specifiedColor=<rgba 1 0 1 1>}
     scrollBarStyle {
      thickness=8
      draggedThumbImage {
       bIsDynamicallyLoaded=false
       drawAs=RoundedBox
       imageType=NoImage
       mirroring=NoMirror
       resourceName=None
       resourceObject=null
       tiling=NoTile
       imageSize=<0 0>
       margin=<ltrb 0 0 0 0>
       outlineSettings {
        bUseBrushTransparency=false
        roundingType=FixedRadius
        width=0
        color {colorUseRule=UseColor_Specified, specifiedColor=<rgba 0 0 0 0>}
        cornerRadii=<4 4 4 4>
       }
       tintColor {colorUseRule=UseColor_Specified, specifiedColor=<rgba 0.2159 0.2159 0.2159 1>}
       uVRegion {bIsValid=false, max=<0 0>, min=<0 0>}
      }
      horizontalBackgroundImage {
       bIsDynamicallyLoaded=false
       drawAs=NoDrawType
       imageType=NoImage
       mirroring=NoMirror
       resourceName=None
       resourceObject=null
       tiling=NoTile
       imageSize=<0 0>
       margin=<ltrb 0 0 0 0>
       outlineSettings {
        bUseBrushTransparency=false
        roundingType=HalfHeightRadius
        width=0
        color {colorUseRule=UseColor_Specified, specifiedColor=<rgba 0 0 0 0>}
        cornerRadii=<0 0 0 0>
       }
       tintColor {colorUseRule=UseColor_Specified, specifiedColor=<rgba 1 1 1 1>}
       uVRegion {bIsValid=false, max=<0 0>, min=<0 0>}
      }
      horizontalBottomSlotImage {
       bIsDynamicallyLoaded=false
       drawAs=NoDrawType
       imageType=NoImage
       mirroring=NoMirror
       resourceName=None
       resourceObject=null
       tiling=NoTile
       imageSize=<0 0>
       margin=<ltrb 0 0 0 0>
       outlineSettings {
        bUseBrushTransparency=false
        roundingType=HalfHeightRadius
        width=0
        color {colorUseRule=UseColor_Specified, specifiedColor=<rgba 0 0 0 0>}
        cornerRadii=<0 0 0 0>
       }
       tintColor {colorUseRule=UseColor_Specified, specifiedColor=<rgba 1 1 1 1>}
       uVRegion {bIsValid=false, max=<0 0>, min=<0 0>}
      }
      horizontalTopSlotImage {
       bIsDynamicallyLoaded=false
       drawAs=NoDrawType
       imageType=NoImage
       mirroring=NoMirror
       resourceName=None
       resourceObject=null
       tiling=NoTile
       imageSize=<0 0>
       margin=<ltrb 0 0 0 0>
       outlineSettings {
        bUseBrushTransparency=false
        roundingType=HalfHeightRadius
        width=0
        color {colorUseRule=UseColor_Specified, specifiedColor=<rgba 0 0 0 0>}
        cornerRadii=<0 0 0 0>
       }
       tintColor {colorUseRule=UseColor_Specified, specifiedColor=<rgba 1 1 1 1>}
       uVRegion {bIsValid=false, max=<0 0>, min=<0 0>}
      }
      hoveredThumbImage {
       bIsDynamicallyLoaded=false
       drawAs=RoundedBox
       imageType=NoImage
       mirroring=NoMirror
       resourceName=None
       resourceObject=null
       tiling=NoTile
       imageSize=<0 0>
       margin=<ltrb 0 0 0 0>
       outlineSettings {
        bUseBrushTransparency=false
        roundingType=FixedRadius
        width=0
        color {colorUseRule=UseColor_Specified, specifiedColor=<rgba 0 0 0 0>}
        cornerRadii=<4 4 4 4>
       }
       tintColor {colorUseRule=UseColor_Specified, specifiedColor=<rgba 0.2159 0.2159 0.2159 1>}
       uVRegion {bIsValid=false, max=<0 0>, min=<0 0>}
      }
      normalThumbImage {
       bIsDynamicallyLoaded=false
       drawAs=RoundedBox
       imageType=NoImage
       mirroring=NoMirror
       resourceName=None
       resourceObject=null
       tiling=NoTile
       imageSize=<0 0>
       margin=<ltrb 0 0 0 0>
       outlineSettings {
        bUseBrushTransparency=false
        roundingType=FixedRadius
        width=0
        color {colorUseRule=UseColor_Specified, specifiedColor=<rgba 0 0 0 0>}
        cornerRadii=<4 4 4 4>
       }
       tintColor {colorUseRule=UseColor_Specified, specifiedColor=<rgba 0.0953 0.0953 0.0953 1>}
       uVRegion {bIsValid=false, max=<0 0>, min=<0 0>}
      }
      verticalBackgroundImage {
       bIsDynamicallyLoaded=false
       drawAs=NoDrawType
       imageType=NoImage
       mirroring=NoMirror
       resourceName=None
       resourceObject=null
       tiling=NoTile
       imageSize=<0 0>
       margin=<ltrb 0 0 0 0>
       outlineSettings {
        bUseBrushTransparency=false
        roundingType=HalfHeightRadius
        width=0
        color {colorUseRule=UseColor_Specified, specifiedColor=<rgba 0 0 0 0>}
        cornerRadii=<0 0 0 0>
       }
       tintColor {colorUseRule=UseColor_Specified, specifiedColor=<rgba 1 1 1 1>}
       uVRegion {bIsValid=false, max=<0 0>, min=<0 0>}
      }
      verticalBottomSlotImage {
       bIsDynamicallyLoaded=false
       drawAs=NoDrawType
       imageType=NoImage
       mirroring=NoMirror
       resourceName=None
       resourceObject=null
       tiling=NoTile
       imageSize=<0 0>
       margin=<ltrb 0 0 0 0>
       outlineSettings {
        bUseBrushTransparency=false
        roundingType=HalfHeightRadius
        width=0
        color {colorUseRule=UseColor_Specified, specifiedColor=<rgba 0 0 0 0>}
        cornerRadii=<0 0 0 0>
       }
       tintColor {colorUseRule=UseColor_Specified, specifiedColor=<rgba 1 1 1 1>}
       uVRegion {bIsValid=false, max=<0 0>, min=<0 0>}
      }
      verticalTopSlotImage {
       bIsDynamicallyLoaded=false
       drawAs=NoDrawType
       imageType=NoImage
       mirroring=NoMirror
       resourceName=None
       resourceObject=null
       tiling=NoTile
       imageSize=<0 0>
       margin=<ltrb 0 0 0 0>
       outlineSettings {
        bUseBrushTransparency=false
        roundingType=HalfHeightRadius
        width=0
        color {colorUseRule=UseColor_Specified, specifiedColor=<rgba 0 0 0 0>}
        cornerRadii=<0 0 0 0>
       }
       tintColor {colorUseRule=UseColor_Specified, specifiedColor=<rgba 1 1 1 1>}
       uVRegion {bIsValid=false, max=<0 0>, min=<0 0>}
      }
     }
     textStyle {
      overflowPolicy=Clip
      transformPolicy=None
      colorAndOpacity {colorUseRule=UseColor_Specified, specifiedColor=<rgba 1 0.2717 0.265 1>}
      font {
       bForceMonospaced=false
       bMaterialIsStencil=false
       fontMaterial=null
       fontObject=&RobotoTiny
       letterSpacing=0
       monospacedWidth=1
       size=18
       skewAmount=0
       typefaceFontName=Light
       outlineSettings {
        bApplyOutlineToDropShadows=false
        bMiteredCorners=false
        bSeparateFillAlpha=false
        outlineMaterial=null
        outlineSize=0
        outlineColor=<rgba 0 0 0 1>
       }
      }
      highlightColor {colorUseRule=UseColor_ColorTable, specifiedColor=<rgba 0 0 0 0>}
      highlightShape {
       bIsDynamicallyLoaded=false
       drawAs=Image
       imageType=NoImage
       mirroring=NoMirror
       resourceName=None
       resourceObject=null
       tiling=NoTile
       imageSize=<0 0>
       margin=<ltrb 0 0 0 0>
       outlineSettings {
        bUseBrushTransparency=false
        roundingType=HalfHeightRadius
        width=0
        color {colorUseRule=UseColor_Specified, specifiedColor=<rgba 0 0 0 0>}
        cornerRadii=<0 0 0 0>
       }
       tintColor {colorUseRule=UseColor_ColorTable, specifiedColor=<rgba 0 0 0 0>}
       uVRegion {bIsValid=false, max=<0 0>, min=<0 0>}
      }
      selectedBackgroundColor {colorUseRule=UseColor_ColorTable, specifiedColor=<rgba 0 0 0 0>}
      shadowColorAndOpacity=<rgba 0 0 0 1>
      shadowOffset=<0 0>
      strikeBrush {
       bIsDynamicallyLoaded=false
       drawAs=Image
       imageType=NoImage
       mirroring=NoMirror
       resourceName=None
       resourceObject=null
       tiling=NoTile
       imageSize=<32 32>
       margin=<ltrb 0 0 0 0>
       outlineSettings {
        bUseBrushTransparency=false
        roundingType=HalfHeightRadius
        width=0
        color {colorUseRule=UseColor_Specified, specifiedColor=<rgba 0 0 0 0>}
        cornerRadii=<0 0 0 0>
       }
       tintColor {colorUseRule=UseColor_Specified, specifiedColor=<rgba 1 1 1 1>}
       uVRegion {bIsValid=false, max=<0 0>, min=<0 0>}
      }
      underlineBrush {
       bIsDynamicallyLoaded=false
       drawAs=Image
       imageType=NoImage
       mirroring=NoMirror
       resourceName=None
       resourceObject=null
       tiling=NoTile
       imageSize=<32 32>
       margin=<ltrb 0 0 0 0>
       outlineSettings {
        bUseBrushTransparency=false
        roundingType=HalfHeightRadius
        width=0
        color {colorUseRule=UseColor_Specified, specifiedColor=<rgba 0 0 0 0>}
        cornerRadii=<0 0 0 0>
       }
       tintColor {colorUseRule=UseColor_Specified, specifiedColor=<rgba 1 1 1 1>}
       uVRegion {bIsValid=false, max=<0 0>, min=<0 0>}
      }
     }
     vScrollBarPadding=<ltrb 0 0 0 0>
    }
   }
   slot {class=HorizontalBoxSlot, properties {size {sizeRule=Fill, value=1}}}
  }
  Spacer_2 {widgetClassPath=&Spacer, parent=VerticalBox_0, widget=Spacer_2, properties {size=<1 48>}, slot {class=VerticalBoxSlot}}
  Spacer_3 {widgetClassPath=&Spacer, parent=VerticalBox_0, widget=Spacer_3, properties {size=<1 48>}, slot {class=VerticalBoxSlot}}
  CanvasPanel_84 {
   widgetClassPath=&CanvasPanel
   parent=WidgetSwitcher_184
   widget=CanvasPanel_84
   slot {class=WidgetSwitcherSlot, properties {padding=<ltrb 40 40 0 0>}}
  }
  EditorUtilityScrollBox_54 {
   widgetClassPath=&EditorUtilityScrollBox
   parent=CanvasPanel_84
   widget=EditorUtilityScrollBox_54
   properties {alwaysShowScrollbar=true, alwaysShowScrollbarTrack=true, consumeMouseWheel=Always, frontPadScrolling=true, scrollbarThickness=<24 9>}
   slot {class=CanvasPanelSlot, properties {layoutData {alignment=<0 0>, anchors {maximum=<0 0>, minimum=<0 0>}, offsets=<ltrb 0 0 1280 1200>}}}
  }
  VerticalBox_59 {
   widgetClassPath=&VerticalBox
   parent=EditorUtilityScrollBox_54
   widget=VerticalBox_59
   slot {class=ScrollBoxSlot, properties {size {sizeRule=Fill, value=1}}}
  }
  CreateMatchingBP {
   widgetClassPath=&EditorUtilityButton
   parent=VerticalBox_59
   widget=CreateMatchingBP
   properties {bIsEnabled=false}
   slot {class=VerticalBoxSlot}
  }
  TextBlock_5 {
   widgetClassPath=&TextBlock
   parent=CreateMatchingBP
   widget=TextBlock_5
   properties {
    text=CreateMatchingBP
    font {
     bForceMonospaced=false
     bMaterialIsStencil=false
     fontMaterial=null
     fontObject=&RobotoTiny
     letterSpacing=0
     monospacedWidth=1
     size=24
     skewAmount=0
     typefaceFontName=Light
     outlineSettings {
      bApplyOutlineToDropShadows=false
      bMiteredCorners=false
      bSeparateFillAlpha=false
      outlineMaterial=null
      outlineSize=0
      outlineColor=<rgba 0 0 0 1>
     }
    }
   }
   slot {class=ButtonSlot, properties {horizontalAlignment=HAlign_Left}}
  }
  FixTDDependencies {widgetClassPath=&EditorUtilityButton, parent=VerticalBox_59, widget=FixTDDependencies, slot {class=VerticalBoxSlot}}
  FixTD {
   widgetClassPath=&TextBlock
   parent=FixTDDependencies
   widget=FixTD
   properties {
    text=Fix TD Dependencies
    font {
     bForceMonospaced=false
     bMaterialIsStencil=false
     fontMaterial=null
     fontObject=&RobotoTiny
     letterSpacing=0
     monospacedWidth=1
     size=24
     skewAmount=0
     typefaceFontName=Light
     outlineSettings {
      bApplyOutlineToDropShadows=false
      bMiteredCorners=false
      bSeparateFillAlpha=false
      outlineMaterial=null
      outlineSize=0
      outlineColor=<rgba 0 0 0 1>
     }
    }
   }
   slot {class=ButtonSlot, properties {horizontalAlignment=HAlign_Left}}
  }
  SwapSnowMats {
   widgetClassPath=&EditorUtilityButton
   parent=VerticalBox_59
   widget=SwapSnowMats
   slot {class=VerticalBoxSlot, properties {horizontalAlignment=HAlign_Left, padding=<ltrb 4 4 4 4>}}
  }
  Swapsnow_Mats_ {
   widgetClassPath=&TextBlock
   parent=SwapSnowMats
   widget=Swapsnow_Mats_
   properties {
    text=SwapSnowMaterials
    font {
     bForceMonospaced=false
     bMaterialIsStencil=false
     fontMaterial=null
     fontObject=&RobotoTiny
     letterSpacing=0
     monospacedWidth=1
     size=24
     skewAmount=0
     typefaceFontName=Light
     outlineSettings {
      bApplyOutlineToDropShadows=false
      bMiteredCorners=false
      bSeparateFillAlpha=false
      outlineMaterial=null
      outlineSize=0
      outlineColor=<rgba 0 0 0 1>
     }
    }
   }
   slot {class=ButtonSlot, properties {horizontalAlignment=HAlign_Left}}
  }
  ASsignIceHLOD {
   widgetClassPath=&EditorUtilityButton
   parent=VerticalBox_59
   widget=ASsignIceHLOD
   slot {class=VerticalBoxSlot, properties {horizontalAlignment=HAlign_Left, padding=<ltrb 4 4 4 4>}}
  }
  IceHLOD {
   widgetClassPath=&TextBlock
   parent=ASsignIceHLOD
   widget=IceHLOD
   properties {
    text=AssignHLODLayer
    font {
     bForceMonospaced=false
     bMaterialIsStencil=false
     fontMaterial=null
     fontObject=&RobotoTiny
     letterSpacing=0
     monospacedWidth=1
     size=24
     skewAmount=0
     typefaceFontName=Light
     outlineSettings {
      bApplyOutlineToDropShadows=false
      bMiteredCorners=false
      bSeparateFillAlpha=false
      outlineMaterial=null
      outlineSize=0
      outlineColor=<rgba 0 0 0 1>
     }
    }
   }
   slot {class=ButtonSlot, properties {horizontalAlignment=HAlign_Left}}
  }
  ClearTagsInLevel {widgetClassPath=&EditorUtilityButton, parent=VerticalBox_59, widget=ClearTagsInLevel, slot {class=VerticalBoxSlot}}
  TextBlock_10 {
   widgetClassPath=&TextBlock
   parent=ClearTagsInLevel
   widget=TextBlock_10
   properties {
    text=ClearTagsInLevel
    font {
     bForceMonospaced=false
     bMaterialIsStencil=false
     fontMaterial=null
     fontObject=&RobotoTiny
     letterSpacing=0
     monospacedWidth=1
     size=24
     skewAmount=0
     typefaceFontName=Light
     outlineSettings {
      bApplyOutlineToDropShadows=false
      bMiteredCorners=false
      bSeparateFillAlpha=false
      outlineMaterial=null
      outlineSize=0
      outlineColor=<rgba 0 0 0 1>
     }
    }
   }
   slot {class=ButtonSlot, properties {horizontalAlignment=HAlign_Left}}
  }
  DeleteFoliage {widgetClassPath=&EditorUtilityButton, parent=VerticalBox_59, widget=DeleteFoliage, slot {class=VerticalBoxSlot}}
  TextBlock_2 {
   widgetClassPath=&TextBlock
   parent=DeleteFoliage
   widget=TextBlock_2
   properties {
    text=DeleteFoliageActorsInLevel
    font {
     bForceMonospaced=false
     bMaterialIsStencil=false
     fontMaterial=null
     fontObject=&RobotoTiny
     letterSpacing=0
     monospacedWidth=1
     size=24
     skewAmount=0
     typefaceFontName=Light
     outlineSettings {
      bApplyOutlineToDropShadows=false
      bMiteredCorners=false
      bSeparateFillAlpha=false
      outlineMaterial=null
      outlineSize=0
      outlineColor=<rgba 0 0 0 1>
     }
    }
   }
   slot {class=ButtonSlot, properties {horizontalAlignment=HAlign_Left}}
  }
  HorizontalBox_3 {widgetClassPath=&HorizontalBox, parent=VerticalBox_59, widget=HorizontalBox_3, slot {class=VerticalBoxSlot}}
  DetectPhysMatInLevel {widgetClassPath=&EditorUtilityButton, parent=HorizontalBox_3, widget=DetectPhysMatInLevel, slot {class=HorizontalBoxSlot}}
  DetectPhysMatInLevel_text {
   widgetClassPath=&TextBlock
   parent=DetectPhysMatInLevel
   widget=DetectPhysMatInLevel_text
   properties {
    text=DetectPhysMatInLevel
    font {
     bForceMonospaced=false
     bMaterialIsStencil=false
     fontMaterial=null
     fontObject=&RobotoTiny
     letterSpacing=0
     monospacedWidth=1
     size=24
     skewAmount=0
     typefaceFontName=Light
     outlineSettings {
      bApplyOutlineToDropShadows=false
      bMiteredCorners=false
      bSeparateFillAlpha=false
      outlineMaterial=null
      outlineSize=0
      outlineColor=<rgba 0 0 0 1>
     }
    }
   }
   slot {class=ButtonSlot, properties {horizontalAlignment=HAlign_Left}}
  }
  Spacer_5 {widgetClassPath=&Spacer, parent=HorizontalBox_3, widget=Spacer_5, properties {size=<12 1>}, slot {class=HorizontalBoxSlot}}
  TextBlock_16 {
   widgetClassPath=&TextBlock
   parent=HorizontalBox_3
   widget=TextBlock_16
   properties {
    text=apply to all in current level:
    font {
     bForceMonospaced=false
     bMaterialIsStencil=false
     fontMaterial=null
     fontObject=&RobotoTiny
     letterSpacing=0
     monospacedWidth=1
     size=12
     skewAmount=0
     typefaceFontName=Light
     outlineSettings {
      bApplyOutlineToDropShadows=false
      bMiteredCorners=false
      bSeparateFillAlpha=false
      outlineMaterial=null
      outlineSize=0
      outlineColor=<rgba 0 0 0 1>
     }
    }
   }
   slot {class=HorizontalBoxSlot, properties {verticalAlignment=VAlign_Center}}
  }
  Spacer_6 {widgetClassPath=&Spacer, parent=HorizontalBox_3, widget=Spacer_6, properties {size=<4 1>}, slot {class=HorizontalBoxSlot}}
  SMConvertApplyToAll_1 {
   widgetClassPath=&EditorUtilityCheckBox
   bIsVariable=true
   parent=HorizontalBox_3
   widget=SMConvertApplyToAll_1
   slot {class=HorizontalBoxSlot}
  }
  DeleteFoliage_1 {widgetClassPath=&EditorUtilityButton, parent=VerticalBox_59, widget=DeleteFoliage_1, slot {class=VerticalBoxSlot}}
  TextBlock_9 {
   widgetClassPath=&TextBlock
   parent=DeleteFoliage_1
   widget=TextBlock_9
   properties {
    text=DeleteFoliageActorsInPath
    font {
     bForceMonospaced=false
     bMaterialIsStencil=false
     fontMaterial=null
     fontObject=&RobotoTiny
     letterSpacing=0
     monospacedWidth=1
     size=24
     skewAmount=0
     typefaceFontName=Light
     outlineSettings {
      bApplyOutlineToDropShadows=false
      bMiteredCorners=false
      bSeparateFillAlpha=false
      outlineMaterial=null
      outlineSize=0
      outlineColor=<rgba 0 0 0 1>
     }
    }
   }
   slot {class=ButtonSlot, properties {horizontalAlignment=HAlign_Left}}
  }
  FactoryWallWPOInPath {widgetClassPath=&EditorUtilityButton, parent=VerticalBox_59, widget=FactoryWallWPOInPath, slot {class=VerticalBoxSlot}}
  TextBlock_13 {
   widgetClassPath=&TextBlock
   parent=FactoryWallWPOInPath
   widget=TextBlock_13
   properties {
    text=FactoryWallWPOInPath
    font {
     bForceMonospaced=false
     bMaterialIsStencil=false
     fontMaterial=null
     fontObject=&RobotoTiny
     letterSpacing=0
     monospacedWidth=1
     size=24
     skewAmount=0
     typefaceFontName=Light
     outlineSettings {
      bApplyOutlineToDropShadows=false
      bMiteredCorners=false
      bSeparateFillAlpha=false
      outlineMaterial=null
      outlineSize=0
      outlineColor=<rgba 0 0 0 1>
     }
    }
   }
   slot {class=ButtonSlot, properties {horizontalAlignment=HAlign_Left}}
  }
  ClearTagsInPath {widgetClassPath=&EditorUtilityButton, parent=VerticalBox_59, widget=ClearTagsInPath, slot {class=VerticalBoxSlot}}
  TextBlock_11 {
   widgetClassPath=&TextBlock
   parent=ClearTagsInPath
   widget=TextBlock_11
   properties {
    text=ClearTagsInPath
    font {
     bForceMonospaced=false
     bMaterialIsStencil=false
     fontMaterial=null
     fontObject=&RobotoTiny
     letterSpacing=0
     monospacedWidth=1
     size=24
     skewAmount=0
     typefaceFontName=Light
     outlineSettings {
      bApplyOutlineToDropShadows=false
      bMiteredCorners=false
      bSeparateFillAlpha=false
      outlineMaterial=null
      outlineSize=0
      outlineColor=<rgba 0 0 0 1>
     }
    }
   }
   slot {class=ButtonSlot, properties {horizontalAlignment=HAlign_Left}}
  }
  Border_47 {
   widgetClassPath=&Border
   parent=VerticalBox_59
   widget=Border_47
   properties {
    background {
     bIsDynamicallyLoaded=false
     drawAs=RoundedBox
     imageType=NoImage
     mirroring=NoMirror
     resourceName=None
     resourceObject=null
     tiling=NoTile
     imageSize=<32 32>
     margin=<ltrb 0 0 0 0>
     outlineSettings {
      bUseBrushTransparency=false
      roundingType=FixedRadius
      width=1.8204
      color {colorUseRule=UseColor_Specified, specifiedColor=<rgba 1 1 1 1>}
      cornerRadii=<20 20 20 20>
     }
     tintColor {colorUseRule=UseColor_Specified, specifiedColor=<rgba 0.0802 0.0808 0.0885 1>}
     uVRegion {bIsValid=false, max=<0 0>, min=<0 0>}
    }
    padding=<ltrb 20 20 20 20>
   }
   slot {class=VerticalBoxSlot}
  }
  VerticalBox_88 {
   widgetClassPath=&VerticalBox
   parent=Border_47
   widget=VerticalBox_88
   slot {class=BorderSlot, properties {padding=<ltrb 20 20 20 20>}}
  }
  HorizontalBox_16 {
   widgetClassPath=&HorizontalBox
   parent=VerticalBox_88
   widget=HorizontalBox_16
   slot {class=VerticalBoxSlot, properties {size {sizeRule=Fill, value=1}}}
  }
  SinglePropertyView_0 {
   widgetClassPath=&SinglePropertyView
   parent=HorizontalBox_16
   widget=SinglePropertyView_0
   properties {propertyName=MICFrom}
   slot {class=HorizontalBoxSlot}
  }
  TextBlock_59 {widgetClassPath=&TextBlock, parent=HorizontalBox_16, widget=TextBlock_59, properties {text="->"}, slot {class=HorizontalBoxSlot}}
  SinglePropertyView {
   widgetClassPath=&SinglePropertyView
   parent=HorizontalBox_16
   widget=SinglePropertyView
   properties {propertyName=MICTo}
   slot {class=HorizontalBoxSlot}
  }
  EditorUtilityButton_5 {
   widgetClassPath=&EditorUtilityButton
   parent=VerticalBox_88
   widget=EditorUtilityButton_5
   slot {class=VerticalBoxSlot, properties {horizontalAlignment=HAlign_Left}}
  }
  TextBlock_4 {
   widgetClassPath=&TextBlock
   parent=EditorUtilityButton_5
   widget=TextBlock_4
   properties {
    text=Load material from selected Actor
    font {
     bForceMonospaced=false
     bMaterialIsStencil=false
     fontMaterial=null
     fontObject=&RobotoTiny
     letterSpacing=0
     monospacedWidth=1
     size=24
     skewAmount=0
     typefaceFontName=Light
     outlineSettings {
      bApplyOutlineToDropShadows=false
      bMiteredCorners=false
      bSeparateFillAlpha=false
      outlineMaterial=null
      outlineSize=0
      outlineColor=<rgba 0 0 0 1>
     }
    }
   }
   slot {class=ButtonSlot, properties {horizontalAlignment=HAlign_Left}}
  }
  EditorUtilityButton_4 {
   widgetClassPath=&EditorUtilityButton
   parent=VerticalBox_88
   widget=EditorUtilityButton_4
   slot {class=VerticalBoxSlot, properties {horizontalAlignment=HAlign_Left}}
  }
  TextBlock_3 {
   widgetClassPath=&TextBlock
   parent=EditorUtilityButton_4
   widget=TextBlock_3
   properties {
    text=Replace Actors using this material with another
    font {
     bForceMonospaced=false
     bMaterialIsStencil=false
     fontMaterial=null
     fontObject=&RobotoTiny
     letterSpacing=0
     monospacedWidth=1
     size=24
     skewAmount=0
     typefaceFontName=Light
     outlineSettings {
      bApplyOutlineToDropShadows=false
      bMiteredCorners=false
      bSeparateFillAlpha=false
      outlineMaterial=null
      outlineSize=0
      outlineColor=<rgba 0 0 0 1>
     }
    }
   }
   slot {class=ButtonSlot, properties {horizontalAlignment=HAlign_Left}}
  }
  Spacer {widgetClassPath=&Spacer, parent=VerticalBox_59, widget=Spacer, properties {size=<1 24>}, slot {class=VerticalBoxSlot}}
  Border {
   widgetClassPath=&Border
   parent=VerticalBox_59
   widget=Border
   properties {
    background {
     bIsDynamicallyLoaded=false
     drawAs=RoundedBox
     imageType=NoImage
     mirroring=NoMirror
     resourceName=None
     resourceObject=null
     tiling=NoTile
     imageSize=<32 32>
     margin=<ltrb 0 0 0 0>
     outlineSettings {
      bUseBrushTransparency=false
      roundingType=FixedRadius
      width=1.8204
      color {colorUseRule=UseColor_Specified, specifiedColor=<rgba 1 1 1 1>}
      cornerRadii=<20 20 20 20>
     }
     tintColor {colorUseRule=UseColor_Specified, specifiedColor=<rgba 0.0802 0.0808 0.0885 1>}
     uVRegion {bIsValid=false, max=<0 0>, min=<0 0>}
    }
    padding=<ltrb 20 20 20 20>
   }
   slot {class=VerticalBoxSlot}
  }
  VerticalBox {widgetClassPath=&VerticalBox, parent=Border, widget=VerticalBox, slot {class=BorderSlot, properties {padding=<ltrb 20 20 20 20>}}}
  HorizontalBox {
   widgetClassPath=&HorizontalBox
   parent=VerticalBox
   widget=HorizontalBox
   slot {class=VerticalBoxSlot, properties {size {sizeRule=Fill, value=1}}}
  }
  SinglePropertyView_1 {
   widgetClassPath=&SinglePropertyView
   parent=HorizontalBox
   widget=SinglePropertyView_1
   properties {propertyName=TD Asset}
   slot {class=HorizontalBoxSlot}
  }
  EditorUtilityButton_7 {
   widgetClassPath=&EditorUtilityButton
   parent=VerticalBox
   widget=EditorUtilityButton_7
   slot {class=VerticalBoxSlot, properties {horizontalAlignment=HAlign_Left}}
  }
  TextBlock_7 {
   widgetClassPath=&TextBlock
   parent=EditorUtilityButton_7
   widget=TextBlock_7
   properties {
    text=Apply this Texture Data Asset to Ground Meshes with no TD
    font {
     bForceMonospaced=false
     bMaterialIsStencil=false
     fontMaterial=null
     fontObject=&RobotoTiny
     letterSpacing=0
     monospacedWidth=1
     size=24
     skewAmount=0
     typefaceFontName=Light
     outlineSettings {
      bApplyOutlineToDropShadows=false
      bMiteredCorners=false
      bSeparateFillAlpha=false
      outlineMaterial=null
      outlineSize=0
      outlineColor=<rgba 0 0 0 1>
     }
    }
   }
   slot {class=ButtonSlot, properties {horizontalAlignment=HAlign_Left}}
  }
  AppltTD {
   widgetClassPath=&EditorUtilityButton
   parent=VerticalBox
   widget=AppltTD
   slot {class=VerticalBoxSlot, properties {horizontalAlignment=HAlign_Left}}
  }
  TextBlock_14 {
   widgetClassPath=&TextBlock
   parent=AppltTD
   widget=TextBlock_14
   properties {
    text="Apply this Texture Data Asset to Ground Meshes "
    font {
     bForceMonospaced=false
     bMaterialIsStencil=false
     fontMaterial=null
     fontObject=&RobotoTiny
     letterSpacing=0
     monospacedWidth=1
     size=24
     skewAmount=0
     typefaceFontName=Light
     outlineSettings {
      bApplyOutlineToDropShadows=false
      bMiteredCorners=false
      bSeparateFillAlpha=false
      outlineMaterial=null
      outlineSize=0
      outlineColor=<rgba 0 0 0 1>
     }
    }
   }
   slot {class=ButtonSlot, properties {horizontalAlignment=HAlign_Left}}
  }
  SwapSnowMeshes {
   widgetClassPath=&EditorUtilityButton
   parent=VerticalBox_59
   widget=SwapSnowMeshes
   properties {toolTipText="When we moved away from dynamic snow, we wanted to force snow meshes to ve 'forcesnow' variants. this uses the snow topo texture to set the meshes in that area to the force variant"}
   slot {class=VerticalBoxSlot, properties {horizontalAlignment=HAlign_Left, padding=<ltrb 4 4 4 4>}}
  }
  Swapsnow {
   widgetClassPath=&TextBlock
   parent=SwapSnowMeshes
   widget=Swapsnow
   properties {
    text=SwapSnowMeshes
    font {
     bForceMonospaced=false
     bMaterialIsStencil=false
     fontMaterial=null
     fontObject=&RobotoTiny
     letterSpacing=0
     monospacedWidth=1
     size=24
     skewAmount=0
     typefaceFontName=Light
     outlineSettings {
      bApplyOutlineToDropShadows=false
      bMiteredCorners=false
      bSeparateFillAlpha=false
      outlineMaterial=null
      outlineSize=0
      outlineColor=<rgba 0 0 0 1>
     }
    }
   }
   slot {class=ButtonSlot, properties {horizontalAlignment=HAlign_Left}}
  }
  FixUpInLevel {
   widgetClassPath=&EditorUtilityButton
   parent=VerticalBox_59
   widget=FixUpInLevel
   slot {class=VerticalBoxSlot, properties {horizontalAlignment=HAlign_Left}}
  }
  TextBlock_0 {
   widgetClassPath=&TextBlock
   parent=FixUpInLevel
   widget=TextBlock_0
   properties {
    text=Fix up actors in current level
    font {
     bForceMonospaced=false
     bMaterialIsStencil=false
     fontMaterial=null
     fontObject=&RobotoTiny
     letterSpacing=0
     monospacedWidth=1
     size=24
     skewAmount=0
     typefaceFontName=Light
     outlineSettings {
      bApplyOutlineToDropShadows=false
      bMiteredCorners=false
      bSeparateFillAlpha=false
      outlineMaterial=null
      outlineSize=0
      outlineColor=<rgba 0 0 0 1>
     }
    }
   }
   slot {class=ButtonSlot, properties {horizontalAlignment=HAlign_Left}}
  }
  PaintConcreteVertexColorsRedInLevel {
   widgetClassPath=&EditorUtilityButton
   parent=VerticalBox_59
   widget=PaintConcreteVertexColorsRedInLevel
   slot {class=VerticalBoxSlot, properties {horizontalAlignment=HAlign_Left}}
  }
  TextBlock_12 {
   widgetClassPath=&TextBlock
   parent=PaintConcreteVertexColorsRedInLevel
   widget=TextBlock_12
   properties {
    text=Paint concrete vertex actors in Path
    font {
     bForceMonospaced=false
     bMaterialIsStencil=false
     fontMaterial=null
     fontObject=&RobotoTiny
     letterSpacing=0
     monospacedWidth=1
     size=24
     skewAmount=0
     typefaceFontName=Light
     outlineSettings {
      bApplyOutlineToDropShadows=false
      bMiteredCorners=false
      bSeparateFillAlpha=false
      outlineMaterial=null
      outlineSize=0
      outlineColor=<rgba 0 0 0 1>
     }
    }
   }
   slot {class=ButtonSlot, properties {horizontalAlignment=HAlign_Left}}
  }
  HorizontalBox_4 {widgetClassPath=&HorizontalBox, parent=VerticalBox_59, widget=HorizontalBox_4, slot {class=VerticalBoxSlot}}
  TextBlock_17 {
   widgetClassPath=&TextBlock
   parent=HorizontalBox_4
   widget=TextBlock_17
   properties {
    text=Target Material Path:
    font {
     bForceMonospaced=false
     bMaterialIsStencil=false
     fontMaterial=null
     fontObject=&RobotoTiny
     letterSpacing=0
     monospacedWidth=1
     size=24
     skewAmount=0
     typefaceFontName=Light
     outlineSettings {
      bApplyOutlineToDropShadows=false
      bMiteredCorners=false
      bSeparateFillAlpha=false
      outlineMaterial=null
      outlineSize=0
      outlineColor=<rgba 0 0 0 1>
     }
    }
   }
   slot {class=HorizontalBoxSlot, properties {padding=<ltrb 8 8 8 8>}}
  }
  VerticalBox_2 {
   widgetClassPath=&VerticalBox
   parent=HorizontalBox_4
   widget=VerticalBox_2
   slot {class=HorizontalBoxSlot, properties {verticalAlignment=VAlign_Center}}
  }
  TargetTextBox {
   widgetClassPath=&EditorUtilityEditableTextBox
   bIsVariable=true
   parent=VerticalBox_2
   widget=TargetTextBox
   properties {
    text=/Figment_S07/LM_7_00/Athena/WinterArt/Materials/
    widgetStyle {
     backgroundColor {colorUseRule=UseColor_Specified, specifiedColor=<rgba 1 1 1 1>}
     backgroundImageFocused {
      bIsDynamicallyLoaded=false
      drawAs=RoundedBox
      imageType=NoImage
      mirroring=NoMirror
      resourceName=None
      resourceObject=null
      tiling=NoTile
      imageSize=<0 0>
      margin=<ltrb 0 0 0 0>
      outlineSettings {
       bUseBrushTransparency=false
       roundingType=FixedRadius
       width=1
       color {colorUseRule=UseColor_ColorTable, specifiedColor=<rgba 0 0 0 0>}
       cornerRadii=<4 4 4 4>
      }
      tintColor {colorUseRule=UseColor_Specified, specifiedColor=<rgba 0.0048 0.0048 0.0048 1>}
      uVRegion {bIsValid=false, max=<0 0>, min=<0 0>}
     }
     backgroundImageHovered {
      bIsDynamicallyLoaded=false
      drawAs=RoundedBox
      imageType=NoImage
      mirroring=NoMirror
      resourceName=None
      resourceObject=null
      tiling=NoTile
      imageSize=<0 0>
      margin=<ltrb 0 0 0 0>
      outlineSettings {
       bUseBrushTransparency=false
       roundingType=FixedRadius
       width=1
       color {colorUseRule=UseColor_ColorTable, specifiedColor=<rgba 0 0 0 0>}
       cornerRadii=<4 4 4 4>
      }
      tintColor {colorUseRule=UseColor_Specified, specifiedColor=<rgba 0.0048 0.0048 0.0048 1>}
      uVRegion {bIsValid=false, max=<0 0>, min=<0 0>}
     }
     backgroundImageNormal {
      bIsDynamicallyLoaded=false
      drawAs=RoundedBox
      imageType=NoImage
      mirroring=NoMirror
      resourceName=None
      resourceObject=null
      tiling=NoTile
      imageSize=<0 0>
      margin=<ltrb 0 0 0 0>
      outlineSettings {
       bUseBrushTransparency=false
       roundingType=FixedRadius
       width=1
       color {colorUseRule=UseColor_ColorTable, specifiedColor=<rgba 0 0 0 0>}
       cornerRadii=<4 4 4 4>
      }
      tintColor {colorUseRule=UseColor_Specified, specifiedColor=<rgba 0.0048 0.0048 0.0048 1>}
      uVRegion {bIsValid=false, max=<0 0>, min=<0 0>}
     }
     backgroundImageReadOnly {
      bIsDynamicallyLoaded=false
      drawAs=RoundedBox
      imageType=NoImage
      mirroring=NoMirror
      resourceName=None
      resourceObject=null
      tiling=NoTile
      imageSize=<0 0>
      margin=<ltrb 0 0 0 0>
      outlineSettings {
       bUseBrushTransparency=false
       roundingType=FixedRadius
       width=1
       color {colorUseRule=UseColor_ColorTable, specifiedColor=<rgba 0 0 0 0>}
       cornerRadii=<4 4 4 4>
      }
      tintColor {colorUseRule=UseColor_Specified, specifiedColor=<rgba 0.0284 0.0284 0.0284 1>}
      uVRegion {bIsValid=false, max=<0 0>, min=<0 0>}
     }
     focusedForegroundColor {colorUseRule=UseColor_Specified, specifiedColor=<rgba 1 1 1 1>}
     foregroundColor {colorUseRule=UseColor_Specified, specifiedColor=<rgba 0.5271 0.5271 0.5271 1>}
     hScrollBarPadding=<ltrb 0 0 0 0>
     padding=<ltrb 8 3 8 4>
     readOnlyForegroundColor {colorUseRule=UseColor_Foreground, specifiedColor=<rgba 1 0 1 1>}
     scrollBarStyle {
      thickness=8
      draggedThumbImage {
       bIsDynamicallyLoaded=false
       drawAs=RoundedBox
       imageType=NoImage
       mirroring=NoMirror
       resourceName=None
       resourceObject=null
       tiling=NoTile
       imageSize=<0 0>
       margin=<ltrb 0 0 0 0>
       outlineSettings {
        bUseBrushTransparency=false
        roundingType=FixedRadius
        width=0
        color {colorUseRule=UseColor_Specified, specifiedColor=<rgba 0 0 0 0>}
        cornerRadii=<4 4 4 4>
       }
       tintColor {colorUseRule=UseColor_Specified, specifiedColor=<rgba 0.2159 0.2159 0.2159 1>}
       uVRegion {bIsValid=false, max=<0 0>, min=<0 0>}
      }
      horizontalBackgroundImage {
       bIsDynamicallyLoaded=false
       drawAs=NoDrawType
       imageType=NoImage
       mirroring=NoMirror
       resourceName=None
       resourceObject=null
       tiling=NoTile
       imageSize=<0 0>
       margin=<ltrb 0 0 0 0>
       outlineSettings {
        bUseBrushTransparency=false
        roundingType=HalfHeightRadius
        width=0
        color {colorUseRule=UseColor_Specified, specifiedColor=<rgba 0 0 0 0>}
        cornerRadii=<0 0 0 0>
       }
       tintColor {colorUseRule=UseColor_Specified, specifiedColor=<rgba 1 1 1 1>}
       uVRegion {bIsValid=false, max=<0 0>, min=<0 0>}
      }
      horizontalBottomSlotImage {
       bIsDynamicallyLoaded=false
       drawAs=NoDrawType
       imageType=NoImage
       mirroring=NoMirror
       resourceName=None
       resourceObject=null
       tiling=NoTile
       imageSize=<0 0>
       margin=<ltrb 0 0 0 0>
       outlineSettings {
        bUseBrushTransparency=false
        roundingType=HalfHeightRadius
        width=0
        color {colorUseRule=UseColor_Specified, specifiedColor=<rgba 0 0 0 0>}
        cornerRadii=<0 0 0 0>
       }
       tintColor {colorUseRule=UseColor_Specified, specifiedColor=<rgba 1 1 1 1>}
       uVRegion {bIsValid=false, max=<0 0>, min=<0 0>}
      }
      horizontalTopSlotImage {
       bIsDynamicallyLoaded=false
       drawAs=NoDrawType
       imageType=NoImage
       mirroring=NoMirror
       resourceName=None
       resourceObject=null
       tiling=NoTile
       imageSize=<0 0>
       margin=<ltrb 0 0 0 0>
       outlineSettings {
        bUseBrushTransparency=false
        roundingType=HalfHeightRadius
        width=0
        color {colorUseRule=UseColor_Specified, specifiedColor=<rgba 0 0 0 0>}
        cornerRadii=<0 0 0 0>
       }
       tintColor {colorUseRule=UseColor_Specified, specifiedColor=<rgba 1 1 1 1>}
       uVRegion {bIsValid=false, max=<0 0>, min=<0 0>}
      }
      hoveredThumbImage {
       bIsDynamicallyLoaded=false
       drawAs=RoundedBox
       imageType=NoImage
       mirroring=NoMirror
       resourceName=None
       resourceObject=null
       tiling=NoTile
       imageSize=<0 0>
       margin=<ltrb 0 0 0 0>
       outlineSettings {
        bUseBrushTransparency=false
        roundingType=FixedRadius
        width=0
        color {colorUseRule=UseColor_Specified, specifiedColor=<rgba 0 0 0 0>}
        cornerRadii=<4 4 4 4>
       }
       tintColor {colorUseRule=UseColor_Specified, specifiedColor=<rgba 0.2159 0.2159 0.2159 1>}
       uVRegion {bIsValid=false, max=<0 0>, min=<0 0>}
      }
      normalThumbImage {
       bIsDynamicallyLoaded=false
       drawAs=RoundedBox
       imageType=NoImage
       mirroring=NoMirror
       resourceName=None
       resourceObject=null
       tiling=NoTile
       imageSize=<0 0>
       margin=<ltrb 0 0 0 0>
       outlineSettings {
        bUseBrushTransparency=false
        roundingType=FixedRadius
        width=0
        color {colorUseRule=UseColor_Specified, specifiedColor=<rgba 0 0 0 0>}
        cornerRadii=<4 4 4 4>
       }
       tintColor {colorUseRule=UseColor_Specified, specifiedColor=<rgba 0.0953 0.0953 0.0953 1>}
       uVRegion {bIsValid=false, max=<0 0>, min=<0 0>}
      }
      verticalBackgroundImage {
       bIsDynamicallyLoaded=false
       drawAs=NoDrawType
       imageType=NoImage
       mirroring=NoMirror
       resourceName=None
       resourceObject=null
       tiling=NoTile
       imageSize=<0 0>
       margin=<ltrb 0 0 0 0>
       outlineSettings {
        bUseBrushTransparency=false
        roundingType=HalfHeightRadius
        width=0
        color {colorUseRule=UseColor_Specified, specifiedColor=<rgba 0 0 0 0>}
        cornerRadii=<0 0 0 0>
       }
       tintColor {colorUseRule=UseColor_Specified, specifiedColor=<rgba 1 1 1 1>}
       uVRegion {bIsValid=false, max=<0 0>, min=<0 0>}
      }
      verticalBottomSlotImage {
       bIsDynamicallyLoaded=false
       drawAs=NoDrawType
       imageType=NoImage
       mirroring=NoMirror
       resourceName=None
       resourceObject=null
       tiling=NoTile
       imageSize=<0 0>
       margin=<ltrb 0 0 0 0>
       outlineSettings {
        bUseBrushTransparency=false
        roundingType=HalfHeightRadius
        width=0
        color {colorUseRule=UseColor_Specified, specifiedColor=<rgba 0 0 0 0>}
        cornerRadii=<0 0 0 0>
       }
       tintColor {colorUseRule=UseColor_Specified, specifiedColor=<rgba 1 1 1 1>}
       uVRegion {bIsValid=false, max=<0 0>, min=<0 0>}
      }
      verticalTopSlotImage {
       bIsDynamicallyLoaded=false
       drawAs=NoDrawType
       imageType=NoImage
       mirroring=NoMirror
       resourceName=None
       resourceObject=null
       tiling=NoTile
       imageSize=<0 0>
       margin=<ltrb 0 0 0 0>
       outlineSettings {
        bUseBrushTransparency=false
        roundingType=HalfHeightRadius
        width=0
        color {colorUseRule=UseColor_Specified, specifiedColor=<rgba 0 0 0 0>}
        cornerRadii=<0 0 0 0>
       }
       tintColor {colorUseRule=UseColor_Specified, specifiedColor=<rgba 1 1 1 1>}
       uVRegion {bIsValid=false, max=<0 0>, min=<0 0>}
      }
     }
     textStyle {
      overflowPolicy=Clip
      transformPolicy=None
      colorAndOpacity {colorUseRule=UseColor_Foreground, specifiedColor=<rgba 1 0 1 1>}
      font {
       bForceMonospaced=false
       bMaterialIsStencil=false
       fontMaterial=null
       fontObject=null
       letterSpacing=0
       monospacedWidth=1
       size=18
       skewAmount=0
       typefaceFontName=Regular
       outlineSettings {
        bApplyOutlineToDropShadows=false
        bMiteredCorners=false
        bSeparateFillAlpha=false
        outlineMaterial=null
        outlineSize=0
        outlineColor=<rgba 0 0 0 1>
       }
      }
      highlightColor {colorUseRule=UseColor_ColorTable, specifiedColor=<rgba 0 0 0 0>}
      highlightShape {
       bIsDynamicallyLoaded=false
       drawAs=Image
       imageType=NoImage
       mirroring=NoMirror
       resourceName=None
       resourceObject=null
       tiling=NoTile
       imageSize=<0 0>
       margin=<ltrb 0 0 0 0>
       outlineSettings {
        bUseBrushTransparency=false
        roundingType=HalfHeightRadius
        width=0
        color {colorUseRule=UseColor_Specified, specifiedColor=<rgba 0 0 0 0>}
        cornerRadii=<0 0 0 0>
       }
       tintColor {colorUseRule=UseColor_ColorTable, specifiedColor=<rgba 0 0 0 0>}
       uVRegion {bIsValid=false, max=<0 0>, min=<0 0>}
      }
      selectedBackgroundColor {colorUseRule=UseColor_ColorTable, specifiedColor=<rgba 0 0 0 0>}
      shadowColorAndOpacity=<rgba 0 0 0 1>
      shadowOffset=<0 0>
      strikeBrush {
       bIsDynamicallyLoaded=false
       drawAs=Image
       imageType=NoImage
       mirroring=NoMirror
       resourceName=None
       resourceObject=null
       tiling=NoTile
       imageSize=<32 32>
       margin=<ltrb 0 0 0 0>
       outlineSettings {
        bUseBrushTransparency=false
        roundingType=HalfHeightRadius
        width=0
        color {colorUseRule=UseColor_Specified, specifiedColor=<rgba 0 0 0 0>}
        cornerRadii=<0 0 0 0>
       }
       tintColor {colorUseRule=UseColor_Specified, specifiedColor=<rgba 1 1 1 1>}
       uVRegion {bIsValid=false, max=<0 0>, min=<0 0>}
      }
      underlineBrush {
       bIsDynamicallyLoaded=false
       drawAs=Image
       imageType=NoImage
       mirroring=NoMirror
       resourceName=None
       resourceObject=null
       tiling=NoTile
       imageSize=<32 32>
       margin=<ltrb 0 0 0 0>
       outlineSettings {
        bUseBrushTransparency=false
        roundingType=HalfHeightRadius
        width=0
        color {colorUseRule=UseColor_Specified, specifiedColor=<rgba 0 0 0 0>}
        cornerRadii=<0 0 0 0>
       }
       tintColor {colorUseRule=UseColor_Specified, specifiedColor=<rgba 1 1 1 1>}
       uVRegion {bIsValid=false, max=<0 0>, min=<0 0>}
      }
     }
     vScrollBarPadding=<ltrb 0 0 0 0>
    }
   }
   slot {class=VerticalBoxSlot}
  }
  ReplaceMaterialsWithTarget {
   widgetClassPath=&EditorUtilityButton
   parent=VerticalBox_59
   widget=ReplaceMaterialsWithTarget
   slot {class=VerticalBoxSlot, properties {horizontalAlignment=HAlign_Left}}
  }
  TextBlock_15 {
   widgetClassPath=&TextBlock
   parent=ReplaceMaterialsWithTarget
   widget=TextBlock_15
   properties {
    text=ReplaceMaterialsWithTarget
    font {
     bForceMonospaced=false
     bMaterialIsStencil=false
     fontMaterial=null
     fontObject=&RobotoTiny
     letterSpacing=0
     monospacedWidth=1
     size=24
     skewAmount=0
     typefaceFontName=Light
     outlineSettings {
      bApplyOutlineToDropShadows=false
      bMiteredCorners=false
      bSeparateFillAlpha=false
      outlineMaterial=null
      outlineSize=0
      outlineColor=<rgba 0 0 0 1>
     }
    }
   }
   slot {class=ButtonSlot, properties {horizontalAlignment=HAlign_Left}}
  }
  HorizontalBox_5 {widgetClassPath=&HorizontalBox, parent=VerticalBox_59, widget=HorizontalBox_5, slot {class=VerticalBoxSlot}}
  TextBlock_18 {
   widgetClassPath=&TextBlock
   parent=HorizontalBox_5
   widget=TextBlock_18
   properties {
    text=Target Tree Material Path:
    font {
     bForceMonospaced=false
     bMaterialIsStencil=false
     fontMaterial=null
     fontObject=&RobotoTiny
     letterSpacing=0
     monospacedWidth=1
     size=24
     skewAmount=0
     typefaceFontName=Light
     outlineSettings {
      bApplyOutlineToDropShadows=false
      bMiteredCorners=false
      bSeparateFillAlpha=false
      outlineMaterial=null
      outlineSize=0
      outlineColor=<rgba 0 0 0 1>
     }
    }
   }
   slot {class=HorizontalBoxSlot, properties {padding=<ltrb 8 8 8 8>}}
  }
  VerticalBox_3 {
   widgetClassPath=&VerticalBox
   parent=HorizontalBox_5
   widget=VerticalBox_3
   slot {class=HorizontalBoxSlot, properties {verticalAlignment=VAlign_Center}}
  }
  TargetTextBox_1 {
   widgetClassPath=&EditorUtilityEditableTextBox
   bIsVariable=true
   parent=VerticalBox_3
   widget=TargetTextBox_1
   properties {
    text=/FigmentEnvironment/Materials/Tree/
    widgetStyle {
     backgroundColor {colorUseRule=UseColor_Specified, specifiedColor=<rgba 1 1 1 1>}
     backgroundImageFocused {
      bIsDynamicallyLoaded=false
      drawAs=RoundedBox
      imageType=NoImage
      mirroring=NoMirror
      resourceName=None
      resourceObject=null
      tiling=NoTile
      imageSize=<0 0>
      margin=<ltrb 0 0 0 0>
      outlineSettings {
       bUseBrushTransparency=false
       roundingType=FixedRadius
       width=1
       color {colorUseRule=UseColor_ColorTable, specifiedColor=<rgba 0 0 0 0>}
       cornerRadii=<4 4 4 4>
      }
      tintColor {colorUseRule=UseColor_Specified, specifiedColor=<rgba 0.0048 0.0048 0.0048 1>}
      uVRegion {bIsValid=false, max=<0 0>, min=<0 0>}
     }
     backgroundImageHovered {
      bIsDynamicallyLoaded=false
      drawAs=RoundedBox
      imageType=NoImage
      mirroring=NoMirror
      resourceName=None
      resourceObject=null
      tiling=NoTile
      imageSize=<0 0>
      margin=<ltrb 0 0 0 0>
      outlineSettings {
       bUseBrushTransparency=false
       roundingType=FixedRadius
       width=1
       color {colorUseRule=UseColor_ColorTable, specifiedColor=<rgba 0 0 0 0>}
       cornerRadii=<4 4 4 4>
      }
      tintColor {colorUseRule=UseColor_Specified, specifiedColor=<rgba 0.0048 0.0048 0.0048 1>}
      uVRegion {bIsValid=false, max=<0 0>, min=<0 0>}
     }
     backgroundImageNormal {
      bIsDynamicallyLoaded=false
      drawAs=RoundedBox
      imageType=NoImage
      mirroring=NoMirror
      resourceName=None
      resourceObject=null
      tiling=NoTile
      imageSize=<0 0>
      margin=<ltrb 0 0 0 0>
      outlineSettings {
       bUseBrushTransparency=false
       roundingType=FixedRadius
       width=1
       color {colorUseRule=UseColor_ColorTable, specifiedColor=<rgba 0 0 0 0>}
       cornerRadii=<4 4 4 4>
      }
      tintColor {colorUseRule=UseColor_Specified, specifiedColor=<rgba 0.0048 0.0048 0.0048 1>}
      uVRegion {bIsValid=false, max=<0 0>, min=<0 0>}
     }
     backgroundImageReadOnly {
      bIsDynamicallyLoaded=false
      drawAs=RoundedBox
      imageType=NoImage
      mirroring=NoMirror
      resourceName=None
      resourceObject=null
      tiling=NoTile
      imageSize=<0 0>
      margin=<ltrb 0 0 0 0>
      outlineSettings {
       bUseBrushTransparency=false
       roundingType=FixedRadius
       width=1
       color {colorUseRule=UseColor_ColorTable, specifiedColor=<rgba 0 0 0 0>}
       cornerRadii=<4 4 4 4>
      }
      tintColor {colorUseRule=UseColor_Specified, specifiedColor=<rgba 0.0284 0.0284 0.0284 1>}
      uVRegion {bIsValid=false, max=<0 0>, min=<0 0>}
     }
     focusedForegroundColor {colorUseRule=UseColor_Specified, specifiedColor=<rgba 1 1 1 1>}
     foregroundColor {colorUseRule=UseColor_Specified, specifiedColor=<rgba 0.5271 0.5271 0.5271 1>}
     hScrollBarPadding=<ltrb 0 0 0 0>
     padding=<ltrb 8 3 8 4>
     readOnlyForegroundColor {colorUseRule=UseColor_Foreground, specifiedColor=<rgba 1 0 1 1>}
     scrollBarStyle {
      thickness=8
      draggedThumbImage {
       bIsDynamicallyLoaded=false
       drawAs=RoundedBox
       imageType=NoImage
       mirroring=NoMirror
       resourceName=None
       resourceObject=null
       tiling=NoTile
       imageSize=<0 0>
       margin=<ltrb 0 0 0 0>
       outlineSettings {
        bUseBrushTransparency=false
        roundingType=FixedRadius
        width=0
        color {colorUseRule=UseColor_Specified, specifiedColor=<rgba 0 0 0 0>}
        cornerRadii=<4 4 4 4>
       }
       tintColor {colorUseRule=UseColor_Specified, specifiedColor=<rgba 0.2159 0.2159 0.2159 1>}
       uVRegion {bIsValid=false, max=<0 0>, min=<0 0>}
      }
      horizontalBackgroundImage {
       bIsDynamicallyLoaded=false
       drawAs=NoDrawType
       imageType=NoImage
       mirroring=NoMirror
       resourceName=None
       resourceObject=null
       tiling=NoTile
       imageSize=<0 0>
       margin=<ltrb 0 0 0 0>
       outlineSettings {
        bUseBrushTransparency=false
        roundingType=HalfHeightRadius
        width=0
        color {colorUseRule=UseColor_Specified, specifiedColor=<rgba 0 0 0 0>}
        cornerRadii=<0 0 0 0>
       }
       tintColor {colorUseRule=UseColor_Specified, specifiedColor=<rgba 1 1 1 1>}
       uVRegion {bIsValid=false, max=<0 0>, min=<0 0>}
      }
      horizontalBottomSlotImage {
       bIsDynamicallyLoaded=false
       drawAs=NoDrawType
       imageType=NoImage
       mirroring=NoMirror
       resourceName=None
       resourceObject=null
       tiling=NoTile
       imageSize=<0 0>
       margin=<ltrb 0 0 0 0>
       outlineSettings {
        bUseBrushTransparency=false
        roundingType=HalfHeightRadius
        width=0
        color {colorUseRule=UseColor_Specified, specifiedColor=<rgba 0 0 0 0>}
        cornerRadii=<0 0 0 0>
       }
       tintColor {colorUseRule=UseColor_Specified, specifiedColor=<rgba 1 1 1 1>}
       uVRegion {bIsValid=false, max=<0 0>, min=<0 0>}
      }
      horizontalTopSlotImage {
       bIsDynamicallyLoaded=false
       drawAs=NoDrawType
       imageType=NoImage
       mirroring=NoMirror
       resourceName=None
       resourceObject=null
       tiling=NoTile
       imageSize=<0 0>
       margin=<ltrb 0 0 0 0>
       outlineSettings {
        bUseBrushTransparency=false
        roundingType=HalfHeightRadius
        width=0
        color {colorUseRule=UseColor_Specified, specifiedColor=<rgba 0 0 0 0>}
        cornerRadii=<0 0 0 0>
       }
       tintColor {colorUseRule=UseColor_Specified, specifiedColor=<rgba 1 1 1 1>}
       uVRegion {bIsValid=false, max=<0 0>, min=<0 0>}
      }
      hoveredThumbImage {
       bIsDynamicallyLoaded=false
       drawAs=RoundedBox
       imageType=NoImage
       mirroring=NoMirror
       resourceName=None
       resourceObject=null
       tiling=NoTile
       imageSize=<0 0>
       margin=<ltrb 0 0 0 0>
       outlineSettings {
        bUseBrushTransparency=false
        roundingType=FixedRadius
        width=0
        color {colorUseRule=UseColor_Specified, specifiedColor=<rgba 0 0 0 0>}
        cornerRadii=<4 4 4 4>
       }
       tintColor {colorUseRule=UseColor_Specified, specifiedColor=<rgba 0.2159 0.2159 0.2159 1>}
       uVRegion {bIsValid=false, max=<0 0>, min=<0 0>}
      }
      normalThumbImage {
       bIsDynamicallyLoaded=false
       drawAs=RoundedBox
       imageType=NoImage
       mirroring=NoMirror
       resourceName=None
       resourceObject=null
       tiling=NoTile
       imageSize=<0 0>
       margin=<ltrb 0 0 0 0>
       outlineSettings {
        bUseBrushTransparency=false
        roundingType=FixedRadius
        width=0
        color {colorUseRule=UseColor_Specified, specifiedColor=<rgba 0 0 0 0>}
        cornerRadii=<4 4 4 4>
       }
       tintColor {colorUseRule=UseColor_Specified, specifiedColor=<rgba 0.0953 0.0953 0.0953 1>}
       uVRegion {bIsValid=false, max=<0 0>, min=<0 0>}
      }
      verticalBackgroundImage {
       bIsDynamicallyLoaded=false
       drawAs=NoDrawType
       imageType=NoImage
       mirroring=NoMirror
       resourceName=None
       resourceObject=null
       tiling=NoTile
       imageSize=<0 0>
       margin=<ltrb 0 0 0 0>
       outlineSettings {
        bUseBrushTransparency=false
        roundingType=HalfHeightRadius
        width=0
        color {colorUseRule=UseColor_Specified, specifiedColor=<rgba 0 0 0 0>}
        cornerRadii=<0 0 0 0>
       }
       tintColor {colorUseRule=UseColor_Specified, specifiedColor=<rgba 1 1 1 1>}
       uVRegion {bIsValid=false, max=<0 0>, min=<0 0>}
      }
      verticalBottomSlotImage {
       bIsDynamicallyLoaded=false
       drawAs=NoDrawType
       imageType=NoImage
       mirroring=NoMirror
       resourceName=None
       resourceObject=null
       tiling=NoTile
       imageSize=<0 0>
       margin=<ltrb 0 0 0 0>
       outlineSettings {
        bUseBrushTransparency=false
        roundingType=HalfHeightRadius
        width=0
        color {colorUseRule=UseColor_Specified, specifiedColor=<rgba 0 0 0 0>}
        cornerRadii=<0 0 0 0>
       }
       tintColor {colorUseRule=UseColor_Specified, specifiedColor=<rgba 1 1 1 1>}
       uVRegion {bIsValid=false, max=<0 0>, min=<0 0>}
      }
      verticalTopSlotImage {
       bIsDynamicallyLoaded=false
       drawAs=NoDrawType
       imageType=NoImage
       mirroring=NoMirror
       resourceName=None
       resourceObject=null
       tiling=NoTile
       imageSize=<0 0>
       margin=<ltrb 0 0 0 0>
       outlineSettings {
        bUseBrushTransparency=false
        roundingType=HalfHeightRadius
        width=0
        color {colorUseRule=UseColor_Specified, specifiedColor=<rgba 0 0 0 0>}
        cornerRadii=<0 0 0 0>
       }
       tintColor {colorUseRule=UseColor_Specified, specifiedColor=<rgba 1 1 1 1>}
       uVRegion {bIsValid=false, max=<0 0>, min=<0 0>}
      }
     }
     textStyle {
      overflowPolicy=Clip
      transformPolicy=None
      colorAndOpacity {colorUseRule=UseColor_Foreground, specifiedColor=<rgba 1 0 1 1>}
      font {
       bForceMonospaced=false
       bMaterialIsStencil=false
       fontMaterial=null
       fontObject=null
       letterSpacing=0
       monospacedWidth=1
       size=18
       skewAmount=0
       typefaceFontName=Regular
       outlineSettings {
        bApplyOutlineToDropShadows=false
        bMiteredCorners=false
        bSeparateFillAlpha=false
        outlineMaterial=null
        outlineSize=0
        outlineColor=<rgba 0 0 0 1>
       }
      }
      highlightColor {colorUseRule=UseColor_ColorTable, specifiedColor=<rgba 0 0 0 0>}
      highlightShape {
       bIsDynamicallyLoaded=false
       drawAs=Image
       imageType=NoImage
       mirroring=NoMirror
       resourceName=None
       resourceObject=null
       tiling=NoTile
       imageSize=<0 0>
       margin=<ltrb 0 0 0 0>
       outlineSettings {
        bUseBrushTransparency=false
        roundingType=HalfHeightRadius
        width=0
        color {colorUseRule=UseColor_Specified, specifiedColor=<rgba 0 0 0 0>}
        cornerRadii=<0 0 0 0>
       }
       tintColor {colorUseRule=UseColor_ColorTable, specifiedColor=<rgba 0 0 0 0>}
       uVRegion {bIsValid=false, max=<0 0>, min=<0 0>}
      }
      selectedBackgroundColor {colorUseRule=UseColor_ColorTable, specifiedColor=<rgba 0 0 0 0>}
      shadowColorAndOpacity=<rgba 0 0 0 1>
      shadowOffset=<0 0>
      strikeBrush {
       bIsDynamicallyLoaded=false
       drawAs=Image
       imageType=NoImage
       mirroring=NoMirror
       resourceName=None
       resourceObject=null
       tiling=NoTile
       imageSize=<32 32>
       margin=<ltrb 0 0 0 0>
       outlineSettings {
        bUseBrushTransparency=false
        roundingType=HalfHeightRadius
        width=0
        color {colorUseRule=UseColor_Specified, specifiedColor=<rgba 0 0 0 0>}
        cornerRadii=<0 0 0 0>
       }
       tintColor {colorUseRule=UseColor_Specified, specifiedColor=<rgba 1 1 1 1>}
       uVRegion {bIsValid=false, max=<0 0>, min=<0 0>}
      }
      underlineBrush {
       bIsDynamicallyLoaded=false
       drawAs=Image
       imageType=NoImage
       mirroring=NoMirror
       resourceName=None
       resourceObject=null
       tiling=NoTile
       imageSize=<32 32>
       margin=<ltrb 0 0 0 0>
       outlineSettings {
        bUseBrushTransparency=false
        roundingType=HalfHeightRadius
        width=0
        color {colorUseRule=UseColor_Specified, specifiedColor=<rgba 0 0 0 0>}
        cornerRadii=<0 0 0 0>
       }
       tintColor {colorUseRule=UseColor_Specified, specifiedColor=<rgba 1 1 1 1>}
       uVRegion {bIsValid=false, max=<0 0>, min=<0 0>}
      }
     }
     vScrollBarPadding=<ltrb 0 0 0 0>
    }
   }
   slot {class=VerticalBoxSlot}
  }
  ReplaceTreeMaterials {
   widgetClassPath=&EditorUtilityButton
   parent=VerticalBox_59
   widget=ReplaceTreeMaterials
   slot {class=VerticalBoxSlot, properties {horizontalAlignment=HAlign_Left}}
  }
  TextBlock_19 {
   widgetClassPath=&TextBlock
   parent=ReplaceTreeMaterials
   widget=TextBlock_19
   properties {
    text=Replace Tree Materials
    font {
     bForceMonospaced=false
     bMaterialIsStencil=false
     fontMaterial=null
     fontObject=&RobotoTiny
     letterSpacing=0
     monospacedWidth=1
     size=24
     skewAmount=0
     typefaceFontName=Light
     outlineSettings {
      bApplyOutlineToDropShadows=false
      bMiteredCorners=false
      bSeparateFillAlpha=false
      outlineMaterial=null
      outlineSize=0
      outlineColor=<rgba 0 0 0 1>
     }
    }
   }
   slot {class=ButtonSlot, properties {horizontalAlignment=HAlign_Left}}
  }
  SetTreeTopo {
   widgetClassPath=&EditorUtilityButton
   parent=VerticalBox_59
   widget=SetTreeTopo
   slot {class=VerticalBoxSlot, properties {horizontalAlignment=HAlign_Left}}
  }
  TextBlock_25 {
   widgetClassPath=&TextBlock
   parent=SetTreeTopo
   widget=TextBlock_25
   properties {
    text=SetTreeTopoMask
    font {
     bForceMonospaced=false
     bMaterialIsStencil=false
     fontMaterial=null
     fontObject=&RobotoTiny
     letterSpacing=0
     monospacedWidth=1
     size=24
     skewAmount=0
     typefaceFontName=Light
     outlineSettings {
      bApplyOutlineToDropShadows=false
      bMiteredCorners=false
      bSeparateFillAlpha=false
      outlineMaterial=null
      outlineSize=0
      outlineColor=<rgba 0 0 0 1>
     }
    }
   }
   slot {class=ButtonSlot, properties {horizontalAlignment=HAlign_Left}}
  }
  ReparentMIs {
   widgetClassPath=&EditorUtilityButton
   parent=VerticalBox_59
   widget=ReparentMIs
   slot {class=VerticalBoxSlot, properties {horizontalAlignment=HAlign_Left}}
  }
  TextBlock_26 {
   widgetClassPath=&TextBlock
   parent=ReparentMIs
   widget=TextBlock_26
   properties {
    text=Reparent MIs
    font {
     bForceMonospaced=false
     bMaterialIsStencil=false
     fontMaterial=null
     fontObject=&RobotoTiny
     letterSpacing=0
     monospacedWidth=1
     size=24
     skewAmount=0
     typefaceFontName=Light
     outlineSettings {
      bApplyOutlineToDropShadows=false
      bMiteredCorners=false
      bSeparateFillAlpha=false
      outlineMaterial=null
      outlineSize=0
      outlineColor=<rgba 0 0 0 1>
     }
    }
   }
   slot {class=ButtonSlot, properties {horizontalAlignment=HAlign_Left}}
  }
  mpc_change {
   widgetClassPath=&EditorUtilityButton
   parent=VerticalBox_59
   widget=mpc_change
   slot {class=VerticalBoxSlot, properties {horizontalAlignment=HAlign_Left}}
  }
  ChangeMPC {
   widgetClassPath=&TextBlock
   parent=mpc_change
   widget=ChangeMPC
   properties {
    text=ChangeMPC
    font {
     bForceMonospaced=false
     bMaterialIsStencil=false
     fontMaterial=null
     fontObject=&RobotoTiny
     letterSpacing=0
     monospacedWidth=1
     size=24
     skewAmount=0
     typefaceFontName=Light
     outlineSettings {
      bApplyOutlineToDropShadows=false
      bMiteredCorners=false
      bSeparateFillAlpha=false
      outlineMaterial=null
      outlineSize=0
      outlineColor=<rgba 0 0 0 1>
     }
    }
   }
   slot {class=ButtonSlot, properties {horizontalAlignment=HAlign_Left}}
  }
  Spacer_4 {
   widgetClassPath=&Spacer
   parent=CanvasPanel_84
   widget=Spacer_4
   properties {size=<1 48>}
   slot {class=CanvasPanelSlot, properties=~}
  }
 ]
}
