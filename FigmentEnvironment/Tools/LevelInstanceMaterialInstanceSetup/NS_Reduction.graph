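# NS_Reduction is a GPU Niagara utility system in the LevelInstanceMaterialInstanceSetup tool, named for a reduction over a 2D grid.
# Its single emitter creates a float Grid2D Collection one and a half times as wide as the User.size height.
# Each update sets a bool from the system tick count and passes the size to a scratch module, whose logic is not shown.
# User parameters expose several render targets, and a default sprite renderer is present.
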NiagaraSystem /FigmentEnvironment/Tools/LevelInstanceMaterialInstanceSetup/NS_Reduction
dependencies {
 usedDataInterfaces [/Script/Niagara.NiagaraDataInterfaceGrid2DCollection]
 usedDynamicInputs []
 usedModules [&SystemState]
}
emitters {Minimal {
  simTarget=GPUComputeSim
  emitterSpawnScript [
   /SetVariables(Emitter.Grid2D Collection=DataInterface() {propertyValues {
      ClearBeforeNonIterationStage=false
      NumAttributes=0
      NumCellsMaxAxis=3
      NumCellsX=3
      NumCellsY=3
      OverrideBufferFormat=Float
      PreviewAttribute=RGBA
      SetGridFromMaxAxis=false
      bOverrideFormat=true
      bPreviewGrid=true
      RenderTargetUserParameter {parameter {name=None, typeDefHandle {registeredTypeIndex=4}, varData []}}
      WorldBBoxSize=<100 100>
     }}) {isSetParameters=true}
   /SetVariables(Emitter.gridWidth:NiagaraFloat=&Multiply_FloatByInt(Float=1.5, Integer=$User.size)) {isSetParameters=true}
   /NiagaraFluids/Modules/Grid2D/Grid2D_SetResolution(
    Grid:NiagaraDataInterfaceGrid2DCollection=$Emitter.Grid2D Collection
    NumCellsX:NiagaraInt32=&MakeIntFromFloat(Float=$Emitter.gridWidth, Conversion Mode=::NewEnumerator2)
    NumCellsY:NiagaraInt32=$User.size
   )
  ]
  emitterUpdateScript [
   &EmitterState()
   /SetVariables(Emitter.bool:NiagaraBool=&SetBoolByIntComparison(A=$Engine.System.TickCount, B=0, Comparison Type=::NewEnumerator4)) {isSetParameters=true}
   &self:ScratchModule_04(res:NiagaraInt32=$User.size) {module=ScratchModule_04}
  ]
  particleSpawnScript [&InitializeParticle(Color Mode=::NewEnumerator0, Lifetime=5, UsePositionOffset=false, Write Material Random=true)]
  particleUpdateScript [&ParticleState()]
  rendererClasses [&NiagaraSpriteRendererProperties]
  renderers [{rendererClass=&NiagaraSpriteRendererProperties}]
 }}
summary {
 systemName=NS_Reduction
 emitters [{emitterName=Minimal, simTarget=GPUComputeSim, rendererClasses [&NiagaraSpriteRendererProperties]}]
 userVariables [
  User.size {type=NiagaraInt32, defaultValue=2048}
  User.Texture Render Target {
   type=TextureRenderTarget
   description=""
   defaultValue {struct=NiagaraExt_VariableValue_Object, value {object=null, objectClass=/Script/Engine.TextureRenderTarget}}
  }
  User.Texture Render Target001 {
   type=TextureRenderTarget
   description=""
   defaultValue {
    struct=NiagaraExt_VariableValue_Object
    value {
     object=&/NewTextureRenderTarget2D2
     objectClass=&TextureRenderTarget
    }
   }
  }
  User.TRTA {
   type=TextureRenderTarget
   description=""
   defaultValue {
    struct=NiagaraExt_VariableValue_Object
    value {object=&/NewTextureRenderTarget2D, objectClass=&TextureRenderTarget}
   }
  }
  User.TRTB {
   type=TextureRenderTarget
   description=""
   defaultValue {
    struct=NiagaraExt_VariableValue_Object
    value {
     object=&/NewTextureRenderTarget2D1
     objectClass=&TextureRenderTarget
    }
   }
  }
 ]
}
